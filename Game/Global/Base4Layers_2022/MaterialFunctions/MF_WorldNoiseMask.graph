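# Material function that optionally combines a scalar BaseMask input with a second mask, intended as world noise by its naming.
# A static bool input, off by default, enables the blend, and another selects Max (default) or Min; otherwise BaseMask passes through.
# The blend source is a named reroute with no declaration shown.
# The WorldAlignedTexture call and WorldPosition node have no visible connection to the output.
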MaterialFunction /Game/Global/Base4Layers_2022/MaterialFunctions/MF_WorldNoiseMask
graph {nodes [
  n1 = FunctionInput() {inputName=BaseMask, inputType=FunctionInput_Scalar}
  n2 = NamedRerouteUsage()
  n3 = Max(A=@n1, B=@n2)
  n4 = Min(A=@n1, B=@n2)
  n5 = FunctionInput(Preview=StaticBool() {value=true}) {bUsePreviewValueAsDefault=true, inputName=Mask_World Noise Min Blend_Blend Max, inputType=FunctionInput_StaticBool, sortPriority=10}
  n6 = StaticBool()
  FunctionOutput(None=StaticSwitch(
    True=StaticSwitch(True=@n3, False=@n4, Value=@n5)
    False=@n1
    Value=FunctionInput(Preview=@n6) {bUsePreviewValueAsDefault=true, inputName=Mask_Use World Noise, inputType=FunctionInput_StaticBool, sortPriority=10}
   ))
  MaterialFunctionCall(TextureObject=TextureObjectParameter() {parameterName=Param, texture=null}) {materialFunction=/Game/Weapons/WeaponSkins/Wraps/MaterialFunctions/WorldAlignedTexture}
  WorldPosition()
 ]}
