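# Material function that optionally blends a scalar base mask with an upward-facing world Z normal mask, for use in layered materials.
# Z comes from either the vertex normal (default) or the input layer's transformed, saturated normal, then is shaped by power, multiply and saturate.
# A static switch combines it with the base mask by Max (default) or Min.
# The world Z blend is off by default, passing the base mask through.
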MaterialFunction /Game/Global/Base4Layers_2022/MaterialFunctions/MF_WorldZBlend
graph {nodes [
  n1 = FunctionInput() {inputName=BaseMask, inputType=FunctionInput_Scalar, sortPriority=3}
  n2 = NamedRerouteUsage()
  n3 = Max(A=@n1, B=@n2)
  n4 = Min(A=@n1, B=@n2)
  n5 = FunctionInput(Preview=StaticBool() {value=true}) {bUsePreviewValueAsDefault=true, inputName=Mask_World Z Min Blend_Blend Max, inputType=FunctionInput_StaticBool, sortPriority=10}
  n6 = StaticBool()
  FunctionOutput(None=StaticSwitch(
    True=StaticSwitch(True=@n3, False=@n4, Value=@n5)
    False=@n1
    Value=FunctionInput(Preview=@n6) {bUsePreviewValueAsDefault=true, inputName=Mask_Use World Z Normal, inputType=FunctionInput_StaticBool, sortPriority=10}
   ))
  n7 = VertexNormalWS()
  n8 = FunctionInput() {bUsePreviewValueAsDefault=true, inputName=MaterialLayer, inputType=FunctionInput_MaterialAttributes}
  n9 = ComponentMask(None=Transform(None=GetMaterialAttributes(None=@n8) {attributeGetTypes [#1]}.Normal)) {b=true}
  n10 = StaticBool() {value=true}
  n11 = FunctionInput() {inputName=Mask_TextureNormalZ_Power, inputType=FunctionInput_Scalar, sortPriority=40, previewValue=<1 0 0 1>}
  n12 = Power(
   Base=StaticSwitch(
    True=ComponentMask(None=@n7) {b=true}
    False=Saturate(None=@n9)
    Value=FunctionInput(Preview=@n10) {bUsePreviewValueAsDefault=true, inputName=Mask_VTX Only, inputType=FunctionInput_StaticBool, sortPriority=20}
   )
   Exp=@n11
  )
  n13 = FunctionInput() {inputName=Mask_TextureNormalZ_Mult, inputType=FunctionInput_Scalar, sortPriority=50, previewValue=<1 0 0 1>}
  NamedRerouteDeclaration(None=Saturate(None=Multiply(A=@n12, B=@n13))) {name=L3_Normal Z, nodeColor=<rgba 0 0.9062 1 1>}
 ]}
