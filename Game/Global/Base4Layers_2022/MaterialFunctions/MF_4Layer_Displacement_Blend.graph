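# Material function that height-blends up to four tinted material-attribute layers and outputs the result with a combined Displacement.
# Masks from mesh paint, vertex color or UV indices, optionally world-Z modified, drive cascaded HeightLerps using UV or world-aligned height maps.
# Static switches set layer count and per-layer overlay, which adds angle-corrected normals.
# Displacement scales and biases heights against Magnitude, either per layer before blending (SmoothBlend) or afterwards using blended values.
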MaterialFunction /Game/Global/Base4Layers_2022/MaterialFunctions/MF_4Layer_Displacement_Blend
graph {nodes [
  n1 = FunctionInput() {inputName=Layer2, inputType=FunctionInput_MaterialAttributes, sortPriority=20}
  Composite!(None=@n1)
  n2 = FunctionInput() {inputName=Layer1, inputType=FunctionInput_MaterialAttributes, sortPriority=5}
  Composite(None=@n2)
  Composite!() {subgraphName=UV Mask}
  n3 = FunctionInput() {inputName=Layer3, inputType=FunctionInput_MaterialAttributes, sortPriority=30}
  Composite!(None=@n3)
  n4 = FunctionInput() {inputName=Layer4, inputType=FunctionInput_MaterialAttributes, sortPriority=40}
  Composite(None=@n4)
  Constant3Vector()
  Constant() {r=1}
  n5 = NamedRerouteUsage()
  n6 = NamedRerouteUsage()
  n7 = NamedRerouteUsage()
  n8 = NamedRerouteUsage()
  n9 = NamedRerouteUsage()
  n10 = MaterialFunctionCall!(Base=@n8, Blend=@n9) {materialFunction=/Game/Global/Materials/Character_Parent_2023/Anime_Parent_2023/MaterialFunctions/MF_BlendOverlay}
  n11 = MaterialFunctionCall!(
   Diffuse 1
   Diffuse 2
   Transistion Phase=NamedRerouteUsage()
   Height Texture 1=@n8
   Height Texture 2=Multiply(A=@n9, B=ScalarParameter() {defaultValue=1, group=2_Layer_Displacement, parameterName=DisplacementBlendContrast_L2_Smooth})
   Contrast=ScalarParameter() {group=2_Layer_Displacement, parameterName=DisplacementBlendContrast_L2}
  ) {materialFunction=/Engine/Functions/Engine_MaterialFunctions02/Texturing/HeightLerpWithTwoHeightMaps}
  n12 = StaticSwitch(
   True=StaticSwitchParameter(True=LinearInterpolate(A=@n10.Result, B=@n8, Alpha=@n11.Alpha), False=@n11.Resulting Height Map) {group=2_Layer_Displacement, parameterName=L2_Displacement_Overlay}
   False=NamedRerouteUsage()
   Value=StaticBoolParameter() {group=2_Layer, parameterName=Use 2 Layers, sortPriority=-1}
  )
  n13 = MaterialFunctionCall(Base=@n12, Blend=@n7) {materialFunction=&MF_BlendOverlay}
  n14 = MaterialFunctionCall(
   Transistion Phase=NamedRerouteUsage()
   Height Texture 1=@n12
   Height Texture 2=Multiply(A=@n7, B=ScalarParameter() {defaultValue=1, group=3_Layer_Displacement, parameterName=DisplacementBlendContrast_L3_Smooth})
   Contrast=ScalarParameter() {group=3_Layer_Displacement, parameterName=DisplacementBlendContrast_L3}
  ) {materialFunction=&HeightLerpWithTwoHeightMaps}
  n15 = StaticSwitch(
   True=StaticSwitchParameter(True=LinearInterpolate(A=@n13.Result, B=@n12, Alpha=@n14.Alpha), False=@n14.Resulting Height Map) {group=3_Layer_Displacement, parameterName=L3_Displacement_Overlay}
   False=@n12
   Value=StaticBoolParameter() {group=3_Layer, parameterName=Use 3 Layers, sortPriority=-1}
  )
  n16 = MaterialFunctionCall(
   Transistion Phase=NamedRerouteUsage()
   Height Texture 1=@n15
   Height Texture 2=Multiply(A=@n6, B=ScalarParameter() {defaultValue=1, group=4_Layer_Displacement, parameterName=DisplacementBlendContrast_L4_Smooth})
   Contrast=ScalarParameter() {group=4_Layer_Displacement, parameterName=DisplacementBlendContrast_L4}
  ) {materialFunction=&HeightLerpWithTwoHeightMaps}
  n17 = NamedRerouteUsage()
  n18 = NamedRerouteUsage()
  n19 = NamedRerouteUsage()
  n20 = BlendMaterialAttributes(A=@n18, B=@n19, Alpha=@n11.Alpha)
  n21 = GetMaterialAttributes(None=@n19) {attributeGetTypes [#1]}
  n22 = GetMaterialAttributes(None=@n18) {attributeGetTypes [#1]}
  n23 = SetMaterialAttributes!(
   MaterialAttributes=@n20
   Normal=LinearInterpolate(
    A=MaterialFunctionCall(BaseNormal=@n21.Normal, AdditionalNormal=@n22.Normal) {materialFunction=&BlendAngleCorrectedNormals}.Result
    B=@n21.Normal
    Alpha=@n11.Alpha
   )
  ) {attributeSetTypes [#1]}
  n24 = NamedRerouteDeclaration(None=StaticSwitch(
    True=StaticSwitchParameter(True=@n23, False=@n20) {group=2_Layer_Displacement, parameterName=L2_Displacement_Overlay}
    False=NamedRerouteUsage()
    Value=StaticBoolParameter() {group=2_Layer, parameterName=Use 2 Layers, sortPriority=-1}
   )) {name=L1 and L2, nodeColor=<rgba 0 1 0.625 1>}
  n25 = BlendMaterialAttributes(A=@n17, B=@n24, Alpha=@n14.Alpha)
  n26 = GetMaterialAttributes(None=@n24) {attributeGetTypes [#1]}
  n27 = GetMaterialAttributes(None=@n17) {attributeGetTypes [#1]}
  n28 = SetMaterialAttributes(
   MaterialAttributes=@n25
   Normal=LinearInterpolate(
    A=MaterialFunctionCall(BaseNormal=@n26.Normal, AdditionalNormal=@n27.Normal) {materialFunction=&BlendAngleCorrectedNormals}.Result
    B=@n26.Normal
    Alpha=@n14.Alpha
   )
  ) {attributeSetTypes [#1]}
  n29 = NamedRerouteDeclaration(None=StaticSwitch(
    True=StaticSwitchParameter(True=@n28, False=@n25) {group=3_Layer_Displacement, parameterName=L3_Displacement_Overlay}
    False=@n24
    Value=StaticBoolParameter() {group=3_Layer, parameterName=Use 3 Layers, sortPriority=-1}
   )) {name=L1 L2 and L3, nodeColor=<rgba 1 0.3047 0 1>}
  n30 = BlendMaterialAttributes(A=@n5, B=@n29, Alpha=@n16.Alpha)
  n31 = GetMaterialAttributes(None=@n29) {attributeGetTypes [#1]}
  n32 = GetMaterialAttributes(None=@n5) {attributeGetTypes [#1]}
  n33 = SetMaterialAttributes(
   MaterialAttributes=@n30
   Normal=LinearInterpolate(
    A=MaterialFunctionCall(BaseNormal=@n31.Normal, AdditionalNormal=@n32.Normal) {materialFunction=&BlendAngleCorrectedNormals}.Result
    B=@n31.Normal
    Alpha=@n16.Alpha
   )
  ) {attributeSetTypes [#1]}
  n34 = MaterialFunctionCall(Base=@n15, Blend=@n6) {materialFunction=&MF_BlendOverlay}
  n35 = StaticSwitch(
   True=StaticSwitchParameter(True=LinearInterpolate(A=@n34.Result, B=@n15, Alpha=@n16.Alpha), False=@n16.Resulting Height Map) {group=4_Layer_Displacement, parameterName=L4_Displacement_Overlay}
   False=@n15
   Value=StaticBoolParameter() {group=4_Layer, parameterName=Use 4 Layers, sortPriority=-1}
  )
  n36 = NamedRerouteUsage()
  n37 = NamedRerouteUsage()
  n38 = NamedRerouteUsage()
  n39 = LinearInterpolate(A=Add(A=@n37, B=@n38), B=@n38, Alpha=@n11.Alpha)
  n40 = StaticSwitch(
   True=LinearInterpolate(A=StaticSwitchParameter(True=@n39, False=@n38) {group=2_Layer_Displacement, parameterName=L2_Displacement_Overlay}, B=@n37, Alpha=@n11.Alpha)
   False=NamedRerouteUsage()
   Value=StaticBoolParameter() {group=2_Layer, parameterName=Use 2 Layers, sortPriority=-1}
  )
  n41 = NamedRerouteUsage()
  n42 = LinearInterpolate(A=Add(A=@n40, B=@n41), B=@n41, Alpha=@n14.Alpha)
  n43 = StaticBoolParameter() {group=3_Layer, parameterName=Use 3 Layers, sortPriority=-1}
  n44 = StaticSwitch(
   True=LinearInterpolate(A=StaticSwitchParameter(True=@n42, False=@n41) {group=3_Layer_Displacement, parameterName=L3_Displacement_Overlay}, B=@n40, Alpha=@n14.Alpha)
   False=@n40
   Value=@n43
  )
  n45 = Add(A=@n36, B=@n44)
  n46 = LinearInterpolate(
   A=StaticSwitchParameter(True=LinearInterpolate(A=@n45, Alpha=@n16.Alpha), False=@n36) {group=4_Layer_Displacement, parameterName=L4_Displacement_Overlay}
   B=@n44
   Alpha=@n16.Alpha
  )
  n47 = StaticBoolParameter() {group=4_Layer, parameterName=Use 4 Layers, sortPriority=-1}
  n48 = NamedRerouteUsage()
  n49 = NamedRerouteUsage()
  n50 = LinearInterpolate(A=Divide(A=Add(A=@n48, B=@n49)), B=@n49, Alpha=@n11.Alpha)
  n51 = StaticSwitch(
   True=LinearInterpolate(A=StaticSwitchParameter(True=@n50, False=@n49) {group=2_Layer_Displacement, parameterName=L2_Displacement_Overlay}, B=@n48, Alpha=@n11.Alpha)
   False=NamedRerouteUsage()
   Value=StaticBoolParameter() {group=2_Layer, parameterName=Use 2 Layers, sortPriority=-1}
  )
  n52 = NamedRerouteUsage()
  n53 = LinearInterpolate(A=Divide(A=Add(A=@n51, B=@n52)), B=@n52, Alpha=@n14.Alpha)
  n54 = StaticSwitch(
   True=LinearInterpolate(A=StaticSwitchParameter(True=@n53, False=@n52) {group=3_Layer_Displacement, parameterName=L3_Displacement_Overlay}, B=@n51, Alpha=@n14.Alpha)
   False=@n51
   Value=@n43
  )
  n55 = NamedRerouteUsage()
  n56 = LinearInterpolate(A=Divide(A=Add(A=@n54, B=@n55)), Alpha=@n16.Alpha)
  n57 = StaticSwitch(
   True=LinearInterpolate(A=StaticSwitchParameter(True=@n56, False=@n55) {group=4_Layer_Displacement, parameterName=L4_Displacement_Overlay}, B=@n54, Alpha=@n16.Alpha)
   False=@n54
   Value=@n47
  )
  n58 = Multiply(A=Divide(A=StaticSwitch(True=@n46, False=@n44, Value=@n47), B=NamedRerouteUsage()), B=Subtract(A=@n35, B=@n57) {constB=0.5})
  n59 = SetMaterialAttributes!(
   MaterialAttributes=StaticSwitch(
    True=StaticSwitchParameter(True=@n33, False=@n30) {group=4_Layer_Displacement, parameterName=L4_Displacement_Overlay}
    False=@n29
    Value=StaticBoolParameter() {group=4_Layer, parameterName=Use 4 Layers, sortPriority=-1}
   )
   Displacement=StaticSwitchParameter(True=@n35, False=Add(A=@n58, B=@n57) {constB=0.5}) {group=Displacement, parameterName=Displacement_SmoothBlend}
  ) {attributeSetTypes [#2]}
  FunctionOutput(None=@n59) {outputName="Switch Param (False)"}
  MaterialFunctionCall(Base=@n48, Blend=@n49) {materialFunction=&MF_BlendOverlay}
  MaterialFunctionCall!(BaseMask, Mask_Use World Noise, Mask_World Noise Min Blend_Blend Max) {materialFunction=&/MF_WorldNoiseMask}
  n60 = NamedRerouteUsage()
  n61 = Constant3Vector() {constant=<rgba 0 0 -1 0>}
  n62 = ScalarParameter() {group=1_layer UV, parameterName=L1 WP Offset Z, sortPriority=1}
  n63 = MaterialFunctionCall!(
   Diffuse Texture Object=@n60
   Spec Texture Object=@n60
   TextureSize=MaterialFunctionCall!(L3 UV2, L2 UV2, L4 UV2) {materialFunction=&/MF_UVsSet_New}.WSM 1
   --------------------------
   Use seperate settings for Z projection
   Use seperate texture for Z projection
   Texture Size Z projection
   Z Plane Diffuse Texture Object
   Z Plane Spec Texture Object
   "--------------------------      "
   ProjectionTransitionContrast=Constant() {r=1}
   Use Texture Alpha Heightmap to Refine the Transistion
   HeightLerpTransitionContrast
   PickTheDiffuseHeightChannel
   TheDiffuseHeightChannel
   "--------------------------       "
   NormalTextureObject
   World Space Normals=StaticBool()
   "--------------------------  "
   Export Float 4 Diffuse
   Export Float 4 Spec
   "--------------------------     "
   WorldPosition=Add(A=WorldPosition().XYZ, B=Multiply(A=@n61, B=@n62))
   World Space Normal
  ) {materialFunction=/Engine/Functions/Engine_MaterialFunctions02/Texturing/WorldAlignedTextures_Complex}
  n64 = StaticBoolParameter() {desc="World Space XYZ ( Expensive)", group=1_layer UV, parameterName=Layer 1 WSM_XYZ}
  n65 = StaticBoolParameter() {
   desc="World Space Z only, you need appropriate mask to make it visible. Use Z or map your layers accordingly "
   group=1_layer UV
   parameterName=Layer 1 WSM_Z
  }
  NamedRerouteDeclaration(None=StaticSwitch(True=@n63.Z Spec, False=StaticSwitch(True=@n63.XYZ Spec, False=@n63.XYZ Spec, Value=@n64), Value=@n65)) {name=WSM Displ L1, nodeColor=<rgba 0 1 0.9766 1>}
  n66 = ScalarParameter() {defaultValue=3, group=1_Layer_Displacement, parameterName=Displacement L1}
  n67 = StaticBoolParameter() {
   desc=Enable to use UV2 for mapping. You may need to switch masks for ohter layers To UV3
   group=2_layer UV
   parameterName=Use UV2 for Layer 2
  }
  n68 = MaterialFunctionCall(
   L3 UV2=StaticBoolParameter() {group=3_layer UV, parameterName=Use UV2 for Layer 3}
   L2 UV2=@n67
   L4 UV2=StaticBoolParameter() {group=4_layer UV, parameterName=Use UV2 for Layer 4}
  ) {materialFunction=&MF_UVsSet_New}
  n69 = TextureObjectParameter() {
   group=1_Layer
   parameterName=Height L1
   samplerType=SAMPLERTYPE_Masks
   sortPriority=0
   texture=/Game/Global/Base4Layers_2022/Textures/Default_Displacement__H
  }
  n70 = TextureSample(UVs=@n68.UVs Layer 1, Tex=NamedRerouteDeclaration(None=@n69) {name=Height_Tex_L1, nodeColor=<rgba 0 0.3906 1 1>}) {
   samplerSource=SSM_Wrap_WorldGroupSettings
   samplerType=SAMPLERTYPE_Masks
   texture=&Default_Displacement__H
  }
  n71 = StaticSwitch(
   True=NamedRerouteUsage()
   False=@n70.R
   Value=StaticBoolParameter() {desc=Enable to use World Space Mapping for layer, group=1_layer UV, parameterName=Layer 1 WS Mapping}
  )
  n72 = ScalarParameter() {defaultValue=0.5, group=1_Layer_Displacement, parameterName=Displacement_Bias L1}
  n73 = Multiply(A=Divide(A=@n66, B=NamedRerouteUsage()), B=Subtract(A=@n71, B=@n72) {constB=0.5})
  NamedRerouteDeclaration(None=StaticSwitchParameter(True=Add(A=@n73, B=@n72) {constB=0.5}, False=@n71) {group=Displacement, parameterName=Displacement_SmoothBlend}) {name=Height L1, nodeColor=<rgba 1 0.1406 0 1>}
  n74 = Constant3Vector() {constant=<rgba 0 0 -1 0>}
  n75 = ScalarParameter() {group=3_layer UV, parameterName=L3 WP Offset Z, sortPriority=1}
  n76 = MaterialFunctionCall(
   Spec Texture Object=NamedRerouteUsage()
   TextureSize=MaterialFunctionCall(L2 UV2=StaticBoolParameter() {group=3_layer UV, parameterName=Use UV2 for Layer 3}) {materialFunction=&MF_UVsSet_New}.WSM 3
   ProjectionTransitionContrast=Constant() {r=1}
   World Space Normals=StaticBool()
   WorldPosition=Add(A=WorldPosition().XYZ, B=Multiply(A=@n74, B=@n75))
  ) {materialFunction=&WorldAlignedTextures_Complex}
  n77 = StaticBoolParameter() {desc="World Space XYZ ( Expensive)", group=3_layer UV, parameterName=Layer 3 WSM_XYZ}
  n78 = StaticBoolParameter() {
   desc="World Space Z only, you need appropriate mask to make it visible. Use Z or map your layers accordingly "
   group=3_layer UV
   parameterName=Layer 3 WSM_Z
  }
  NamedRerouteDeclaration(None=StaticSwitch(True=@n76.Z Spec, False=StaticSwitch(True=@n76.XYZ Spec, False=@n76.XYZ Spec, Value=@n77), Value=@n78)) {name=WSM Displ L3, nodeColor=<rgba 0 1 0.9766 1>}
  n79 = Constant3Vector() {constant=<rgba 0 0 -1 0>}
  n80 = ScalarParameter() {group=4_layer UV, parameterName=L4 WP Offset Z, sortPriority=1}
  n81 = MaterialFunctionCall(
   Spec Texture Object=NamedRerouteUsage()
   TextureSize=MaterialFunctionCall(L2 UV2=StaticBoolParameter() {group=4_layer UV, parameterName=Use UV2 for Layer 4}) {materialFunction=&MF_UVsSet_New}.WSM 4
   ProjectionTransitionContrast=Constant() {r=1}
   World Space Normals=StaticBool()
   WorldPosition=Add(A=WorldPosition().XYZ, B=Multiply(A=@n79, B=@n80))
  ) {materialFunction=&WorldAlignedTextures_Complex}
  n82 = StaticBoolParameter() {desc="World Space XYZ ( Expensive)", group=3_layer UV, parameterName=Layer 3 WSM_XYZ}
  n83 = StaticBoolParameter() {
   desc="World Space Z only, you need appropriate mask to make it visible. Use Z or map your layers accordingly "
   group=3_layer UV
   parameterName=Layer 3 WSM_Z
  }
  NamedRerouteDeclaration(None=StaticSwitch(True=@n81.Z Spec, False=StaticSwitch(True=@n81.XYZ Spec, False=@n81.XYZ Spec, Value=@n82), Value=@n83)) {name=WSM Displ L4, nodeColor=<rgba 0 1 0.9766 1>}
  n84 = TextureObjectParameter() {
   group=3_Layer
   parameterName=Height_L3
   samplerType=SAMPLERTYPE_Masks
   sortPriority=0
   texture=&Default_Displacement__H
  }
  n85 = TextureSample(UVs=@n68.UVs Layer 3, Tex=NamedRerouteDeclaration(None=@n84) {name=Height_Tex_L3, nodeColor=<rgba 0 0.3906 1 1>}) {
   samplerSource=SSM_Wrap_WorldGroupSettings
   samplerType=SAMPLERTYPE_Masks
   texture=&Default_Displacement__H
  }
  n86 = StaticBoolParameter() {desc=Enable to use World Space Mapping for layer. Only XY or Z available in Layer 3, group=3_layer UV, parameterName=Layer 3 WS Mapping}
  n87 = StaticSwitch(True=NamedRerouteUsage(), False=@n85.R, Value=@n86)
  NamedRerouteDeclaration(None=@n87) {name=Height_L3_Raw, nodeColor=<rgba 1 0 0.8203 1>}
  n88 = TextureObjectParameter() {
   group=4_Layer
   parameterName=Height L4
   samplerType=SAMPLERTYPE_Masks
   sortPriority=0
   texture=&Default_Displacement__H
  }
  n89 = TextureSample(UVs=@n68.UVs Layer 4, Tex=NamedRerouteDeclaration(None=@n88) {name=Height_Tex_L4, nodeColor=<rgba 0 0.3906 1 1>}) {
   samplerSource=SSM_Wrap_WorldGroupSettings
   samplerType=SAMPLERTYPE_Masks
   texture=&Default_Displacement__H
  }
  n90 = StaticBoolParameter() {desc=Enable to use World Space Mapping for layer. Only Z projection available in Layer 4, group=4_layer UV, parameterName=Layer 4 WS Mapping}
  n91 = StaticSwitch(True=NamedRerouteUsage(), False=@n89.R, Value=@n90)
  NamedRerouteDeclaration(None=@n91) {name=Height_L4_Raw, nodeColor=<rgba 1 0 0.8203 1>}
  NamedRerouteDeclaration(None=@n71) {name=Height_L1_Raw, nodeColor=<rgba 1 0 0.8203 1>}
  NamedRerouteDeclaration(None=@n66) {name=Displacement_L1, nodeColor=<rgba 1 0.1172 0 1>}
  NamedRerouteDeclaration(None=@n72) {name=Displacement_L1_Bias, nodeColor=<rgba 1 0.1172 0 1>}
  n92 = ScalarParameter() {defaultValue=3, group=2_Layer_Displacement, parameterName=Displacement L2}
  n93 = TextureObjectParameter() {
   group=2_Layer
   parameterName=Height L2
   samplerType=SAMPLERTYPE_Masks
   sortPriority=0
   texture=&Default_Displacement__H
  }
  n94 = TextureSample(UVs=@n68.UVs Layer 2, Tex=NamedRerouteDeclaration(None=@n93) {name=Height_Tex_L2, nodeColor=<rgba 0 0.3906 1 1>}) {
   samplerSource=SSM_Wrap_WorldGroupSettings
   samplerType=SAMPLERTYPE_Masks
   texture=&Default_Displacement__H
  }
  n95 = StaticSwitch(
   True=NamedRerouteUsage()
   False=@n94.R
   Value=StaticBoolParameter() {desc=Enable to use World Space Mapping for layer, group=2_layer UV, parameterName=Layer 2 WS Mapping}
  )
  n96 = ScalarParameter() {defaultValue=0.5, group=2_Layer_Displacement, parameterName=Displacement_Bias L2}
  n97 = Multiply(A=Divide(A=@n92, B=NamedRerouteUsage()), B=Subtract(A=@n95, B=@n96) {constB=0.5})
  NamedRerouteDeclaration(None=StaticSwitchParameter(True=Add(A=@n97, B=@n96) {constB=0.5}, False=@n95) {group=Displacement, parameterName=Displacement_SmoothBlend}) {name=Height L2, nodeColor=<rgba 1 0.1406 0 1>}
  NamedRerouteDeclaration(None=@n92) {name=Displacement_L2, nodeColor=<rgba 1 0.1172 0 1>}
  NamedRerouteDeclaration(None=@n96) {name=Displacement_L2_Bias, nodeColor=<rgba 1 0.1172 0 1>}
  n98 = ScalarParameter() {defaultValue=3, group=3_Layer_Displacement, parameterName=Displacement L3}
  NamedRerouteDeclaration(None=@n98) {name=Displacement_L3, nodeColor=<rgba 1 0.1172 0 1>}
  n99 = ScalarParameter() {defaultValue=0.5, group=3_Layer_Displacement, parameterName=Displacement_Bias L3}
  NamedRerouteDeclaration(None=@n99) {name=Displacement_L3_Bias, nodeColor=<rgba 1 0.1172 0 1>}
  n100 = ScalarParameter() {defaultValue=3, group=4_Layer_Displacement, parameterName=Displacement L4}
  NamedRerouteDeclaration(None=@n100) {name=Displacement_L4, nodeColor=<rgba 1 0.1172 0 1>}
  MaterialExpressionReroute_14 = Reroute(None=@n4)
  n101 = GetMaterialAttributes(None=@MaterialExpressionReroute_14.Input) {attributeGetTypes [#3, #4, #5]}
  n102 = VectorParameter() {group=4_Layer, parameterName="L4 Tint  (Expand)", channelNames {a=L4 Specular Tweak, b="", g="", r=""}, defaultValue=<rgba 1 1 1 1>}
  n103 = SetMaterialAttributes!(
   MaterialAttributes=@n101.MaterialAttributes
   BaseColor=Multiply(A=@n101.BaseColor, B=@n102)
   Roughness=Multiply(A=@n101.Roughness, B=ScalarParameter() {defaultValue=1, group=4_Layer, parameterName=L4 Roughness Tweak})
   Specular=Multiply(A=@n101.Specular, B=@n102.L4 Specular Tweak)
  ) {attributeSetTypes [#3, #4, #5]}
  MaterialExpressionReroute_15 = Reroute(None=@n103)
  NamedRerouteDeclaration(None=@MaterialExpressionReroute_15.Output) {name=L4, nodeColor=<rgba 0.6172 0 1 1>}
  MaterialExpressionReroute_12 = Reroute(None=@n3)
  n104 = GetMaterialAttributes(None=@MaterialExpressionReroute_12.Input) {attributeGetTypes [#3, #4, #5]}
  n105 = VectorParameter() {group=3_Layer, parameterName="L3 Tint  (Expand)", channelNames {a=L3 Specular Tweak, b="", g="", r=""}, defaultValue=<rgba 1 1 1 1>}
  n106 = SetMaterialAttributes(
   MaterialAttributes=@n104.MaterialAttributes
   BaseColor=Multiply(A=@n104.BaseColor, B=@n105)
   Roughness=Multiply(A=@n104.Roughness, B=ScalarParameter() {defaultValue=1, group=3_Layer, parameterName=L3 Roughness Tweak})
   Specular=Multiply(A=@n104.Specular, B=@n105.L3 Specular Tweak)
  ) {attributeSetTypes [#3, #4, #5]}
  MaterialExpressionReroute_13 = Reroute(None=@n106)
  NamedRerouteDeclaration(None=@MaterialExpressionReroute_13.Output) {name=L3, nodeColor=<rgba 0.6406 1 0 1>}
  MaterialExpressionReroute_21 = Reroute(None=@n2)
  n107 = GetMaterialAttributes(None=@MaterialExpressionReroute_21.Input) {attributeGetTypes [#3, #4, #5]}
  n108 = VectorParameter() {group=1_Layer, parameterName="L1 Tint (Expand)", channelNames {a=L1 Specular Tweak, b="", g="", r=""}, defaultValue=<rgba 1 1 1 1>}
  n109 = SetMaterialAttributes(
   MaterialAttributes=@n107.MaterialAttributes
   BaseColor=Multiply(A=@n107.BaseColor, B=@n108)
   Roughness=Multiply(A=@n107.Roughness, B=ScalarParameter() {defaultValue=1, group=1_Layer, parameterName=L1 Roughness Tweak})
   Specular=Multiply(A=@n107.Specular, B=@n108.L1 Specular Tweak)
  ) {attributeSetTypes [#3, #4, #5]}
  MaterialExpressionReroute_23 = Reroute(None=@n109)
  NamedRerouteDeclaration(None=@MaterialExpressionReroute_23.Output) {name=L1, nodeColor=<rgba 0 0.5078 1 1>}
  MaterialExpressionReroute_10 = Reroute(None=@n1)
  n110 = GetMaterialAttributes(None=@MaterialExpressionReroute_10.Input) {attributeGetTypes [#3, #4, #5]}
  n111 = VectorParameter() {group=2_Layer, parameterName="L2 Tint  (Expand)", channelNames {a=L2 Specular Tweak, b="", g="", r=""}, defaultValue=<rgba 1 1 1 1>}
  n112 = SetMaterialAttributes(
   MaterialAttributes=@n110.MaterialAttributes
   BaseColor=Multiply(A=@n110.BaseColor, B=@n111)
   Roughness=Multiply(A=@n110.Roughness, B=ScalarParameter() {defaultValue=1, group=2_Layer, parameterName=L2 Roughness Tweak})
   Specular=Multiply(A=@n110.Specular, B=@n111.L2 Specular Tweak)
  ) {attributeSetTypes [#3, #4, #5]}
  MaterialExpressionReroute_11 = Reroute(None=@n112)
  NamedRerouteDeclaration(None=@MaterialExpressionReroute_11.Output) {name=L2, nodeColor=<rgba 1 0 0.5859 1>}
  n113 = ScalarParameter() {defaultValue=0.5, group=4_Layer_Displacement, parameterName=Displacement_Bias L4}
  NamedRerouteDeclaration(None=@n113) {name=Displacement_L4_Bias, nodeColor=<rgba 1 0.1172 0 1>}
  n114 = Multiply(A=Divide(A=@n98, B=NamedRerouteUsage()), B=Subtract(A=@n87, B=@n99) {constB=0.5})
  NamedRerouteDeclaration(None=StaticSwitchParameter(True=Add(A=@n114, B=@n99) {constB=0.5}, False=@n87) {group=Displacement, parameterName=Displacement_SmoothBlend}) {name=Height L3, nodeColor=<rgba 1 0.1406 0 1>}
  n115 = VertexColor()
  n116 = TextureCoordinate() {coordinateIndex=1}
  n117 = TextureCoordinate()
  n118 = TextureSample(
   UVs=StaticSwitchParameter(
    True=TextureCoordinate() {coordinateIndex=2}
    False=StaticSwitchParameter(True=@n116, False=@n117) {group=MeshPaint_Blending, parameterName="Use2ndUVchannelforMeshPaint(UV1)"}
   ) {defaultValue=true, group=MeshPaint_Blending, parameterName="Use3rdUVchannelforMeshPaint(UV2)"}
   Tex=MeshPaintTextureObject()
  ) {samplerType=SAMPLERTYPE_LinearColor}
  n119 = ComponentMask(None=TextureCoordinate() {coordinateIndex=2}) {r=true}
  n120 = ComponentMask(None=TextureCoordinate() {coordinateIndex=1}) {r=true}
  n121 = Constant()
  n122 = Constant() {r=1}
  n123 = StaticSwitch(
   True=Constant() {r=1}
   False=If(
    A=Constant() {r=1}
    B=StaticSwitchParameter(True=@n119, False=@n120) {group=2_Layer, parameterName=Use UV2 For Layer 2 Mask, sortPriority=-2}
    "A > B"=@n121
    "A < B"=@n122
   ) {desc=0}
   Value=StaticBoolParameter() {group=2_Layer_Blending, parameterName=Blend Overlay Layer 2, sortPriority=-1}
  )
  MaterialExpressionReroute_6 = Reroute(None=@n123)
  n124 = StaticSwitch(
   True=Constant() {r=1}
   False=If(
    A=Constant() {r=2}
    B=StaticSwitchParameter(True=@n119, False=@n120) {group=3_Layer, parameterName=Use UV2 For Layer 3 Mask, sortPriority=-2}
    "A > B"=@n121
    "A < B"=@n122
   )
   Value=StaticBoolParameter() {group=3_Layer_Blending, parameterName=Blend Overlay Layer 3, sortPriority=-1}
  )
  MaterialExpressionReroute_7 = Reroute(None=@n124)
  n125 = AppendVector(A=@MaterialExpressionReroute_6.Output, B=@MaterialExpressionReroute_7.Output2)
  n126 = StaticSwitch(
   True=Constant() {r=1}
   False=If(
    A=Constant() {r=3}
    B=StaticSwitchParameter(True=@n119, False=@n120) {group=4_Layer, parameterName=Use UV2 For Layer 4 Mask, sortPriority=-2}
    "A > B"=@n121
    "A < B"=@n122
   )
   Value=StaticBoolParameter() {group=3_Layer_Blending, parameterName=Blend Overlay Layer 4, sortPriority=-1}
  )
  MaterialExpressionReroute_8 = Reroute(None=@n126)
  n127 = MaterialFunctionCall(Float3=StaticSwitchParameter(
    True=MeshPaintTextureReplace!(Default=@n115, MeshPaintTexture=@n118.RGB)
    False=AppendVector(A=@n125, B=@MaterialExpressionReroute_8.Output3)
   ) {defaultValue=true, group=MeshPaint_Blending, parameterName=UseMeshPaintMask/Vertex paint}) {materialFunction=&BreakOutFloat3Components}
  NamedRerouteDeclaration(None=@n127.R) {name=MeshPaint Mask L2, nodeColor=<rgba 0 1 0.8594 1>}
  NamedRerouteDeclaration(None=@n127.G) {name=MeshPaint Mask L3, nodeColor=<rgba 0 1 0.8594 1>}
  NamedRerouteDeclaration(None=@n127.B) {name=MeshPaint Mask L4, nodeColor=<rgba 0 1 0.8594 1>}
  n128 = StaticBoolParameter() {
   desc=Enable to use UV2 for mapping. You may need to switch masks for ohter layers To UV3
   group=2_layer UV
   parameterName=Use UV2 for Layer 2
  }
  n129 = Constant3Vector() {constant=<rgba 0 0 -1 0>}
  n130 = ScalarParameter() {group=2_layer UV, parameterName=L2 WP Offset Z, sortPriority=1}
  n131 = MaterialFunctionCall(
   Spec Texture Object=NamedRerouteUsage()
   TextureSize=MaterialFunctionCall(L2 UV2=@n128) {materialFunction=&MF_UVsSet_New}.WSM 2
   ProjectionTransitionContrast=Constant() {r=1}
   World Space Normals=StaticBool()
   WorldPosition=Add(A=WorldPosition().XYZ, B=Multiply(A=@n129, B=@n130))
  ) {materialFunction=&WorldAlignedTextures_Complex}
  n132 = StaticBoolParameter() {desc="World Space XYZ ( Expensive)", group=2_layer UV, parameterName=Layer 2 WSM_XYZ}
  n133 = StaticBoolParameter() {
   desc="World Space Z only, you need appropriate mask to make it visible. Use Z or map your layers accordingly "
   group=2_layer UV
   parameterName=Layer 2 WSM_Z
  }
  NamedRerouteDeclaration(None=StaticSwitch(True=@n131.Z Spec, False=StaticSwitch(True=@n131.XYZ Spec, False=@n131.XYZ Spec, Value=@n132), Value=@n133)) {name=WSM Displ L2, nodeColor=<rgba 0 1 0.9766 1>}
  n134 = Multiply(A=Divide(A=@n100, B=NamedRerouteUsage()), B=Subtract(A=@n91, B=@n113) {constB=0.5})
  NamedRerouteDeclaration(None=StaticSwitchParameter(True=Add(A=@n134, B=@n113) {constB=0.5}, False=@n91) {group=Displacement, parameterName=Displacement_SmoothBlend}) {name=Height L4, nodeColor=<rgba 1 0.1406 0 1>}
  NamedRerouteDeclaration(None=MaterialFunctionCall!(
    MaterialLayer=NamedRerouteUsage()
    BaseMask=NamedRerouteUsage()
    Mask_Use World Z Normal=StaticBoolParameter() {group=2_Layer_Mask, parameterName="== L2_Mask_Use_World Z Normal =="}
    Mask_World Z Min Blend_Blend Max=StaticBoolParameter() {defaultValue=true, group=2_Layer_Mask, parameterName=L2_Mask_Use_World Z Min Blend_Blend Max}
    Mask_VTX Only=StaticBoolParameter() {defaultValue=true, group=2_Layer_Mask, parameterName=L2_Mask_Use_World Z VTX Only}
    Mask_TextureNormalZ_Power=ScalarParameter() {defaultValue=1, group=2_Layer_Mask, parameterName=L2_Mask_Use_World Z - Contrast}
    Mask_TextureNormalZ_Mult=ScalarParameter() {defaultValue=1, group=2_Layer_Mask, parameterName=L2_Mask_Use_World Z - Mult}
   ) {materialFunction=&/MF_WorldZBlend}.Result) {name=LayerMask L2, nodeColor=<rgba 0 1 0.8594 1>}
  NamedRerouteDeclaration(None=MaterialFunctionCall(
    MaterialLayer=NamedRerouteUsage()
    BaseMask=NamedRerouteUsage()
    Mask_Use World Z Normal=StaticBoolParameter() {group=3_Layer_Mask, parameterName="== L3_Mask_Use_World Z Normal =="}
    Mask_World Z Min Blend_Blend Max=StaticBoolParameter() {defaultValue=true, group=3_Layer_Mask, parameterName=L3_Mask_Use_World Z Min Blend_Blend Max}
    Mask_VTX Only=StaticBoolParameter() {defaultValue=true, group=3_Layer_Mask, parameterName=L3_Mask_Use_World Z VTX Only}
    Mask_TextureNormalZ_Power=ScalarParameter() {defaultValue=1, group=3_Layer_Mask, parameterName=L3_Mask_Use_World Z Contrast}
    Mask_TextureNormalZ_Mult=ScalarParameter() {defaultValue=1, group=3_Layer_Mask, parameterName=L3_Mask_Use_World Z Mult}
   ) {materialFunction=&MF_WorldZBlend}.Result) {name=LayerMask L3, nodeColor=<rgba 0 1 0.8594 1>}
  NamedRerouteDeclaration(None=MaterialFunctionCall(
    MaterialLayer=NamedRerouteUsage()
    BaseMask=NamedRerouteUsage()
    Mask_Use World Z Normal=StaticBoolParameter() {group=4_Layer_Mask, parameterName="== L4_Mask_Use_World Z Normal =="}
    Mask_World Z Min Blend_Blend Max=StaticBoolParameter() {defaultValue=true, group=4_Layer_Mask, parameterName=L4_Mask_Use_World Z Min Blend_Blend Max}
    Mask_VTX Only=StaticBoolParameter() {defaultValue=true, group=4_Layer_Mask, parameterName=L4_Mask_Use_World Z VTX Only}
    Mask_TextureNormalZ_Power=ScalarParameter() {defaultValue=1, group=4_Layer_Mask, parameterName=L4_Mask_Use_World Z - Contrast}
    Mask_TextureNormalZ_Mult=ScalarParameter() {defaultValue=1, group=4_Layer_Mask, parameterName=L4_Mask_Use_World Z - Mult}
   ) {materialFunction=&MF_WorldZBlend}.Result) {name=LayerMask L4, nodeColor=<rgba 0 1 0.8594 1>}
  NamedRerouteDeclaration(None=ScalarParameter() {defaultValue=32, desc=Make This the Same as the Material magnitude, group=Displacement, parameterName=Magnitude}) {name=Magnitude, nodeColor=<rgba 0 1 0.5312 1>}
  PinBase!(Output=@n112) {reroutePins [Output {expression=&self:MaterialExpressionReroute_11}]}
  PinBase!() {reroutePins [Input {expression=&self:MaterialExpressionReroute_10}]}
  PinBase() {reroutePins [Input {expression=&self:MaterialExpressionReroute_12}]}
  PinBase!(Output=@n106) {reroutePins [Output {expression=&self:MaterialExpressionReroute_13}]}
  PinBase!() {reroutePins [Input {expression=&self:MaterialExpressionReroute_14}]}
  PinBase!(Output=@n103) {reroutePins [Output {expression=&self:MaterialExpressionReroute_15}]}
  PinBase!() {reroutePins [Input {expression=&self:MaterialExpressionReroute_21}]}
  PinBase!(Output=@n109) {reroutePins [Output {expression=&self:MaterialExpressionReroute_23}]}
  PinBase!()
  PinBase!(Output=@n123, Output2=@n124, Output3=@n126) {reroutePins [
    Output {expression=&self:MaterialExpressionReroute_6}
    Output2 {expression=&self:MaterialExpressionReroute_7}
    Output3 {expression=&self:MaterialExpressionReroute_8}
   ]}
  Reroute(None=Constant4Vector() {constant=<rgba 1 0 0 0>}.RGBA)
  n135 = StaticBoolParameter() {
   desc=Uses the alpha channel of the diffuse to do a height-blend of XY or XYZ
   group=1_layer UV
   parameterName=L1 WSM_HeightLerpUsingAlpha
   sortPriority=100
  }
  Reroute(None=@n135)
  Reroute(None=ScalarParameter() {group=1_layer UV, parameterName=L1 WSM_HeightLerpUsingAlpha_Contrast, sortPriority=101})
  Reroute(None=@n135)
  Reroute(None=ScalarParameter() {defaultValue=1, group=1_layer UV, parameterName=L1 WSM_ProjectionTransitionContrast, sortPriority=40})
  Saturate(None=@n16.Alpha)
  StaticSwitchParameter(True=@n59, False=FunctionInput() {inputName=Material In, inputType=FunctionInput_MaterialAttributes, sortPriority=1}) {group=Displacement, parameterName=Use Displacement}
 ]}
