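# Material function that projects a blockout grid texture onto vertical faces in object local space, as a debug scale grid.
# Local position, scaled to a fixed tile size, drives XZ and YZ samples, blended by a contrast-adjusted local normal X.
# A saturated local Y offset from the actor forms a wall-side mask that flips horizontal UV direction.
# Static switches add an optional interior tint and a four-channel export.
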MaterialFunction /Game/Global/Base4Layers_2022/MaterialFunctions/MF_LocalSpaceDebugGrid
graph {nodes [
  Constant() {r=-1}
  n1 = WorldPosition()
  n2 = Constant3Vector() {constant=<rgba 512 512 512 0>}
  n3 = Constant3Vector() {constant=<rgba -1 -1 -1 0>}
  n4 = Constant3Vector() {constant=<rgba 1 1 -1 0>}
  n5 = Subtract(A=WorldPosition().XYZ, B=ActorPositionWS())
  n6 = Saturate(None=ComponentMask(None=Transform(None=@n5) {transformSourceType=TRANSFORMSOURCE_World, transformType=TRANSFORM_Local}) {g=true})
  n7 = NamedRerouteDeclaration(None=@n6) {name=Local Wall Mask, nodeColor=<rgba 0.6172 1 0 1>}
  n8 = Divide(
   A=TransformPosition(None=FunctionInput(Preview=@n1.XYZ) {bUsePreviewValueAsDefault=true, inputName=World Position}) {transformSourceType=TRANSFORMPOSSOURCE_World}
   B=Multiply(A=Abs(None=@n2), B=LinearInterpolate(A=@n3, B=@n4, Alpha=@n7)) {constB=-1}
  )
  n9 = FunctionInput(Preview=TextureObject() {texture=/Game/Global/Base4Layers_2022/Textures/T_Blockout_Grid_Scale}) {inputName=TextureObject, inputType=FunctionInput_Texture2D}
  n10 = TextureSample(UVs=ComponentMask(None=@n8) {b=true, r=true}, Tex=@n9) {samplerSource=SSM_Wrap_WorldGroupSettings}
  n11 = FunctionInput(Preview=StaticBool()) {bUsePreviewValueAsDefault=true, description="Default = False", inputName=Export Float 4, inputType=FunctionInput_StaticBool}
  n12 = TextureSample(UVs=ComponentMask(None=@n8) {b=true, g=true}, Tex=@n9) {samplerSource=SSM_Wrap_WorldGroupSettings}
  n13 = Constant3Vector() {constant=<rgba 0 0 1 0>}
  n14 = Abs(None=ComponentMask(None=Transform(None=@n13) {transformType=TRANSFORM_Local}) {r=true})
  n15 = Constant() {r=1}
  n16 = LinearInterpolate(
   A=StaticSwitch(True=AppendVector(A=@n10.RGB, B=@n10.RGB), False=@n10.RGB, Value=@n11)
   B=StaticSwitch(True=AppendVector(A=@n12.RGB, B=@n12.RGB), False=@n12.RGB, Value=@n11)
   Alpha=ComponentMask(None=MaterialFunctionCall(In=@n14, Contrast=@n15) {materialFunction=&CheapContrast}.Result) {r=true}
  ) {desc=Float3}
  n17 = Multiply(A=FunctionInput(Preview=Constant3Vector() {constant=<rgba 1 1 1 0>}) {bUsePreviewValueAsDefault=true, inputName=Interior Tint}, B=@n16)
  n18 = NamedRerouteUsage()
  n19 = StaticBool()
  FunctionOutput(None=StaticSwitch(
    True=LinearInterpolate(A=@n17, B=@n16, Alpha=@n18)
    False=@n16
    Value=FunctionInput(Preview=@n19) {bUsePreviewValueAsDefault=true, inputName=Interior Tint true, inputType=FunctionInput_StaticBool}
   )) {outputName=XY}
  Reroute(None=@n6)
  VertexNormalWS()
 ]}
