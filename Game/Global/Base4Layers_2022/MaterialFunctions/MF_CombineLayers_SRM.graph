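# Material function that builds one PBR layer's material attributes from diffuse, normal and packed specular/roughness/metallic textures.
# Dot products select SFX texture channels to mask tint, emissive, opacity, subsurface and clearcoat, blending the latter two as shading-model overrides.
# Static switches choose UV or world-aligned projection and Nanite-specific texture sets, and can output only diffuse or normals.
# Debug parameters add UV checker, blockout grid and character-scale overlays.
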MaterialFunction /Game/Global/Base4Layers_2022/MaterialFunctions/MF_CombineLayers_SRM
graph {nodes [
  n1 = Constant() {r=0.5}
  n2 = StaticBoolParameter() {desc="Enable for Color tint via BP.  Used ONLY for BP color tint.", group=0_SFX BP, parameterName="SFX Layers (Tint)", sortPriority=-3}
  n3 = StaticSwitch(True=NamedRerouteUsage(), False=NamedRerouteUsage(), Value=@n2)
  n4 = SetMaterialAttributes!(
   MaterialAttributes
   BaseColor=@n3
   Normal=Constant3Vector() {constant=<rgba 0 0 1 0>}
   Specular=@n1
   Roughness=@n1
   EmissiveColor=NamedRerouteUsage()
  ) {attributeSetTypes [
    #1
    #2
    #3
    #4
    #5
   ]}
  n5 = SetMaterialAttributes!(MaterialAttributes, Normal=NamedRerouteUsage()) {attributeSetTypes [#2]}
  n6 = NamedRerouteUsage()
  n7 = NamedRerouteUsage()
  n8 = ShadingModel() {shadingModel=MSM_ClearCoat}
  n9 = NamedRerouteDeclaration(None=NamedRerouteUsage()) {name=SFX Path, nodeColor=<rgba 1 0.4453 0 1>}
  n10 = NamedRerouteUsage()
  n11 = NamedRerouteUsage()
  n12 = StaticSwitch(
   True=ShadingModel() {shadingModel=MSM_TwoSidedFoliage}
   False=ShadingModel() {shadingModel=MSM_Subsurface}
   Value=FunctionInput(Preview=StaticBool()) {bUsePreviewValueAsDefault=true, inputName=SFX SSS Two Sided Foliange, inputType=FunctionInput_StaticBool}
  )
  n13 = FunctionInput() {inputName=SFX SSS, inputType=FunctionInput_StaticBool}
  n14 = StaticSwitch(
   True=BlendMaterialAttributes(
    A=@n9
    B=SetMaterialAttributes!(MaterialAttributes=@n9, Opacity=@n10, SubsurfaceColor=@n11, ShadingModel=@n12) {attributeSetTypes [#6, #7, #8]}
    Alpha=NamedRerouteUsage()
   )
   False=NamedRerouteUsage()
   Value=NamedRerouteDeclaration(None=@n13) {name=SFX SSS, nodeColor=<rgba 0 1 0.8828 1>}
  )
  n15 = StaticSwitch(
   True=BlendMaterialAttributes(
    A=@n14
    B=SetMaterialAttributes!(MaterialAttributes=@n14, CustomData0=@n6, CustomData1=@n7, ShadingModel=@n8) {attributeSetTypes [#9, #10, #8]}
    Alpha=NamedRerouteUsage()
   )
   False=@n14
   Value=FunctionInput() {bUsePreviewValueAsDefault=true, inputName=SFX Clearcoat, inputType=FunctionInput_StaticBool}
  )
  n16 = StaticBool()
  n17 = FunctionInput(Preview=@n16) {bUsePreviewValueAsDefault=true, inputName=Output Normals Only, inputType=FunctionInput_StaticBool, sortPriority=513}
  FunctionOutput(None=StaticSwitch(
    True=@n4
    False=StaticSwitch(True=@n5, False=@n15, Value=@n17)
    Value=FunctionInput(Preview=@n16) {bUsePreviewValueAsDefault=true, inputName=Output Diffuse Only, inputType=FunctionInput_StaticBool, sortPriority=512}
   ))
  n18 = Constant3Vector() {constant=<rgba 1 1 1 0>}
  FunctionOutput(None=NamedRerouteDeclaration(None=FunctionInput(Preview=@n18) {bUsePreviewValueAsDefault=true, inputName=SFX Texture Complete}) {name=SFX Tex, nodeColor=<rgba 0 0.4609 1 1>}) {outputName=SFX}
  n19 = FunctionInput() {desc=Works when MI changed to CC, inputName="Clearcoat (R) CC Roughness (G)", sortPriority=7, previewValue=<1 0 0 0>}
  n20 = MaterialFunctionCall(Float3=@n19) {materialFunction=&BreakOutFloat3Components}
  NamedRerouteDeclaration(None=Multiply(A=@n20.R, B=NamedRerouteUsage())) {name=Clearcoat, nodeColor=<rgba 0.4297 1 0 1>}
  n21 = NamedRerouteUsage()
  n22 = FunctionInput() {bUsePreviewValueAsDefault=true, inputName=Normals, inputType=FunctionInput_Texture2D, sortPriority=4, previewValue=<0 0 1 1>}
  n23 = TextureSample(UVs=@n21, Tex=NamedRerouteDeclaration(None=@n22) {name=Normals TexObj, nodeColor=<rgba 0 0.0156 1 1>}) {samplerSource=SSM_Wrap_WorldGroupSettings, samplerType=SAMPLERTYPE_Normal}
  n24 = FunctionInput() {bUsePreviewValueAsDefault=true, inputName=Normals Power, inputType=FunctionInput_Scalar, sortPriority=5, previewValue=<1 0.5 0.5 1>}
  n25 = NamedRerouteDeclaration(None=@n24) {name=BoostNormals, nodeColor=<rgba 0 0.9766 1 1>}
  n26 = AppendVector(A=ComponentMask(None=Multiply(A=@n23.RGB, B=@n25)) {g=true, r=true}, B=@n23.B)
  n27 = FunctionInput() {bUsePreviewValueAsDefault=true, inputName=Normals_NN, inputType=FunctionInput_Texture2D, sortPriority=-20, previewValue=<0 0 1 1>}
  n28 = TextureSample(UVs=@n21, Tex=@n27) {samplerSource=SSM_Wrap_WorldGroupSettings, samplerType=SAMPLERTYPE_Normal}
  n29 = ComponentMask(None=Multiply(A=@n25, B=@n28.RGB)) {g=true, r=true}
  n30 = StaticSwitchParameter(True=NaniteReplace!(Default=@n26, Nanite=AppendVector(A=@n29, B=@n28.B)), False=@n26) {group=0_Nanite, parameterName=Use Nanite Texture Sets, sortPriority=-25}
  n31 = Transform(None=MaterialFunctionCall!(UVs=NamedRerouteUsage(), Normal=@n30) {materialFunction=/Engine/Functions/Engine_MaterialFunctions01/Texturing/DeriveTangentBasis}.World Space Normal) {transformSourceType=TRANSFORMSOURCE_World, transformType=TRANSFORM_Tangent}
  n32 = FunctionInput(Preview=StaticBool()) {bUsePreviewValueAsDefault=true, inputName=UV2 True, inputType=FunctionInput_StaticBool, sortPriority=-99}
  n33 = FunctionInput(Preview=StaticBool()) {bUsePreviewValueAsDefault=true, inputName=Use WS Mapping, inputType=FunctionInput_StaticBool, sortPriority=-103}
  NamedRerouteDeclaration(None=StaticSwitch(True=NamedRerouteUsage(), False=StaticSwitch(True=@n31, False=@n30, Value=@n32), Value=@n33)) {name=Normals, nodeColor=<rgba 0 1 0.0625 1>}
  n34 = NamedRerouteUsage()
  NamedRerouteDeclaration(None=DotProduct(A=@n34, B=FunctionInput() {inputName=SFX Tint Channel, sortPriority=1, previewValue=<1 0 0 0>})) {name=Tint Mask, nodeColor=<rgba 0.4766 1 0 1>}
  n35 = NamedRerouteUsage()
  n36 = TextureObject() {texture=/Game/FN_PublicTurntable/Turntable_Lowpoly/Content/BlueprintAssets/Checker}
  n37 = FunctionInput() {bUsePreviewValueAsDefault=true, inputName=Diffuse, inputType=FunctionInput_Texture2D, previewValue=<0.5 0.5 0.5 1>}
  n38 = TextureSample(
   UVs=@n35
   Tex=NamedRerouteDeclaration(None=StaticSwitchParameter(True=@n36, False=@n37) {group=0_Debug, parameterName=Check Mesh UVs, sortPriority=11}) {name=Diffuse TexObj, nodeColor=<rgba 0 1 0.3438 1>}
  ) {samplerSource=SSM_Wrap_WorldGroupSettings}
  n39 = FunctionInput() {bUsePreviewValueAsDefault=true, inputName=Diffuse_NN, inputType=FunctionInput_Texture2D, sortPriority=-20, previewValue=<0.5 0.5 0.5 1>}
  n40 = TextureSample(UVs=@n35, Tex=StaticSwitchParameter(True=@n36, False=@n39) {group=0_Debug, parameterName=Check Mesh UVs, sortPriority=11}) {samplerSource=SSM_Wrap_WorldGroupSettings}
  n41 = StaticSwitch(
   True=NamedRerouteUsage()
   False=StaticSwitchParameter(True=NaniteReplace(Default=@n38.RGB, Nanite=@n40.RGB), False=@n38.RGB) {group=0_Nanite, parameterName=Use Nanite Texture Sets, sortPriority=-25}
   Value=@n33
  )
  n42 = StaticSwitchParameter(True=Multiply(A=NamedRerouteUsage(), B=@n41), False=@n41) {group=0_Debug, parameterName=Use WB_Grid, sortPriority=1}
  n43 = Multiply(A=@n42, B=VectorParameter() {group=1_Layer SFX, parameterName=Mobile SSS Boost, defaultValue=<rgba 1 1 1 1>}.RGB)
  n44 = NamedRerouteUsage()
  NamedRerouteDeclaration(None=ShadingPathSwitch(Default=@n42, Mobile=StaticSwitch(True=@n43, False=@n42, Value=@n44))) {name=Diff RGB, nodeColor=<rgba 0.0078 1 0 1>}
  n45 = NamedRerouteUsage()
  n46 = NamedRerouteUsage()
  NamedRerouteDeclaration(None=LinearInterpolate(A=@n45, B=Multiply(A=@n46, B=@n45), Alpha=NamedRerouteUsage())) {name=Diffuse Tinted, nodeColor=<rgba 0.0312 1 0 1>}
  n47 = FunctionInput() {bUsePreviewValueAsDefault=true, inputName=Data Ramp RGB, sortPriority=14, previewValue=<1 1 1 1>}
  NamedRerouteDeclaration(None=@n47) {name=Diffuse Tint, nodeColor=<rgba 1 0.3984 0 1>}
  n48 = FunctionInput(Preview=TextureCoordinate()) {bUsePreviewValueAsDefault=true, inputName=UV's, inputType=FunctionInput_Vector2, sortPriority=-59}
  n49 = ScalarParameter() {defaultValue=1, group=0_Debug, parameterName=Tile Mesh UV, sortPriority=12}
  NamedRerouteDeclaration(None=StaticSwitchParameter(True=Multiply(A=@n48, B=@n49), False=@n48) {group=0_Debug, parameterName=Check Mesh UVs, sortPriority=11}) {name=Uv's, nodeColor=<rgba 0.3125 0 1 1>}
  n50 = NamedRerouteUsage()
  n51 = FunctionInput() {inputName=Emissive Tint, inputType=FunctionInput_Vector4, sortPriority=13}
  n52 = NamedRerouteUsage()
  n53 = Constant() {r=1}
  n54 = StaticBoolParameter() {desc="Enable for Color tint via BP.  Used ONLY for BP color tint.", group=0_SFX BP, parameterName="SFX Layers (Tint)", sortPriority=-3}
  n55 = StaticSwitch(
   True=LinearInterpolate(
    A=@n51
    B=AppendVector(A=@n52, B=@n53)
    Alpha=ScalarParameter() {group=0_SFX BP, parameterName=Use SFX Layers as Emissive Tint, sortPriority=100}
   )
   False=@n51
   Value=@n54
  )
  n56 = ComponentMask(None=@n55) {b=true, g=true, r=true}
  NamedRerouteDeclaration(None=Multiply(A=@n50, B=@n56)) {name=Emissive Color Mult, nodeColor=<rgba 1 0 0.3281 1>}
  NamedRerouteDeclaration(None=@n20.G) {name=Clearcoat Roughness, nodeColor=<rgba 0 0.8594 1 1>}
  n57 = VectorParameter() {group=0_Debug, parameterName=Interior Tint, sortPriority=6, defaultValue=<rgba 0.9618 0.8133 0.6311 1>}
  n58 = TextureObjectParameter() {
   desc=A channel for character scale
   group=0_Debug
   parameterName=BlockoutGrid
   sortPriority=4
   texture=/Game/Global/Base4Layers_2022/Textures/T_Blockout_Grid_Scale
  }
  n59 = MaterialFunctionCall!(
   Export Float 4
   World Position
   Interior Tint=@n57.RGB
   Interior Tint true=StaticBoolParameter() {group=0_Debug, parameterName=Interior Tint True, sortPriority=5}
   TextureObject=@n58
  ) {materialFunction=&/MF_LocalSpaceDebugGrid}
  n60 = WorldPosition()
  n61 = Subtract(A=@n60.XYZ, B=Multiply(A=Constant3Vector() {constant=<rgba 0 0 1 0>}, B=Constant() {r=12}))
  n62 = StaticBoolParameter() {group=0_Debug, parameterName=Add Character Floor Offset, sortPriority=4}
  n63 = Constant3Vector() {constant=<rgba 1 1 0 0>}
  n64 = ScalarParameter() {group=0_Debug, parameterName=CharacterI Pivot Offset, sortPriority=3}
  n65 = MaterialFunctionCall(
   Export Float 4=StaticBool() {value=true}
   World Position=Add(A=StaticSwitch(True=@n61, False=@n60.XYZ, Value=@n62), B=Multiply(A=@n63, B=@n64))
   TextureObject=@n58
  ) {materialFunction=&MF_LocalSpaceDebugGrid}
  n66 = VertexNormalWS()
  n67 = Saturate(None=ComponentMask(None=TransformPosition(None=@n66) {transformSourceType=TRANSFORMPOSSOURCE_World}) {b=true})
  n68 = Multiply(A=ComponentMask(None=@n59.XY) {b=true, g=true, r=true}, B=ComponentMask(None=LinearInterpolate(A=@n65.XY, Alpha=@n67)) {a=true})
  n69 = StaticBoolParameter() {group=0_Debug, parameterName=Add Character Scale Check, sortPriority=2}
  NamedRerouteDeclaration(None=ShadingPathSwitch(Default=StaticSwitch(True=@n68, False=@n59.XY, Value=@n69), Mobile=Constant3Vector() {constant=<rgba 1 1 1 0>})) {name=WS Grid Texture, nodeColor=<rgba 0 0.7891 1 1>}
  n70 = NamedRerouteUsage()
  NamedRerouteDeclaration(None=DotProduct(A=@n70, B=FunctionInput() {inputName="SFX Emissive  Channel", sortPriority=13, previewValue=<1 0 0 0>})) {name=Emissive Mask, nodeColor=<rgba 0.4766 1 0 1>}
  n71 = NamedRerouteUsage()
  n72 = Multiply(A=@n71, B=@n50)
  n73 = NamedRerouteDeclaration(None=ComponentMask(None=@n55) {a=true}) {name=Emissive A, nodeColor=<rgba 0 1 0.2734 1>}
  NamedRerouteDeclaration(None=Multiply(A=LinearInterpolate(A=@n71, B=@n72, Alpha=@n73), B=@n56)) {name=Emissive Color SFX, nodeColor=<rgba 1 0 0.3281 1>}
  NamedRerouteDeclaration(None=DotProduct(A=@n70, B=FunctionInput() {inputName=Clearcoat Mask Channal, sortPriority=7, previewValue=<1 0 0 0>})) {name=Clearcoat Mask, nodeColor=<rgba 0.4766 1 0 1>}
  NamedRerouteDeclaration(None=DotProduct(A=FunctionInput() {inputName="SFX  SSS Channel", sortPriority=6, previewValue=<1 0 0 0>}, B=@n34)) {name=SSS Mask, nodeColor=<rgba 0.4766 1 0 1>}
  n74 = SetMaterialAttributes!(
   MaterialAttributes
   BaseColor=@n3
   Metallic=NamedRerouteUsage()
   Specular=NamedRerouteUsage()
   Roughness=NamedRerouteUsage()
   Normal=NamedRerouteUsage()
   ShadingModel=ShadingModel()
  ) {attributeSetTypes [
    #1
    #11
    #3
    #4
    #2
    #8
   ]}
  n75 = StaticSwitchParameter(
   True=SetMaterialAttributes!(MaterialAttributes=@n74, EmissiveColor=NamedRerouteUsage()) {attributeSetTypes [#5]}
   False=@n74
  ) {
   desc="Emissive RGB Available if SFX RGB is activated in 0 SFX Global. Will replace layers Emissive unless Emissive RGB or Emissive Texture Add is set= true  ( adds texture sample)"
   group=A_Emissive_Texture
   parameterName=SFX Emissive Texture RGB
   sortPriority=-32
  }
  n76 = StaticSwitchParameter(True=@n75, False=@n74) {
   desc="Enable to access various emissive functions, TOD,Fresnel, Flipbook etc."
   group=A__EmissiveFilters
   parameterName=Emissive Filters
   sortPriority=-9
  }
  n77 = FunctionInput(Preview=StaticBool()) {
   bUsePreviewValueAsDefault=true
   desc="Enables controls for masked inputs (Emissive, "
   inputName=Use SFX Emissive
   inputType=FunctionInput_StaticBool
   sortPriority=-5
   previewValue=<1 0.5 0.5 1>
  }
  n78 = StaticSwitch(
   True=SetMaterialAttributes(MaterialAttributes=@n74, EmissiveColor=NamedRerouteUsage()) {attributeSetTypes [#5]}
   False=@n76
   Value=@n77
  )
  n79 = NamedRerouteUsage()
  NamedRerouteDeclaration(None=StaticSwitch(
    True=SetMaterialAttributes!(MaterialAttributes=@n78, OpacityMask=@n79) {attributeSetTypes [#12]}
    False=@n78
    Value=FunctionInput() {inputName=SFX Opacity Mask, inputType=FunctionInput_StaticBool}
   )) {name=No SFX path, nodeColor=<rgba 0.3359 0 1 1>}
  n80 = NamedRerouteUsage()
  n81 = FunctionInput() {bUsePreviewValueAsDefault=true, inputName=SpecularMasks, inputType=FunctionInput_Texture2D, sortPriority=3, previewValue=<0.5 0.5 0.5 1>}
  n82 = TextureSample(UVs=@n80, Tex=NamedRerouteDeclaration(None=@n81) {name=Spec Tex Obj, nodeColor=<rgba 0 0.4609 1 1>}) {samplerSource=SSM_Wrap_WorldGroupSettings, samplerType=SAMPLERTYPE_Masks}
  n83 = FunctionInput() {
   bUsePreviewValueAsDefault=true
   inputName=SpecularMasks_NN
   inputType=FunctionInput_Texture2D
   sortPriority=-20
   previewValue=<0.5 0.5 0.5 1>
  }
  n84 = NaniteReplace(Default=@n82.RGB, Nanite=TextureSample(UVs=@n80, Tex=@n83) {samplerSource=SSM_Wrap_WorldGroupSettings, samplerType=SAMPLERTYPE_Masks}.RGB)
  n85 = MaterialFunctionCall(Float3=StaticSwitch(
    True=NamedRerouteUsage()
    False=StaticSwitchParameter(True=@n84, False=@n82.RGB) {group=0_Nanite, parameterName=Use Nanite Texture Sets, sortPriority=-25}
    Value=@n33
   )) {materialFunction=&BreakOutFloat3Components}
  NamedRerouteDeclaration(None=@n85.R) {name=S, nodeColor=<rgba 1 0 0.0469 1>}
  n86 = FunctionInput() {bUsePreviewValueAsDefault=true, inputName=WSM Scale, inputType=FunctionInput_Scalar, sortPriority=-105, previewValue=<512 0 0 1>}
  n87 = FunctionInput(Preview=Constant() {r=1}) {bUsePreviewValueAsDefault=true, inputName=WSM_ProjectionTransitionContrast, inputType=FunctionInput_Scalar, sortPriority=-100}
  n88 = FunctionInput(Preview=Constant()) {bUsePreviewValueAsDefault=true, inputName=WSM_HeightLerpUsingAlpha_Contrast, inputType=FunctionInput_Scalar, sortPriority=-98}
  n89 = Constant3Vector() {constant=<rgba 0 0 -1 0>}
  n90 = FunctionInput() {bUsePreviewValueAsDefault=true, inputName=World Position Z, inputType=FunctionInput_Scalar}
  n91 = MaterialFunctionCall!(
   Diffuse Texture Object=NamedRerouteUsage()
   Spec Texture Object=NamedRerouteUsage()
   TextureSize=@n86
   --------------------------
   Use seperate settings for Z projection
   Use seperate texture for Z projection
   Texture Size Z projection
   Z Plane Diffuse Texture Object
   Z Plane Spec Texture Object
   "--------------------------      "
   ProjectionTransitionContrast=@n87
   Use Texture Alpha Heightmap to Refine the Transistion=FunctionInput(Preview=StaticBool()) {bUsePreviewValueAsDefault=true, inputName=WSM_HeightLerpUsingAlpha, inputType=FunctionInput_StaticBool, sortPriority=-99}
   HeightLerpTransitionContrast=@n88
   PickTheDiffuseHeightChannel
   TheDiffuseHeightChannel
   "--------------------------       "
   NormalTextureObject=NamedRerouteUsage()
   World Space Normals=StaticBool()
   "--------------------------  "
   Export Float 4 Diffuse
   Export Float 4 Spec
   "--------------------------     "
   WorldPosition=Add(A=WorldPosition().XYZ, B=Multiply(A=@n89, B=@n90))
   World Space Normal
  ) {materialFunction=/Engine/Functions/Engine_MaterialFunctions02/Texturing/WorldAlignedTextures_Complex}
  n92 = FunctionInput(Preview=StaticBool()) {bUsePreviewValueAsDefault=true, inputName=Use XYZ, inputType=FunctionInput_StaticBool, sortPriority=-101}
  n93 = FunctionInput(Preview=StaticBool()) {bUsePreviewValueAsDefault=true, inputName=Use Z, inputType=FunctionInput_StaticBool, sortPriority=-102}
  n94 = StaticSwitch(
   True=NamedRerouteDeclaration(None=@n91.Normal Z Texture) {name=Normal Z, nodeColor=<rgba 0 0.6719 1 1>}
   False=StaticSwitch(
    True=NamedRerouteDeclaration(None=@n91.Normal XYZ Texture) {name=Normal XYZ, nodeColor=<rgba 0 0.6719 1 1>}
    False=NamedRerouteDeclaration(None=@n91.Normal XY Texture) {name=Normal XY, nodeColor=<rgba 0 0.6719 1 1>}
    Value=@n92
   )
   Value=@n93
  )
  n95 = Multiply(A=NamedRerouteUsage(), B=@n94)
  NamedRerouteDeclaration(None=AppendVector(A=ComponentMask(None=@n95) {g=true, r=true}, B=ComponentMask(None=@n94) {b=true})) {name=WSM Normals, nodeColor=<rgba 0 1 0.1562 1>}
  NamedRerouteDeclaration(None=StaticSwitch(True=@n91.Z Spec, False=StaticSwitch(True=@n91.XYZ Spec, False=@n91.XYZ Spec, Value=@n92), Value=@n93)) {name=WSM Spec, nodeColor=<rgba 0 1 0.9766 1>}
  NamedRerouteDeclaration(None=StaticSwitch(True=@n91.Z Diffuse Texture, False=StaticSwitch(True=@n91.XYZ Diffuse Texture, False=@n91.XYZ Diffuse Texture, Value=@n92), Value=@n93)) {name=WSM Diffuse, nodeColor=<rgba 0.3359 0 1 1>}
  NamedRerouteDeclaration(None=@n85.B) {name=M, nodeColor=<rgba 1 0 0.0469 1>}
  NamedRerouteDeclaration(None=@n85.G) {name=R, nodeColor=<rgba 0.5234 0 1 1>}
  NamedRerouteDeclaration(None=@n56) {name=Emissive Color Clear, nodeColor=<rgba 1 0 0.3281 1>}
  n96 = FunctionInput() {
   bUsePreviewValueAsDefault=true
   desc=Works when MI changed to any SSS materials
   inputName=SSS Color Value
   inputType=FunctionInput_Vector4
   sortPriority=6
   previewValue=<1 0.5 0.5 1>
  }
  n97 = ComponentMask(None=@n96) {a=true}
  NamedRerouteDeclaration(None=Saturate(None=Add(A=@n97) {constB=0.05})) {name=SSS Depth, nodeColor=<rgba 0 1 0.6484 1>}
  n98 = ComponentMask(None=@n96) {b=true, g=true, r=true}
  n99 = NamedRerouteUsage()
  NamedRerouteDeclaration(None=Multiply(A=@n41, B=LinearInterpolate(B=@n98, Alpha=@n99))) {name=SSS Color, nodeColor=<rgba 0.9922 1 0 1>}
  n100 = FunctionInput() {inputName=SFX Opacilty Mask Channel, sortPriority=8, previewValue=<1 0 0 0>}
  n101 = FunctionInput() {inputName=Opacilty Mask Cancel, inputType=FunctionInput_Scalar, sortPriority=9, previewValue=<0 0 0 0>}
  NamedRerouteDeclaration(None=Max(A=DotProduct(A=@n100, B=@n70), B=@n101)) {name=Opacity Mask, nodeColor=<rgba 0.4766 1 0 1>}
  Reroute(None=@n96)
  Reroute(None=@n13)
 ]}
