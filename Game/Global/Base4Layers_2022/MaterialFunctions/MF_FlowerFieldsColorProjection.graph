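# Material function that projects a flower field color texture onto surfaces from world XY position, outputting the tinted color and its UVs.
# UVs come from a field origin and size, with a static switch choosing between custom primitive data and material parameters.
# Pixel depth drives two brightness fades, a nearer one and a farther skydive one.
# The skydive fade also blends desaturation between near and far amounts.
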MaterialFunction /Game/Global/Base4Layers_2022/MaterialFunctions/MF_FlowerFieldsColorProjection
graph {nodes [
  n1 = PixelDepth()
  n2 = ScalarParameter() {defaultValue=20000, group=Transition, parameterName=Skydive_FlowerField_Blend_offset}
  n3 = Saturate(None=Divide(A=Subtract(A=@n1.R, B=@n2), B=ScalarParameter() {defaultValue=20000, group=Transition, parameterName=SkyDive_FlowerField_Blend_fade}))
  n4 = LinearInterpolate(B=ScalarParameter() {defaultValue=0.76, group=Transition, parameterName=Brightness_SkyDive_FlowerField_Far}, Alpha=@n3) {constA=1}
  n5 = Divide(
   A=Subtract(A=@n1.R, B=ScalarParameter() {defaultValue=2000, group=Transition, parameterName=Brightness_FlowerField_Blend_offset})
   B=ScalarParameter() {defaultValue=5000, group=Transition, parameterName=Brightness_FlowerField_Blend_fade}
  )
  n6 = ComponentMask(None=VectorParameter() {bUseCustomPrimitiveData=true, parameterName=Origin Prim, defaultValue=<rgba 0 0 0 1>}.RGB) {g=true, r=true}
  n7 = ComponentMask(None=VectorParameter() {parameterName=FlowerField_Origin, defaultValue=<rgba 40384.3281 40768.3281 -9551.6719 1>}.RGB) {g=true, r=true}
  n8 = ScalarParameter() {bUseCustomPrimitiveData=true, parameterName=Size Prim, primitiveDataIndex=4}
  n9 = VectorParameter() {parameterName=FlowerField_Size, defaultValue=<rgba 29407.3438 29407.3438 29407.3438 1>}
  n10 = Divide(
   A=Subtract(A=WorldPosition().XY, B=StaticSwitchParameter(True=@n6, False=@n7) {parameterName=UseFlowerFieldPrimData})
   B=StaticSwitchParameter(True=AppendVector(A=@n8, B=@n8), False=ComponentMask(None=@n9.RGB) {g=true, r=true}) {parameterName=UseFlowerFieldPrimData}
  )
  n11 = Multiply(
   A=LinearInterpolate(
    A=ScalarParameter() {defaultValue=1, group=Transition, parameterName=Brightness_FlowerField_Near}
    B=ScalarParameter() {defaultValue=0.6, group=Transition, parameterName=Brightness_FlowerField_Far}
    Alpha=Saturate(None=@n5)
   )
   B=TextureSampleParameter2D(UVs=@n10) {parameterName=FlowerFieldColorProjection, texture=/Engine/EngineResources/WhiteSquareTexture}.RGB
  )
  n12 = ScalarParameter() {group=Transition, parameterName=Desat_FlowerField_Near}
  n13 = ScalarParameter() {defaultValue=0.416, group=Transition, parameterName=Desat_FlowerField_Far}
  FunctionOutput(None=Desaturation(None=Multiply(A=@n4, B=@n11), Fraction=LinearInterpolate(A=@n12, B=@n13, Alpha=@n3)))
  FunctionOutput(None=@n10) {outputName=UVs, sortPriority=4}
 ]}
