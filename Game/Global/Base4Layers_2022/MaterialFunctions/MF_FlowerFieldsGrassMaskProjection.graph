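# Material function that projects the FlowerGrassMask texture onto flower fields, giving per-type flower grass masks.
# UVs from MF_FlowerFieldsColorProjection are clamped to sample the texture, which a ScaleUVsByCenter 0-1 mask blends toward black.
# The red, green and blue channels output as FlowerType1, FlowerType2 and FlowerType3.
# The Combined output saturates red added twice plus blue, so it leaves out green.
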MaterialFunction /Game/Global/Base4Layers_2022/MaterialFunctions/MF_FlowerFieldsGrassMaskProjection
graph {nodes [
  n1 = MaterialFunctionCall() {materialFunction=&/MF_FlowerFieldsColorProjection}
  n2 = Clamp(None=@n1.UVs)
  n3 = MaterialFunctionCall(Float3=LinearInterpolate(
    A=TextureSample(UVs=@n2) {samplerType=SAMPLERTYPE_LinearColor, texture=/Hermes/Techart/FlowerFields/Blueprint/FlowerGrassMask}.RGB
    B=Constant3Vector()
    Alpha=MaterialFunctionCall(Texture Scale=@n1.UVs) {materialFunction=&ScaleUVsByCenter}.0-1 mask
   )) {materialFunction=&BreakOutFloat3Components}
  n4 = Add(A=@n3.R, B=@n3.R)
  FunctionOutput(None=Saturate(None=Add(A=@n4, B=@n3.B))) {outputName=Combined}
  FunctionOutput(None=@n3.R) {outputName=FlowerType1, sortPriority=1}
  FunctionOutput(None=@n3.G) {outputName=FlowerType2, sortPriority=2}
  FunctionOutput(None=@n3.B) {outputName=FlowerType3, sortPriority=3}
 ]}
