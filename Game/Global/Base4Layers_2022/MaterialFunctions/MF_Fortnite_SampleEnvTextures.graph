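# Material function exposed to the library as the standard Fortnite environment texture sampler, with blank default textures.
# It samples Diffuse, SpecularMasks and Normals inputs on shared coordinates, with masks packing specular, metallic and roughness in RGB.
# It writes the results into an input material attributes set and also outputs each channel separately.
# Within those attributes, Metallic and Roughness are wired to the red Specular channel, unlike the separate outputs.
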MaterialFunction /Game/Global/Base4Layers_2022/MaterialFunctions/MF_Fortnite_SampleEnvTextures
graph {
 properties {
  bExposeToLibrary=true
  description=Standard sampler for textures in Fortnite
  userExposedCaption=FortniteEnvTextures
  libraryCategoriesText [Fortnite]
 }
 nodes [
  n1 = FunctionInput(Preview=TextureCoordinate()) {bUsePreviewValueAsDefault=true, inputName=Coordinates, inputType=FunctionInput_Vector2}
  n2 = TextureSample(
   UVs=@n1
   Tex=FunctionInput(Preview=TextureObjectParameter() {parameterName=Diffuse, sortPriority=0, texture=/Game/Global/Textures/Default/Blanks/T_Blank_Grey_sRGBColor}) {bUsePreviewValueAsDefault=true, inputName=Diffuse, inputType=FunctionInput_Texture2D, sortPriority=1}
  )
  n3 = TextureObjectParameter() {
   desc="R = Specular\r\nG = Metal\r\nB = Roughness"
   parameterName=SpecularMasks
   samplerType=SAMPLERTYPE_Masks
   sortPriority=0
   texture=/Game/Global/Textures/Default/Blanks/T_Blank_SpecularMetallicRoughness_SMR_Masks
   channelNames {a="", b=Roughness, g=Metallic, r=Specular}
  }
  n4 = TextureSample(
   UVs=@n1
   Tex=FunctionInput(Preview=@n3) {bUsePreviewValueAsDefault=true, inputName=SpecularMasks, inputType=FunctionInput_Texture2D, sortPriority=2}
  ) {channelNames {a="", b=Roughness, g=Metallic, r=Specular}}
  n5 = TextureObjectParameter() {parameterName=Normals, samplerType=SAMPLERTYPE_Normal, sortPriority=0, texture=/Game/Global/Textures/Default/Blanks/T_Blank_Normal3D_Normal}
  n6 = TextureSample(
   UVs=@n1
   Tex=FunctionInput(Preview=@n5) {bUsePreviewValueAsDefault=true, inputName=Normals, inputType=FunctionInput_Texture2D, sortPriority=3}
  )
  n7 = SetMaterialAttributes!(
   MaterialAttributes=FunctionInput(Preview=SetMaterialAttributes!(MaterialAttributes)) {bUsePreviewValueAsDefault=true, inputName=MA, inputType=FunctionInput_MaterialAttributes}
   BaseColor=@n2.RGB
   Specular=@n4.Specular
   Metallic=@n4.Specular
   Roughness=@n4.Specular
   Normal=@n6.RGB
  ) {attributeSetTypes [
    #1
    #2
    #3
    #4
    #5
   ]}
  FunctionOutput(None=@n7) {outputName=MA}
  FunctionOutput(None=@n2.RGB) {outputName=Diffuse, sortPriority=1}
  FunctionOutput(None=@n4.Specular) {outputName=Specular, sortPriority=2}
  FunctionOutput(None=@n4.Metallic) {outputName=Metallic, sortPriority=3}
  FunctionOutput(None=@n4.Roughness) {outputName=Roughness, sortPriority=4}
  FunctionOutput(None=@n6.RGB) {outputName=Normal, sortPriority=5}
 ]
}
