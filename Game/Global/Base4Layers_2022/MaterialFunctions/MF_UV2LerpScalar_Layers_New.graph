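# Material function that selects one of up to four scalar layer inputs based on the U value of UV channel 1.
# Chained If nodes split U at 1, 2 and 3, giving a hard per-tile choice rather than a blend.
# Static layer-count switches compile in layers two to four.
# Nanite Texture Sets limits the inner chain to two layers, and Emissive Filters with SFX Emissive RGB bypasses the fourth.
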MaterialFunction /Game/Global/Base4Layers_2022/MaterialFunctions/MF_UV2LerpScalar_Layers_New
graph {nodes [
  n1 = ComponentMask(None=TextureCoordinate() {coordinateIndex=1}) {r=true}
  n2 = FunctionInput() {inputName=1, inputType=FunctionInput_Scalar}
  n3 = If(A=Constant() {r=1}, B=@n1, "A > B"=@n2, "A < B"=FunctionInput() {inputName=2, inputType=FunctionInput_Scalar, sortPriority=1}) {desc=0}
  n4 = StaticSwitch(True=@n3, False=@n2, Value=StaticBoolParameter() {group=2_Layer, parameterName=Use 2 Layers, sortPriority=-1})
  n5 = If(A=Constant() {r=2}, B=@n1, "A > B"=@n3, "A < B"=FunctionInput() {inputName=3, inputType=FunctionInput_Scalar, sortPriority=2})
  n6 = StaticSwitchParameter(True=@n4, False=StaticSwitch(True=@n5, False=@n4, Value=StaticBoolParameter() {group=3_Layer, parameterName=Use 3 Layers, sortPriority=-1})) {group=0_Nanite, parameterName=Use Nanite Texture Sets, sortPriority=-25}
  n7 = StaticSwitch(
   True=If(A=Constant() {r=3}, B=@n1, "A > B"=@n5, "A < B"=FunctionInput() {inputName=4, inputType=FunctionInput_Scalar, sortPriority=3})
   False=@n6
   Value=StaticBoolParameter() {group=4_Layer, parameterName=Use 4 Layers, sortPriority=-1}
  )
  n8 = StaticSwitchParameter(True=@n6, False=@n7) {
   desc="Emissive RGB Available if SFX RGB is activated in 0 SFX Global. Will replace layers Emissive unless Emissive RGB or Emissive Texture Add is set= true  ( adds texture sample)"
   group=A_Emissive_Texture
   parameterName=SFX Emissive Texture RGB
   sortPriority=-32
  }
  FunctionOutput(None=StaticSwitchParameter(True=@n8, False=@n7) {group=A__EmissiveFilters, parameterName=Emissive Filters, sortPriority=-31})
 ]}
