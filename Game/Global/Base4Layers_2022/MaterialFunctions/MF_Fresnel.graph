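# Material function for layered materials that adds optional Fresnel reflections, making surfaces look wetter and glossier at grazing angles by scaling down roughness and base color.
# Four per-layer level parameters, combined by MF_UV2LerpScalar_Layers_New, set the exponent of a vertex-normal Fresnel mask.
# A static switch toggles the effect.
# Quality and shading path switches limit it to deferred shading at High, Epic or default quality, with other branches fed by reroutes.
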MaterialFunction /Game/Global/Base4Layers_2022/MaterialFunctions/MF_Fresnel
graph {nodes [
  n1 = GetMaterialAttributes(None=NamedRerouteDeclaration(None=FunctionInput() {inputName=Mat In, inputType=FunctionInput_MaterialAttributes}) {name=Mat Input, nodeColor=<rgba 0.5938 0 1 1>}) {attributeGetTypes [#1, #2]}
  n2 = NamedRerouteUsage()
  n3 = LinearInterpolate(Alpha=@n2) {constA=1, constB=0.1}
  n4 = NamedRerouteUsage()
  n5 = LinearInterpolate(Alpha=@n2) {constA=1, constB=0}
  n6 = SetMaterialAttributes!(
   MaterialAttributes=@n1.MaterialAttributes
   Roughness=LinearInterpolate(A=@n1.Roughness, B=Multiply(A=@n1.Roughness, B=@n3), Alpha=@n4)
   BaseColor=LinearInterpolate(A=@n1.BaseColor, B=Multiply(A=@n1.BaseColor, B=@n5), Alpha=@n4)
  ) {attributeSetTypes [#1, #2]}
  n7 = NamedRerouteUsage()
  n8 = QualitySwitch(Default=@n6, Low=@n7, High=@n6, Medium=@n7, Epic=@n6)
  n9 = StaticSwitchParameter(True=ShadingPathSwitch(Default=@n8, Deferred=@n8, Forward=@n7, Mobile=@n7), False=NamedRerouteUsage()) {
   desc="Useful for making surfaces more believable/wet.  Use exponent controlls to smooth out surfces on grazing angles"
   group=A_Fresnel_Wet_Reflections
   parameterName=Fresnel Reflections
   sortPriority=-1
  }
  FunctionOutput(None=@n9)
  n10 = ScalarParameter() {
   defaultValue=0.5
   desc="Controls the exponent of refletions from low reflective ( 0) to high gloss (1) "
   group=A_FresnelReflections
   parameterName=Fresnel Level Exp
   sliderMax=1
   sortPriority=2
  }
  n11 = ScalarParameter() {
   defaultValue=0.5
   desc="Controls the exponent of refletions from low reflective ( 0) to high gloss (1) "
   group=A_FresnelReflections
   parameterName=Fresnel Level Exp L2
   sliderMax=1
   sortPriority=2
  }
  n12 = ScalarParameter() {
   defaultValue=0.5
   desc="Controls the exponent of refletions from low reflective ( 0) to high gloss (1) "
   group=A_FresnelReflections
   parameterName=Fresnel Level Exp L3
   sliderMax=1
   sortPriority=2
  }
  n13 = ScalarParameter() {
   defaultValue=0.5
   desc="Controls the exponent of refletions from low reflective ( 0) to high gloss (1) "
   group=A_FresnelReflections
   parameterName=Fresnel Level Exp L4
   sliderMax=1
   sortPriority=2
  }
  n14 = MaterialFunctionCall!(1=@n10, 2=@n11, 3=@n12, 4=@n13) {materialFunction=&/MF_UV2LerpScalar_Layers_New}
  n15 = LinearInterpolate(Alpha=NamedRerouteDeclaration(None=Saturate(None=@n14.Result)) {name=Refl Exponent, nodeColor=<rgba 1 0 0.3984 1>}) {constA=7, constB=1.5}
  n16 = VertexNormalWS()
  NamedRerouteDeclaration(None=Saturate(None=Fresnel(ExponentIn=@n15, Normal=@n16))) {name=FresnelMask, nodeColor=<rgba 0.0781 0 1 1>}
 ]}
