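# Material function that blends a road material into the landscape gravel path, toggled by UseGravelPathBlending.
# The background comes from the Helios landscape-match gravel layer or Apollo landscape RVT, and feeds a height blend with the road.
# A road edge mask is switchable between two world-space box masks and actor-local sphere-mask gradients with adjustable direction.
# World-projected Gaussian noise also builds a SnowMask.
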MaterialFunction /Game/Global/Base4Layers_2022/MaterialFunctions/MF_GravelPathBlend
graph {nodes [
  FunctionInput(Preview=Constant()) {bUsePreviewValueAsDefault=true, inputName=RoadHeight, inputType=FunctionInput_Scalar}
  n1 = NamedRerouteUsage()
  FunctionOutput(None=StaticSwitchParameter(True=SetMaterialAttributes!(MaterialAttributes=@n1), False=NamedRerouteUsage()) {defaultValue=true, group=RVT_Blending, parameterName=UseGravelPathBlending})
  MaterialFunctionCall(Float3=ObjectLocalBounds().Extents) {materialFunction=&BreakOutFloat3Components}
  n2 = GetMaterialAttributes(None=FunctionInput() {inputName=Material, inputType=FunctionInput_MaterialAttributes}) {attributeGetTypes [#1]}
  NamedRerouteDeclaration(None=@n2.MaterialAttributes) {name=RoadMat, nodeColor=<rgba 0 0.4844 1 1>}
  NamedRerouteDeclaration(None=@n2.Roughness) {name=RoadHeight, nodeColor=<rgba 1 0 0.0938 1>}
  NamedRerouteDeclaration(None=MaterialFunctionCall!(
    Material 1=NamedRerouteUsage()
    Height Texture 1=NamedRerouteUsage()
    Material 2=NamedRerouteUsage()
    Height Texture 2=NamedRerouteUsage()
    Transistion Phase=NamedRerouteUsage()
    UseContrast
    Contrast
    AddLayerMask
   ) {materialFunction=/Game/Athena/Helios/Environments/Landscape/Materials/MaterialFunctions_Utilities/MF_HeightBlendLandscapeLayers}.BlendedMaterial) {name=FinalMat, nodeColor=<rgba 1 0.5391 0 1>}
  n3 = ScalarParameter() {defaultValue=-0.25, group=HeliosSnow, parameterName=NoiseMin}
  n4 = ScalarParameter() {defaultValue=1, group=HeliosSnow, parameterName=NoiseMax}
  n5 = WorldPosition()
  n6 = ScalarParameter() {defaultValue=4096, group=HeliosSnow, parameterName=SnowNoise}
  n7 = TextureSampleParameter2D(UVs=ComponentMask(None=Divide(A=@n5.XYZ, B=@n6)) {g=true, r=true}) {group=HeliosSnow, parameterName=SnowCoverageNoise, texture=/Game/Effects/Fort_Effects/Textures/Tile/T_GaussianNoise_01_M}
  NamedRerouteDeclaration(None=Multiply(A=LinearInterpolate(A=@n3, B=@n4, Alpha=@n7.R), B=ScalarParameter() {defaultValue=1, group=HeliosSnow, parameterName=NoiseStrength})) {name=SnowMask, nodeColor=<rgba 0 0.7656 1 1>}
  n8 = MaterialFunctionCall!(Grass_UseCustomSlopeMask, Grass_CustomSlopeMask) {materialFunction=/Game/Athena/Helios/Environments/Landscape/Materials/MaterialFunctions_Utilities/MF_Helios_StaticMeshLandscapeLayerMatch}
  NamedRerouteDeclaration(None=GetMaterialAttributes(None=@n8.GravelLayer) {attributeGetTypes [#2]}.OpacityMask) {name=BG-Height, nodeColor=<rgba 1 0 0.6797 1>}
  n9 = RuntimeVirtualTextureSample!(UVs, World Position) {materialType=BaseColor_Normal_Specular_YCoCg, virtualTexture=/Game/Athena/Apollo/Maps/Landscape/VT/RVT_Apollo_Landscape}
  n10 = SetMaterialAttributes!(MaterialAttributes, BaseColor=@n9.BaseColor, Specular=@n9.BaseColor, Roughness=@n9.BaseColor, Normal=@n9.BaseColor) {attributeSetTypes [
    #3
    #4
    #1
    #5
   ]}
  NamedRerouteDeclaration(None=MaterialFunctionCall(NoRVT=@n8.GravelLayer, RVT=@n10) {materialFunction=&FN_RVT_Switch_Material}.Result) {name=BG-Material, nodeColor=<rgba 0.5469 0 1 1>}
  n11 = WorldPosition()
  n12 = MaterialFunctionCall!(
   A=@n11.XYZ
   B=VectorParameter() {group=RVT_Blending, parameterName=BoxMaskPosition, defaultValue=<rgba 59968 -11776 2688 1>}.RGB
   Bounds=VectorParameter() {group=RVT_Blending, parameterName=BoxMaskBounds, defaultValue=<rgba 2048 1023 1024 1>}.RGB
   Edge Falloff=ScalarParameter() {defaultValue=32, group=RVT_Blending, parameterName=EdgeFalloff}
  ) {materialFunction=/Engine/Functions/Engine_MaterialFunctions01/Gradient/BoxMask-3D}
  n13 = MaterialFunctionCall(
   A=@n11.XYZ
   B=VectorParameter() {group=RVT_Blending, parameterName=V_BoxMaskPosition, defaultValue=<rgba 59968 -11776 2688 1>}.RGB
   Bounds=VectorParameter() {group=RVT_Blending, parameterName=V_BoxMaskBounds, defaultValue=<rgba 2048 1023 1024 1>}.RGB
   Edge Falloff=ScalarParameter() {defaultValue=32, group=RVT_Blending, parameterName=V_EdgeFalloff}
  ) {materialFunction=&BoxMask-3D}
  n14 = WorldPosition()
  n15 = ActorPositionWS()
  n16 = MaterialFunctionCall(Float3=Transform(None=Subtract(A=@n14.XYZ, B=@n15)) {transformSourceType=TRANSFORMSOURCE_World, transformType=TRANSFORM_Local}) {materialFunction=&BreakOutFloat3Components}
  n17 = MaterialFunctionCall(A=@n16.R, B=@n16.R, C=@n16.R, Alpha=ScalarParameter() {defaultValue=0.5, group=RVT_Blending, parameterName=LinearFadeDirection}) {materialFunction=&Lerp_3Color}
  n18 = LinearInterpolate(A=@n17.Result, B=OneMinus(None=@n17.Result), Alpha=ScalarParameter() {defaultValue=1, group=RVT_Blending, parameterName=FlipFadeDirection})
  n19 = ScalarParameter() {defaultValue=-128, group=RVT_Blending, parameterName=Mask3SamplePos}
  n20 = ScalarParameter() {defaultValue=256, group=RVT_Blending, parameterName=Fade3Radius}
  n21 = ScalarParameter() {group=RVT_Blending, parameterName=Fade3EdgeHardness}
  n22 = Saturate(None=Multiply(
    A=SphereMask(A=@n18, B=@n19, Radius=@n20, Hardness=@n21)
    B=ScalarParameter() {defaultValue=1, group=RVT_Blending, parameterName=Fade3GradientStrength}
   ))
  n23 = MaterialFunctionCall(A=@n16.R, B=@n16.R, C=@n16.R, Alpha=ScalarParameter() {group=RVT_Blending, parameterName=LinearFade2Direction}) {materialFunction=&Lerp_3Color}
  n24 = SphereMask(
   A=LinearInterpolate(A=@n23.Result, B=OneMinus(None=@n23.Result), Alpha=ScalarParameter() {defaultValue=1, group=RVT_Blending, parameterName=FlipFade2Direction})
   B=ScalarParameter() {defaultValue=-128, group=RVT_Blending, parameterName=Mask4SamplePos}
   Radius=ScalarParameter() {defaultValue=256, group=RVT_Blending, parameterName=Fade4Radius}
   Hardness=ScalarParameter() {group=RVT_Blending, parameterName=Fade4EdgeHardness}
  )
  n25 = ScalarParameter() {defaultValue=1, group=RVT_Blending, parameterName=Fade4GradientStrength}
  n26 = Multiply(A=@n22, B=Saturate(None=Multiply(A=@n24, B=@n25)))
  n27 = Multiply(
   A=SphereMask(
    A=@n16.R
    B=ScalarParameter() {group=RVT_Blending, parameterName=Mask2SamplePos}
    Radius=ScalarParameter() {defaultValue=128, group=RVT_Blending, parameterName=Fade2Radius}
    Hardness=ScalarParameter() {group=RVT_Blending, parameterName=Fade2EdgeHardness}
   )
   B=ScalarParameter() {defaultValue=1, group=RVT_Blending, parameterName=Fade2GradientStrength}
  )
  n28 = Saturate(None=@n27)
  n29 = StaticSwitchParameter(True=Multiply(A=@n28, B=@n22), False=@n28) {group=RVT_Blending, parameterName=UseMultiSideGradient}
  NamedRerouteDeclaration(None=StaticSwitchParameter(
    True=Max(A=@n12.Result, B=@n13.Result)
    False=StaticSwitchParameter(True=@n26, False=@n29) {group=RVT_Blending, parameterName=AdjacentSideGradient}
   ) {group=RVT_Blending, parameterName=UseGlobalMsk}) {name=RoadEdgeMask, nodeColor=<rgba 0 0.1094 1 1>}
  Subtract(A=@n27, B=NamedRerouteUsage())
 ]}
