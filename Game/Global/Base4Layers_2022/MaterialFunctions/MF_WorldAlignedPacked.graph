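# Triplanar world-aligned material function that projects a packed texture by world position and size, returning blended color and normal data as XYZ, XY-only and Z-only outputs.
# Three axis projections, optionally rotated, are blended by contrast-adjusted absolute world normal, with an optional float4 export of the texture channels.
# Normals are unpacked from RG or AG, counter-rotated, strength-scaled and combined with the surface normal via reoriented normal mapping or simpler addition, then output in world or transformed space.
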MaterialFunction /Game/Global/Base4Layers_2022/MaterialFunctions/MF_WorldAlignedPacked
graph {
 expressions {
  MaterialExpressionAbs_1 {class=MaterialExpressionAbs, inputs [None {from=MaterialExpressionFunctionInput_1, output=""}], properties {}}
  MaterialExpressionAbs_2 {class=MaterialExpressionAbs, inputs [None {from=MaterialExpressionNamedRerouteUsage_0, output=""}], properties {}}
  MaterialExpressionAbs_5 {class=MaterialExpressionAbs, inputs [None {from=MaterialExpressionNamedRerouteDeclaration_0, output=""}], properties {}}
  MaterialExpressionAdd_1 {
   class=MaterialExpressionAdd
   inputs [
    A {from=MaterialExpressionNamedRerouteUsage_1}
    B {from=MaterialExpressionLinearInterpolate_3}
   ]
  }
  MaterialExpressionAdd_12 {
   class=MaterialExpressionAdd
   inputs [
    A {from=MaterialExpressionComponentMask_45}
    B {from=""}
   ]
  }
  MaterialExpressionAdd_13 {
   class=MaterialExpressionAdd
   inputs [
    A {from=MaterialExpressionComponentMask_58}
    B {from=""}
   ]
  }
  MaterialExpressionAdd_14 {
   class=MaterialExpressionAdd
   inputs [
    A {from=MaterialExpressionComponentMask_62}
    B {from=""}
   ]
  }
  MaterialExpressionAdd_2 {
   class=MaterialExpressionAdd
   inputs [
    A {from=MaterialExpressionNamedRerouteUsage_2}
    B {from=MaterialExpressionLinearInterpolate_2}
   ]
  }
  MaterialExpressionAdd_3 {
   class=MaterialExpressionAdd
   inputs [
    A {from=MaterialExpressionNamedRerouteUsage_3}
    B {from=MaterialExpressionConvert_32}
   ]
  }
  MaterialExpressionAppendVector_15 {
   class=MaterialExpressionAppendVector
   inputs [
    A {from=MaterialExpressionComponentMask_44, output=""}
    B {from=MaterialExpressionAdd_12, output=""}
   ]
  }
  MaterialExpressionAppendVector_18 {
   class=MaterialExpressionAppendVector
   inputs [
    A {from=MaterialExpressionMultiply_35, output=""}
    B {from=MaterialExpressionComponentMask_47, output=""}
   ]
  }
  MaterialExpressionAppendVector_19 {
   class=MaterialExpressionAppendVector
   inputs [
    A {from=MaterialExpressionMultiply_36, output=""}
    B {from=MaterialExpressionComponentMask_48, output=""}
   ]
  }
  MaterialExpressionAppendVector_20 {
   class=MaterialExpressionAppendVector
   inputs [
    A {from=MaterialExpressionComponentMask_59, output=""}
    B {from=MaterialExpressionAdd_13, output=""}
   ]
  }
  MaterialExpressionAppendVector_21 {
   class=MaterialExpressionAppendVector
   inputs [
    A {from=MaterialExpressionMultiply_39, output=""}
    B {from=MaterialExpressionComponentMask_60, output=""}
   ]
  }
  MaterialExpressionAppendVector_22 {
   class=MaterialExpressionAppendVector
   inputs [
    A {from=MaterialExpressionComponentMask_63, output=""}
    B {from=MaterialExpressionAdd_14, output=""}
   ]
  }
  MaterialExpressionComponentMask_0 {class=MaterialExpressionComponentMask, inputs [None {from=MaterialExpressionDivide_0}], properties {g=true, r=true}}
  MaterialExpressionComponentMask_1 {class=MaterialExpressionComponentMask, inputs [None {from=MaterialExpressionNamedRerouteUsage_12}], properties {b=true, g=true, r=true}}
  MaterialExpressionComponentMask_2 {class=MaterialExpressionComponentMask, inputs [None {from=MaterialExpressionDivide_0}], properties {b=true, g=true}}
  MaterialExpressionComponentMask_4 {class=MaterialExpressionComponentMask, inputs [None {from=MaterialExpressionDivide_0}], properties {b=true, r=true}}
  MaterialExpressionComponentMask_44 {class=MaterialExpressionComponentMask, inputs [None {from=MaterialExpressionReroute_7}], properties {g=true, r=true}}
  MaterialExpressionComponentMask_45 {class=MaterialExpressionComponentMask, inputs [None {from=MaterialExpressionReroute_7}], properties {b=true}}
  MaterialExpressionComponentMask_46 {class=MaterialExpressionComponentMask, inputs [None {from=MaterialExpressionReroute_8}], properties {g=true, r=true}}
  MaterialExpressionComponentMask_47 {class=MaterialExpressionComponentMask, inputs [None {from=MaterialExpressionReroute_8}], properties {b=true}}
  MaterialExpressionComponentMask_48 {class=MaterialExpressionComponentMask, inputs [None {from=MaterialExpressionReroute_13}], properties {b=true}}
  MaterialExpressionComponentMask_49 {class=MaterialExpressionComponentMask, inputs [None {from=MaterialExpressionReroute_13}], properties {g=true, r=true}}
  MaterialExpressionComponentMask_5 {class=MaterialExpressionComponentMask, inputs [None {from=MaterialExpressionAbs_5}], properties {r=true}}
  MaterialExpressionComponentMask_58 {class=MaterialExpressionComponentMask, inputs [None {from=MaterialExpressionReroute_9}], properties {b=true}}
  MaterialExpressionComponentMask_59 {class=MaterialExpressionComponentMask, inputs [None {from=MaterialExpressionReroute_9}], properties {g=true, r=true}}
  MaterialExpressionComponentMask_6 {class=MaterialExpressionComponentMask, inputs [None {from=MaterialExpressionNamedRerouteUsage_13}], properties {b=true, g=true, r=true}}
  MaterialExpressionComponentMask_60 {class=MaterialExpressionComponentMask, inputs [None {from=MaterialExpressionReroute_20}], properties {b=true}}
  MaterialExpressionComponentMask_61 {class=MaterialExpressionComponentMask, inputs [None {from=MaterialExpressionReroute_20}], properties {g=true, r=true}}
  MaterialExpressionComponentMask_62 {class=MaterialExpressionComponentMask, inputs [None {from=MaterialExpressionReroute_14}], properties {b=true}}
  MaterialExpressionComponentMask_63 {class=MaterialExpressionComponentMask, inputs [None {from=MaterialExpressionReroute_14}], properties {g=true, r=true}}
  MaterialExpressionComponentMask_7 {class=MaterialExpressionComponentMask, inputs [None {from=MaterialExpressionAbs_5}], properties {b=true}}
  MaterialExpressionComponentMask_8 {class=MaterialExpressionComponentMask, inputs [None {from=MaterialExpressionNamedRerouteUsage_14}], properties {b=true, g=true, r=true}}
  MaterialExpressionConstantBiasScale_0 {class=MaterialExpressionConstantBiasScale, inputs [None {from=MaterialExpressionStaticSwitch_1}], properties {bias=-0.5, scale=2}}
  MaterialExpressionConstantBiasScale_1 {class=MaterialExpressionConstantBiasScale, inputs [None {from=MaterialExpressionStaticSwitch_3}], properties {bias=-0.5, scale=2}}
  MaterialExpressionConstantBiasScale_2 {class=MaterialExpressionConstantBiasScale, inputs [None {from=MaterialExpressionStaticSwitch_5}], properties {bias=-0.5, scale=2}}
  MaterialExpressionConvert_10 {
   class=MaterialExpressionConvert
   inputs [
    None {from=MaterialExpressionNamedRerouteUsage_0}
    None {from=MaterialExpressionAbs_2}
   ]
   properties {
    nodeName=Make Vector3
    convertInputs [
     {
      type=Vector3
      expressionInput {
       expression=&self:MaterialExpressionNamedRerouteUsage_0
       inputName=None
       mask=0
       maskA=0
       maskB=0
       maskG=0
       maskR=0
       outputIndex=0
      }
     }
     {
      type=Vector3
      expressionInput {
       expression=&self:MaterialExpressionAbs_2
       inputName=None
       mask=0
       maskA=0
       maskB=0
       maskG=0
       maskR=0
       outputIndex=0
      }
     }
    ]
    convertOutputs [{type=Vector3}]
   }
  }
  MaterialExpressionConvert_12 {
   class=MaterialExpressionConvert
   inputs [None {from=MaterialExpressionSubtract_13}]
   properties {
    nodeName=Break Vector3
    convertInputs [{
      type=Vector3
      expressionInput {
       expression=&self:MaterialExpressionSubtract_13
       inputName=None
       mask=0
       maskA=0
       maskB=0
       maskG=0
       maskR=0
       outputIndex=0
      }
     }]
    convertOutputs [{type=Scalar}, {type=Scalar}, {type=Scalar}]
   }
  }
  MaterialExpressionConvert_13 {
   class=MaterialExpressionConvert
   inputs [None {from=MaterialExpressionSubtract_10}]
   properties {
    nodeName=Break Vector3
    convertInputs [{
      type=Vector3
      expressionInput {
       expression=&self:MaterialExpressionSubtract_10
       inputName=None
       mask=0
       maskA=0
       maskB=0
       maskG=0
       maskR=0
       outputIndex=0
      }
     }]
    convertOutputs [{type=Scalar}, {type=Scalar}, {type=Scalar}]
   }
  }
  MaterialExpressionConvert_14 {
   class=MaterialExpressionConvert
   inputs [None {from=MaterialExpressionSubtract_9}]
   properties {
    nodeName=Break Vector3
    convertInputs [{
      type=Vector3
      expressionInput {
       expression=&self:MaterialExpressionSubtract_9
       inputName=None
       mask=0
       maskA=0
       maskB=0
       maskG=0
       maskR=0
       outputIndex=0
      }
     }]
    convertOutputs [{type=Scalar}, {type=Scalar}, {type=Scalar}]
   }
  }
  MaterialExpressionConvert_15 {
   class=MaterialExpressionConvert
   inputs [
    None {from=MaterialExpressionConvert_14}
    None {from=MaterialExpressionConvert_14}
    None {from=MaterialExpressionMultiply_18}
   ]
   properties {
    nodeName=Make Vector3
    convertInputs [
     {
      type=Scalar
      expressionInput {
       expression=&self:MaterialExpressionConvert_14
       inputName=None
       mask=0
       maskA=0
       maskB=0
       maskG=0
       maskR=0
       outputIndex=0
      }
     }
     {
      type=Scalar
      expressionInput {
       expression=&self:MaterialExpressionConvert_14
       inputName=None
       mask=0
       maskA=0
       maskB=0
       maskG=0
       maskR=0
       outputIndex=1
      }
     }
     {
      type=Scalar
      expressionInput {
       expression=&self:MaterialExpressionMultiply_18
       inputName=None
       mask=0
       maskA=0
       maskB=0
       maskG=0
       maskR=0
       outputIndex=0
      }
     }
    ]
    convertOutputs [{type=Vector3}]
   }
  }
  MaterialExpressionConvert_16 {
   class=MaterialExpressionConvert
   inputs [
    None {from=MaterialExpressionConvert_13}
    None {from=MaterialExpressionConvert_13}
    None {from=MaterialExpressionMultiply_19}
   ]
   properties {
    nodeName=Make Vector3
    convertInputs [
     {
      type=Scalar
      expressionInput {
       expression=&self:MaterialExpressionConvert_13
       inputName=None
       mask=0
       maskA=0
       maskB=0
       maskG=0
       maskR=0
       outputIndex=0
      }
     }
     {
      type=Scalar
      expressionInput {
       expression=&self:MaterialExpressionConvert_13
       inputName=None
       mask=0
       maskA=0
       maskB=0
       maskG=0
       maskR=0
       outputIndex=1
      }
     }
     {
      type=Scalar
      expressionInput {
       expression=&self:MaterialExpressionMultiply_19
       inputName=None
       mask=0
       maskA=0
       maskB=0
       maskG=0
       maskR=0
       outputIndex=0
      }
     }
    ]
    convertOutputs [{type=Vector3}]
   }
  }
  MaterialExpressionConvert_17 {
   class=MaterialExpressionConvert
   inputs [
    None {from=MaterialExpressionConvert_12}
    None {from=MaterialExpressionConvert_12}
    None {from=MaterialExpressionMultiply_20}
   ]
   properties {
    nodeName=Make Vector3
    convertInputs [
     {
      type=Scalar
      expressionInput {
       expression=&self:MaterialExpressionConvert_12
       inputName=None
       mask=0
       maskA=0
       maskB=0
       maskG=0
       maskR=0
       outputIndex=0
      }
     }
     {
      type=Scalar
      expressionInput {
       expression=&self:MaterialExpressionConvert_12
       inputName=None
       mask=0
       maskA=0
       maskB=0
       maskG=0
       maskR=0
       outputIndex=1
      }
     }
     {
      type=Scalar
      expressionInput {
       expression=&self:MaterialExpressionMultiply_20
       inputName=None
       mask=0
       maskA=0
       maskB=0
       maskG=0
       maskR=0
       outputIndex=0
      }
     }
    ]
    convertOutputs [{type=Vector3}]
   }
  }
  MaterialExpressionConvert_27 {
   class=MaterialExpressionConvert
   inputs [None {from=MaterialExpressionMultiply_1}]
   properties {
    nodeName=Break Vector2
    convertInputs [{
      type=Vector2
      expressionInput {
       expression=&self:MaterialExpressionMultiply_1
       inputName=None
       mask=0
       maskA=0
       maskB=0
       maskG=0
       maskR=0
       outputIndex=0
      }
     }]
    convertOutputs [{type=Scalar}, {type=Scalar}]
   }
  }
  MaterialExpressionConvert_28 {
   class=MaterialExpressionConvert
   inputs [None {from=MaterialExpressionMultiply_6}]
   properties {
    nodeName=Break Vector2
    convertInputs [{
      type=Vector2
      expressionInput {
       expression=&self:MaterialExpressionMultiply_6
       inputName=None
       mask=0
       maskA=0
       maskB=0
       maskG=0
       maskR=0
       outputIndex=0
      }
     }]
    convertOutputs [{type=Scalar}, {type=Scalar}]
   }
  }
  MaterialExpressionConvert_29 {
   class=MaterialExpressionConvert
   inputs [None {from=MaterialExpressionMultiply_4}]
   properties {
    nodeName=Break Vector2
    convertInputs [{
      type=Vector2
      expressionInput {
       expression=&self:MaterialExpressionMultiply_4
       inputName=None
       mask=0
       maskA=0
       maskB=0
       maskG=0
       maskR=0
       outputIndex=0
      }
     }]
    convertOutputs [{type=Scalar}, {type=Scalar}]
   }
  }
  MaterialExpressionConvert_3 {
   class=MaterialExpressionConvert
   inputs [None {from=MaterialExpressionSign_0}]
   properties {
    nodeName=Break Vector3
    convertInputs [{
      type=Vector3
      expressionInput {
       expression=&self:MaterialExpressionSign_0
       inputName=None
       mask=0
       maskA=0
       maskB=0
       maskG=0
       maskR=0
       outputIndex=0
      }
     }]
    convertOutputs [{type=Scalar}, {type=Scalar}, {type=Scalar}]
   }
  }
  MaterialExpressionConvert_30 {
   class=MaterialExpressionConvert
   inputs [
    None {from=MaterialExpressionConvert_27}
    None {from=MaterialExpressionConvert_27}
   ]
   properties {
    nodeName=Make Vector3
    convertInputs [
     {
      type=Scalar
      expressionInput {
       expression=&self:MaterialExpressionConvert_27
       inputName=None
       mask=0
       maskA=0
       maskB=0
       maskG=0
       maskR=0
       outputIndex=0
      }
     }
     {
      type=Scalar
      expressionInput {
       expression=&self:MaterialExpressionConvert_27
       inputName=None
       mask=0
       maskA=0
       maskB=0
       maskG=0
       maskR=0
       outputIndex=1
      }
     }
    ]
    convertOutputs [{type=Vector3}]
   }
  }
  MaterialExpressionConvert_31 {
   class=MaterialExpressionConvert
   inputs [
    None {from=MaterialExpressionConvert_29}
    None {from=MaterialExpressionConvert_29}
   ]
   properties {
    nodeName=Make Vector3
    convertInputs [
     {
      type=Scalar
      expressionInput {
       expression=&self:MaterialExpressionConvert_29
       inputName=None
       mask=0
       maskA=0
       maskB=0
       maskG=0
       maskR=0
       outputIndex=0
      }
     }
     {
      type=Scalar
      expressionInput {
       expression=&self:MaterialExpressionConvert_29
       inputName=None
       mask=0
       maskA=0
       maskB=0
       maskG=0
       maskR=0
       outputIndex=1
      }
     }
    ]
    convertOutputs [{type=Vector3}]
   }
  }
  MaterialExpressionConvert_32 {
   class=MaterialExpressionConvert
   inputs [
    None {from=MaterialExpressionConvert_28}
    None {from=MaterialExpressionConvert_28}
   ]
   properties {
    nodeName=Make Vector3
    convertInputs [
     {
      type=Scalar
      expressionInput {
       expression=&self:MaterialExpressionConvert_28
       inputName=None
       mask=0
       maskA=0
       maskB=0
       maskG=0
       maskR=0
       outputIndex=0
      }
     }
     {
      type=Scalar
      expressionInput {
       expression=&self:MaterialExpressionConvert_28
       inputName=None
       mask=0
       maskA=0
       maskB=0
       maskG=0
       maskR=0
       outputIndex=1
      }
     }
    ]
    convertOutputs [{type=Vector3}]
   }
  }
  MaterialExpressionConvert_8 {
   class=MaterialExpressionConvert
   inputs [
    None {from=MaterialExpressionNamedRerouteUsage_0}
    None {from=MaterialExpressionAbs_2}
   ]
   properties {
    nodeName=Make Vector3
    convertInputs [
     {
      type=Vector3
      expressionInput {
       expression=&self:MaterialExpressionNamedRerouteUsage_0
       inputName=None
       mask=0
       maskA=0
       maskB=0
       maskG=0
       maskR=0
       outputIndex=0
      }
     }
     {
      type=Vector3
      expressionInput {
       expression=&self:MaterialExpressionAbs_2
       inputName=None
       mask=0
       maskA=0
       maskB=0
       maskG=0
       maskR=0
       outputIndex=0
      }
     }
    ]
    convertOutputs [{type=Vector3}]
   }
  }
  MaterialExpressionConvert_9 {
   class=MaterialExpressionConvert
   inputs [
    None {from=MaterialExpressionNamedRerouteUsage_0}
    None {from=MaterialExpressionAbs_2}
   ]
   properties {
    nodeName=Make Vector3
    convertInputs [
     {
      type=Vector3
      expressionInput {
       expression=&self:MaterialExpressionNamedRerouteUsage_0
       inputName=None
       mask=0
       maskA=0
       maskB=0
       maskG=0
       maskR=0
       outputIndex=0
      }
     }
     {
      type=Vector3
      expressionInput {
       expression=&self:MaterialExpressionAbs_2
       inputName=None
       mask=0
       maskA=0
       maskB=0
       maskG=0
       maskR=0
       outputIndex=0
      }
     }
    ]
    convertOutputs [{type=Vector3}]
   }
  }
  MaterialExpressionDeriveNormalZ_0 {class=MaterialExpressionDeriveNormalZ, inputs [InXY {from=MaterialExpressionMultiply_1, output=""}], properties {}}
  MaterialExpressionDeriveNormalZ_1 {class=MaterialExpressionDeriveNormalZ, inputs [InXY {from=MaterialExpressionMultiply_4, output=""}], properties {}}
  MaterialExpressionDeriveNormalZ_2 {class=MaterialExpressionDeriveNormalZ, inputs [InXY {from=MaterialExpressionMultiply_6, output=""}], properties {}}
  MaterialExpressionDivide_0 {
   class=MaterialExpressionDivide
   inputs [
    A {from=MaterialExpressionFunctionInput_3}
    B {from=MaterialExpressionMultiply_0}
   ]
  }
  MaterialExpressionDotProduct_5 {
   class=MaterialExpressionDotProduct
   inputs [
    A {from=MaterialExpressionAppendVector_15, output=""}
    B {from=MaterialExpressionAppendVector_18, output=""}
   ]
   properties {}
  }
  MaterialExpressionDotProduct_6 {
   class=MaterialExpressionDotProduct
   inputs [
    A {from=MaterialExpressionAppendVector_20, output=""}
    B {from=MaterialExpressionAppendVector_19, output=""}
   ]
   properties {}
  }
  MaterialExpressionDotProduct_7 {
   class=MaterialExpressionDotProduct
   inputs [
    A {from=MaterialExpressionAppendVector_22, output=""}
    B {from=MaterialExpressionAppendVector_21, output=""}
   ]
   properties {}
  }
  MaterialExpressionFunctionInput_0 {
   class=MaterialExpressionFunctionInput
   inputs [Preview {from=MaterialExpressionTextureObject_0, output=""}]
   properties {inputName=TextureObject, inputType=FunctionInput_Texture2D}
  }
  MaterialExpressionFunctionInput_1 {
   class=MaterialExpressionFunctionInput
   inputs [Preview {from="", output=""}]
   properties {
    bUsePreviewValueAsDefault=true
    description=Size of texture in worldspace units
    inputName=TextureSize
    sortPriority=1
    previewValue=<64 64 64 6>
   }
  }
  MaterialExpressionFunctionInput_10 {
   class=MaterialExpressionFunctionInput
   inputs [Preview {from=MaterialExpressionStaticBool_3, output=""}]
   properties {bUsePreviewValueAsDefault=true, description="Default = False", inputName=Export Float 4, inputType=FunctionInput_StaticBool, sortPriority=5}
  }
  MaterialExpressionFunctionInput_11 {
   class=MaterialExpressionFunctionInput
   inputs [Preview {from=MaterialExpressionStaticBool_4, output=""}]
   properties {
    bUsePreviewValueAsDefault=true
    description=Default True
    inputName=Use High Quality Normals
    inputType=FunctionInput_StaticBool
    sortPriority=51
   }
  }
  MaterialExpressionFunctionInput_2 {
   class=MaterialExpressionFunctionInput
   inputs [Preview {from=MaterialExpressionVertexNormalWS_0, output=""}]
   properties {bUsePreviewValueAsDefault=true, inputName=WorldNormal, sortPriority=9, previewValue=<0 0 1 1>}
  }
  MaterialExpressionFunctionInput_3 {
   class=MaterialExpressionFunctionInput
   inputs [Preview {from=MaterialExpressionWorldPosition_0, output=XYZ}]
   properties {bUsePreviewValueAsDefault=true, inputName=WorldPosition, sortPriority=8}
  }
  MaterialExpressionFunctionInput_4 {
   class=MaterialExpressionFunctionInput
   inputs [Preview {from=MaterialExpressionStaticBool_0, output=""}]
   properties {
    bUsePreviewValueAsDefault=true
    description=This switch defaults to false.
    inputName=WorldSpace
    inputType=FunctionInput_StaticBool
    sortPriority=7
   }
  }
  MaterialExpressionFunctionInput_5 {
   class=MaterialExpressionFunctionInput
   inputs [Preview {from=MaterialExpressionStaticBool_2, output=""}]
   properties {
    bUsePreviewValueAsDefault=true
    description=Use AG channel for normals isntead of RG. Default False
    inputName=UseAGNormals
    inputType=FunctionInput_StaticBool
    sortPriority=7
   }
  }
  MaterialExpressionFunctionInput_6 {
   class=MaterialExpressionFunctionInput
   inputs [Preview {from="", output=""}]
   properties {bUsePreviewValueAsDefault=true, inputName=ProjectionTransitionContrast, inputType=FunctionInput_Scalar, sortPriority=50}
  }
  MaterialExpressionFunctionInput_7 {
   class=MaterialExpressionFunctionInput
   inputs [Preview {from="", output=""}]
   properties {bUsePreviewValueAsDefault=true, inputName=NormalStrength, inputType=FunctionInput_Scalar, sortPriority=6, previewValue=<1 0 0 1>}
  }
  MaterialExpressionFunctionInput_8 {
   class=MaterialExpressionFunctionInput
   inputs [Preview {from=MaterialExpressionStaticBool_1, output=""}]
   properties {bUsePreviewValueAsDefault=true, description=Default False, inputName=RotateUVs, inputType=FunctionInput_StaticBool, sortPriority=3}
  }
  MaterialExpressionFunctionInput_9 {
   class=MaterialExpressionFunctionInput
   inputs [Preview {from="", output=""}]
   properties {bUsePreviewValueAsDefault=true, inputName=RotationAngle, inputType=FunctionInput_Scalar, sortPriority=4}
  }
  MaterialExpressionFunctionOutput_0 {
   class=MaterialExpressionFunctionOutput
   inputs [None {from=MaterialExpressionNormalize_0, output=""}]
   properties {outputName=XYZ Normal, sortPriority=2}
  }
  MaterialExpressionFunctionOutput_1 {
   class=MaterialExpressionFunctionOutput
   inputs [None {from=MaterialExpressionLinearInterpolate_1, output=""}]
   properties {outputName=XYZ Texture}
  }
  MaterialExpressionFunctionOutput_2 {
   class=MaterialExpressionFunctionOutput
   inputs [None {from=MaterialExpressionNormalize_4, output=""}]
   properties {outputName=XY Normal, sortPriority=5}
  }
  MaterialExpressionFunctionOutput_3 {
   class=MaterialExpressionFunctionOutput
   inputs [None {from=MaterialExpressionNormalize_7, output=""}]
   properties {outputName=Z Normal, sortPriority=8}
  }
  MaterialExpressionFunctionOutput_4 {
   class=MaterialExpressionFunctionOutput
   inputs [None {from=MaterialExpressionLinearInterpolate_0, output=""}]
   properties {outputName=XY Texture, sortPriority=3}
  }
  MaterialExpressionFunctionOutput_5 {
   class=MaterialExpressionFunctionOutput
   inputs [None {from=MaterialExpressionLinearInterpolate_4, output=""}]
   properties {outputName=XYZ RawNorm, sortPriority=1}
  }
  MaterialExpressionFunctionOutput_6 {
   class=MaterialExpressionFunctionOutput
   inputs [None {from=MaterialExpressionStaticSwitch_15, output=""}]
   properties {outputName=Z Texture, sortPriority=6}
  }
  MaterialExpressionFunctionOutput_7 {
   class=MaterialExpressionFunctionOutput
   inputs [None {from=MaterialExpressionLinearInterpolate_5, output=""}]
   properties {outputName=XY RawNorm, sortPriority=4}
  }
  MaterialExpressionFunctionOutput_8 {
   class=MaterialExpressionFunctionOutput
   inputs [None {from=MaterialExpressionDeriveNormalZ_2, output=""}]
   properties {outputName=Z RawNorm, sortPriority=7}
  }
  MaterialExpressionLinearInterpolate_0 {
   class=MaterialExpressionLinearInterpolate
   inputs [
    A {from=MaterialExpressionStaticSwitch_13}
    B {from=MaterialExpressionStaticSwitch_14}
    Alpha {from=MaterialExpressionNamedRerouteUsage_10}
   ]
   properties {desc=Float3}
  }
  MaterialExpressionLinearInterpolate_1 {
   class=MaterialExpressionLinearInterpolate
   inputs [
    A {from=MaterialExpressionLinearInterpolate_0}
    B {from=MaterialExpressionStaticSwitch_15}
    Alpha {from=MaterialExpressionNamedRerouteUsage_11}
   ]
  }
  MaterialExpressionLinearInterpolate_2 {
   class=MaterialExpressionLinearInterpolate
   inputs [
    A {from=MaterialExpressionConvert_30}
    B {from=MaterialExpressionConvert_31}
    Alpha {from=MaterialExpressionNamedRerouteUsage_5}
   ]
  }
  MaterialExpressionLinearInterpolate_3 {
   class=MaterialExpressionLinearInterpolate
   inputs [
    A {from=MaterialExpressionLinearInterpolate_2}
    B {from=MaterialExpressionConvert_32}
    Alpha {from=MaterialExpressionNamedRerouteUsage_4}
   ]
  }
  MaterialExpressionLinearInterpolate_4 {
   class=MaterialExpressionLinearInterpolate
   inputs [
    A {from=MaterialExpressionLinearInterpolate_5}
    B {from=MaterialExpressionDeriveNormalZ_2}
    Alpha {from=MaterialExpressionNamedRerouteUsage_9}
   ]
  }
  MaterialExpressionLinearInterpolate_5 {
   class=MaterialExpressionLinearInterpolate
   inputs [
    A {from=MaterialExpressionDeriveNormalZ_0}
    B {from=MaterialExpressionDeriveNormalZ_1}
    Alpha {from=MaterialExpressionNamedRerouteUsage_8}
   ]
   properties {desc=Float3}
  }
  MaterialExpressionLinearInterpolate_6 {
   class=MaterialExpressionLinearInterpolate
   inputs [
    A {from=MaterialExpressionConvert_15}
    B {from=MaterialExpressionConvert_16}
    Alpha {from=MaterialExpressionNamedRerouteUsage_6}
   ]
  }
  MaterialExpressionLinearInterpolate_7 {
   class=MaterialExpressionLinearInterpolate
   inputs [
    A {from=MaterialExpressionLinearInterpolate_6}
    B {from=MaterialExpressionConvert_17}
    Alpha {from=MaterialExpressionNamedRerouteUsage_7}
   ]
  }
  MaterialExpressionMaterialFunctionCall_0 {
   class=MaterialExpressionMaterialFunctionCall
   inputs [
    In {from=MaterialExpressionComponentMask_5}
    Contrast {from=MaterialExpressionFunctionInput_6}
   ]
   properties {materialFunction=&CheapContrast}
  }
  MaterialExpressionMaterialFunctionCall_1 {
   class=MaterialExpressionMaterialFunctionCall
   inputs [
    X {from=MaterialExpressionTextureSample_3, output=A}
    Y {from=MaterialExpressionTextureSample_3, output=A}
   ]
   properties {materialFunction=&MakeFloat2}
  }
  MaterialExpressionMaterialFunctionCall_10 {
   class=MaterialExpressionMaterialFunctionCall
   inputs [
    X {from=MaterialExpressionTextureSample_5, output=R}
    Y {from=MaterialExpressionTextureSample_5, output=R}
   ]
   properties {materialFunction=&MakeFloat2}
  }
  MaterialExpressionMaterialFunctionCall_2 {
   class=MaterialExpressionMaterialFunctionCall
   inputs [
    In {from=MaterialExpressionComponentMask_7}
    Contrast {from=MaterialExpressionFunctionInput_6}
   ]
   properties {materialFunction=&CheapContrast}
  }
  MaterialExpressionMaterialFunctionCall_6 {
   class=MaterialExpressionMaterialFunctionCall
   inputs [
    X {from=MaterialExpressionTextureSample_3, output=R}
    Y {from=MaterialExpressionTextureSample_3, output=R}
   ]
   properties {materialFunction=&MakeFloat2}
  }
  MaterialExpressionMaterialFunctionCall_7 {
   class=MaterialExpressionMaterialFunctionCall
   inputs [
    X {from=MaterialExpressionTextureSample_0, output=A}
    Y {from=MaterialExpressionTextureSample_0, output=A}
   ]
   properties {materialFunction=&MakeFloat2}
  }
  MaterialExpressionMaterialFunctionCall_8 {
   class=MaterialExpressionMaterialFunctionCall
   inputs [
    X {from=MaterialExpressionTextureSample_0, output=R}
    Y {from=MaterialExpressionTextureSample_0, output=R}
   ]
   properties {materialFunction=&MakeFloat2}
  }
  MaterialExpressionMaterialFunctionCall_9 {
   class=MaterialExpressionMaterialFunctionCall
   inputs [
    X {from=MaterialExpressionTextureSample_5, output=A}
    Y {from=MaterialExpressionTextureSample_5, output=A}
   ]
   properties {materialFunction=&MakeFloat2}
  }
  MaterialExpressionMultiply_0 {
   class=MaterialExpressionMultiply
   inputs [
    A {from=MaterialExpressionAbs_1}
    B {from=""}
   ]
   properties {constA=-1, constB=-1}
  }
  MaterialExpressionMultiply_1 {
   class=MaterialExpressionMultiply
   inputs [
    A {from=MaterialExpressionStaticSwitch_10}
    B {from=MaterialExpressionMultiply_24}
   ]
  }
  MaterialExpressionMultiply_18 {
   class=MaterialExpressionMultiply
   inputs [
    A {from=MaterialExpressionConvert_14}
    B {from=MaterialExpressionConvert_3}
   ]
  }
  MaterialExpressionMultiply_19 {
   class=MaterialExpressionMultiply
   inputs [
    A {from=MaterialExpressionConvert_13}
    B {from=MaterialExpressionConvert_3}
   ]
  }
  MaterialExpressionMultiply_20 {
   class=MaterialExpressionMultiply
   inputs [
    A {from=MaterialExpressionConvert_12}
    B {from=MaterialExpressionConvert_3}
   ]
  }
  MaterialExpressionMultiply_24 {
   class=MaterialExpressionMultiply
   inputs [
    A {from=MaterialExpressionFunctionInput_7}
    B {from=""}
   ]
   properties {constB=-1}
  }
  MaterialExpressionMultiply_33 {
   class=MaterialExpressionMultiply
   inputs [
    A {from=MaterialExpressionAppendVector_15}
    B {from=MaterialExpressionDotProduct_5}
   ]
  }
  MaterialExpressionMultiply_34 {
   class=MaterialExpressionMultiply
   inputs [
    A {from=MaterialExpressionAdd_12}
    B {from=MaterialExpressionAppendVector_18}
   ]
  }
  MaterialExpressionMultiply_35 {
   class=MaterialExpressionMultiply
   inputs [
    A {from=MaterialExpressionComponentMask_46}
    B {from=""}
   ]
   properties {constB=-1}
  }
  MaterialExpressionMultiply_36 {
   class=MaterialExpressionMultiply
   inputs [
    A {from=MaterialExpressionComponentMask_49}
    B {from=""}
   ]
   properties {constB=-1}
  }
  MaterialExpressionMultiply_37 {
   class=MaterialExpressionMultiply
   inputs [
    A {from=MaterialExpressionAdd_13}
    B {from=MaterialExpressionAppendVector_19}
   ]
  }
  MaterialExpressionMultiply_38 {
   class=MaterialExpressionMultiply
   inputs [
    A {from=MaterialExpressionAppendVector_20}
    B {from=MaterialExpressionDotProduct_6}
   ]
  }
  MaterialExpressionMultiply_39 {
   class=MaterialExpressionMultiply
   inputs [
    A {from=MaterialExpressionComponentMask_61}
    B {from=""}
   ]
   properties {constB=-1}
  }
  MaterialExpressionMultiply_4 {
   class=MaterialExpressionMultiply
   inputs [
    A {from=MaterialExpressionStaticSwitch_11}
    B {from=MaterialExpressionMultiply_24}
   ]
  }
  MaterialExpressionMultiply_40 {
   class=MaterialExpressionMultiply
   inputs [
    A {from=MaterialExpressionAdd_14}
    B {from=MaterialExpressionAppendVector_21}
   ]
  }
  MaterialExpressionMultiply_47 {
   class=MaterialExpressionMultiply
   inputs [
    A {from=MaterialExpressionAppendVector_22}
    B {from=MaterialExpressionDotProduct_7}
   ]
  }
  MaterialExpressionMultiply_6 {
   class=MaterialExpressionMultiply
   inputs [
    A {from=MaterialExpressionStaticSwitch_12}
    B {from=MaterialExpressionMultiply_24}
   ]
  }
  MaterialExpressionMultiply_80 {
   class=MaterialExpressionMultiply
   inputs [
    A {from=MaterialExpressionFunctionInput_9}
    B {from=""}
   ]
   properties {constB=-1}
  }
  MaterialExpressionNamedRerouteDeclaration_0 {
   class=MaterialExpressionNamedRerouteDeclaration
   inputs [None {from=MaterialExpressionFunctionInput_2}]
   properties {name=InWorldNormal, nodeColor=<rgba 0 0.9531 1 1>}
  }
  MaterialExpressionNamedRerouteDeclaration_1 {
   class=MaterialExpressionNamedRerouteDeclaration
   inputs [None {from=MaterialExpressionMaterialFunctionCall_2, output=Result}]
   properties {name=ZLerpFactor, nodeColor=<rgba 0 0 1 1>}
  }
  MaterialExpressionNamedRerouteDeclaration_2 {
   class=MaterialExpressionNamedRerouteDeclaration
   inputs [None {from=MaterialExpressionMaterialFunctionCall_0, output=Result}]
   properties {name=XLerpFactor, nodeColor=<rgba 1 0 0 1>}
  }
  MaterialExpressionNamedRerouteDeclaration_3 {
   class=MaterialExpressionNamedRerouteDeclaration
   inputs [None {from=MaterialExpressionTextureSample_3, output=RGBA}]
   properties {name=RawY, nodeColor=<rgba 0 1 0 1>}
  }
  MaterialExpressionNamedRerouteDeclaration_4 {
   class=MaterialExpressionNamedRerouteDeclaration
   inputs [None {from=MaterialExpressionTextureSample_0, output=RGBA}]
   properties {name=RawX, nodeColor=<rgba 1 0 0 1>}
  }
  MaterialExpressionNamedRerouteDeclaration_5 {
   class=MaterialExpressionNamedRerouteDeclaration
   inputs [None {from=MaterialExpressionTextureSample_5, output=RGBA}]
   properties {name=RawZ, nodeColor=<rgba 0 0 1 1>}
  }
  MaterialExpressionNamedRerouteUsage_0 {class=MaterialExpressionNamedRerouteUsage}
  MaterialExpressionNamedRerouteUsage_1 {class=MaterialExpressionNamedRerouteUsage}
  MaterialExpressionNamedRerouteUsage_10 {class=MaterialExpressionNamedRerouteUsage}
  MaterialExpressionNamedRerouteUsage_11 {class=MaterialExpressionNamedRerouteUsage}
  MaterialExpressionNamedRerouteUsage_12 {class=MaterialExpressionNamedRerouteUsage}
  MaterialExpressionNamedRerouteUsage_13 {class=MaterialExpressionNamedRerouteUsage}
  MaterialExpressionNamedRerouteUsage_14 {class=MaterialExpressionNamedRerouteUsage}
  MaterialExpressionNamedRerouteUsage_2 {class=MaterialExpressionNamedRerouteUsage}
  MaterialExpressionNamedRerouteUsage_3 {class=MaterialExpressionNamedRerouteUsage}
  MaterialExpressionNamedRerouteUsage_4 {class=MaterialExpressionNamedRerouteUsage}
  MaterialExpressionNamedRerouteUsage_5 {class=MaterialExpressionNamedRerouteUsage}
  MaterialExpressionNamedRerouteUsage_6 {class=MaterialExpressionNamedRerouteUsage}
  MaterialExpressionNamedRerouteUsage_7 {class=MaterialExpressionNamedRerouteUsage}
  MaterialExpressionNamedRerouteUsage_8 {class=MaterialExpressionNamedRerouteUsage}
  MaterialExpressionNamedRerouteUsage_9 {class=MaterialExpressionNamedRerouteUsage}
  MaterialExpressionNormalize_0 {class=MaterialExpressionNormalize, inputs [VectorInput {from=MaterialExpressionStaticSwitch_0, output=""}]}
  MaterialExpressionNormalize_4 {class=MaterialExpressionNormalize, inputs [VectorInput {from=MaterialExpressionStaticSwitch_4, output=""}]}
  MaterialExpressionNormalize_7 {class=MaterialExpressionNormalize, inputs [VectorInput {from=MaterialExpressionStaticSwitch_6, output=""}]}
  MaterialExpressionReroute_0 {class=MaterialExpressionReroute, inputs [None {from=MaterialExpressionFunctionInput_8}]}
  MaterialExpressionReroute_13 {class=MaterialExpressionReroute, inputs [None {from=MaterialExpressionDeriveNormalZ_1}]}
  MaterialExpressionReroute_14 {class=MaterialExpressionReroute, inputs [None {from=MaterialExpressionConvert_10}]}
  MaterialExpressionReroute_20 {class=MaterialExpressionReroute, inputs [None {from=MaterialExpressionDeriveNormalZ_2}]}
  MaterialExpressionReroute_7 {class=MaterialExpressionReroute, inputs [None {from=MaterialExpressionConvert_8}]}
  MaterialExpressionReroute_8 {class=MaterialExpressionReroute, inputs [None {from=MaterialExpressionDeriveNormalZ_0}]}
  MaterialExpressionReroute_9 {class=MaterialExpressionReroute, inputs [None {from=MaterialExpressionConvert_9}]}
  MaterialExpressionRotator_1 {
   class=MaterialExpressionRotator
   inputs [
    Coordinate {from=MaterialExpressionComponentMask_4, output=""}
    Time {from=MaterialExpressionFunctionInput_9, output=""}
   ]
   properties {centerX=0, centerY=0, speed=6.2832}
  }
  MaterialExpressionRotator_2 {
   class=MaterialExpressionRotator
   inputs [
    Coordinate {from=MaterialExpressionComponentMask_2, output=""}
    Time {from=MaterialExpressionFunctionInput_9, output=""}
   ]
   properties {centerX=0, centerY=0, speed=6.2832}
  }
  MaterialExpressionRotator_3 {
   class=MaterialExpressionRotator
   inputs [
    Coordinate {from=MaterialExpressionComponentMask_0, output=""}
    Time {from=MaterialExpressionFunctionInput_9, output=""}
   ]
   properties {centerX=0, centerY=0, speed=6.2832}
  }
  MaterialExpressionRotator_4 {
   class=MaterialExpressionRotator
   inputs [
    Coordinate {from=MaterialExpressionConstantBiasScale_0, output=""}
    Time {from=MaterialExpressionMultiply_80, output=""}
   ]
   properties {centerX=0, centerY=0, speed=6.2832}
  }
  MaterialExpressionRotator_5 {
   class=MaterialExpressionRotator
   inputs [
    Coordinate {from=MaterialExpressionConstantBiasScale_1, output=""}
    Time {from=MaterialExpressionMultiply_80, output=""}
   ]
   properties {centerX=0, centerY=0, speed=6.2832}
  }
  MaterialExpressionRotator_6 {
   class=MaterialExpressionRotator
   inputs [
    Coordinate {from=MaterialExpressionConstantBiasScale_2, output=""}
    Time {from=MaterialExpressionMultiply_80, output=""}
   ]
   properties {centerX=0, centerY=0, speed=6.2832}
  }
  MaterialExpressionSign_0 {class=MaterialExpressionSign, inputs [None {from=MaterialExpressionNamedRerouteUsage_0, output=""}], properties {}}
  MaterialExpressionStaticBool_0 {class=MaterialExpressionStaticBool}
  MaterialExpressionStaticBool_1 {class=MaterialExpressionStaticBool, properties {value=true}}
  MaterialExpressionStaticBool_2 {class=MaterialExpressionStaticBool}
  MaterialExpressionStaticBool_3 {class=MaterialExpressionStaticBool}
  MaterialExpressionStaticBool_4 {class=MaterialExpressionStaticBool, properties {value=true}}
  MaterialExpressionStaticSwitch_0 {
   class=MaterialExpressionStaticSwitch
   inputs [
    True {from=MaterialExpressionStaticSwitch_2, output=""}
    False {from=MaterialExpressionTransform_1, output=""}
    Value {from=MaterialExpressionFunctionInput_4, output=""}
   ]
   properties {}
  }
  MaterialExpressionStaticSwitch_1 {
   class=MaterialExpressionStaticSwitch
   inputs [
    True {from=MaterialExpressionMaterialFunctionCall_1, output=Result}
    False {from=MaterialExpressionMaterialFunctionCall_6, output=Result}
    Value {from=MaterialExpressionFunctionInput_5, output=""}
   ]
   properties {}
  }
  MaterialExpressionStaticSwitch_10 {
   class=MaterialExpressionStaticSwitch
   inputs [
    True {from=MaterialExpressionRotator_4, output=""}
    False {from=MaterialExpressionConstantBiasScale_0, output=""}
    Value {from=MaterialExpressionReroute_0, output=""}
   ]
   properties {}
  }
  MaterialExpressionStaticSwitch_11 {
   class=MaterialExpressionStaticSwitch
   inputs [
    True {from=MaterialExpressionRotator_5, output=""}
    False {from=MaterialExpressionConstantBiasScale_1, output=""}
    Value {from=MaterialExpressionReroute_0, output=""}
   ]
   properties {}
  }
  MaterialExpressionStaticSwitch_12 {
   class=MaterialExpressionStaticSwitch
   inputs [
    True {from=MaterialExpressionRotator_6, output=""}
    False {from=MaterialExpressionConstantBiasScale_2, output=""}
    Value {from=MaterialExpressionReroute_0, output=""}
   ]
   properties {}
  }
  MaterialExpressionStaticSwitch_13 {
   class=MaterialExpressionStaticSwitch
   inputs [
    True {from=MaterialExpressionNamedRerouteUsage_12, output=""}
    False {from=MaterialExpressionComponentMask_1, output=""}
    Value {from=MaterialExpressionFunctionInput_10, output=""}
   ]
   properties {}
  }
  MaterialExpressionStaticSwitch_14 {
   class=MaterialExpressionStaticSwitch
   inputs [
    True {from=MaterialExpressionNamedRerouteUsage_13, output=""}
    False {from=MaterialExpressionComponentMask_6, output=""}
    Value {from=MaterialExpressionFunctionInput_10, output=""}
   ]
   properties {}
  }
  MaterialExpressionStaticSwitch_15 {
   class=MaterialExpressionStaticSwitch
   inputs [
    True {from=MaterialExpressionNamedRerouteUsage_14, output=""}
    False {from=MaterialExpressionComponentMask_8, output=""}
    Value {from=MaterialExpressionFunctionInput_10, output=""}
   ]
   properties {}
  }
  MaterialExpressionStaticSwitch_16 {
   class=MaterialExpressionStaticSwitch
   inputs [
    True {from=MaterialExpressionLinearInterpolate_6, output=""}
    False {from=MaterialExpressionAdd_2, output=""}
    Value {from=MaterialExpressionFunctionInput_11, output=""}
   ]
   properties {}
  }
  MaterialExpressionStaticSwitch_17 {
   class=MaterialExpressionStaticSwitch
   inputs [
    True {from=MaterialExpressionConvert_17, output=""}
    False {from=MaterialExpressionAdd_3, output=""}
    Value {from=MaterialExpressionFunctionInput_11, output=""}
   ]
   properties {}
  }
  MaterialExpressionStaticSwitch_2 {
   class=MaterialExpressionStaticSwitch
   inputs [
    True {from=MaterialExpressionLinearInterpolate_7, output=""}
    False {from=MaterialExpressionAdd_1, output=""}
    Value {from=MaterialExpressionFunctionInput_11, output=""}
   ]
   properties {}
  }
  MaterialExpressionStaticSwitch_3 {
   class=MaterialExpressionStaticSwitch
   inputs [
    True {from=MaterialExpressionMaterialFunctionCall_7, output=Result}
    False {from=MaterialExpressionMaterialFunctionCall_8, output=Result}
    Value {from=MaterialExpressionFunctionInput_5, output=""}
   ]
   properties {}
  }
  MaterialExpressionStaticSwitch_4 {
   class=MaterialExpressionStaticSwitch
   inputs [
    True {from=MaterialExpressionStaticSwitch_16, output=""}
    False {from=MaterialExpressionTransform_3, output=""}
    Value {from=MaterialExpressionFunctionInput_4, output=""}
   ]
   properties {}
  }
  MaterialExpressionStaticSwitch_5 {
   class=MaterialExpressionStaticSwitch
   inputs [
    True {from=MaterialExpressionMaterialFunctionCall_9, output=Result}
    False {from=MaterialExpressionMaterialFunctionCall_10, output=Result}
    Value {from=MaterialExpressionFunctionInput_5, output=""}
   ]
   properties {}
  }
  MaterialExpressionStaticSwitch_6 {
   class=MaterialExpressionStaticSwitch
   inputs [
    True {from=MaterialExpressionStaticSwitch_17, output=""}
    False {from=MaterialExpressionTransform_5, output=""}
    Value {from=MaterialExpressionFunctionInput_4, output=""}
   ]
   properties {}
  }
  MaterialExpressionStaticSwitch_7 {
   class=MaterialExpressionStaticSwitch
   inputs [
    True {from=MaterialExpressionRotator_1, output=""}
    False {from=MaterialExpressionComponentMask_4, output=""}
    Value {from=MaterialExpressionFunctionInput_8, output=""}
   ]
   properties {}
  }
  MaterialExpressionStaticSwitch_8 {
   class=MaterialExpressionStaticSwitch
   inputs [
    True {from=MaterialExpressionRotator_2, output=""}
    False {from=MaterialExpressionComponentMask_2, output=""}
    Value {from=MaterialExpressionFunctionInput_8, output=""}
   ]
   properties {}
  }
  MaterialExpressionStaticSwitch_9 {
   class=MaterialExpressionStaticSwitch
   inputs [
    True {from=MaterialExpressionRotator_3, output=""}
    False {from=MaterialExpressionComponentMask_0, output=""}
    Value {from=MaterialExpressionFunctionInput_8, output=""}
   ]
   properties {}
  }
  MaterialExpressionSubtract_10 {
   class=MaterialExpressionSubtract
   inputs [
    A {from=MaterialExpressionMultiply_38}
    B {from=MaterialExpressionMultiply_37}
   ]
  }
  MaterialExpressionSubtract_13 {
   class=MaterialExpressionSubtract
   inputs [
    A {from=MaterialExpressionMultiply_47}
    B {from=MaterialExpressionMultiply_40}
   ]
  }
  MaterialExpressionSubtract_9 {
   class=MaterialExpressionSubtract
   inputs [
    A {from=MaterialExpressionMultiply_33}
    B {from=MaterialExpressionMultiply_34}
   ]
  }
  MaterialExpressionTextureObject_0 {
   class=MaterialExpressionTextureObject
   inputs []
   properties {samplerType=SAMPLERTYPE_Masks, texture=/Game/Global/Base4Layers_2022/Textures/T_VegetableVines_Packed}
  }
  MaterialExpressionTextureSample_0 {
   class=MaterialExpressionTextureSample
   inputs [
    UVs {from=MaterialExpressionStaticSwitch_8}
    Tex {from=MaterialExpressionFunctionInput_0}
    Apply View MipBias {from=""}
   ]
   properties {samplerType=SAMPLERTYPE_Masks}
  }
  MaterialExpressionTextureSample_3 {
   class=MaterialExpressionTextureSample
   inputs [
    UVs {from=MaterialExpressionStaticSwitch_7}
    Tex {from=MaterialExpressionFunctionInput_0}
    Apply View MipBias {from=""}
   ]
   properties {samplerType=SAMPLERTYPE_Masks}
  }
  MaterialExpressionTextureSample_5 {
   class=MaterialExpressionTextureSample
   inputs [
    UVs {from=MaterialExpressionStaticSwitch_9}
    Tex {from=MaterialExpressionFunctionInput_0}
    Apply View MipBias {from=""}
   ]
   properties {samplerType=SAMPLERTYPE_Masks}
  }
  MaterialExpressionTransform_1 {
   class=MaterialExpressionTransform
   inputs [None {from=MaterialExpressionStaticSwitch_2}]
   properties {transformSourceType=TRANSFORMSOURCE_World, transformType=TRANSFORM_Tangent}
  }
  MaterialExpressionTransform_3 {
   class=MaterialExpressionTransform
   inputs [None {from=MaterialExpressionStaticSwitch_16}]
   properties {transformSourceType=TRANSFORMSOURCE_World, transformType=TRANSFORM_Tangent}
  }
  MaterialExpressionTransform_5 {
   class=MaterialExpressionTransform
   inputs [None {from=MaterialExpressionStaticSwitch_17}]
   properties {transformSourceType=TRANSFORMSOURCE_World, transformType=TRANSFORM_Tangent}
  }
  MaterialExpressionVertexNormalWS_0 {class=MaterialExpressionVertexNormalWS, inputs [], properties {}}
  MaterialExpressionWorldPosition_0 {class=MaterialExpressionWorldPosition, inputs [], properties {worldPositionShaderOffset=WPT_ExcludeAllShaderOffsets}}
 }
 properties {bExposeToLibrary=true, description=Rotates a 2D Vector, libraryCategoriesText [Misc,Texturing]}
}
