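# Material function that layers procedural vegetable vines, with burnt edges, burn shadowing and mud, over input material attributes, toggled by UseVegetableVines.
# Two rotated, world-aligned packed vine projections are height-blended through noise for color, roughness, normals and optional displacement.
# Input mask, per-layer multipliers and world-aligned noise shape vine and burn coverage, while mud weights vary across walls, floors and ceilings.
# The IsLandscape input disables mud and limits textures to Z projection.
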MaterialFunction /Game/Global/Base4Layers_2022/MaterialFunctions/MF_VegetableVines
graph {nodes [
  n1 = NamedRerouteUsage()
  n2 = NamedRerouteUsage()
  n3 = BlendMaterialAttributes(A=@n2, B=NamedRerouteUsage(), Alpha=NamedRerouteUsage()) {vertexAttributeBlendType=UseB}
  n4 = NamedRerouteUsage()
  n5 = Constant()
  n6 = GetMaterialAttributes(None=BlendMaterialAttributes(
    A=StaticSwitch(True=@n1, False=@n3, Value=@n4)
    B=SetMaterialAttributes!(MaterialAttributes=@n2, BaseColor=@n5, Metallic=@n5, Specular=@n5) {attributeSetTypes [#1, #2, #3]}
    Alpha=NamedRerouteUsage()
   ) {vertexAttributeBlendType=UseB}) {attributeGetTypes [
    #1
    #2
    #3
    #4
   ]}
  n7 = SetMaterialAttributes!(MaterialAttributes=NamedRerouteUsage(), BaseColor=@n6.BaseColor, Metallic=@n6.BaseColor, Specular=@n6.BaseColor, Roughness=@n6.BaseColor) {attributeSetTypes [
    #1
    #2
    #3
    #4
   ]}
  n8 = NamedRerouteUsage()
  n9 = NamedRerouteUsage()
  FunctionOutput(None=StaticSwitchParameter(True=BlendMaterialAttributes(A=@n7, B=@n8, Alpha=@n9) {vertexAttributeBlendType=UseB}, False=NamedRerouteUsage()) {group=Vegetable, parameterName=UseVegetableVines, sortPriority=0})
  n10 = NamedRerouteUsage()
  n11 = Abs(None=ConstantBiasScale(None=ComponentMask(None=@n10) {g=true, r=true}) {bias=-0.5, scale=2})
  n12 = DotProduct(A=@n11, B=@n11)
  n13 = NamedRerouteUsage()
  n14 = Abs(None=ConstantBiasScale(None=ComponentMask(None=@n13) {g=true, r=true}) {bias=-0.5, scale=2})
  n15 = DotProduct(A=@n14, B=@n14)
  n16 = LinearInterpolate(A=@n12, B=@n15, Alpha=NamedRerouteUsage())
  n17 = LinearInterpolate(A=@n15, B=@n12, Alpha=NamedRerouteUsage())
  Max(A=@n16, B=@n17)
  n18 = NamedRerouteUsage()
  n19 = NamedRerouteUsage()
  n20 = ComponentMask(None=Saturate(None=Max(A=@n18, B=@n19))) {b=true}
  n21 = Multiply(A=@n20, B=@n20)
  Multiply(A=@n21, B=@n21)
  n22 = TextureObjectParameter() {
   group=Vegetable
   parameterName=VegetableVines
   samplerType=SAMPLERTYPE_Masks
   texture=/Game/Global/Base4Layers_2022/Textures/T_VegetableVines_Packed
  }
  NamedRerouteDeclaration(None=@n22) {name=VineTex, nodeColor=<rgba 0 0.2734 1 1>}
  n23 = ScalarParameter() {defaultValue=280.5032, group=Vegetable, parameterName=VineSizeA, sortPriority=1}
  n24 = StaticBool() {value=true}
  n25 = MaterialFunctionCall!(
   TextureObject=NamedRerouteUsage()
   TextureSize=MaterialFunctionCall(X=@n23, Y=@n23, Z=@n23) {materialFunction=&MakeFloat3}.Result
   RotateUVs=@n24
   RotationAngle=ScalarParameter() {defaultValue=0.1288, group=Vegetable, parameterName=VineRotationA, sortPriority=1}
   Export Float 4=@n24
   NormalStrength=ScalarParameter() {defaultValue=1, group=Vegetable, parameterName=VineNormalStrength, sortPriority=1}
   UseAGNormals
   WorldSpace=@n24
   WorldPosition
   WorldNormal
   ProjectionTransitionContrast=ScalarParameter() {defaultValue=0.1, group=Vegetable, parameterName=VineProjectionContrast, sortPriority=1}
   Use High Quality Normals=NamedRerouteUsage()
  ) {materialFunction=&/MF_WorldAlignedPacked}
  n26 = NamedRerouteUsage()
  NamedRerouteDeclaration(None=StaticSwitch(True=@n25.Z Normal, False=@n25.Z Normal, Value=@n26)) {name=VineWorldNormalA, nodeColor=<rgba 0.0312 0 1 1>}
  n27 = StaticBool() {value=true}
  n28 = StaticBool()
  n29 = QualitySwitch(Default=@n27, Low=@n28, High=@n27, Medium=@n28, Epic=@n27) {desc=has GBuffers}
  NamedRerouteDeclaration(None=ShadingPathSwitch(Default=@n29, Deferred=@n29, Forward=@n28, Mobile=@n28)) {name=UseHQNormals, nodeColor=<rgba 0 0.2969 1 1>}
  n30 = StaticSwitch(True=@n25.Z Texture, False=@n25.Z Texture, Value=@n26)
  NamedRerouteDeclaration(None=OneMinus(None=ComponentMask(None=@n30) {a=true})) {name=BurnHeightA, nodeColor=<rgba 0.3594 0 1 1>}
  n31 = ScalarParameter() {defaultValue=304.906, group=Vegetable, parameterName=VineSizeB, sortPriority=1}
  n32 = StaticBool() {value=true}
  n33 = MaterialFunctionCall(
   TextureObject=NamedRerouteUsage()
   TextureSize=MaterialFunctionCall(X=@n31, Y=@n31, Z=@n31) {materialFunction=&MakeFloat3}.Result
   RotateUVs=@n32
   RotationAngle=ScalarParameter() {defaultValue=-0.071, group=Vegetable, parameterName=VineRotationB, sortPriority=1}
   Export Float 4=@n32
   NormalStrength=ScalarParameter() {defaultValue=1, group=Vegetable, parameterName=VineNormalStrength, sortPriority=1}
   WorldSpace=@n32
   ProjectionTransitionContrast=ScalarParameter() {defaultValue=0.1, group=Vegetable, parameterName=VineProjectionContrast, sortPriority=1}
   Use High Quality Normals=NamedRerouteUsage()
  ) {materialFunction=&MF_WorldAlignedPacked}
  n34 = NamedRerouteUsage()
  n35 = StaticSwitch(True=@n33.Z Texture, False=@n33.Z Texture, Value=@n34)
  NamedRerouteDeclaration(None=OneMinus(None=ComponentMask(None=@n35) {a=true})) {name=BurnHeightB, nodeColor=<rgba 0.3594 0 1 1>}
  NamedRerouteDeclaration(None=@n30) {name=VineRawA, nodeColor=<rgba 1 0 0.6094 1>}
  NamedRerouteDeclaration(None=@n35) {name=VineRawB, nodeColor=<rgba 1 0.75 0 1>}
  n36 = MaterialFunctionCall(
   TextureObject=NamedRerouteUsage()
   TextureSize=ScalarParameter() {defaultValue=640, group=Vegetable, parameterName=VineHeightNoiseSize, sortPriority=4}
  ) {materialFunction=&WorldAlignedTexture}
  n37 = StaticSwitch(True=@n36.Z Texture, False=@n36.Z Texture, Value=NamedRerouteUsage())
  n38 = ComponentMask(None=Saturate(None=ConstantBiasScale(None=@n37) {bias=-0.1, scale=3})) {g=true}
  n39 = NamedRerouteUsage()
  n40 = NamedRerouteUsage()
  n41 = ScalarParameter() {defaultValue=2, group=Vegetable, parameterName=VineLayerContrast, sortPriority=4}
  NamedRerouteDeclaration(None=Saturate(None=MaterialFunctionCall!(Diffuse 1, Diffuse 2, Transistion Phase=@n38, Height Texture 1=@n39, Height Texture 2=@n40, Contrast=@n41) {materialFunction=/Engine/Functions/Engine_MaterialFunctions02/Texturing/HeightLerpWithTwoHeightMaps}.Alpha)) {name=VineABLerp, nodeColor=<rgba 0 1 0.4609 1>}
  n42 = TextureObjectParameter() {
   group=Vegetable
   parameterName=VegetableNoise
   samplerType=SAMPLERTYPE_Masks
   texture=/Game/Global/Base4Layers_2022/Textures/T_VegetableVines_Noise
  }
  NamedRerouteDeclaration(None=@n42) {name=NoiseTex, nodeColor=<rgba 1 0 0.4453 1>}
  n43 = VectorParameter() {group=Vegetable, parameterName=VegetableMudColor, sortPriority=19, defaultValue=<rgba 0.1322 0.051 0.0122 1>}
  NamedRerouteDeclaration(None=SetMaterialAttributes!(
    MaterialAttributes
    BaseColor=@n43.RGB
    Specular=ScalarParameter() {defaultValue=0.25, group=Vegetable, parameterName=VegetableMudSpecular, sortPriority=22}
    Roughness=ScalarParameter() {defaultValue=0.8, group=Vegetable, parameterName=VegetableMudRoughness, sortPriority=22}
   ) {attributeSetTypes [#1, #3, #4]}) {name=MudMat, nodeColor=<rgba 1 0 0.2109 1>}
  n44 = VertexNormalWS()
  n45 = Constant3Vector() {constant=<rgba 1 1 1.25 0>}
  NamedRerouteDeclaration(None=MaterialFunctionCall(
    TextureObject=NamedRerouteUsage()
    TextureSize=ScalarParameter() {defaultValue=1307.7621, group=Vegetable, parameterName=VineNoiseSize, sortPriority=1}
    World Space Normal=Multiply(A=@n44, B=@n45)
    ProjectionTransitionContrast=ScalarParameter() {defaultValue=1, group=Vegetable, parameterName=VineNoiseProjectionContrast, sortPriority=1}
   ) {materialFunction=&WorldAlignedTexture}.XYZ Texture) {name=NoiseRaw, nodeColor=<rgba 0 0.5312 1 1>}
  n46 = MaterialFunctionCall(Float3=NamedRerouteUsage()) {materialFunction=&BreakOutFloat3Components}
  n47 = Multiply(A=@n46.R, B=ScalarParameter() {defaultValue=0.5, group=Vegetable, parameterName=VegetableMudWallIntensity, sortPriority=20})
  n48 = ComponentMask(None=VertexNormalWS()) {b=true}
  n49 = Multiply(
   A=@n46.B
   B=If(
    A=@n48
    "A > B"=ScalarParameter() {defaultValue=0.4, group=Vegetable, parameterName=VegetableMudFloorIntensity, sortPriority=20}
    "A < B"=ScalarParameter() {defaultValue=0.2, group=Vegetable, parameterName=VegetableMudCeilingIntensity, sortPriority=20}
   )
  )
  n50 = MaterialFunctionCall(
   In=Abs(None=@n48)
   Contrast=ScalarParameter() {defaultValue=1, group=Vegetable, parameterName=VegetableMudWallTransitionContrast, sortPriority=20}
  ) {materialFunction=&CheapContrast}
  NamedRerouteDeclaration(None=Multiply(A=@n43.A, B=LinearInterpolate(A=@n47, B=@n49, Alpha=@n50.Result))) {name=MudMask, nodeColor=<rgba 0.8516 1 0 1>}
  n51 = ComponentMask(None=@n30) {b=true}
  n52 = Saturate(None=InverseLinearInterpolate(Value=@n51) {constA=0.5, constB=0.75})
  NamedRerouteDeclaration(None=@n52) {name=VineMaskA, nodeColor=<rgba 0.2891 1 0 1>}
  n53 = VectorParameter() {group=Vegetable, parameterName=VineBurnColor, sortPriority=5, defaultValue=<rgba 0.005 0.0033 0.0011 1>}
  n54 = VectorParameter() {
   group=Vegetable
   parameterName=VineBurnLayerMult
   sortPriority=5
   channelNames {a=Layer 4, b=Layer 3, g=Layer 2, r=Layer 1}
   defaultValue=<rgba 1 1 1 1>
  }
  n55 = Multiply(
   A=@n53.RGB
   B=MaterialFunctionCall!(1=@n54.Layer 1, 2=@n54.Layer 1, 3=@n54.Layer 1, 4=@n54.Layer 1) {materialFunction=&/MF_UV2LerpScalar_Layers_New}.Result
  )
  n56 = NamedRerouteUsage()
  n57 = VectorParameter() {group=Vegetable, parameterName=VineColor, sortPriority=1, defaultValue=<rgba 0.1683 0.0685 0.0382 1>}
  n58 = VectorParameter() {group=Vegetable, parameterName=VineColorDark, sortPriority=1, defaultValue=<rgba 0.0193 0.025 0.0074 1>}
  n59 = NamedRerouteUsage()
  n60 = ScalarParameter() {defaultValue=0.3278, group=Vegetable, parameterName=VineRoughness, sortPriority=3}
  n61 = ScalarParameter() {defaultValue=0.0409, group=Vegetable, parameterName=VineRoughnessB, sortPriority=3}
  n62 = NamedRerouteUsage()
  n63 = NamedRerouteUsage()
  n64 = NamedRerouteUsage()
  n65 = LinearInterpolate(A=@n63, B=@n64, Alpha=NamedRerouteUsage())
  n66 = LinearInterpolate(A=@n64, B=@n63, Alpha=NamedRerouteUsage())
  n67 = NamedRerouteUsage()
  n68 = SetMaterialAttributes!(
   MaterialAttributes
   BaseColor=LinearInterpolate(A=StaticSwitch(True=@n53.RGB, False=@n55, Value=@n56), B=LinearInterpolate(A=@n57.RGB, B=@n58.RGB, Alpha=@n59), Alpha=NamedRerouteUsage())
   Roughness=LinearInterpolate(
    A=ScalarParameter() {defaultValue=0.9, group=Vegetable, parameterName=VineBurnRoughness, sortPriority=6}
    B=LinearInterpolate(A=@n60, B=@n61, Alpha=@n59)
    Alpha=@n62
   )
   Normal=Transform(None=LinearInterpolate(A=@n65, B=@n66, Alpha=@n67)) {transformSourceType=TRANSFORMSOURCE_World, transformType=TRANSFORM_Tangent}
   Specular=LinearInterpolate(
    A=ScalarParameter() {defaultValue=0.15, group=Vegetable, parameterName=VineBurnSpecular, sortPriority=6}
    B=ScalarParameter() {defaultValue=0.21, group=Vegetable, parameterName=VineSpecular, sortPriority=3}
    Alpha=@n62
   ) {constA=0.25, constB=0.5}
  ) {attributeSetTypes [
    #1
    #4
    #5
    #3
   ]}
  n69 = NamedRerouteUsage()
  n70 = Saturate(None=ScalarParameter() {defaultValue=0.1, group=Vegetable, parameterName=VineDisplacementAmt, sliderMax=1})
  n71 = NamedRerouteUsage()
  n72 = NamedRerouteUsage()
  n73 = Multiply(A=Saturate(None=Multiply(A=@n71) {constB=5}), B=Add(A=@n71, B=OneMinus(None=@n72)))
  n74 = NamedRerouteUsage()
  n75 = Multiply(A=Add(A=@n74, B=@n72), B=Saturate(None=Multiply(A=@n74) {constB=5}))
  n76 = Add(
   A=Saturate(None=GetMaterialAttributes(None=@n69) {attributeGetTypes [#6]}.Displacement)
   B=Multiply(A=Multiply(A=@n70) {constB=0.5}, B=Max(A=@n73, B=@n75))
  )
  NamedRerouteDeclaration(None=StaticSwitch(
    True=SetMaterialAttributes!(MaterialAttributes=@n68, Displacement=@n76) {attributeSetTypes [#6]}
    False=@n68
    Value=NamedRerouteUsage()
   )) {name=VegetableMat, nodeColor=<rgba 0 1 0.1562 1>}
  n77 = Max(A=NamedRerouteUsage(), B=NamedRerouteUsage())
  n78 = NamedRerouteUsage()
  n79 = NamedRerouteUsage()
  n80 = VectorParameter() {
   desc=Range between 0.0 and 2.0
   group=Vegetable
   parameterName=VineLayerMask
   sortPriority=4
   channelNames {a=Layer 4, b=Layer 3, g=Layer 2, r=Layer 1}
   defaultValue=<rgba 1 1 1 1>
  }
  n81 = Multiply(
   A=NamedRerouteUsage()
   B=MaterialFunctionCall(1=@n80.Layer 1, 2=@n80.Layer 1, 3=@n80.Layer 1, 4=@n80.Layer 1) {materialFunction=&MF_UV2LerpScalar_Layers_New}.Result
  )
  n82 = ConstantBiasScale(None=ComponentMask(None=NamedRerouteUsage()) {g=true}) {bias=-0.1, scale=1.5}
  n83 = Max(A=Add(A=OneMinus(None=@n81), B=Saturate(None=@n82))) {constB=0}
  n84 = ScalarParameter() {defaultValue=1.5, group=Vegetable, parameterName=VineHeightNoiseInfluence, sortPriority=4}
  n85 = SmoothStep(Value=Subtract(A=Max(A=@n78, B=@n79), B=Multiply(A=@n83, B=@n84) {constB=1.1})) {constMax=0.25}
  n86 = NamedRerouteDeclaration(None=Saturate(None=Min(A=@n77, B=@n85))) {name=VineMask, nodeColor=<rgba 0.4766 0 1 1>}
  n87 = NamedRerouteUsage()
  n88 = NamedRerouteUsage()
  n89 = Add(
   A=Multiply(A=Min(A=@n87, B=@n88), B=ScalarParameter() {defaultValue=4.0763, group=Vegetable, parameterName=VineBurnHeightAdjustScale, sortPriority=6})
   B=ScalarParameter() {defaultValue=0.1315, group=Vegetable, parameterName=VineBurnHeightAdjustOffset, sortPriority=6}
  ) {constB=-0.1}
  n90 = Max(B=InverseLinearInterpolate(
    A=ScalarParameter() {defaultValue=0.232, group=Vegetable, parameterName=VineNoiseMaskAdjustA, sortPriority=4}
    B=ScalarParameter() {defaultValue=0.688, group=Vegetable, parameterName=VineNoiseMaskAdjustB, sortPriority=4}
    Value=@n83
   ))
  n91 = SmoothStep(
   Max=ScalarParameter() {defaultValue=0.5, group=Vegetable, parameterName=VineBurnMaskEdge, sortPriority=6}
   Value=Subtract(A=OneMinus(None=@n89), B=@n90)
  ) {constMax=0.5}
  NamedRerouteDeclaration(None=Multiply(A=Max(A=@n86, B=@n91), B=LinearInterpolate(A=@n53.A, Alpha=@n62) {constA=1})) {name=VegetableMask, nodeColor=<rgba 0 1 0.6953 1>}
  n92 = Constant() {r=0.212}
  n93 = Constant() {r=1}
  NamedRerouteDeclaration(None=FunctionInput(Preview=SetMaterialAttributes!(MaterialAttributes, BaseColor=@n92, Roughness=@n93) {attributeSetTypes [#1, #4]}) {bUsePreviewValueAsDefault=true, inputType=FunctionInput_MaterialAttributes}) {name=InMat, nodeColor=<rgba 1 0.375 0 1>}
  n94 = NamedRerouteUsage()
  n95 = NamedRerouteUsage()
  n96 = Subtract(A=OneMinus(None=Min(A=@n94, B=@n95)), B=@n90)
  n97 = ScalarParameter() {defaultValue=0.642, group=Vegetable, parameterName=VineBurnShadowIntensity, sortPriority=10}
  NamedRerouteDeclaration(None=Saturate(None=Multiply(A=@n96, B=@n97) {constB=0.2})) {name=BurnShadowMask, nodeColor=<rgba 0 0 0 1>}
  NamedRerouteDeclaration(None=@n38) {name=VineHeightLerpNoise, nodeColor=<rgba 0.6641 0 1 1>}
  n98 = LinearInterpolate(Alpha=ConstantBiasScale(None=@n51) {bias=-0.5, scale=2}) {constA=0.5}
  NamedRerouteDeclaration(None=Saturate(None=Multiply(A=@n52, B=@n98))) {name=VineHeightA, nodeColor=<rgba 1 0 0.6094 1>}
  n99 = ComponentMask(None=@n35) {b=true}
  n100 = Saturate(None=InverseLinearInterpolate(Value=@n99) {constB=0.75})
  NamedRerouteDeclaration(None=@n100) {name=VineMaskB, nodeColor=<rgba 0 0.9297 1 1>}
  n101 = LinearInterpolate(Alpha=ConstantBiasScale(None=@n99) {bias=-0.5, scale=2}) {constA=0.5}
  NamedRerouteDeclaration(None=Saturate(None=Multiply(A=@n100, B=@n101))) {name=VineHeightB, nodeColor=<rgba 1 0.75 0 1>}
  NamedRerouteDeclaration(None=StaticSwitch(True=@n33.Z Normal, False=@n33.Z Normal, Value=@n34)) {name=VineWorldNormalB, nodeColor=<rgba 0.4297 0 1 1>}
  n102 = FunctionInput() {
   bUsePreviewValueAsDefault=true
   description="0.0 to 2.0 range. 0.0 removes vines, 2.0 shows full vines without noise mask."
   inputName=Mask
   inputType=FunctionInput_Scalar
   sortPriority=3
   previewValue=<1 0 0 1>
  }
  NamedRerouteDeclaration(None=@n102) {name=InMask, nodeColor=<rgba 1 0.6563 0 1>}
  n103 = FunctionInput(Preview=StaticBool()) {
   bUsePreviewValueAsDefault=true
   description=Disables mud and uses Z projected textures only
   inputName=IsLandscape
   inputType=FunctionInput_StaticBool
   sortPriority=1
  }
  NamedRerouteDeclaration(None=@n103) {name=IsLandscape, nodeColor=<rgba 0.3359 0 1 1>}
  n104 = NamedRerouteUsage()
  n105 = Abs(None=VertexNormalWS())
  n106 = ScalarParameter() {defaultValue=0.1, group=Vegetable, parameterName=VineProjectionContrast, sortPriority=1}
  n107 = ConstantBiasScale(None=MaterialFunctionCall(In=ComponentMask(None=@n105) {r=true}, Contrast=@n106) {materialFunction=&CheapContrast}.Result) {bias=-0.5, scale=2}
  n108 = MaterialFunctionCall(In=ComponentMask(None=@n105) {b=true}, Contrast=@n106) {materialFunction=&CheapContrast}
  n109 = ConstantBiasScale(None=@n108.Result) {bias=-0.5, scale=2}
  n110 = LinearInterpolate(A=OneMinus(None=Abs(None=@n107)), B=OneMinus(None=Abs(None=@n109)), Alpha=@n108.Result)
  n111 = NamedRerouteDeclaration(None=Multiply(A=LinearInterpolate(A=@n16, B=@n17, Alpha=@n104), B=LinearInterpolate(Alpha=@n110) {constA=1, constB=1.3})) {name=BurnColorLerp, nodeColor=<rgba 1 0.7969 0 1>}
  NamedRerouteDeclaration(None=Saturate(None=InverseLinearInterpolate(Value=@n111) {constA=0.05, constB=0.3})) {name=VineColorLerp, nodeColor=<rgba 1 0 0.5625 1>}
  NamedRerouteUsage()
 ]}
