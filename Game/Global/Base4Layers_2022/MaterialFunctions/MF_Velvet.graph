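# Material function that optionally adds a velvet sheen to incoming material attributes, selectable for each of four layers.
# A tiled distortion texture perturbs the normal and shapes specular; camera-to-normal dot product becomes a repeating fresnel band.
# It overrides metallic and blends base color toward a brightened, saturation-adjusted reflection color.
# UV1's U tile selects the layer weight, optionally multiplied by an SFX mask channel, to blend velvet over the original.
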MaterialFunction /Game/Global/Base4Layers_2022/MaterialFunctions/MF_Velvet
graph {nodes [
  n1 = NamedRerouteUsage()
  n2 = Multiply(A=NamedRerouteUsage(), B=ScalarParameter() {defaultValue=0.8, group=Velvet, parameterName=Velvet BaseColor Multiplier, sortPriority=4})
  n3 = Multiply(A=@n2, B=ScalarParameter() {defaultValue=3, group=Velvet, parameterName=Velvet Reflection Multiplier, sortPriority=5}) {constB=8}
  n4 = Multiply(A=Add(A=ScalarParameter() {defaultValue=1, group=Velvet, parameterName=Velvet Reflection Saturation, sortPriority=6}) {constB=-1}) {constB=-1}
  n5 = SetMaterialAttributes!(
   MaterialAttributes=NamedRerouteUsage()
   BaseColor=LinearInterpolate(A=@n2, B=Desaturation(None=@n3, Fraction=@n4), Alpha=NamedRerouteUsage())
   Metallic=ScalarParameter() {defaultValue=0.4, group=Velvet, parameterName=Velvet Metallic Override}
   Normal=NamedRerouteUsage()
   Specular=Multiply(A=NamedRerouteUsage(), B=NamedRerouteUsage())
  ) {attributeSetTypes [
    #1
    #2
    #3
    #4
   ]}
  n6 = FunctionInput() {bUsePreviewValueAsDefault=true, inputName=SFX, sortPriority=3}
  n7 = Add(A=ComponentMask(None=TextureCoordinate() {coordinateIndex=1}) {r=true}) {constB=-0.5}
  n8 = Abs(None=@n7)
  n9 = OneMinus(None=Saturate(None=Round(None=@n8)))
  n10 = ScalarParameter() {defaultValue=1, group=Velvet, parameterName=Velvet L1, sortPriority=-4}
  n11 = Add(A=@n7) {constB=-1}
  n12 = Abs(None=@n11)
  n13 = OneMinus(None=Saturate(None=Round(None=@n12)))
  n14 = ScalarParameter() {group=Velvet, parameterName=Velvet L2, sortPriority=-3}
  n15 = Add(A=@n11) {constB=-1}
  n16 = Abs(None=@n15)
  n17 = OneMinus(None=Saturate(None=Round(None=@n16)))
  n18 = ScalarParameter() {group=Velvet, parameterName=Velvet L3, sortPriority=-2}
  n19 = Abs(None=Add(A=@n15) {constB=-1})
  n20 = OneMinus(None=Saturate(None=Round(None=@n19)))
  n21 = ScalarParameter() {group=Velvet, parameterName=Velvet L4, sortPriority=-1}
  n22 = Max(A=Max(A=Multiply(A=@n9, B=@n10), B=Multiply(A=@n13, B=@n14)), B=Max(A=Multiply(A=@n17, B=@n18), B=Multiply(A=@n20, B=@n21)))
  n23 = StaticSwitchParameter(True=Multiply(A=ChannelMaskParameter(None=@n6) {group=Velvet, parameterName=Velvet Mask Channel, sortPriority=0}.R, B=@n22), False=@n22) {defaultValue=true, group=Velvet, parameterName=Use SFX Mask, sortPriority=-99}
  FunctionOutput(None=StaticSwitchParameter(True=BlendMaterialAttributes(A=@n1, B=@n5, Alpha=@n23), False=@n1) {group=Velvet, parameterName=UseVelvet, sortPriority=-100})
  n24 = GetMaterialAttributes(None=FunctionInput() {inputName=Material, inputType=FunctionInput_MaterialAttributes}) {attributeGetTypes [#1, #3, #4]}
  NamedRerouteDeclaration(None=@n24.MaterialAttributes) {name=Material Input, nodeColor=<rgba 1 0.4922 0 1>}
  NamedRerouteDeclaration(None=@n24.BaseColor) {name=BaseColor Input, nodeColor=<rgba 1 0 0.2109 1>}
  NamedRerouteDeclaration(None=@n24.Normal) {name=Normal Input, nodeColor=<rgba 0 0.5312 1 1>}
  NamedRerouteDeclaration(None=@n24.Specular) {name=Specular Input, nodeColor=<rgba 0 0.0859 1 1>}
  n25 = TextureSampleParameter2D(UVs=Multiply(A=TextureCoordinate(), B=ScalarParameter() {defaultValue=4, group=Velvet, parameterName=Velvet Texture Tiling, sortPriority=2})) {
   group=Velvet
   parameterName=Velvet Texture
   samplerType=SAMPLERTYPE_LinearColor
   sortPriority=1
   texture=/Game/Global/Textures/Default/Masks_and_Gradients/T_Velvet_Distortion
  }
  NamedRerouteDeclaration(None=Power(Base=@n25.B, Exp=ScalarParameter() {defaultValue=1, group=Velvet, parameterName=Velvet Specular Power}) {constExponent=8}) {name=Specular Multiplier, nodeColor=<rgba 0.4062 1 0 1>}
  n26 = NamedRerouteUsage()
  n27 = Multiply(
   A=Add(A=@n25.RGB) {constB=-0.5}
   B=ScalarParameter() {defaultValue=0.7, group=Velvet, parameterName=Velvet Distortion Multiplier, sortPriority=3}
  ) {constB=1.6}
  n28 = AppendVector(A=Add(A=ComponentMask(None=@n26) {g=true, r=true}, B=ComponentMask(None=@n27) {g=true, r=true}), B=ComponentMask(None=@n26) {b=true})
  n29 = DotProduct(A=CameraVectorWS(), B=Transform(None=NamedRerouteDeclaration(None=@n28) {name=Normal Distorted, nodeColor=<rgba 1 0.3281 0 1>}))
  n30 = Frac(None=Multiply(A=Saturate(None=@n29), B=ScalarParameter() {defaultValue=1.4, group=Velvet, parameterName=Velvet Fresnel Tiling}) {constB=1.2})
  n31 = Multiply(A=Abs(None=Add(A=@n30) {constB=-0.5})) {constB=2}
  n32 = ScalarParameter() {defaultValue=2, group=Velvet, parameterName=Velvet Fresnel Power}
  NamedRerouteDeclaration(None=Multiply(A=Power(Base=@n31, Exp=@n32), B=NamedRerouteUsage())) {name=Fresnel Mask, nodeColor=<rgba 0 0.5781 1 1>}
 ]}
