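# Material function for Asteria foliage that tints near-grayscale areas of a MaterialAttributes Base Color, outputting the updated attributes and the mask.
# The mask derives from (max + min) × min of red and blue only; green is unused.
# Saturated, it remaps through lerp parameters and blends Base Color toward its product with InTint, which defaults to a Masked Tint parameter.
# The exponent and multiplier branch feeds nothing.
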MaterialFunction /Game/Environments/Asteria/Foliage/HeliosMaterials/Masters/Functions/MF_MPColorMA
graph {
 properties {description="Generates a mask from non-saturated colors (grayscale) in order to tint parts of the Albedo"}
 nodes [
  n1 = GetMaterialAttributes(None=FunctionInput() {description=Reads Base Color, inputName=MA, inputType=FunctionInput_MaterialAttributes}) {attributeGetTypes [#1]}
  n2 = Multiply(
   A=@n1.BaseColor
   B=FunctionInput(Preview=VectorParameter() {group=MPColor, parameterName=Masked Tint, defaultValue=<rgba 1 1 1 0>}.RGB) {bUsePreviewValueAsDefault=true, inputName=InTint, sortPriority=1}
  )
  n3 = NamedRerouteUsage()
  FunctionOutput(None=SetMaterialAttributes!(MaterialAttributes=@n1.MaterialAttributes, BaseColor=LinearInterpolate(A=@n1.BaseColor, B=@n2, Alpha=@n3)) {attributeSetTypes [#1]}) {description=Writes Base Color, outputName="Result (MA)"}
  FunctionOutput(None=@n3) {outputName="Mask (S)", sortPriority=1}
  n4 = ScalarParameter() {defaultValue=-0.8, group=MPColor, parameterName=MP Mask LerpA}
  n5 = ScalarParameter() {defaultValue=15, group=MPColor, parameterName=MP Mask LerpB}
  n6 = MaterialFunctionCall(Float3=@n1.BaseColor) {materialFunction=&BreakOutFloat3Components}
  n7 = Max(A=@n6.R, B=@n6.R)
  n8 = Min(A=Min(A=@n6.R, B=@n6.R), B=@n6.B)
  n9 = Multiply(A=Add(A=Max(A=@n7, B=@n6.B), B=@n8), B=@n8)
  n10 = Saturate(None=@n9)
  NamedRerouteDeclaration(None=Saturate(None=LinearInterpolate(A=@n4, B=@n5, Alpha=@n10))) {name=outMask, nodeColor=<rgba 1 0.6563 0 1>}
  n11 = ScalarParameter() {defaultValue=2, desc=Mask Exponent Value, group=MPColor, parameterName=MP Mask Exp}
  Saturate(None=Multiply(A=Power(Base=@n9, Exp=@n11), B=ScalarParameter() {defaultValue=2, desc=Mask Multiplier, group=MPColor, parameterName=MP Mask Mul}) {constB=2})
 ]
}
