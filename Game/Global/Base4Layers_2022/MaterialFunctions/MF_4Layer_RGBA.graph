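# Material function for the four-layer base material that outputs an RGBA one-hot mask, with each layer mapped to one channel.
# It truncates the U texture coordinate to an integer tile index, so UV placement selects the layer.
# Four If nodes compare that index to each layer number and output one on a match, zero otherwise.
# AppendMany packs the results into a vec4, with red as the first layer through alpha as the fourth.
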MaterialFunction /Game/Global/Base4Layers_2022/MaterialFunctions/MF_4Layer_RGBA
graph {nodes [
  n1 = Truncate!(None=ComponentMask(None=TextureCoordinate()) {r=true}) {is=class MaterialExpressionTruncate}
  n2 = Constant()
  n3 = Constant() {r=1}
  n4 = Constant()
  n5 = Constant() {r=1}
  n6 = Constant()
  n7 = Constant() {r=1}
  n8 = Constant()
  n9 = Constant() {r=1}
  FunctionOutput(None=MaterialFunctionCall(
    R=If(A=@n1, "A > B"=@n2, "A == B"=@n3, "A < B"=@n2)
    G=If(A=@n1, "A > B"=@n4, "A == B"=@n5, "A < B"=@n4) {constB=1}
    B=If(A=@n1, "A > B"=@n6, "A == B"=@n7, "A < B"=@n6) {constB=2}
    A=If(A=@n1, "A > B"=@n8, "A == B"=@n9, "A < B"=@n8) {constB=3}
   ) {materialFunction=&AppendMany}.RGBA) {description="Results a vec4 with each layer representing a single component (ex. layer 1 = red, 2 = green, etc)"}
 ]}
