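# Optional 4-layer material function that adds a thin-film iridescent color effect to incoming material attributes.
# Normal-aware Fresnel, optionally offset by a mask with view-dependent shift and animated time, indexes a curve-atlas spectrum.
# Static switches modulate that color by Fresnel, SFX mask, layer mask, albedo and roughness, then blend it into BaseColor and optionally eye-adaptation-compensated Emissive.
# Low quality and mobile bypass the effect unless forced.
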MaterialFunction /Game/Global/Base4Layers_2022/MaterialFunctions/MF_4Layer_ThinFilm
graph {nodes [
  n1 = NamedRerouteUsage()
  n2 = NamedRerouteUsage()
  n3 = NamedRerouteUsage()
  n4 = LinearInterpolate(
   B=StaticSwitchParameter(True=OneMinus(None=@n3), False=@n3) {defaultValue=true, group=4LayerThinFilm, parameterName=4LayerThinFilm_Invert_Roughness}
   Alpha=ScalarParameter() {defaultValue=1, group=4LayerThinFilm, parameterName=4LayerThinFilm_ModulateByRoughnessScale}
  ) {constA=1}
  n5 = NamedRerouteUsage()
  n6 = ScalarParameter() {defaultValue=1, group=4LayerThinFilm, parameterName=4LayerThinFilm_ModulateByAlbedoScale}
  n7 = LinearInterpolate(
   A=Constant3Vector() {constant=<rgba 0 0 1 0>}
   B=NamedRerouteUsage()
   Alpha=ScalarParameter() {defaultValue=1, group=4LayerThinFilm, parameterName=4LayerThinFilm_FresnelNormalIntensity}
  )
  n8 = MaterialFunctionCall(
   Brightness=ScalarParameter() {defaultValue=1, group=4LayerThinFilm, parameterName=4LayerThinFilm_FresnelIntensity}
   Exponent=ScalarParameter() {defaultValue=1, group=4LayerThinFilm, parameterName=4LayerThinFilm_FresnelExponent}
   Vertex / Pixel Normal=Normalize(VectorInput=Transform(None=@n7))
  ) {materialFunction=&MF_Fresnel_FX}
  n9 = NamedRerouteDeclaration(None=@n8.Fresnel Unclamped) {name=ThinFIlm.Fresnel, nodeColor=<rgba 0 0.6953 1 1>}
  n10 = ScalarParameter() {defaultValue=1, group=4LayerThinFilm, parameterName=4LayerThinFilm_OffsetMaskTileU}
  n11 = ScalarParameter() {defaultValue=1, group=4LayerThinFilm, parameterName=4LayerThinFilm_OffsetMaskTileV}
  n12 = TextureSampleParameter2D(UVs=Multiply(A=AppendVector(A=@n10, B=@n11), B=TextureCoordinate())) {
   group=4LayerThinFilm
   parameterName=4LayerThinFilm_OffsetMask
   samplerType=SAMPLERTYPE_Masks
   texture=/Game/Global/Textures/Default/Blanks/T_MidgreyMask
  }
  n13 = Add(
   A=DotProduct(
    A=ObjectOrientation()
    B=MaterialFunctionCall() {materialFunction=&CameraDirectionVector}.Result
   )
   B=Subtract(A=ChannelMaskParameter(None=@n12.RGBA) {group=4LayerThinFilm, parameterName=4LayerThinFilm_OffsetMaskChannel}.R) {constB=0.5}
  )
  n14 = ScalarParameter() {defaultValue=1, group=4LayerThinFilm, parameterName=4LayerThinFilm_OffsetMaskIntensity}
  n15 = StaticSwitchParameter(True=Add(A=@n9, B=Multiply(A=@n13, B=@n14)), False=@n9) {group=4LayerThinFilm, parameterName=4LayerThinFilm_UseOffsetMask}
  n16 = Add(
   A=@n15
   B=MaterialFunctionCall!(
    Time
    Divisor=ScalarParameter() {defaultValue=1, group=4LayerThinFilm, parameterName=4LayerThinFilm_AnimatedTimeDivisor}
    Mod=ScalarParameter() {defaultValue=1, group=4LayerThinFilm, parameterName=4LayerThinFilm_AnimatedTimeModulo}
   ) {materialFunction=/Game/Effects/Fort_Effects/Materials/_Functions/MF_ModuloTime}.Unclamped
  )
  n17 = CurveAtlasRowParameter(CurveTime=Frac(None=StaticSwitchParameter(True=@n16, False=@n15) {group=4LayerThinFilm, parameterName=4LayerThinFilm_Animated})) {
   atlas=/Game/Global/Base4Layers_2022/Textures/CA_4layer_ThinFilm
   curve=/Game/TechArt/Seasonal/LackLuster/Curves/C_LackLuster_Spectrum_03
   defaultValue=3
   group=4LayerThinFilm
   parameterName=4LayerThinfilmCurveAtlas
  }
  n18 = StaticSwitchParameter(True=Multiply(A=@n17, B=NamedRerouteUsage()), False=@n17) {group=4LayerThinFilm, parameterName=4LayerThinFilm_MultiplyFinalResultByFresnel}
  n19 = FunctionInput() {inputName=SFXMask, sortPriority=1}
  n20 = StaticSwitchParameter(True=Multiply(A=@n18, B=ChannelMaskParameter(None=@n19) {group=4LayerThinFilm, parameterName=4LayerThinFilm_SFXMaskChannel}.R), False=@n18) {group=4LayerThinFilm, parameterName=4LayerThinFilm_UseSFXMask}
  n21 = MaterialFunctionCall() {materialFunction=&/MF_4Layer_RGBA}
  n22 = VectorParameter() {group=4LayerThinFilm, parameterName=4LayerThinFilm_4LayerMask, defaultValue=<rgba 1 1 1 1>}
  n23 = AppendVector(A=@n22.RGB, B=@n22.RGB)
  n24 = StaticSwitchParameter(True=Multiply(A=@n20, B=DotProduct(A=@n21.Result, B=@n23)), False=@n20) {group=4LayerThinFilm, parameterName=4LayerThinFilm_Use4LayerMask}
  n25 = StaticSwitchParameter(True=Multiply(A=LinearInterpolate(B=@n5, Alpha=@n6) {constA=1}, B=@n24), False=@n24) {group=4LayerThinFilm, parameterName=4LayerThinFilm_ModulateByAlbedo}
  n26 = StaticSwitchParameter(True=Multiply(A=Saturate(None=@n4), B=@n25), False=@n25) {group=4LayerThinFilm, parameterName=4LayerThinFilm_ModulateByRougness}
  n27 = Multiply(A=@n26, B=ScalarParameter() {defaultValue=1, group=4LayerThinFilm, parameterName=4LayerThinFilm_DiffuseScale})
  n28 = NamedRerouteUsage()
  n29 = EyeAdaptationInverse(
   LightValueInput=Multiply(A=@n26, B=ScalarParameter() {defaultValue=1, group=4LayerThinFilm, parameterName=4LayerThinFilm_EmissiveScale})
   AlphaInput=ScalarParameter() {defaultValue=1, group=4LayerThinFilm, parameterName=4LayerThinFilm_EmissiveEyeAdaptionScale}
  )
  n30 = SetMaterialAttributes!(
   MaterialAttributes=NamedRerouteUsage()
   BaseColor=StaticSwitchParameter(
    True=MaterialFunctionCall(Base=@n2, Blend=@n27) {materialFunction=&Blend_Overlay}.Result
    False=Add(A=@n2, B=@n27)
   ) {group=4LayerThinFilm, parameterName=4LayerThinFilm_Overlay}
   EmissiveColor=StaticSwitchParameter(True=Add(A=@n28, B=@n29.EyeAdaptationInverse), False=@n28) {group=4LayerThinFilm, parameterName=4LayerThinFilm_AddToEmissive}
  ) {attributeSetTypes [#1, #2]}
  n31 = ShadingPathSwitch(Default=QualitySwitch(Default=@n30, Low=@n1), Mobile=@n1)
  FunctionOutput(None=StaticSwitchParameter(
    True=StaticSwitchParameter(True=@n30, False=@n31) {group=4LayerThinFilm, parameterName=4LayerThinFilm_ForceOnLowMobilePlatforms}
    False=NamedRerouteUsage()
   ) {group=4LayerThinFilm, parameterName=Use4LayerThinFilm})
  n32 = GetMaterialAttributes(None=NamedRerouteDeclaration(None=FunctionInput() {inputType=FunctionInput_MaterialAttributes}) {name=MA.In, nodeColor=<rgba 1 0 0.4453 1>}) {attributeGetTypes [
    #1
    #2
    #3
    #4
   ]}
  NamedRerouteDeclaration(None=@n32.EmissiveColor) {name=Emissive.In, nodeColor=<rgba 0.5234 0 1 1>}
  NamedRerouteDeclaration(None=@n32.BaseColor) {name=BaseColor.In, nodeColor=<rgba 1 0 0.4687 1>}
  NamedRerouteDeclaration(None=@n32.Normal) {name=Normal.In, nodeColor=<rgba 1 0 0.5156 1>}
  NamedRerouteDeclaration(None=@n8.Fresnel Inverted) {name=ThinFilm.FresnelInverted, nodeColor=<rgba 0 0.6953 1 1>}
  NamedRerouteDeclaration(None=@n32.Roughness) {name=Roughness.In, nodeColor=<rgba 1 0.8672 0 1>}
 ]}
