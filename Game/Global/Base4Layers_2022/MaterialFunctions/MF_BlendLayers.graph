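# Material function that blends an upper material layer with a lower one, with static switches for three modes.
# Simple Swatch changes only base color by masked multiply; Normals Only adds angle-corrected normals through the same mask.
# That mask comes from three per-layer scalar parameters selected by a UV2 helper function.
# Default mode lerps from the upper layer to overlaid color and roughness plus combined normals, using angle correction on Epic quality.
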MaterialFunction /Game/Global/Base4Layers_2022/MaterialFunctions/MF_BlendLayers
graph {nodes [
  n1 = FunctionInput() {inputName=Lower Layer, inputType=FunctionInput_MaterialAttributes}
  n2 = GetMaterialAttributes(None=@n1) {attributeGetTypes [#1, #2, #3]}
  n3 = GetMaterialAttributes(None=FunctionInput() {inputName="Upper Layer (Blended)", inputType=FunctionInput_MaterialAttributes}) {attributeGetTypes [#1, #2, #3]}
  n4 = ScalarParameter() {defaultValue=1, desc=If Simple swatch enabled in upper stack layer, group=1_Layer, parameterName=Swatch To Layer 1, sortPriority=1}
  n5 = ScalarParameter() {defaultValue=1, desc=If Simple swatch enabled in upper stack layer, group=2_Layer, parameterName=Swatch To Layer 2, sortPriority=2}
  n6 = ScalarParameter() {defaultValue=1, desc=If Simple swatch enabled in upper stack layer, group=3_Layer, parameterName=Swatch ToLayer 3, sortPriority=3}
  n7 = Constant() {r=1}
  n8 = Saturate(None=ComponentMask(None=MaterialFunctionCall!(1=@n4, 2=@n5, 3=@n6, 4=@n7) {materialFunction=&/MF_UV2LerpScalar_Layers_New}.Result) {r=true})
  n9 = Saturate(None=FunctionInput() {inputName=Blending Value, inputType=FunctionInput_Scalar})
  n10 = LinearInterpolate(A=@n3.BaseColor, B=Multiply(A=@n2.BaseColor, B=LinearInterpolate(B=@n3.BaseColor, Alpha=@n8) {constA=1}), Alpha=@n9)
  n11 = GetMaterialAttributes(None=@n1) {attributeGetTypes [#2]}
  n12 = SetMaterialAttributes!(
   MaterialAttributes=@n11.MaterialAttributes
   Normal=LinearInterpolate(
    A=@n11.Normal
    B=MaterialFunctionCall(BaseNormal=@n11.Normal, AdditionalNormal=@n3.Normal) {materialFunction=&BlendAngleCorrectedNormals}.Result
    Alpha=@n8
   ) {constA=1}
  ) {attributeSetTypes [#2]}
  n13 = Add(A=@n2.Normal, B=Multiply(A=@n3.Normal, B=Constant3Vector() {constant=<rgba 1 1 0 0>}))
  n14 = MaterialFunctionCall(BaseNormal=@n2.Normal, AdditionalNormal=@n3.Normal) {materialFunction=&BlendAngleCorrectedNormals}
  n15 = BlendMaterialAttributes(
   A=@n3.MaterialAttributes
   B=SetMaterialAttributes!(
    MaterialAttributes=@n3.MaterialAttributes
    BaseColor=MaterialFunctionCall(Base=@n2.BaseColor, Blend=@n3.BaseColor) {materialFunction=&Blend_Overlay}.Result
    Normal=QualitySwitch(Default=@n13, Epic=@n14.Result)
    Roughness=MaterialFunctionCall(Base=@n2.Roughness, Blend=@n3.Roughness) {materialFunction=&Blend_Overlay}.Result
   ) {attributeSetTypes [#1, #2, #3]}
   Alpha=@n9
  )
  n16 = FunctionInput(Preview=StaticBool()) {bUsePreviewValueAsDefault=true, inputName="Use  Normals Only", inputType=FunctionInput_StaticBool, previewValue=<1 0 0 1>}
  n17 = FunctionInput(Preview=StaticBool()) {bUsePreviewValueAsDefault=true, inputName=Use Simple Swatch, inputType=FunctionInput_StaticBool, previewValue=<1 0 0 1>}
  FunctionOutput(None=StaticSwitch(
    True=SetMaterialAttributes!(MaterialAttributes=@n2.MaterialAttributes, BaseColor=@n10) {attributeSetTypes [#1]}
    False=StaticSwitch(True=@n12, False=@n15, Value=@n16)
    Value=@n17
   ))
 ]}
