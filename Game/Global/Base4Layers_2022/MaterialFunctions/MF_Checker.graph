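# Material function that optionally overlays a checker texture on the base color of an incoming material, sampled at supplied UVs.
# The base color is multiplied by the checker, then blended with a brightened, saturated version of itself.
# That brightening is the checker scaled by a sphere mask of vertex normal against camera vector, lifting camera-facing surfaces.
# A static bool input switches between the modified attributes and the untouched material.
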MaterialFunction /Game/Global/Base4Layers_2022/MaterialFunctions/MF_Checker
graph {nodes [
  n1 = FunctionInput() {inputName=Material In, inputType=FunctionInput_MaterialAttributes}
  n2 = GetMaterialAttributes(None=@n1) {attributeGetTypes [#1]}
  n3 = TextureSample(UVs=FunctionInput(Preview=TextureCoordinate()) {inputName=UV's, inputType=FunctionInput_Vector2}) {texture=/Game/FN_PublicTurntable/Turntable_Lowpoly/Content/BlueprintAssets/Checker}
  n4 = Multiply(A=@n2.BaseColor, B=@n3.RGB)
  n5 = VertexNormalWS()
  n6 = CameraVectorWS()
  n7 = Multiply(A=Multiply(A=@n3.RGB, B=SphereMask(A=@n5, B=@n6) {attenuationRadius=1, hardnessPercent=0})) {constB=2}
  n8 = LinearInterpolate(A=@n4, B=Saturate(None=Add(A=@n4, B=@n7)))
  n9 = StaticBool()
  FunctionOutput(None=StaticSwitch(
    True=SetMaterialAttributes!(MaterialAttributes=@n2.MaterialAttributes, BaseColor=@n8) {attributeSetTypes [#1]}
    False=@n1
    Value=FunctionInput(Preview=@n9) {inputName=Checker true, inputType=FunctionInput_StaticBool}
   )) {outputName=Switch}
 ]}
