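# Material function that builds world-space XZ and YZ projection UVs for cliff sediment layers, with optional wavy distortion.
# A repeating squared-parabola wave from fractional world position, summed as twice X plus Z and scaled by intensity, is also output as Sediment Wave Distortion.
# A static switch, off by default, subtracts that wave from height.
# The position then gets its Z flipped, is offset and scaled, and is masked into both UV pairs.
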MaterialFunction /Game/Global/Base4Layers_2022/MaterialFunctions/MF_Cliffs_SedimentWS_UVs
graph {nodes [
  n1 = FunctionInput(Preview=WorldPosition().XYZ) {bUsePreviewValueAsDefault=true, inputName=WorldPosition, sortPriority=-1}
  n2 = ComponentMask(None=@n1) {b=true}
  n3 = ComponentMask(None=@n1) {b=true, g=true, r=true}
  n4 = FunctionInput(Preview=Constant3Vector()) {bUsePreviewValueAsDefault=true, inputName=Sediment L2_WS | Wave DistPos Offset, sortPriority=15}
  n5 = Constant() {r=6000}
  n6 = Frac(None=Divide(
    A=Add(A=@n3, B=@n4)
    B=FunctionInput(Preview=@n5) {bUsePreviewValueAsDefault=true, inputName=Wave Distortion Freq, inputType=FunctionInput_Scalar, sortPriority=20}
   ))
  n7 = Multiply(A=OneMinus(None=@n6), B=@n6)
  n8 = MaterialFunctionCall(Color=Multiply(A=@n7, B=@n7)) {materialFunction=&SplitComponents}
  n9 = Add(A=@n8.R, B=@n8.R)
  n10 = Constant() {r=5000}
  n11 = FunctionOutput(None=Multiply(
    A=Add(A=@n9, B=@n8.B)
    B=FunctionInput(Preview=@n10) {bUsePreviewValueAsDefault=true, inputName=Wave Distortion Intensity, inputType=FunctionInput_Scalar, sortPriority=10}
   )) {outputName=Sediment Wave Distortion, sortPriority=10}
  n12 = StaticSwitch(
   True=AppendVector(A=ComponentMask(None=@n1) {g=true, r=true}, B=Subtract(A=@n2, B=@n11))
   False=@n3
   Value=FunctionInput(Preview=StaticBool()) {bUsePreviewValueAsDefault=true, inputName=Use Wavy Distortion, inputType=FunctionInput_StaticBool, sortPriority=5}
  )
  n13 = Constant3Vector() {constant=<rgba 1 1 -1 0>}
  n14 = Constant3Vector()
  n15 = Divide(
   A=Add(A=Multiply(A=@n12, B=@n13), B=FunctionInput(Preview=@n14) {bUsePreviewValueAsDefault=true, inputName=Sediment L2_WS | Position Offset_1})
   B=FunctionInput(Preview=Constant() {r=2550}) {bUsePreviewValueAsDefault=true, inputName=Sediment L2_WS | Scale, inputType=FunctionInput_Scalar, sortPriority=1}
  )
  FunctionOutput(None=ComponentMask(None=@n15) {b=true, r=true}) {outputName=XZ}
  FunctionOutput(None=ComponentMask(None=@n15) {b=true, g=true}) {outputName=YZ, sortPriority=2}
 ]}
