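# Material function that projects sediment cliff diffuse, normal and specular textures in world space on XZ and YZ planes.
# Projection UVs default to MF_Cliffs_SedimentWS_UVs; normals are swizzled onto vertex normals, outputting material attributes plus a normal-blue mask.
# The contrast-powered normal blend divides X by X+X, not X+Y, so it tends toward 0.5 wherever X dominates the epsilon.
# Specular-map blue drives Roughness, Specular and Metallic alike; diffuse alpha feeds Opacity.
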MaterialFunction /Game/Global/Base4Layers_2022/MaterialFunctions/MF_Cliffs_SedimentWS
graph {nodes [
  n1 = FunctionInput(Preview=TextureObject() {samplerType=SAMPLERTYPE_Normal, texture=/Game/Environments/Helios/Terrain/Textures/T_Helios_RustBucket_Cliff_Strata_01_N}) {inputName=T_Normal, inputType=FunctionInput_Texture2D, sortPriority=6}
  n2 = MaterialFunctionCall!(
   WorldPosition
   Sediment L2_WS | Position Offset_1=Constant3Vector()
   Sediment L2_WS | Scale=Constant() {r=2550}
   Use Wavy Distortion=StaticBool()
   Wave Distortion Intensity=Constant() {r=5000}
   Sediment L2_WS | Wave DistPos Offset=Constant3Vector()
   Wave Distortion Freq=Constant() {r=6000}
  ) {materialFunction=&/MF_Cliffs_SedimentWS_UVs}
  n3 = FunctionInput(Preview=@n2.XZ) {inputName=UV_XZ, inputType=FunctionInput_Vector2, sortPriority=-2}
  n4 = MaterialFunctionCall(Color=TextureSample(UVs=@n3, Tex=@n1).RGB) {materialFunction=&SplitComponents}
  n5 = FunctionInput(Preview=@n2.YZ) {inputName=UV_YZ, inputType=FunctionInput_Vector2, sortPriority=-1}
  n6 = MaterialFunctionCall(Color=TextureSample(UVs=@n5, Tex=@n1).RGB) {materialFunction=&SplitComponents}
  n7 = VertexNormalWS()
  n8 = ComponentMask(None=Power(
    Base=Abs(None=@n7)
    Exp=FunctionInput() {desc=Projection Transition Contrast, inputName=PTC, inputType=FunctionInput_Scalar, sortPriority=30}
   ) {constExponent=10}) {g=true, r=true}
  n9 = MaterialFunctionCall(Float2=@n8) {materialFunction=&BreakOutFloat2Components}
  n10 = Add(A=@n9.R, B=@n9.R)
  n11 = ComponentMask(None=Divide(A=@n8, B=Add(A=@n10) {constB=0.001, desc=fix division by zero for upward faces})) {r=true}
  FunctionOutput(None=LinearInterpolate(A=@n4.B, B=@n6.B, Alpha=@n11)) {outputName=Sediment Normal B Mask, sortPriority=10}
  n12 = FunctionInput(Preview=TextureObject() {texture=/Game/Environments/Helios/Terrain/Textures/T_Helios_RustBucket_Cliff_Strata_01_D}) {inputName=T_Diffuse, inputType=FunctionInput_Texture2D, sortPriority=4}
  n13 = LinearInterpolate(A=TextureSample(UVs=@n3, Tex=@n12).RGBA, B=TextureSample(UVs=@n5, Tex=@n12).RGBA, Alpha=@n11)
  n14 = FunctionInput(Preview=TextureObject() {samplerType=SAMPLERTYPE_Masks, texture=/Game/Environments/Helios/Terrain/Textures/T_Helios_RustBucket_Cliff_Strata_01_S}) {inputName=T_Specular, inputType=FunctionInput_Texture2D, sortPriority=5}
  n15 = MaterialFunctionCall(Color=LinearInterpolate(A=TextureSample(UVs=@n3, Tex=@n14).RGB, B=TextureSample(UVs=@n5, Tex=@n14).RGB, Alpha=@n11)) {materialFunction=&SplitComponents}
  n16 = VertexNormalWS()
  n17 = ComponentMask(None=@n16) {r=true}
  n18 = Add(A=@n4.R, B=@n17)
  n19 = ComponentMask(None=@n16) {g=true}
  n20 = Multiply(A=@n4.B, B=@n19)
  n21 = ComponentMask(None=@n16) {b=true}
  n22 = Add(A=Multiply(A=@n4.G) {constB=-1}, B=@n21)
  n23 = Multiply(A=@n6.B, B=@n17)
  n24 = Add(A=@n6.R, B=@n19)
  n25 = Add(A=Multiply(A=@n6.G) {constB=-1}, B=@n21)
  n26 = SetMaterialAttributes!(
   MaterialAttributes
   BaseColor=ComponentMask(None=@n13) {b=true, g=true, r=true}
   Opacity=ComponentMask(None=@n13) {a=true}
   Roughness=@n15.B
   Specular=@n15.B
   Metallic=@n15.B
   Normal=LinearInterpolate(
    A=MaterialFunctionCall(R=@n18, G=@n20, B=@n22) {materialFunction=&AppendMany}.RGB
    B=MaterialFunctionCall(R=@n23, G=@n24, B=@n25) {materialFunction=&AppendMany}.RGB
    Alpha=@n11
   )
  ) {attributeSetTypes [
    #1
    #2
    #3
    #4
    #5
    #6
   ]}
  FunctionOutput(None=@n26) {outputName=SetMaterialAttributes}
 ]}
