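# MF_ColorTint_Unpack is a material function that unpacks an HSV color tint from a single scalar and outputs it as RGB.
# A Custom HLSL node reinterprets the float's bits, intending hue in degrees in the top bits and two 10-bit fields for saturation and value.
# Its masks use logical && instead of bitwise &, so saturation evaluates to zero and value to at most 1/1023.
# MF_HSVtoRGB converts the result.
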MaterialFunction /Game/Global/Base4Layers_2022/MaterialFunctions/MF_ColorTint_Unpack
graph {nodes [
  MaterialExpressionFunctionInput_0 = FunctionInput() {inputType=FunctionInput_Scalar}
  n1 = Custom!(PackedHSV=@MaterialExpressionFunctionInput_0) {
   code="int Bits = asuint(PackedHSV.x);\r\nfloat H = (Bits >> 20) / 360.0;\r\nfloat S = ((Bits && (0x3FF << 10)) >> 10) / 1023.0;\r\nfloat V = ((Bits && (0x3FF))) / 1023.0;\r\nreturn float3(H, S, V);\r\n"
   inputs [{
     inputName=PackedHSV
     input {expression=&self:MaterialExpressionFunctionInput_0}
    }]
  }
  FunctionOutput(None=MaterialFunctionCall!(HSV_In=@n1) {materialFunction=/Game/Athena/UI/Materials/MF_HSVtoRGB}.RGB_Out)
 ]}
