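# Material function providing two static-switch debug views for four-layer materials, applied to incoming material attributes.
# The POI check replaces emissive with a parameter color, green by default.
# The layer view turns the UV2-selected layer index into a flipbook frame of a mask texture, tinted and alpha-scaled by an input color, then added to emissive.
# Shading path and material proxy switches bypass both views on mobile and in proxy materials.
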MaterialFunction /Game/Global/Base4Layers_2022/MaterialFunctions/MF_DebugViews
graph {nodes [
  n1 = FunctionInput() {inputName=Mat, inputType=FunctionInput_MaterialAttributes}
  n2 = StaticSwitchParameter(
   True=SetMaterialAttributes!(
    MaterialAttributes=@n1
    EmissiveColor=VectorParameter() {group=0_Debug, parameterName=POI Check, sortPriority=22, defaultValue=<rgba 0 1 0 0>}.RGB
   ) {attributeSetTypes [#1]}
   False=@n1
  ) {group=0_Debug, parameterName=CheckMyPOI, sortPriority=21}
  FunctionOutput(None=ShadingPathSwitch(Default=MaterialProxyReplace(Realtime=@n2, MaterialProxy=@n1), Mobile=@n1)) {outputName=POI Color}
  n3 = FunctionInput() {inputName=Layer Debug, inputType=FunctionInput_MaterialAttributes}
  n4 = GetMaterialAttributes(None=@n3) {attributeGetTypes [#1]}
  n5 = Add(A=MaterialFunctionCall!(1=Constant() {r=1}, 2=Constant() {r=2}, 3=Constant() {r=3}, 4=Constant() {r=4}) {materialFunction=&/MF_UV2LerpScalar_Layers_New}.Result) {constB=-1}
  n6 = MaterialFunctionCall("Animation  Phase (0-1)"=Divide(A=Ceil(None=@n5)) {constB=4}, Number of Rows=Constant() {r=2}) {materialFunction=&FlipBook}
  n7 = FunctionInput() {bUsePreviewValueAsDefault=true, inputName=Debug Layer Color, inputType=FunctionInput_Vector4, previewValue=<1 1 1 1>}
  n8 = Multiply(
   A=Multiply(
    A=TextureSample(UVs=@n6.UVs) {samplerType=SAMPLERTYPE_Masks, texture=/Game/Global/Base4Layers_2022/Textures/T_Flipbook_M}.R
    B=ComponentMask(None=@n7) {b=true, g=true, r=true}
   )
   B=ComponentMask(None=@n7) {a=true}
  )
  n9 = StaticSwitchParameter(
   True=SetMaterialAttributes(MaterialAttributes=@n4.MaterialAttributes, EmissiveColor=Add(A=@n4.EmissiveColor, B=@n8)) {attributeSetTypes [#1]}
   False=@n4.MaterialAttributes
  ) {desc=Uses flipbook projection on available Material Layers via UV mask, group=0_Debug, parameterName=Layers Debug View, sortPriority=31}
  FunctionOutput(None=ShadingPathSwitch(Default=MaterialProxyReplace(Realtime=@n9, MaterialProxy=@n3), Mobile=@n3)) {outputName=Layer Debug}
 ]}
