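# Material function that adds world-space dirt, waterline and leak weathering to incoming material attributes, with a separate DFAO mask output.
# Height-lerped dirt uses detail roughness, surface orientation, per-layer values and an optional landscape RVT height gradient.
# Waterline and leaks use world height, with optional object-bounds leaks and SFX-channel masking.
# Quality switch bypasses dirt on low/mobile tiers; proxy replacements drop local-space effects for HLOD.
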MaterialFunction /Game/Global/Base4Layers_2022/MaterialFunctions/MF_DetailsWS
graph {nodes [
  Constant() {r=1}
  n1 = NamedRerouteUsage()
  n2 = NamedRerouteUsage()
  n3 = DotProduct(A=FunctionInput() {inputName=SFX V3}, B=VectorParameter() {group=C_Dirt, parameterName=SFX_DirtMaskCh, defaultValue=<rgba 1 0 0 0>}.RGB)
  n4 = StaticSwitchParameter(True=BlendMaterialAttributes(A=NamedRerouteUsage(), B=@n2, Alpha=Saturate(None=@n3)), False=@n2) {
   desc="Enables World Space detail texture from your SFX Functions texture. Select channel. Default scale is 1 = 128  You can enable DFAO mask inside this function"
   group=C_Dirt
   parameterName=Use Dirt SFX Mask
   sortPriority=-3
  }
  n5 = StaticSwitchParameter(
   True=MaterialFunctionCall!(
    Low _ PC Low and Med=@n1
    "High _ XB1, PS4"=@n4
    "Epic _ PC High, XSX, PS5"=@n4
    --------
    "Mobile Low _ PC Perf, Android"=@n1
    Mobile Med _ Switch=@n1
   ) {materialFunction=/Game/Global/MaterialFunctions/TechArt/MF_Fortnite_QualitySwitch_MaterialAttributes}.Result
   False=@n1
  ) {
   desc="Enables World Space detail texture from your SFX Functions texture. Select channel. Default scale is 1 = 128  You can enable DFAO mask inside this function"
   group=C_Dirt
   parameterName=Use Dirt
   sortPriority=-3
  }
  FunctionOutput(None=@n5)
  NamedRerouteDeclaration(None=FunctionInput() {inputName=Mat In, inputType=FunctionInput_MaterialAttributes}) {name=Mat Input, nodeColor=<rgba 1 0 0.6328 1>}
  n6 = ScalarParameter() {group=C_Dirt_Waterline, parameterName=Waterline Height}
  n7 = ComponentMask(None=Add(A=WorldPosition().XYZ, B=OneMinus(None=@n6))) {b=true}
  n8 = ScalarParameter() {defaultValue=1500, group=C_Dirt_Waterline, parameterName=Waterline Blur}
  NamedRerouteDeclaration(None=Saturate(None=Divide(A=@n7, B=@n8))) {name=Water Height, nodeColor=<rgba 0.9922 1 0 1>}
  n9 = Multiply(A=@n8) {constB=0.1}
  NamedRerouteDeclaration(None=Saturate(None=Divide(A=@n7, B=@n9) {constB=15})) {name=Waterline Contact, nodeColor=<rgba 0.3828 0 1 1>}
  n10 = ComponentMask(None=WorldPosition().XYZ) {b=true}
  n11 = ScalarParameter() {defaultValue=-9618.5, group=C_Dirt_Leak, parameterName=Leak Height}
  n12 = ScalarParameter() {defaultValue=2500, group=C_Dirt_Leak, parameterName=Leak Fade down}
  NamedRerouteDeclaration(None=Divide(A=Add(A=@n10, B=@n11), B=OneMinus(None=@n12))) {name=Leak Location, nodeColor=<rgba 0 1 0.4609 1>}
  n13 = VectorParameter() {
   group=C_Dirt_Waterline
   parameterName=Waterline Bottom
   channelNames {a=Water Transparency, b="", g="", r=""}
   defaultValue=<rgba 0.3099 0.3142 0.0182 0.65>
  }
  n14 = Multiply(A=@n13) {constB=0.5}
  n15 = GetMaterialAttributes(None=NamedRerouteUsage()) {attributeGetTypes [#1, #2, #3]}
  n16 = NamedRerouteUsage()
  n17 = SetMaterialAttributes!(
   MaterialAttributes=@n16
   Specular=Constant() {r=0.05}
   BaseColor=LinearInterpolate(A=@n14, B=Multiply(A=@n14, B=@n15.BaseColor)) {constAlpha=0.9}
   Roughness=Constant()
  ) {
   desc=Waterline
   attributeSetTypes [#2, #3, #1]
  }
  n18 = NamedRerouteUsage()
  n19 = LinearInterpolate(A=Multiply(A=@n13) {constB=1.5}, Alpha=@n18)
  n20 = Constant() {r=0.1}
  n21 = Constant() {r=0.15}
  n22 = BlendMaterialAttributes(
   A=@n16
   B=SetMaterialAttributes!(
    MaterialAttributes=@n16
    BaseColor=Multiply(A=@n19, B=@n15.BaseColor)
    Specular=LinearInterpolate(A=@n20, B=@n15.Specular, Alpha=@n18)
    Roughness=LinearInterpolate(A=@n21, B=@n15.Roughness, Alpha=@n18)
   ) {attributeSetTypes [#3, #2, #1]}
   Alpha=@n13.Water Transparency
  )
  n23 = NamedRerouteUsage()
  n24 = NamedRerouteUsage()
  n25 = OneMinus(None=@n15.Roughness)
  n26 = ScalarParameter() {defaultValue=0.5, group=C_Dirt_Waterline, parameterName=Waterline Contrast}
  n27 = StaticSwitchParameter(
   True=BlendMaterialAttributes(
    A=BlendMaterialAttributes(A=@n17, B=@n22, Alpha=@n23)
    B=@n16
    Alpha=MaterialFunctionCall(Transition Phase=@n24, Height Texture=@n25, Contrast=@n26) {materialFunction=&HeightLerp}.Alpha
   )
   False=@n16
  ) {
   desc="Enables World Space detail texture from your SFX Functions texture. Select channel. Default scale is 1 = 128  You can enable DFAO mask inside this function"
   group=C_Dirt
   parameterName=Use Waterline
   sortPriority=-3
  }
  NamedRerouteDeclaration(None=@n27) {name=After Water, nodeColor=<rgba 0 0.25 1 1>}
  n28 = NamedRerouteUsage()
  n29 = GetMaterialAttributes(None=@n28) {attributeGetTypes [#3, #1, #2]}
  n30 = VectorParameter() {
   group=C_Dirt_Layer
   parameterName=Dirt Color
   sortPriority=0
   channelNames {a=Dirt Global Transparency, b="", g="", r=""}
   defaultValue=<rgba 0.4149 0.2605 0.1379 1>
  }
  n31 = Desaturation(None=@n30, Fraction=Constant() {r=0.5})
  n32 = Multiply(A=@n29.BaseColor, B=@n31)
  n33 = MaterialFunctionCall() {materialFunction=&ObjectLocalBounds}
  n34 = Add(A=ScalarParameter() {group=C_Dirt_Leak, parameterName=LeakOffset}, B=@n33.Local Bounds Max)
  n35 = TransformPosition(None=WorldPosition() {worldPositionShaderOffset=WPT_ExcludeAllShaderOffsets}.XYZ) {transformSourceType=TRANSFORMPOSSOURCE_World}
  n36 = ScalarParameter() {defaultValue=1, group=C_Dirt_Leak, parameterName=LeakScaleTweak}
  n37 = ComponentMask(None=Divide(A=Subtract(A=@n34, B=@n35), B=Multiply(A=@n33.Local Bounds Size, B=@n36))) {b=true}
  n38 = ComponentMask(None=VertexNormalWS()) {b=true}
  n39 = Saturate(None=Add(A=Saturate(None=@n37), B=@n38))
  n40 = NamedRerouteUsage()
  n41 = Multiply(B=@n38) {constA=0.75}
  n42 = NamedRerouteDeclaration(None=Saturate(None=Add(A=@n40, B=@n41))) {name=Top Gradient, nodeColor=<rgba 0 0.2734 1 1>}
  n43 = BlendMaterialAttributes(
   A=SetMaterialAttributes!(
    MaterialAttributes=@n29.MaterialAttributes
    BaseColor=LinearInterpolate(A=@n32, B=@n31) {constAlpha=0.15}
    Roughness=Multiply(A=@n29.Roughness) {constB=0.75}
    Specular=Multiply(A=@n29.Specular) {constB=0.5}
   ) {attributeSetTypes [#3, #1, #2]}
   B=@n28
   Alpha=StaticSwitchParameter(True=Multiply(A=@n39, B=@n42), False=@n42) {desc=Only on building meshes, group=C_Dirt_Leak, parameterName=Add Actor Space Leak}
  )
  NamedRerouteDeclaration(None=StaticSwitchParameter(True=MaterialProxyReplace(Realtime=@n43, MaterialProxy=@n28) {desc="HLOD can't support  Object Bounds"}, False=@n28) {group=C_Dirt, parameterName=Use Dirt Leak}) {name=Dirt Leak IN, nodeColor=<rgba 1 0.8203 0 1>}
  n44 = ComponentMask(None=WorldPosition().XYZ) {b=true}
  n45 = Constant() {r=-18680}
  n46 = Constant() {r=2200}
  n47 = Saturate(None=Divide(A=Add(A=@n44, B=@n45), B=OneMinus(None=@n46)))
  n48 = Saturate(None=SphereMask(A=WorldPosition().XYZ, B=Constant3Vector() {constant=<rgba -311296 -297984 16128 0>}, Radius=Constant() {r=40000}, Hardness=Constant() {r=1}))
  NamedRerouteDeclaration(None=LinearInterpolate(
    A=NamedRerouteUsage()
    B=NamedRerouteDeclaration(None=@n47) {name=Leak Starting Island, nodeColor=<rgba 0 1 0.4609 1>}
    Alpha=NamedRerouteDeclaration(None=@n48) {name=SI Mask, nodeColor=<rgba 1 0 0.5625 1>}
   )) {name=LEak Locaiton Mix, nodeColor=<rgba 0 0.3672 1 1>}
  n49 = NamedRerouteUsage()
  n50 = WorldPosition()
  n51 = ActorPositionWS()
  n52 = Multiply(A=Transform(None=Subtract(A=@n50.XYZ, B=@n51)) {transformSourceType=TRANSFORMSOURCE_World, transformType=TRANSFORM_Local}) {constB=0.1}
  n53 = ScalarParameter() {
   defaultValue=1
   desc="Mask based on mesh local Y. Will make what is on negative side from pivot free of dirt effect. In some cases you can't use it ( Curved walls) so instead use additional material layer to controll dirt level inside buildings"
   group=C_Dirt
   parameterName=Interior Walls Dirt Value
   sortPriority=5
  }
  n54 = Max(A=Saturate(None=ComponentMask(None=@n52) {g=true}), B=@n53)
  n55 = GetMaterialAttributes(None=NamedRerouteUsage()) {attributeGetTypes [#2, #1]}
  n56 = NamedRerouteUsage()
  n57 = VectorParameter() {
   group=C_Dirt_Layer
   parameterName=Dirt Color Down
   sortPriority=0
   channelNames {a=Dirt Gradient Transparency, b="", g="", r=""}
   defaultValue=<rgba 0.2396 0.111 0.0344 0.5>
  }
  n58 = VectorParameter() {
   group=C_Dirt_Layer
   parameterName=Dirt Color
   sortPriority=0
   channelNames {a=Dirt Global Transparency, b="", g="", r=""}
   defaultValue=<rgba 0.4149 0.2605 0.1379 1>
  }
  n59 = NamedRerouteUsage()
  n60 = Constant()
  n61 = ScalarParameter() {defaultValue=1, group=C_Dirt_Layer, parameterName=Dirt R Min, sortPriority=4}
  n62 = ScalarParameter() {defaultValue=1, group=C_Dirt_Layer, parameterName=Dirt R Max, sortPriority=5}
  n63 = SetMaterialAttributes!(
   MaterialAttributes=@n56
   BaseColor=StaticSwitchParameter(True=LinearInterpolate(A=@n57, B=@n58, Alpha=@n59), False=@n58) {group=C_Dirt, parameterName=Use Dirt Gradient}
   Roughness=StaticSwitchParameter(
    True=LinearInterpolate(A=@n61, B=@n62, Alpha=@n55.Roughness)
    False=ScalarParameter() {defaultValue=0.75, group=C_Dirt_Layer, parameterName=Dirt R, sortPriority=3}
   ) {desc=Used base material roughness as the base., group=C_Dirt_Layer, parameterName=Dirt Roughness Edit, sortPriority=3}
   Specular=ScalarParameter() {defaultValue=0.5, group=C_Dirt_Layer, parameterName=Dirt S, sortPriority=2}
   Metallic=@n60
  ) {attributeSetTypes [
    #3
    #1
    #2
    #4
   ]}
  n64 = SetMaterialAttributes(
   MaterialAttributes=@n56
   BaseColor=Constant3Vector() {constant=<rgba 0.4271 0.3402 0.2677 1>}
   Roughness=Constant() {r=0.75}
   Specular=Constant() {r=0.5}
   Metallic=@n60
  ) {attributeSetTypes [
    #3
    #1
    #2
    #4
   ]}
  n65 = NamedRerouteUsage()
  n66 = Multiply(A=NamedRerouteUsage(), B=Constant() {r=0.5}) {constB=0.5}
  n67 = MaterialFunctionCall(
   Transition Phase=@n65
   Height Texture=Multiply(A=OneMinus(None=@n55.Specular), B=OneMinus(None=@n66))
   Contrast=Multiply(A=@n65) {constB=0.5}
  ) {materialFunction=&HeightLerp}
  n68 = BlendMaterialAttributes(
   A=@n55.MaterialAttributes
   B=StaticSwitchParameter(True=@n63, False=@n64) {group=C_Dirt_AdvancedControls, parameterName=Dirt Layer Edit, sortPriority=55}
   Alpha=StaticSwitchParameter(True=Multiply(A=@n58.Dirt Global Transparency, B=@n67.Alpha), False=@n67.Alpha) {group=C_Dirt_AdvancedControls, parameterName=Dirt Layer Edit, sortPriority=55}
  )
  n69 = GetMaterialAttributes(None=@n68) {attributeGetTypes [#3]}
  n70 = ScalarParameter() {defaultValue=3, group=C_Dirt_Layer, parameterName=Dirt Fresnel Exponent}
  n71 = Transform(None=NamedRerouteUsage())
  n72 = Add(A=@n69.BaseColor, B=Multiply(A=@n69.BaseColor, B=Fresnel(ExponentIn=@n70, Normal=@n71) {baseReflectFraction=0}))
  n73 = NamedRerouteUsage()
  n74 = GetMaterialAttributes(None=@n73) {attributeGetTypes [#3, #2]}
  n75 = ScalarParameter() {
   desc=Uses Base color as specular replacement for materials using dirt
   group=C_Dirt_AdvancedControls
   parameterName=Custom Spec Min
   sortPriority=17
  }
  n76 = ScalarParameter() {
   defaultValue=1
   desc=Uses Base color as specular replacement for materials using dirt
   group=C_Dirt_AdvancedControls
   parameterName=Custom Spec Max
   sortPriority=16
  }
  n77 = Desaturation(None=@n74.BaseColor, Fraction=Constant() {r=1})
  n78 = SetMaterialAttributes!(
   MaterialAttributes=@n74.MaterialAttributes
   Specular=StaticSwitchParameter(True=LinearInterpolate(A=@n75, B=@n76, Alpha=@n77), False=@n77) {group=C_Dirt_AdvancedControls, parameterName=Dirt Spec Edit, sortPriority=15}
  ) {attributeSetTypes [#2]}
  n79 = ScalarParameter() {defaultValue=0.7, group=C_Dirt_AdvancedControls, parameterName=Ceiling Balance, sortPriority=2}
  n80 = ScalarParameter() {defaultValue=0.7, group=C_Dirt_AdvancedControls, parameterName=Floors Balance, sortPriority=4}
  n81 = NamedRerouteUsage()
  n82 = NamedRerouteUsage()
  n83 = Constant() {r=0.7}
  n84 = Constant() {r=0.7}
  n85 = StaticSwitchParameter(
   True=LinearInterpolate(
    A=ScalarParameter() {defaultValue=0.73, group=C_Dirt_AdvancedControls, parameterName=Walls Balance, sortPriority=0}
    B=LinearInterpolate(A=@n79, B=@n80, Alpha=@n81)
    Alpha=@n82
   )
   False=LinearInterpolate(A=Constant() {r=0.73}, B=LinearInterpolate(A=@n83, B=@n84, Alpha=@n81), Alpha=@n82)
  ) {group=C_Dirt_AdvancedControls, parameterName=Dirt Advanced Edit, sortPriority=-1}
  n86 = NamedRerouteUsage()
  n87 = StaticSwitchParameter(
   True=StaticSwitchParameter(True=@n85, False=Min(A=@n86, B=@n85) {constB=0.2}) {group=C_Dirt_Gradient, parameterName=Use Dirt Gradient Color Only}
   False=@n85
  ) {group=C_Dirt, parameterName=Use Dirt Gradient}
  n88 = ScalarParameter() {group=C_Dirt_AdvancedControls, parameterName=Dirt Texture Min, sortPriority=7}
  n89 = ScalarParameter() {defaultValue=1, group=C_Dirt_AdvancedControls, parameterName="Dirt Texture  Max", sortPriority=8}
  n90 = OneMinus(None=GetMaterialAttributes(None=NamedRerouteUsage()) {attributeGetTypes [#1, #3]}.Roughness)
  n91 = Multiply(
   A=NamedRerouteUsage()
   B=StaticSwitchParameter(True=LinearInterpolate(A=@n88, B=@n89, Alpha=@n90), False=@n90) {desc=Roughness is Dirt inut for details, group=C_Dirt_AdvancedControls, parameterName=Dirt Texture Edit, sortPriority=6}
  )
  n92 = ComponentMask(None=VertexNormalWS()) {b=true}
  n93 = OneMinus(None=Multiply(A=ScalarParameter() {group=C_Dirt, parameterName=Dirt Floor Flood}, B=Saturate(None=@n92)))
  n94 = LinearInterpolate(
   A=ScalarParameter() {defaultValue=0.35, group=C_Dirt_AdvancedControls, parameterName=Walls Contrast, sortPriority=1}
   B=LinearInterpolate(
    A=ScalarParameter() {defaultValue=0.5, group=C_Dirt_AdvancedControls, parameterName=Ceiling Contrast, sortPriority=3}
    B=ScalarParameter() {defaultValue=0.25, group=C_Dirt_AdvancedControls, parameterName=Floors Contrast, sortPriority=5}
    Alpha=@n81
   )
   Alpha=@n82
  )
  n95 = LinearInterpolate(A=Constant() {r=0.35}, B=LinearInterpolate(A=Constant() {r=0.5}, B=Constant() {r=0.25}, Alpha=@n81), Alpha=@n82)
  n96 = BlendMaterialAttributes(
   A=StaticSwitchParameter(
    True=SetMaterialAttributes!(MaterialAttributes=@n69.MaterialAttributes, BaseColor=@n72) {attributeSetTypes [#3]}
    False=@n68
   ) {group=C_Dirt_Layer, parameterName=DirtFresnel}
   B=NamedRerouteDeclaration(None=StaticSwitchParameter(True=@n73, False=@n78) {group=C_Dirt, parameterName=Use_DefaultSpecForDirt}) {name=Base Mat Updated Spec, nodeColor=<rgba 0 1 0.3906 1>}
   Alpha=MaterialFunctionCall(
    Transition Phase=StaticSwitchParameter(True=@n87, False=@n87) {group=C_Dirt, parameterName=Use Dirt Leak}
    Height Texture=Multiply(A=@n91, B=@n93)
    Contrast=StaticSwitchParameter(True=@n94, False=@n95) {group=C_Dirt_AdvancedControls, parameterName=Dirt Advanced Edit, sortPriority=-1}
   ) {materialFunction=&HeightLerp}.Alpha
  ) {vertexAttributeBlendType=UseA}
  NamedRerouteDeclaration(None=MaterialProxyReplace(Realtime=BlendMaterialAttributes(A=@n49, B=@n96, Alpha=@n54) {vertexAttributeBlendType=UseA}, MaterialProxy=@n49) {desc=Local space stuff doesn't work in HLOD}) {name=After dirt, nodeColor=<rgba 0.9922 0 1 1>}
  NamedRerouteDeclaration(None=GetMaterialAttributes(None=@n1) {attributeGetTypes [#5]}.Normal) {name=NormalFromMaterial, nodeColor=<rgba 0.4062 0 1 1>}
  n97 = StaticSwitchParameter(True=ScalarParameter() {defaultValue=120, group=C_Dirt_AO, parameterName=Details AO Distance, sortPriority=22}, False=Constant() {r=120}) {group=C_Dirt_AdvancedControls, parameterName=AO Edit, sortPriority=21}
  n98 = ScalarParameter() {defaultValue=1, group=C_Dirt_AO, parameterName=Ceilings AO Scale, sortPriority=26}
  n99 = ScalarParameter() {defaultValue=1, group=C_Dirt_AO, parameterName=Floors AO Scale, sortPriority=27}
  n100 = VertexNormalWS()
  n101 = NamedRerouteDeclaration(None=Saturate(None=ComponentMask(None=@n100) {b=true})) {name=Up Mask, nodeColor=<rgba 0 0.7187 1 1>}
  n102 = Constant() {r=1}
  n103 = StaticSwitchParameter(
   True=StaticSwitchParameter(True=LinearInterpolate(A=@n98, B=@n99, Alpha=@n101), False=@n102) {group=C_Dirt_AO, parameterName=AO Advanced Edit, sortPriority=25}
   False=@n102
  ) {group=C_Dirt_AdvancedControls, parameterName=AO Edit, sortPriority=21}
  n104 = NamedRerouteUsage()
  n105 = LinearInterpolate(A=@n97, B=Multiply(A=@n97, B=@n103), Alpha=@n104)
  n106 = StaticSwitchParameter(True=ScalarParameter() {defaultValue=500, group=C_Dirt_AO, parameterName=Details AO Fade, sortPriority=23}, False=Constant() {r=500}) {group=C_Dirt_AdvancedControls, parameterName=AO Edit, sortPriority=21}
  n107 = LinearInterpolate(A=@n106, B=Multiply(A=@n103, B=@n106), Alpha=@n104)
  n108 = FunctionOutput(None=ShadingPathSwitch(Default=DistanceFieldApproxAO!(World Position, Normal, BaseDistance=@n105, Radius=@n107) {numSteps=3}, Mobile=Constant() {r=1}) {desc=Migrated from FeatureLevelSwitch}) {outputName=AO Mask}
  n109 = NamedRerouteUsage()
  n110 = NamedRerouteDeclaration(None=Multiply(
    A=GetMaterialAttributes(None=@n109) {attributeGetTypes [#4]}.Metallic
    B=ScalarParameter() {defaultValue=0.25, group=C_Dirt, parameterName=Metal to Dirt Mask Value, sortPriority=6}
   )) {name=metal Mask, nodeColor=<rgba 0.6641 1 0 1>}
  NamedRerouteDeclaration(None=Saturate(None=Add(A=@n108, B=@n110))) {name=Ao, nodeColor=<rgba 1 0.9844 0 1>}
  NamedRerouteDeclaration(None=Abs(None=@n92)) {name=Sides, nodeColor=<rgba 1 0.2578 0 1>}
  n111 = ScalarParameter() {defaultValue=1, group=C_Dirt, parameterName=Dirt Value Layer 1, sortPriority=1}
  n112 = ScalarParameter() {defaultValue=1, group=C_Dirt, parameterName=Dirt Value Layer 2, sortPriority=2}
  n113 = ScalarParameter() {defaultValue=1, group=C_Dirt, parameterName=Dirt Value Layer 3, sortPriority=3}
  n114 = ScalarParameter() {defaultValue=1, group=C_Dirt, parameterName=Dirt Value Layer 4, sortPriority=4}
  NamedRerouteDeclaration(None=Saturate(None=MaterialFunctionCall!(1=@n111, 2=@n112, 3=@n113, 4=@n114) {materialFunction=&/MF_UV2LerpScalar_Layers_New}.Result)) {name=Layer Walls Mask, nodeColor=<rgba 0.1484 0 1 1>}
  n115 = Constant() {r=1}
  n116 = GetMaterialAttributes(None=NamedRerouteUsage()) {attributeGetTypes [#5]}
  n117 = Abs(None=SphereMask(
    A=Transform(None=@n116.Normal)
    B=Constant3Vector() {constant=<rgba 0 0 1 0>}
    Radius=ScalarParameter() {defaultValue=0.85, group=C_Dirt_Gradient, parameterName=Up/Side Mask}
    Hardness=ScalarParameter() {defaultValue=0.25, group=C_Dirt_Gradient, parameterName=Up/Side Mask Contrast}
   ))
  n118 = WorldPosition()
  n119 = RuntimeVirtualTextureSampleParameter() {
   group=D_RVT
   materialType=WorldHeight
   parameterName=RVTHeight
   sortPriority=1
   virtualTexture=/Game/Athena/Apollo/Maps/Landscape/VT/RVT_Apollo_Landscape_Height
  }
  n120 = Subtract(
   A=Add(A=ScalarParameter() {group=C_Dirt_Gradient, parameterName=Dirt Gradient Offset, sortPriority=1}, B=ComponentMask(None=@n118.XYZ) {b=true})
   B=@n119.WorldHeight
  )
  n121 = Add(A=WorldPosition().XYZ) {constB=-16128}
  n122 = NamedRerouteUsage()
  n123 = Subtract(A=WorldPosition().XYZ, B=ActorPositionWS())
  n124 = Divide(
   A=StaticSwitchParameter(True=LinearInterpolate(A=@n120, B=@n121, Alpha=@n122), False=ComponentMask(None=@n123) {b=true}) {group=C_Dirt_Gradient, parameterName=Use Dirt Gradient RVT}
   B=ScalarParameter() {defaultValue=1500, group=C_Dirt_Gradient, parameterName=Dirt_Gradient_Fade}
  )
  n125 = LinearInterpolate(A=NamedRerouteUsage(), Alpha=Saturate(None=@n124))
  n126 = Max(A=Max(A=Multiply(A=@n117, B=@n125), B=@n124), B=NamedRerouteUsage())
  NamedRerouteDeclaration(None=Saturate(None=QualitySwitch(Default=@n115, High=@n126, Epic=@n126))) {name=RVT Grad, nodeColor=<rgba 1 0.0469 0 1>}
  NamedRerouteDeclaration(None=OneMinus(None=@n57.Dirt Gradient Transparency)) {name=Gradient Mask, nodeColor=<rgba 1 0 0.7734 1>}
  Reroute()
 ]}
