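# Material function for Base4Layers materials that shifts the hue and saturation of BaseColor, bypassed unless UseDiffuseHueShift is on.
# A vertex color channel, scaled by five and rounded, serves as the index picking among five preset amounts.
# A static switch chooses shared amounts or four per-layer sets blended by MF_UV2LerpScalar_Layers_New.
# Results, optionally multiplied by a rounded SFX mask to prevent bleeding, drive HueShift, then Desaturation with negated values as saturation boost.
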MaterialFunction /Game/Global/Base4Layers_2022/MaterialFunctions/MF_DiffuseHueShift
graph {expressions {
  MaterialExpressionChannelMaskParameter_0 {
   class=MaterialExpressionChannelMaskParameter
   inputs [None {from=MaterialExpressionVertexColor_0, output=""}]
   properties {group=DiffuseHueShift, parameterName=DiffuseHueShiftVertexColorChannel, sortPriority=0}
  }
  MaterialExpressionChannelMaskParameter_1 {
   class=MaterialExpressionChannelMaskParameter
   inputs [None {from=MaterialExpressionFunctionInput_0, output=""}]
   properties {group=DiffuseHueShift, parameterName=DiffuseHueShiftSFXMaskChannel, sortPriority=0}
  }
  MaterialExpressionChannelMaskParameter_2 {
   class=MaterialExpressionChannelMaskParameter
   inputs [None {from=MaterialExpressionFunctionInput_0, output=""}]
   properties {group=DiffuseHueShift, maskChannel=Green, parameterName=L2-DiffuseHueShiftSFXMaskChannel, sortPriority=2, defaultValue=<rgba 0 1 0 0>}
  }
  MaterialExpressionChannelMaskParameter_3 {
   class=MaterialExpressionChannelMaskParameter
   inputs [None {from=MaterialExpressionFunctionInput_0, output=""}]
   properties {group=DiffuseHueShift, maskChannel=Blue, parameterName=L3-DiffuseHueShiftSFXMaskChannel, sortPriority=3, defaultValue=<rgba 0 0 1 0>}
  }
  MaterialExpressionChannelMaskParameter_4 {
   class=MaterialExpressionChannelMaskParameter
   inputs [None {from=MaterialExpressionFunctionInput_0, output=""}]
   properties {group=DiffuseHueShift, maskChannel=Alpha, parameterName=L4-DiffuseHueShiftSFXMaskChannel, sortPriority=4, defaultValue=<rgba 0 0 0 1>}
  }
  MaterialExpressionChannelMaskParameter_5 {
   class=MaterialExpressionChannelMaskParameter
   inputs [None {from=MaterialExpressionFunctionInput_0, output=""}]
   properties {group=DiffuseHueShift, parameterName=L1-DiffuseHueShiftSFXMaskChannel, sortPriority=1}
  }
  MaterialExpressionConstant_1 {class=MaterialExpressionConstant}
  MaterialExpressionConstant_10 {class=MaterialExpressionConstant}
  MaterialExpressionConstant_11 {class=MaterialExpressionConstant}
  MaterialExpressionConstant_12 {class=MaterialExpressionConstant}
  MaterialExpressionConstant_13 {class=MaterialExpressionConstant}
  MaterialExpressionConstant_14 {class=MaterialExpressionConstant}
  MaterialExpressionConstant_15 {class=MaterialExpressionConstant}
  MaterialExpressionConstant_16 {class=MaterialExpressionConstant}
  MaterialExpressionConstant_2 {class=MaterialExpressionConstant, properties {r=0.2}}
  MaterialExpressionConstant_3 {class=MaterialExpressionConstant, properties {r=0.4}}
  MaterialExpressionConstant_4 {class=MaterialExpressionConstant, properties {r=0.6}}
  MaterialExpressionConstant_5 {class=MaterialExpressionConstant, properties {r=0.8}}
  MaterialExpressionConstant_6 {class=MaterialExpressionConstant, properties {r=1}}
  MaterialExpressionConstant_7 {class=MaterialExpressionConstant}
  MaterialExpressionConstant_8 {class=MaterialExpressionConstant}
  MaterialExpressionConstant_9 {class=MaterialExpressionConstant}
  MaterialExpressionDesaturation_1 {
   class=MaterialExpressionDesaturation
   inputs [
    None {from=MaterialExpressionReroute_2, output=""}
    Fraction {from=MaterialExpressionStaticSwitchParameter_6, output=""}
   ]
   properties {}
  }
  MaterialExpressionFunctionInput_0 {class=MaterialExpressionFunctionInput, inputs [Preview {from="", output=""}], properties {inputName=SFXMask}}
  MaterialExpressionFunctionInput_1 {class=MaterialExpressionFunctionInput, inputs [Preview {from="", output=""}], properties {inputType=FunctionInput_MaterialAttributes}}
  MaterialExpressionFunctionOutput_0 {class=MaterialExpressionFunctionOutput, inputs [None {from=MaterialExpressionStaticSwitchParameter_0, output=""}]}
  MaterialExpressionGetMaterialAttributes_0 {
   class=MaterialExpressionGetMaterialAttributes
   inputs [None {from=MaterialExpressionFunctionInput_1, output=""}]
   properties {attributeGetTypes [#1]}
  }
  MaterialExpressionMaterialFunctionCall_0 {
   class=MaterialExpressionMaterialFunctionCall
   inputs [
    1 {from=MaterialExpressionChannelMaskParameter_5, output=R}
    2 {from=MaterialExpressionChannelMaskParameter_2, output=R}
    3 {from=MaterialExpressionChannelMaskParameter_3, output=R}
    4 {from=MaterialExpressionChannelMaskParameter_4, output=R}
   ]
   properties {materialFunction=&/MF_UV2LerpScalar_Layers_New}
  }
  MaterialExpressionMaterialFunctionCall_1 {
   class=MaterialExpressionMaterialFunctionCall
   inputs [
    Hue Shift Percentage {from=MaterialExpressionStaticSwitchParameter_3}
    Texture {from=MaterialExpressionNamedRerouteUsage_0}
   ]
   properties {materialFunction=&HueShift}
  }
  MaterialExpressionMaterialFunctionCall_2 {
   class=MaterialExpressionMaterialFunctionCall
   inputs [
    1 {from=MaterialExpressionNamedRerouteUsage_19}
    2 {from=MaterialExpressionNamedRerouteUsage_20}
    3 {from=MaterialExpressionNamedRerouteUsage_21}
    4 {from=MaterialExpressionNamedRerouteUsage_22}
   ]
   properties {materialFunction=&MF_UV2LerpScalar_Layers_New}
  }
  MaterialExpressionMaterialFunctionCall_3 {
   class=MaterialExpressionMaterialFunctionCall
   inputs [
    1 {from=MaterialExpressionNamedRerouteUsage_1}
    2 {from=MaterialExpressionNamedRerouteUsage_6}
    3 {from=MaterialExpressionNamedRerouteUsage_7}
    4 {from=MaterialExpressionNamedRerouteUsage_8}
   ]
   properties {materialFunction=&MF_UV2LerpScalar_Layers_New}
  }
  MaterialExpressionMultiply_0 {
   class=MaterialExpressionMultiply
   inputs [
    A {from=MaterialExpressionChannelMaskParameter_0, output=R}
    B {from=""}
   ]
   properties {constB=5}
  }
  MaterialExpressionMultiply_1 {
   class=MaterialExpressionMultiply
   inputs [
    A {from=MaterialExpressionStaticSwitchParameter_1}
    B {from=MaterialExpressionNamedRerouteUsage_13}
   ]
  }
  MaterialExpressionMultiply_2 {
   class=MaterialExpressionMultiply
   inputs [
    A {from=MaterialExpressionSwitch_5}
    B {from=""}
   ]
   properties {constB=-1}
  }
  MaterialExpressionMultiply_3 {
   class=MaterialExpressionMultiply
   inputs [
    A {from=MaterialExpressionSwitch_6}
    B {from=""}
   ]
   properties {constB=-1}
  }
  MaterialExpressionMultiply_4 {
   class=MaterialExpressionMultiply
   inputs [
    A {from=MaterialExpressionSwitch_7}
    B {from=""}
   ]
   properties {constB=-1}
  }
  MaterialExpressionMultiply_5 {
   class=MaterialExpressionMultiply
   inputs [
    A {from=MaterialExpressionSwitch_8}
    B {from=""}
   ]
   properties {constB=-1}
  }
  MaterialExpressionMultiply_6 {
   class=MaterialExpressionMultiply
   inputs [
    A {from=MaterialExpressionSwitch_9}
    B {from=""}
   ]
   properties {constB=-1}
  }
  MaterialExpressionMultiply_7 {
   class=MaterialExpressionMultiply
   inputs [
    A {from=MaterialExpressionStaticSwitchParameter_5}
    B {from=MaterialExpressionNamedRerouteUsage_23}
   ]
  }
  MaterialExpressionNamedRerouteDeclaration_0 {
   class=MaterialExpressionNamedRerouteDeclaration
   inputs [None {from=MaterialExpressionGetMaterialAttributes_0, output=BaseColor}]
   properties {name=DiffuseIn, nodeColor=<rgba 1 0 0.4453 1>}
  }
  MaterialExpressionNamedRerouteDeclaration_1 {
   class=MaterialExpressionNamedRerouteDeclaration
   inputs [None {from=MaterialExpressionSwitch_0}]
   properties {name=L1-HueShiftPercentage, nodeColor=<rgba 0 0.7656 1 1>}
  }
  MaterialExpressionNamedRerouteDeclaration_10 {
   class=MaterialExpressionNamedRerouteDeclaration
   inputs [None {from=MaterialExpressionMultiply_2}]
   properties {name=L1-SaturationBoost, nodeColor=<rgba 0 0.7656 1 1>}
  }
  MaterialExpressionNamedRerouteDeclaration_11 {
   class=MaterialExpressionNamedRerouteDeclaration
   inputs [None {from=MaterialExpressionRound_1}]
   properties {name=SFXMasks, nodeColor=<rgba 0 0.4609 1 1>}
  }
  MaterialExpressionNamedRerouteDeclaration_12 {
   class=MaterialExpressionNamedRerouteDeclaration
   inputs [None {from=MaterialExpressionMultiply_3}]
   properties {name=L2-SaturationBoost, nodeColor=<rgba 0.6436 0 1 1>}
  }
  MaterialExpressionNamedRerouteDeclaration_13 {
   class=MaterialExpressionNamedRerouteDeclaration
   inputs [None {from=MaterialExpressionMultiply_4}]
   properties {name=L3-SaturationBoost, nodeColor=<rgba 1 0.0239 0 1>}
  }
  MaterialExpressionNamedRerouteDeclaration_14 {
   class=MaterialExpressionNamedRerouteDeclaration
   inputs [None {from=MaterialExpressionMultiply_5}]
   properties {name=L4-SaturationBoost, nodeColor=<rgba 0 1 0.1615 1>}
  }
  MaterialExpressionNamedRerouteDeclaration_15 {
   class=MaterialExpressionNamedRerouteDeclaration
   inputs [None {from=MaterialExpressionMultiply_6}]
   properties {name=AllLayer-SaturationBoost, nodeColor=<rgba 0 0.2266 1 1>}
  }
  MaterialExpressionNamedRerouteDeclaration_2 {
   class=MaterialExpressionNamedRerouteDeclaration
   inputs [None {from=MaterialExpressionRound_0}]
   properties {name=VertexColorMask, nodeColor=<rgba 0 1 0.2734 1>}
  }
  MaterialExpressionNamedRerouteDeclaration_3 {
   class=MaterialExpressionNamedRerouteDeclaration
   inputs [None {from=MaterialExpressionSwitch_1}]
   properties {name=L2-HueShiftPercentage, nodeColor=<rgba 0.6436 0 1 1>}
  }
  MaterialExpressionNamedRerouteDeclaration_4 {
   class=MaterialExpressionNamedRerouteDeclaration
   inputs [None {from=MaterialExpressionSwitch_2}]
   properties {name=L3-HueShiftPercentage, nodeColor=<rgba 1 0.0239 0 1>}
  }
  MaterialExpressionNamedRerouteDeclaration_5 {
   class=MaterialExpressionNamedRerouteDeclaration
   inputs [None {from=MaterialExpressionSwitch_3}]
   properties {name=L4-HueShiftPercentage, nodeColor=<rgba 0 1 0.1615 1>}
  }
  MaterialExpressionNamedRerouteDeclaration_6 {
   class=MaterialExpressionNamedRerouteDeclaration
   inputs [None {from=MaterialExpressionDesaturation_1}]
   properties {name=DiffuseOut, nodeColor=<rgba 0.9922 0 1 1>}
  }
  MaterialExpressionNamedRerouteDeclaration_7 {
   class=MaterialExpressionNamedRerouteDeclaration
   inputs [None {from=MaterialExpressionGetMaterialAttributes_0, output=MaterialAttributes}]
   properties {name=MatIn, nodeColor=<rgba 0 1 0.6719 1>}
  }
  MaterialExpressionNamedRerouteDeclaration_8 {
   class=MaterialExpressionNamedRerouteDeclaration
   inputs [None {from=MaterialExpressionChannelMaskParameter_1, output=R}]
   properties {name=AllLayer-SFXMask, nodeColor=<rgba 0 1 0.9766 1>}
  }
  MaterialExpressionNamedRerouteDeclaration_9 {
   class=MaterialExpressionNamedRerouteDeclaration
   inputs [None {from=MaterialExpressionSwitch_4}]
   properties {name=AllLayer-HueShiftPercentage, nodeColor=<rgba 0 0.2266 1 1>}
  }
  MaterialExpressionNamedRerouteUsage_0 {class=MaterialExpressionNamedRerouteUsage}
  MaterialExpressionNamedRerouteUsage_1 {class=MaterialExpressionNamedRerouteUsage}
  MaterialExpressionNamedRerouteUsage_10 {class=MaterialExpressionNamedRerouteUsage}
  MaterialExpressionNamedRerouteUsage_11 {class=MaterialExpressionNamedRerouteUsage}
  MaterialExpressionNamedRerouteUsage_12 {class=MaterialExpressionNamedRerouteUsage}
  MaterialExpressionNamedRerouteUsage_13 {class=MaterialExpressionNamedRerouteUsage}
  MaterialExpressionNamedRerouteUsage_14 {class=MaterialExpressionNamedRerouteUsage}
  MaterialExpressionNamedRerouteUsage_15 {class=MaterialExpressionNamedRerouteUsage}
  MaterialExpressionNamedRerouteUsage_16 {class=MaterialExpressionNamedRerouteUsage}
  MaterialExpressionNamedRerouteUsage_17 {class=MaterialExpressionNamedRerouteUsage}
  MaterialExpressionNamedRerouteUsage_18 {class=MaterialExpressionNamedRerouteUsage}
  MaterialExpressionNamedRerouteUsage_19 {class=MaterialExpressionNamedRerouteUsage}
  MaterialExpressionNamedRerouteUsage_2 {class=MaterialExpressionNamedRerouteUsage}
  MaterialExpressionNamedRerouteUsage_20 {class=MaterialExpressionNamedRerouteUsage}
  MaterialExpressionNamedRerouteUsage_21 {class=MaterialExpressionNamedRerouteUsage}
  MaterialExpressionNamedRerouteUsage_22 {class=MaterialExpressionNamedRerouteUsage}
  MaterialExpressionNamedRerouteUsage_23 {class=MaterialExpressionNamedRerouteUsage}
  MaterialExpressionNamedRerouteUsage_3 {class=MaterialExpressionNamedRerouteUsage}
  MaterialExpressionNamedRerouteUsage_4 {class=MaterialExpressionNamedRerouteUsage}
  MaterialExpressionNamedRerouteUsage_5 {class=MaterialExpressionNamedRerouteUsage}
  MaterialExpressionNamedRerouteUsage_6 {class=MaterialExpressionNamedRerouteUsage}
  MaterialExpressionNamedRerouteUsage_7 {class=MaterialExpressionNamedRerouteUsage}
  MaterialExpressionNamedRerouteUsage_8 {class=MaterialExpressionNamedRerouteUsage}
  MaterialExpressionNamedRerouteUsage_9 {class=MaterialExpressionNamedRerouteUsage}
  MaterialExpressionReroute_0 {class=MaterialExpressionReroute, inputs [None {from=MaterialExpressionNamedRerouteUsage_9}]}
  MaterialExpressionReroute_1 {class=MaterialExpressionReroute, inputs [None {from=MaterialExpressionMaterialFunctionCall_1, output=Result}]}
  MaterialExpressionReroute_2 {class=MaterialExpressionReroute, inputs [None {from=MaterialExpressionReroute_1}]}
  MaterialExpressionRound_0 {class=MaterialExpressionRound, inputs [None {from=MaterialExpressionMultiply_0, output=""}], properties {}}
  MaterialExpressionRound_1 {
   class=MaterialExpressionRound
   inputs [None {from=MaterialExpressionStaticSwitchParameter_2, output=""}]
   properties {desc=Preventing from mask bleeding}
  }
  MaterialExpressionScalarParameter_0 {class=MaterialExpressionScalarParameter, properties {group=DiffuseHueShift, parameterName=L1-DiffuseHueShiftAmount1, sortPriority=2}}
  MaterialExpressionScalarParameter_1 {class=MaterialExpressionScalarParameter, properties {group=DiffuseHueShift, parameterName=L1-DiffuseHueShiftAmount2, sortPriority=2}}
  MaterialExpressionScalarParameter_10 {class=MaterialExpressionScalarParameter, properties {group=DiffuseHueShift, parameterName=L3-DiffuseHueShiftAmount1, sortPriority=4}}
  MaterialExpressionScalarParameter_11 {class=MaterialExpressionScalarParameter, properties {group=DiffuseHueShift, parameterName=L3-DiffuseHueShiftAmount2, sortPriority=4}}
  MaterialExpressionScalarParameter_12 {class=MaterialExpressionScalarParameter, properties {group=DiffuseHueShift, parameterName=L3-DiffuseHueShiftAmount3, sortPriority=4}}
  MaterialExpressionScalarParameter_13 {class=MaterialExpressionScalarParameter, properties {group=DiffuseHueShift, parameterName=L3-DiffuseHueShiftAmount4, sortPriority=4}}
  MaterialExpressionScalarParameter_14 {class=MaterialExpressionScalarParameter, properties {group=DiffuseHueShift, parameterName=L3-DiffuseHueShiftAmount5, sortPriority=4}}
  MaterialExpressionScalarParameter_15 {class=MaterialExpressionScalarParameter, properties {group=DiffuseHueShift, parameterName=L4-DiffuseHueShiftAmount1, sortPriority=5}}
  MaterialExpressionScalarParameter_16 {class=MaterialExpressionScalarParameter, properties {group=DiffuseHueShift, parameterName=L4-DiffuseHueShiftAmount2, sortPriority=5}}
  MaterialExpressionScalarParameter_17 {class=MaterialExpressionScalarParameter, properties {group=DiffuseHueShift, parameterName=L4-DiffuseHueShiftAmount3, sortPriority=5}}
  MaterialExpressionScalarParameter_18 {class=MaterialExpressionScalarParameter, properties {group=DiffuseHueShift, parameterName=L4-DiffuseHueShiftAmount4, sortPriority=5}}
  MaterialExpressionScalarParameter_19 {class=MaterialExpressionScalarParameter, properties {group=DiffuseHueShift, parameterName=L4-DiffuseHueShiftAmount5, sortPriority=5}}
  MaterialExpressionScalarParameter_2 {class=MaterialExpressionScalarParameter, properties {group=DiffuseHueShift, parameterName=L1-DiffuseHueShiftAmount3, sortPriority=2}}
  MaterialExpressionScalarParameter_20 {class=MaterialExpressionScalarParameter, properties {group=DiffuseHueShift, parameterName=DiffuseHueShiftAmount1, sortPriority=1}}
  MaterialExpressionScalarParameter_21 {class=MaterialExpressionScalarParameter, properties {group=DiffuseHueShift, parameterName=DiffuseHueShiftAmount2, sortPriority=1}}
  MaterialExpressionScalarParameter_22 {class=MaterialExpressionScalarParameter, properties {group=DiffuseHueShift, parameterName=DiffuseHueShiftAmount3, sortPriority=1}}
  MaterialExpressionScalarParameter_23 {class=MaterialExpressionScalarParameter, properties {group=DiffuseHueShift, parameterName=DiffuseHueShiftAmount4, sortPriority=1}}
  MaterialExpressionScalarParameter_24 {class=MaterialExpressionScalarParameter, properties {group=DiffuseHueShift, parameterName=DiffuseHueShiftAmount5, sortPriority=1}}
  MaterialExpressionScalarParameter_25 {class=MaterialExpressionScalarParameter, properties {group=DiffuseHueShift, parameterName=L1-DiffuseSaturationAmount1, sortPriority=2}}
  MaterialExpressionScalarParameter_26 {class=MaterialExpressionScalarParameter, properties {group=DiffuseHueShift, parameterName=L1-DiffuseSaturationAmount2, sortPriority=2}}
  MaterialExpressionScalarParameter_27 {class=MaterialExpressionScalarParameter, properties {group=DiffuseHueShift, parameterName=L1-DiffuseSaturationAmount3, sortPriority=2}}
  MaterialExpressionScalarParameter_28 {class=MaterialExpressionScalarParameter, properties {group=DiffuseHueShift, parameterName=L1-DiffuseHueShiftAmount4, sortPriority=2}}
  MaterialExpressionScalarParameter_29 {class=MaterialExpressionScalarParameter, properties {group=DiffuseHueShift, parameterName=L1-DiffuseSaturationAmount5, sortPriority=2}}
  MaterialExpressionScalarParameter_3 {class=MaterialExpressionScalarParameter, properties {group=DiffuseHueShift, parameterName=L1-DiffuseHueShiftAmount4, sortPriority=2}}
  MaterialExpressionScalarParameter_30 {class=MaterialExpressionScalarParameter, properties {group=DiffuseHueShift, parameterName=L2-DiffuseSaturation1, sortPriority=3}}
  MaterialExpressionScalarParameter_31 {class=MaterialExpressionScalarParameter, properties {group=DiffuseHueShift, parameterName=L2-DiffuseSaturation2, sortPriority=3}}
  MaterialExpressionScalarParameter_32 {class=MaterialExpressionScalarParameter, properties {group=DiffuseHueShift, parameterName=L2-DiffuseSaturation3, sortPriority=3}}
  MaterialExpressionScalarParameter_33 {class=MaterialExpressionScalarParameter, properties {group=DiffuseHueShift, parameterName=L2-DiffuseSaturation4, sortPriority=3}}
  MaterialExpressionScalarParameter_34 {class=MaterialExpressionScalarParameter, properties {group=DiffuseHueShift, parameterName=L2-DiffuseSaturation5, sortPriority=3}}
  MaterialExpressionScalarParameter_35 {class=MaterialExpressionScalarParameter, properties {group=DiffuseHueShift, parameterName=L3-DiffuseSaturation1, sortPriority=4}}
  MaterialExpressionScalarParameter_36 {class=MaterialExpressionScalarParameter, properties {group=DiffuseHueShift, parameterName=L3-DiffuseSaturation2, sortPriority=4}}
  MaterialExpressionScalarParameter_37 {class=MaterialExpressionScalarParameter, properties {group=DiffuseHueShift, parameterName=L3-DiffuseSaturation3, sortPriority=4}}
  MaterialExpressionScalarParameter_38 {class=MaterialExpressionScalarParameter, properties {group=DiffuseHueShift, parameterName=L3-DiffuseSaturation4, sortPriority=4}}
  MaterialExpressionScalarParameter_39 {class=MaterialExpressionScalarParameter, properties {group=DiffuseHueShift, parameterName=L3-DiffuseSaturation5, sortPriority=4}}
  MaterialExpressionScalarParameter_4 {class=MaterialExpressionScalarParameter, properties {group=DiffuseHueShift, parameterName=L1-DiffuseHueShiftAmount5, sortPriority=2}}
  MaterialExpressionScalarParameter_40 {class=MaterialExpressionScalarParameter, properties {group=DiffuseHueShift, parameterName=L4-DiffuseSaturation1, sortPriority=5}}
  MaterialExpressionScalarParameter_41 {class=MaterialExpressionScalarParameter, properties {group=DiffuseHueShift, parameterName=L4-DiffuseSaturation2, sortPriority=5}}
  MaterialExpressionScalarParameter_42 {class=MaterialExpressionScalarParameter, properties {group=DiffuseHueShift, parameterName=L4-DiffuseSaturation3, sortPriority=5}}
  MaterialExpressionScalarParameter_43 {class=MaterialExpressionScalarParameter, properties {group=DiffuseHueShift, parameterName=L4-DiffuseSaturation4, sortPriority=5}}
  MaterialExpressionScalarParameter_44 {class=MaterialExpressionScalarParameter, properties {group=DiffuseHueShift, parameterName=L4-DiffuseSaturation5, sortPriority=5}}
  MaterialExpressionScalarParameter_45 {class=MaterialExpressionScalarParameter, properties {group=DiffuseHueShift, parameterName=DiffuseSaturation1, sortPriority=1}}
  MaterialExpressionScalarParameter_46 {class=MaterialExpressionScalarParameter, properties {group=DiffuseHueShift, parameterName=DiffuseSaturation2, sortPriority=1}}
  MaterialExpressionScalarParameter_47 {class=MaterialExpressionScalarParameter, properties {group=DiffuseHueShift, parameterName=DiffuseSaturation3, sortPriority=1}}
  MaterialExpressionScalarParameter_48 {class=MaterialExpressionScalarParameter, properties {group=DiffuseHueShift, parameterName=DiffuseSaturation4, sortPriority=1}}
  MaterialExpressionScalarParameter_49 {class=MaterialExpressionScalarParameter, properties {group=DiffuseHueShift, parameterName=DiffuseSaturation5, sortPriority=1}}
  MaterialExpressionScalarParameter_5 {class=MaterialExpressionScalarParameter, properties {group=DiffuseHueShift, parameterName=L2-DiffuseHueShiftAmount1, sortPriority=3}}
  MaterialExpressionScalarParameter_6 {class=MaterialExpressionScalarParameter, properties {group=DiffuseHueShift, parameterName=L2-DiffuseHueShiftAmount2, sortPriority=3}}
  MaterialExpressionScalarParameter_7 {class=MaterialExpressionScalarParameter, properties {group=DiffuseHueShift, parameterName=L2-DiffuseHueShiftAmount3, sortPriority=3}}
  MaterialExpressionScalarParameter_8 {class=MaterialExpressionScalarParameter, properties {group=DiffuseHueShift, parameterName=L2-DiffuseHueShiftAmount4, sortPriority=3}}
  MaterialExpressionScalarParameter_9 {class=MaterialExpressionScalarParameter, properties {group=DiffuseHueShift, parameterName=L2-DiffuseHueShiftAmount5, sortPriority=3}}
  MaterialExpressionSetMaterialAttributes_0 {
   class=MaterialExpressionSetMaterialAttributes
   inputs [
    MaterialAttributes {from=MaterialExpressionNamedRerouteUsage_9, output=""}
    BaseColor {from=MaterialExpressionNamedRerouteDeclaration_6, output=""}
   ]
   properties {attributeSetTypes [#1]}
  }
  MaterialExpressionStaticSwitchParameter_0 {
   class=MaterialExpressionStaticSwitchParameter
   inputs [
    True {from=MaterialExpressionSetMaterialAttributes_0, output=""}
    False {from=MaterialExpressionReroute_0, output=""}
   ]
   properties {group=DiffuseHueShift, parameterName=UseDiffuseHueShift}
  }
  MaterialExpressionStaticSwitchParameter_1 {
   class=MaterialExpressionStaticSwitchParameter
   inputs [
    True {from=MaterialExpressionMaterialFunctionCall_3, output=Result}
    False {from=MaterialExpressionNamedRerouteUsage_12, output=""}
   ]
   properties {group=DiffuseHueShift, parameterName=Use4layersDiffuseHueShift, sortPriority=0}
  }
  MaterialExpressionStaticSwitchParameter_2 {
   class=MaterialExpressionStaticSwitchParameter
   inputs [
    True {from=MaterialExpressionMaterialFunctionCall_0, output=Result}
    False {from=MaterialExpressionNamedRerouteDeclaration_8, output=""}
   ]
   properties {group=DiffuseHueShift, parameterName=Use4layersDiffuseHueShift, sortPriority=0}
  }
  MaterialExpressionStaticSwitchParameter_3 {
   class=MaterialExpressionStaticSwitchParameter
   inputs [
    True {from=MaterialExpressionMultiply_1, output=""}
    False {from=MaterialExpressionStaticSwitchParameter_1, output=""}
   ]
   properties {group=DiffuseHueShift, parameterName=UseSFXasMask}
  }
  MaterialExpressionStaticSwitchParameter_5 {
   class=MaterialExpressionStaticSwitchParameter
   inputs [
    True {from=MaterialExpressionMaterialFunctionCall_2, output=Result}
    False {from=MaterialExpressionNamedRerouteUsage_18, output=""}
   ]
   properties {group=DiffuseHueShift, parameterName=Use4layersDiffuseHueShift, sortPriority=0}
  }
  MaterialExpressionStaticSwitchParameter_6 {
   class=MaterialExpressionStaticSwitchParameter
   inputs [
    True {from=MaterialExpressionMultiply_7, output=""}
    False {from=MaterialExpressionStaticSwitchParameter_5, output=""}
   ]
   properties {group=DiffuseHueShift, parameterName=UseSFXasMask}
  }
  MaterialExpressionSwitch_0 {
   class=MaterialExpressionSwitch
   inputs [
    SwitchValue {from=MaterialExpressionNamedRerouteUsage_2, output=""}
    Default {from="", output=""}
    None {from=MaterialExpressionScalarParameter_0, output=""}
    None {from=MaterialExpressionScalarParameter_1, output=""}
    None {from=MaterialExpressionScalarParameter_2, output=""}
    None {from=MaterialExpressionScalarParameter_3, output=""}
    None {from=MaterialExpressionScalarParameter_4, output=""}
    None {from=MaterialExpressionConstant_7, output=""}
   ]
   properties {
    constDefault=6
    constSwitchValue=1
    inputs [
     {
      inputName=None
      input {
       expression=&self:MaterialExpressionScalarParameter_0
       inputName=None
       mask=0
       maskA=0
       maskB=0
       maskG=0
       maskR=0
       outputIndex=0
      }
     }
     {
      inputName=None
      input {
       expression=&self:MaterialExpressionScalarParameter_1
       inputName=None
       mask=0
       maskA=0
       maskB=0
       maskG=0
       maskR=0
       outputIndex=0
      }
     }
     {
      inputName=None
      input {
       expression=&self:MaterialExpressionScalarParameter_2
       inputName=None
       mask=0
       maskA=0
       maskB=0
       maskG=0
       maskR=0
       outputIndex=0
      }
     }
     {
      inputName=None
      input {
       expression=&self:MaterialExpressionScalarParameter_3
       inputName=None
       mask=0
       maskA=0
       maskB=0
       maskG=0
       maskR=0
       outputIndex=0
      }
     }
     {
      inputName=None
      input {
       expression=&self:MaterialExpressionScalarParameter_4
       inputName=None
       mask=0
       maskA=0
       maskB=0
       maskG=0
       maskR=0
       outputIndex=0
      }
     }
     {
      inputName=None
      input {
       expression=&self:MaterialExpressionConstant_7
       inputName=None
       mask=0
       maskA=0
       maskB=0
       maskG=0
       maskR=0
       outputIndex=0
      }
     }
    ]
   }
  }
  MaterialExpressionSwitch_1 {
   class=MaterialExpressionSwitch
   inputs [
    SwitchValue {from=MaterialExpressionNamedRerouteUsage_3, output=""}
    Default {from="", output=""}
    None {from=MaterialExpressionScalarParameter_5, output=""}
    None {from=MaterialExpressionScalarParameter_6, output=""}
    None {from=MaterialExpressionScalarParameter_7, output=""}
    None {from=MaterialExpressionScalarParameter_8, output=""}
    None {from=MaterialExpressionScalarParameter_9, output=""}
    None {from=MaterialExpressionConstant_13, output=""}
   ]
   properties {
    constDefault=6
    constSwitchValue=1
    inputs [
     {
      inputName=None
      input {
       expression=&self:MaterialExpressionScalarParameter_5
       inputName=None
       mask=0
       maskA=0
       maskB=0
       maskG=0
       maskR=0
       outputIndex=0
      }
     }
     {
      inputName=None
      input {
       expression=&self:MaterialExpressionScalarParameter_6
       inputName=None
       mask=0
       maskA=0
       maskB=0
       maskG=0
       maskR=0
       outputIndex=0
      }
     }
     {
      inputName=None
      input {
       expression=&self:MaterialExpressionScalarParameter_7
       inputName=None
       mask=0
       maskA=0
       maskB=0
       maskG=0
       maskR=0
       outputIndex=0
      }
     }
     {
      inputName=None
      input {
       expression=&self:MaterialExpressionScalarParameter_8
       inputName=None
       mask=0
       maskA=0
       maskB=0
       maskG=0
       maskR=0
       outputIndex=0
      }
     }
     {
      inputName=None
      input {
       expression=&self:MaterialExpressionScalarParameter_9
       inputName=None
       mask=0
       maskA=0
       maskB=0
       maskG=0
       maskR=0
       outputIndex=0
      }
     }
     {
      inputName=None
      input {
       expression=&self:MaterialExpressionConstant_13
       inputName=None
       mask=0
       maskA=0
       maskB=0
       maskG=0
       maskR=0
       outputIndex=0
      }
     }
    ]
   }
  }
  MaterialExpressionSwitch_2 {
   class=MaterialExpressionSwitch
   inputs [
    SwitchValue {from=MaterialExpressionNamedRerouteUsage_4, output=""}
    Default {from="", output=""}
    None {from=MaterialExpressionScalarParameter_10, output=""}
    None {from=MaterialExpressionScalarParameter_11, output=""}
    None {from=MaterialExpressionScalarParameter_12, output=""}
    None {from=MaterialExpressionScalarParameter_13, output=""}
    None {from=MaterialExpressionScalarParameter_14, output=""}
    None {from=MaterialExpressionConstant_14, output=""}
   ]
   properties {
    constDefault=6
    constSwitchValue=1
    inputs [
     {
      inputName=None
      input {
       expression=&self:MaterialExpressionScalarParameter_10
       inputName=None
       mask=0
       maskA=0
       maskB=0
       maskG=0
       maskR=0
       outputIndex=0
      }
     }
     {
      inputName=None
      input {
       expression=&self:MaterialExpressionScalarParameter_11
       inputName=None
       mask=0
       maskA=0
       maskB=0
       maskG=0
       maskR=0
       outputIndex=0
      }
     }
     {
      inputName=None
      input {
       expression=&self:MaterialExpressionScalarParameter_12
       inputName=None
       mask=0
       maskA=0
       maskB=0
       maskG=0
       maskR=0
       outputIndex=0
      }
     }
     {
      inputName=None
      input {
       expression=&self:MaterialExpressionScalarParameter_13
       inputName=None
       mask=0
       maskA=0
       maskB=0
       maskG=0
       maskR=0
       outputIndex=0
      }
     }
     {
      inputName=None
      input {
       expression=&self:MaterialExpressionScalarParameter_14
       inputName=None
       mask=0
       maskA=0
       maskB=0
       maskG=0
       maskR=0
       outputIndex=0
      }
     }
     {
      inputName=None
      input {
       expression=&self:MaterialExpressionConstant_14
       inputName=None
       mask=0
       maskA=0
       maskB=0
       maskG=0
       maskR=0
       outputIndex=0
      }
     }
    ]
   }
  }
  MaterialExpressionSwitch_3 {
   class=MaterialExpressionSwitch
   inputs [
    SwitchValue {from=MaterialExpressionNamedRerouteUsage_5, output=""}
    Default {from="", output=""}
    None {from=MaterialExpressionScalarParameter_15, output=""}
    None {from=MaterialExpressionScalarParameter_16, output=""}
    None {from=MaterialExpressionScalarParameter_17, output=""}
    None {from=MaterialExpressionScalarParameter_18, output=""}
    None {from=MaterialExpressionScalarParameter_19, output=""}
    None {from=MaterialExpressionConstant_15, output=""}
   ]
   properties {
    constDefault=6
    constSwitchValue=1
    inputs [
     {
      inputName=None
      input {
       expression=&self:MaterialExpressionScalarParameter_15
       inputName=None
       mask=0
       maskA=0
       maskB=0
       maskG=0
       maskR=0
       outputIndex=0
      }
     }
     {
      inputName=None
      input {
       expression=&self:MaterialExpressionScalarParameter_16
       inputName=None
       mask=0
       maskA=0
       maskB=0
       maskG=0
       maskR=0
       outputIndex=0
      }
     }
     {
      inputName=None
      input {
       expression=&self:MaterialExpressionScalarParameter_17
       inputName=None
       mask=0
       maskA=0
       maskB=0
       maskG=0
       maskR=0
       outputIndex=0
      }
     }
     {
      inputName=None
      input {
       expression=&self:MaterialExpressionScalarParameter_18
       inputName=None
       mask=0
       maskA=0
       maskB=0
       maskG=0
       maskR=0
       outputIndex=0
      }
     }
     {
      inputName=None
      input {
       expression=&self:MaterialExpressionScalarParameter_19
       inputName=None
       mask=0
       maskA=0
       maskB=0
       maskG=0
       maskR=0
       outputIndex=0
      }
     }
     {
      inputName=None
      input {
       expression=&self:MaterialExpressionConstant_15
       inputName=None
       mask=0
       maskA=0
       maskB=0
       maskG=0
       maskR=0
       outputIndex=0
      }
     }
    ]
   }
  }
  MaterialExpressionSwitch_4 {
   class=MaterialExpressionSwitch
   inputs [
    SwitchValue {from=MaterialExpressionNamedRerouteUsage_11, output=""}
    Default {from="", output=""}
    None {from=MaterialExpressionScalarParameter_20, output=""}
    None {from=MaterialExpressionScalarParameter_21, output=""}
    None {from=MaterialExpressionScalarParameter_22, output=""}
    None {from=MaterialExpressionScalarParameter_23, output=""}
    None {from=MaterialExpressionScalarParameter_24, output=""}
    None {from=MaterialExpressionConstant_16, output=""}
   ]
   properties {
    constDefault=6
    constSwitchValue=1
    inputs [
     {
      inputName=None
      input {
       expression=&self:MaterialExpressionScalarParameter_20
       inputName=None
       mask=0
       maskA=0
       maskB=0
       maskG=0
       maskR=0
       outputIndex=0
      }
     }
     {
      inputName=None
      input {
       expression=&self:MaterialExpressionScalarParameter_21
       inputName=None
       mask=0
       maskA=0
       maskB=0
       maskG=0
       maskR=0
       outputIndex=0
      }
     }
     {
      inputName=None
      input {
       expression=&self:MaterialExpressionScalarParameter_22
       inputName=None
       mask=0
       maskA=0
       maskB=0
       maskG=0
       maskR=0
       outputIndex=0
      }
     }
     {
      inputName=None
      input {
       expression=&self:MaterialExpressionScalarParameter_23
       inputName=None
       mask=0
       maskA=0
       maskB=0
       maskG=0
       maskR=0
       outputIndex=0
      }
     }
     {
      inputName=None
      input {
       expression=&self:MaterialExpressionScalarParameter_24
       inputName=None
       mask=0
       maskA=0
       maskB=0
       maskG=0
       maskR=0
       outputIndex=0
      }
     }
     {
      inputName=None
      input {
       expression=&self:MaterialExpressionConstant_16
       inputName=None
       mask=0
       maskA=0
       maskB=0
       maskG=0
       maskR=0
       outputIndex=0
      }
     }
    ]
   }
  }
  MaterialExpressionSwitch_5 {
   class=MaterialExpressionSwitch
   inputs [
    SwitchValue {from=MaterialExpressionNamedRerouteUsage_10, output=""}
    Default {from="", output=""}
    None {from=MaterialExpressionScalarParameter_25, output=""}
    None {from=MaterialExpressionScalarParameter_26, output=""}
    None {from=MaterialExpressionScalarParameter_27, output=""}
    None {from=MaterialExpressionScalarParameter_28, output=""}
    None {from=MaterialExpressionScalarParameter_29, output=""}
    None {from=MaterialExpressionConstant_8, output=""}
   ]
   properties {
    constDefault=6
    constSwitchValue=1
    inputs [
     {
      inputName=None
      input {
       expression=&self:MaterialExpressionScalarParameter_25
       inputName=None
       mask=0
       maskA=0
       maskB=0
       maskG=0
       maskR=0
       outputIndex=0
      }
     }
     {
      inputName=None
      input {
       expression=&self:MaterialExpressionScalarParameter_26
       inputName=None
       mask=0
       maskA=0
       maskB=0
       maskG=0
       maskR=0
       outputIndex=0
      }
     }
     {
      inputName=None
      input {
       expression=&self:MaterialExpressionScalarParameter_27
       inputName=None
       mask=0
       maskA=0
       maskB=0
       maskG=0
       maskR=0
       outputIndex=0
      }
     }
     {
      inputName=None
      input {
       expression=&self:MaterialExpressionScalarParameter_28
       inputName=None
       mask=0
       maskA=0
       maskB=0
       maskG=0
       maskR=0
       outputIndex=0
      }
     }
     {
      inputName=None
      input {
       expression=&self:MaterialExpressionScalarParameter_29
       inputName=None
       mask=0
       maskA=0
       maskB=0
       maskG=0
       maskR=0
       outputIndex=0
      }
     }
     {
      inputName=None
      input {
       expression=&self:MaterialExpressionConstant_8
       inputName=None
       mask=0
       maskA=0
       maskB=0
       maskG=0
       maskR=0
       outputIndex=0
      }
     }
    ]
   }
  }
  MaterialExpressionSwitch_6 {
   class=MaterialExpressionSwitch
   inputs [
    SwitchValue {from=MaterialExpressionNamedRerouteUsage_14, output=""}
    Default {from="", output=""}
    None {from=MaterialExpressionScalarParameter_30, output=""}
    None {from=MaterialExpressionScalarParameter_31, output=""}
    None {from=MaterialExpressionScalarParameter_32, output=""}
    None {from=MaterialExpressionScalarParameter_33, output=""}
    None {from=MaterialExpressionScalarParameter_34, output=""}
    None {from=MaterialExpressionConstant_9, output=""}
   ]
   properties {
    constDefault=6
    constSwitchValue=1
    inputs [
     {
      inputName=None
      input {
       expression=&self:MaterialExpressionScalarParameter_30
       inputName=None
       mask=0
       maskA=0
       maskB=0
       maskG=0
       maskR=0
       outputIndex=0
      }
     }
     {
      inputName=None
      input {
       expression=&self:MaterialExpressionScalarParameter_31
       inputName=None
       mask=0
       maskA=0
       maskB=0
       maskG=0
       maskR=0
       outputIndex=0
      }
     }
     {
      inputName=None
      input {
       expression=&self:MaterialExpressionScalarParameter_32
       inputName=None
       mask=0
       maskA=0
       maskB=0
       maskG=0
       maskR=0
       outputIndex=0
      }
     }
     {
      inputName=None
      input {
       expression=&self:MaterialExpressionScalarParameter_33
       inputName=None
       mask=0
       maskA=0
       maskB=0
       maskG=0
       maskR=0
       outputIndex=0
      }
     }
     {
      inputName=None
      input {
       expression=&self:MaterialExpressionScalarParameter_34
       inputName=None
       mask=0
       maskA=0
       maskB=0
       maskG=0
       maskR=0
       outputIndex=0
      }
     }
     {
      inputName=None
      input {
       expression=&self:MaterialExpressionConstant_9
       inputName=None
       mask=0
       maskA=0
       maskB=0
       maskG=0
       maskR=0
       outputIndex=0
      }
     }
    ]
   }
  }
  MaterialExpressionSwitch_7 {
   class=MaterialExpressionSwitch
   inputs [
    SwitchValue {from=MaterialExpressionNamedRerouteUsage_15, output=""}
    Default {from="", output=""}
    None {from=MaterialExpressionScalarParameter_35, output=""}
    None {from=MaterialExpressionScalarParameter_36, output=""}
    None {from=MaterialExpressionScalarParameter_37, output=""}
    None {from=MaterialExpressionScalarParameter_38, output=""}
    None {from=MaterialExpressionScalarParameter_39, output=""}
    None {from=MaterialExpressionConstant_10, output=""}
   ]
   properties {
    constDefault=6
    constSwitchValue=1
    inputs [
     {
      inputName=None
      input {
       expression=&self:MaterialExpressionScalarParameter_35
       inputName=None
       mask=0
       maskA=0
       maskB=0
       maskG=0
       maskR=0
       outputIndex=0
      }
     }
     {
      inputName=None
      input {
       expression=&self:MaterialExpressionScalarParameter_36
       inputName=None
       mask=0
       maskA=0
       maskB=0
       maskG=0
       maskR=0
       outputIndex=0
      }
     }
     {
      inputName=None
      input {
       expression=&self:MaterialExpressionScalarParameter_37
       inputName=None
       mask=0
       maskA=0
       maskB=0
       maskG=0
       maskR=0
       outputIndex=0
      }
     }
     {
      inputName=None
      input {
       expression=&self:MaterialExpressionScalarParameter_38
       inputName=None
       mask=0
       maskA=0
       maskB=0
       maskG=0
       maskR=0
       outputIndex=0
      }
     }
     {
      inputName=None
      input {
       expression=&self:MaterialExpressionScalarParameter_39
       inputName=None
       mask=0
       maskA=0
       maskB=0
       maskG=0
       maskR=0
       outputIndex=0
      }
     }
     {
      inputName=None
      input {
       expression=&self:MaterialExpressionConstant_10
       inputName=None
       mask=0
       maskA=0
       maskB=0
       maskG=0
       maskR=0
       outputIndex=0
      }
     }
    ]
   }
  }
  MaterialExpressionSwitch_8 {
   class=MaterialExpressionSwitch
   inputs [
    SwitchValue {from=MaterialExpressionNamedRerouteUsage_16, output=""}
    Default {from="", output=""}
    None {from=MaterialExpressionScalarParameter_40, output=""}
    None {from=MaterialExpressionScalarParameter_41, output=""}
    None {from=MaterialExpressionScalarParameter_42, output=""}
    None {from=MaterialExpressionScalarParameter_43, output=""}
    None {from=MaterialExpressionScalarParameter_44, output=""}
    None {from=MaterialExpressionConstant_11, output=""}
   ]
   properties {
    constDefault=6
    constSwitchValue=1
    inputs [
     {
      inputName=None
      input {
       expression=&self:MaterialExpressionScalarParameter_40
       inputName=None
       mask=0
       maskA=0
       maskB=0
       maskG=0
       maskR=0
       outputIndex=0
      }
     }
     {
      inputName=None
      input {
       expression=&self:MaterialExpressionScalarParameter_41
       inputName=None
       mask=0
       maskA=0
       maskB=0
       maskG=0
       maskR=0
       outputIndex=0
      }
     }
     {
      inputName=None
      input {
       expression=&self:MaterialExpressionScalarParameter_42
       inputName=None
       mask=0
       maskA=0
       maskB=0
       maskG=0
       maskR=0
       outputIndex=0
      }
     }
     {
      inputName=None
      input {
       expression=&self:MaterialExpressionScalarParameter_43
       inputName=None
       mask=0
       maskA=0
       maskB=0
       maskG=0
       maskR=0
       outputIndex=0
      }
     }
     {
      inputName=None
      input {
       expression=&self:MaterialExpressionScalarParameter_44
       inputName=None
       mask=0
       maskA=0
       maskB=0
       maskG=0
       maskR=0
       outputIndex=0
      }
     }
     {
      inputName=None
      input {
       expression=&self:MaterialExpressionConstant_11
       inputName=None
       mask=0
       maskA=0
       maskB=0
       maskG=0
       maskR=0
       outputIndex=0
      }
     }
    ]
   }
  }
  MaterialExpressionSwitch_9 {
   class=MaterialExpressionSwitch
   inputs [
    SwitchValue {from=MaterialExpressionNamedRerouteUsage_17, output=""}
    Default {from="", output=""}
    None {from=MaterialExpressionScalarParameter_45, output=""}
    None {from=MaterialExpressionScalarParameter_46, output=""}
    None {from=MaterialExpressionScalarParameter_47, output=""}
    None {from=MaterialExpressionScalarParameter_48, output=""}
    None {from=MaterialExpressionScalarParameter_49, output=""}
    None {from=MaterialExpressionConstant_12, output=""}
   ]
   properties {
    constDefault=6
    constSwitchValue=1
    inputs [
     {
      inputName=None
      input {
       expression=&self:MaterialExpressionScalarParameter_45
       inputName=None
       mask=0
       maskA=0
       maskB=0
       maskG=0
       maskR=0
       outputIndex=0
      }
     }
     {
      inputName=None
      input {
       expression=&self:MaterialExpressionScalarParameter_46
       inputName=None
       mask=0
       maskA=0
       maskB=0
       maskG=0
       maskR=0
       outputIndex=0
      }
     }
     {
      inputName=None
      input {
       expression=&self:MaterialExpressionScalarParameter_47
       inputName=None
       mask=0
       maskA=0
       maskB=0
       maskG=0
       maskR=0
       outputIndex=0
      }
     }
     {
      inputName=None
      input {
       expression=&self:MaterialExpressionScalarParameter_48
       inputName=None
       mask=0
       maskA=0
       maskB=0
       maskG=0
       maskR=0
       outputIndex=0
      }
     }
     {
      inputName=None
      input {
       expression=&self:MaterialExpressionScalarParameter_49
       inputName=None
       mask=0
       maskA=0
       maskB=0
       maskG=0
       maskR=0
       outputIndex=0
      }
     }
     {
      inputName=None
      input {
       expression=&self:MaterialExpressionConstant_12
       inputName=None
       mask=0
       maskA=0
       maskB=0
       maskG=0
       maskR=0
       outputIndex=0
      }
     }
    ]
   }
  }
  MaterialExpressionVertexColor_0 {class=MaterialExpressionVertexColor, inputs [], properties {}}
 }}
