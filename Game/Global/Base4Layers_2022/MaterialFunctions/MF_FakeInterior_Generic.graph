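# Material function that adds fake room interiors for windows to the emissive of incoming material attributes, toggled by UseFakeInterior.
# An interior-cubemap function, randomized by object-position noise, feeds long-lat lookups into a texture array, with sine distortion and explicit derivatives.
# Time-of-day tint and contrast, optional inside/outside blending and distance fade shape the result, compensated for eye adaptation.
# Mobile shading substitutes a tinted, bump-offset sphere-mask glow for the cubemap.
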MaterialFunction /Game/Global/Base4Layers_2022/MaterialFunctions/MF_FakeInterior_Generic
graph {nodes [
  Add(
   A=Divide(A=WorldPosition().XYZ, B=VectorParameter() {parameterName=RandomMaskWorldUVDivisor, defaultValue=<rgba 512 512 384 1>}.RGB)
   B=VectorParameter() {parameterName=RandomMaskWorldUVBias, defaultValue=<rgba 0.51 0.51 0.01 1>}.RGB
  ) {constB=0.5}
  AppendVector(A=NamedRerouteUsage(), B=Constant())
  Constant3Vector() {constant=<rgba 0.6179 0.9527 2.3 1>}
  Constant3Vector() {constant=<rgba 0.0816 0.0708 0.0656 1>}
  Constant3Vector() {constant=<rgba 1.1379 1.6437 1.8 1>}
  Constant3Vector() {constant=<rgba 1.4 0.9836 0.9028 1>}
  Constant3Vector() {constant=<rgba 1.2535 0.9286 0.8857 1>}
  Constant3Vector() {constant=<rgba 0.4263 0.5191 0.5144 1>}
  Constant3Vector() {constant=<rgba 0.5 0.55 0 1>}
  Constant3Vector() {constant=<rgba 1.3 0.4 0 1>}
  Constant()
  n1 = ObjectPositionWS()
  n2 = VectorNoise(World Position=@n1)
  n3 = NamedRerouteDeclaration(None=ComponentMask(None=@n2) {r=true}) {name=WindowRandomMask.R, nodeColor=<rgba 0.9922 1 0 1>}
  n4 = ComponentMask(None=@n2) {b=true}
  Frac(None=Add(A=@n3, B=NamedRerouteDeclaration(None=@n4) {name=WindowRandomMask.B, nodeColor=<rgba 0 0 1 1>}))
  n5 = MaterialFunctionCall(Float3=@n1) {materialFunction=&BreakOutFloat3Components}
  n6 = Add(A=Add(A=@n5.R, B=@n5.R), B=@n5.B)
  Frac(None=Divide(A=@n6) {constB=122})
  Frac(None=Divide(A=@n6) {constB=99})
  n7 = GetMaterialAttributes(None=NamedRerouteUsage()) {attributeGetTypes [#1]}
  n8 = Multiply(A=NamedRerouteUsage(), B=NamedRerouteUsage())
  n9 = EyeAdaptationInverse()
  n10 = SetMaterialAttributes!(MaterialAttributes=@n7.MaterialAttributes, EmissiveColor=Add(A=Multiply(A=@n8, B=@n9.EyeAdaptationInverse), B=@n7.EmissiveColor)) {attributeSetTypes [#1]}
  FunctionOutput(None=StaticSwitchParameter(True=NamedRerouteDeclaration(None=@n10) {name=FakeInterior_Material, nodeColor=<rgba 0 1 0.7656 1>}, False=NamedRerouteUsage()) {group=Fake Windows, parameterName=UseFakeInterior, sortPriority=-1})
  n11 = TextureCoordinate()
  n12 = ComponentMask(None=VectorParameter() {group=Fake Windows, parameterName=Mobile_FadeLocation, defaultValue=<rgba 0 0 0 1>}.RGB) {g=true, r=true}
  n13 = VectorParameter() {group=Fake Windows, parameterName=Mobile_WindowBounds, defaultValue=<rgba 0.9 0.3 0 1>}
  LinearInterpolate(Alpha=MaterialFunctionCall!(
    A=@n11
    B=@n12
    Bounds=ComponentMask(None=@n13.RGB) {g=true, r=true}
    Edge Falloff=ScalarParameter() {defaultValue=0.1, desc=0.075, group=Fake Windows, parameterName=Mobile_WindowEdgeFalloff}
   ) {materialFunction=/Engine/Functions/Engine_MaterialFunctions01/Gradient/BoxMask-2D}.Result) {constB=0.5}
  n14 = BumpOffset(Coordinate=@n11, Height=ScalarParameter() {defaultValue=-0.35, group=Fake Windows Mobile, parameterName=Mobile_Bump Offset}) {heightRatio=0.5}
  n15 = Add(A=Multiply(A=Subtract(A=@n14, B=@n12), B=Constant2Vector() {g=3, r=0.95})) {constB=0.25}
  n16 = MaterialFunctionCall(Float2=ConstantBiasScale(None=Sine(None=@n15))) {materialFunction=&BreakOutFloat2Components}
  LinearInterpolate(
   A=VectorParameter() {group=Fake Windows, parameterName=Mobile_WindowEmissive_1, defaultValue=<rgba 0.023 0.0178 0.0153 1>}.RGB
   B=VectorParameter() {group=Fake Windows, parameterName=Mobile_WindowEmissive_2, defaultValue=<rgba 0.1214 0.1823 0.1457 1>}.RGB
   Alpha=Multiply(A=@n16.R, B=@n16.R)
  )
  n17 = GetMaterialAttributes(None=FunctionInput(Preview=SetMaterialAttributes!(MaterialAttributes)) {bUsePreviewValueAsDefault=true, inputName=Material, inputType=FunctionInput_MaterialAttributes}) {attributeGetTypes [#2]}
  NamedRerouteDeclaration(None=@n17.MaterialAttributes) {name=Input_Material, nodeColor=<rgba 0.5469 0 1 1>}
  NamedRerouteDeclaration(None=Frac(None=Divide(A=@n6) {constB=133})) {name=Debug, nodeColor=<rgba 1 0.9609 0 1>}
  NamedRerouteDeclaration(None=@n17.Normal) {name=Input_Normal, nodeColor=<rgba 1 0 0.3984 1>}
  n18 = VectorParameter() {group=Fake Windows Mobile, parameterName=Mobile_InsideCubemapMorning, sortPriority=11, defaultValue=<rgba 1.8 0.8801 0.683 1>}
  n19 = VectorParameter() {group=Fake Windows Mobile, parameterName=Mobile_InsideCubemapDay, sortPriority=12, defaultValue=<rgba 1.35 1.3381 1.2502 1>}
  n20 = VectorParameter() {group=Fake Windows Mobile, parameterName=Mobile_InsideCubemapEvening, sortPriority=13, defaultValue=<rgba 1.5 0.834 0.6703 1>}
  n21 = VectorParameter() {group=Fake Windows Mobile, parameterName=Mobile_InsideCubemapNight, sortPriority=14, defaultValue=<rgba 0.6828 0.9558 1.5 1>}
  n22 = VectorParameter() {bUseCustomPrimitiveData=true, parameterName=CubeMapTintPrimData, primitiveDataIndex=17, defaultValue=<rgba 0 0 0 1>}
  n23 = VectorParameter() {group=Fake Windows Mobile, parameterName=Mobile_CubemapMorning, sortPriority=1, defaultValue=<rgba 1.8 0.8801 0.683 1>}
  n24 = VectorParameter() {group=Fake Windows Mobile, parameterName=Mobile_CubemapDay, sortPriority=2, defaultValue=<rgba 1.35 1.3381 1.2502 1>}
  n25 = VectorParameter() {group=Fake Windows Mobile, parameterName=Mobile_CubemapEvening, sortPriority=3, defaultValue=<rgba 1.5 0.834 0.6703 1>}
  n26 = VectorParameter() {group=Fake Windows Mobile, parameterName=Mobile_CubemapNight, sortPriority=4, defaultValue=<rgba 0.6828 0.9558 1.5 1>}
  n27 = MaterialFunctionCall(Color_Morning=@n23.RGB, Color_Day=@n24.RGB, Color_Evening=@n25.RGB, Color_Night=@n26.RGB) {materialFunction=&Color_Time_of_Day}
  n28 = StaticSwitchParameter(True=Multiply(A=@n22.RGB, B=@n27.Result), False=@n27.Result) {group=Fake Windows, parameterName=FakeWindows_UsePrimDataTint}
  n29 = LinearInterpolate(
   A=MaterialFunctionCall(Color_Morning=@n18.RGB, Color_Day=@n19.RGB, Color_Evening=@n20.RGB, Color_Night=@n21.RGB) {materialFunction=&Color_Time_of_Day}.Result
   B=@n28
   Alpha=NamedRerouteUsage()
  )
  n30 = Constant2Vector() {g=0.5, r=0.5}
  n31 = ScalarParameter() {defaultValue=1.4, group=Fake Windows Mobile, parameterName=Mobile_InteriorGlowSize}
  n32 = ScalarParameter() {defaultValue=-2.5, group=Fake Windows Mobile, parameterName=Mobile_InteriorGlowHardness}
  n33 = Multiply(
   A=Multiply(
    A=StaticSwitchParameter(True=@n29, False=@n28) {group=Fake Windows, parameterName=FakeWindows_UseInsideOutSideMask, sortPriority=4}
    B=SphereMask(A=@n14, B=@n30, Radius=@n31, Hardness=@n32)
   )
   B=NamedRerouteUsage()
  )
  n34 = VectorParameter() {group=Fake Windows, parameterName=InsideCubemapMorning, sortPriority=11, defaultValue=<rgba 2 0.4561 0.5357 1>}
  n35 = VectorParameter() {group=Fake Windows, parameterName=InsideCubemapTintDay, sortPriority=12, defaultValue=<rgba 1 0.8667 0.6646 1>}
  n36 = VectorParameter() {group=Fake Windows, parameterName=InsideCubemapEvening, sortPriority=13, defaultValue=<rgba 2 0.5474 0.0849 1>}
  n37 = VectorParameter() {group=Fake Windows, parameterName=InsideCubemapTintNight, sortPriority=13, defaultValue=<rgba 0.3229 0.7422 2 1>}
  n38 = VectorParameter() {bUseCustomPrimitiveData=true, parameterName=CubeMapTintPrimData, primitiveDataIndex=17, defaultValue=<rgba 0 0 0 1>}
  n39 = VectorParameter() {group=Fake Windows, parameterName=CubemapMorning, sortPriority=1, defaultValue=<rgba 2 0.4561 0.5357 1>}
  n40 = VectorParameter() {group=Fake Windows, parameterName=CubemapTintDay, sortPriority=2, defaultValue=<rgba 1 0.8667 0.6646 1>}
  n41 = VectorParameter() {group=Fake Windows, parameterName=CubemapTintEvening, sortPriority=3, defaultValue=<rgba 2 0.5474 0.0849 1>}
  n42 = VectorParameter() {group=Fake Windows, parameterName=CubemapTintNight, sortPriority=4, defaultValue=<rgba 0.3229 0.7422 2 1>}
  n43 = MaterialFunctionCall(Color_Morning=@n39.RGB, Color_Day=@n40.RGB, Color_Evening=@n41.RGB, Color_Night=@n42.RGB) {materialFunction=&Color_Time_of_Day}
  n44 = StaticSwitchParameter(True=Multiply(A=@n38.RGB, B=@n43.Result), False=@n43.Result) {group=Fake Windows, parameterName=FakeWindows_UsePrimDataTint, sortPriority=0}
  n45 = StaticSwitchParameter(
   True=LinearInterpolate(
    A=MaterialFunctionCall(Color_Morning=@n34.RGB, Color_Day=@n35.RGB, Color_Evening=@n36.RGB, Color_Night=@n37.RGB) {materialFunction=&Color_Time_of_Day}.Result
    B=@n44
    Alpha=NamedRerouteUsage()
   )
   False=@n44
  ) {group=Fake Windows, parameterName=FakeWindows_UseInsideOutSideMask, sortPriority=4}
  n46 = TextureCoordinate()
  n47 = Divide(A=@n46, B=ScalarParameter() {defaultValue=0.02, group=Fake Windows, parameterName=DistortionNoiseSize, sortPriority=900})
  n48 = Divide(A=@n46, B=ScalarParameter() {defaultValue=0.2, group=Fake Windows, parameterName=DistortionNoiseSize_Far, sortPriority=900})
  n49 = PixelDepth()
  n50 = ScalarParameter() {defaultValue=5000, group=Fake Windows, parameterName=DistortionNoiseDistance, sortPriority=900}
  n51 = LinearInterpolate(
   A=Multiply(A=Sine(None=@n47), B=ScalarParameter() {defaultValue=0.01, group=Fake Windows, parameterName=DistortCubemapIntensity_Near, sortPriority=900})
   B=Multiply(A=Sine(None=@n48), B=ScalarParameter() {defaultValue=0.005, group=Fake Windows, parameterName=DistortCubemapIntensity_Far, sortPriority=900})
   Alpha=Saturate(None=Divide(A=@n49.R, B=@n50))
  )
  n52 = Constant3Vector() {constant=<rgba 0 0 1 0>}
  n53 = Add(
   A=CameraVectorWS()
   B=Multiply(
    A=Transform(None=@n52) {transformSourceType=TRANSFORMSOURCE_Camera}
    B=ScalarParameter() {defaultValue=0.25, group=Fake Windows, parameterName=Inside_CubeDistorition, sortPriority=600}
   )
  )
  n54 = TextureCoordinate()
  n55 = Multiply(
   A=ComponentMask(None=NamedRerouteUsage()) {g=true, r=true}
   B=ScalarParameter() {group=Fake Windows, parameterName=WindowRefractionScale, sortPriority=500}
  )
  n56 = AppendVector(A=@n3, B=NamedRerouteDeclaration(None=ComponentMask(None=@n2) {g=true}) {name=WindowRandomMask.G, nodeColor=<rgba 0 1 0 1>})
  n57 = ScalarParameter() {defaultValue=2.33, group=Fake Windows, parameterName=RandomMult}
  n58 = MaterialFunctionCall!(
   UVs=NamedRerouteDeclaration(None=Add(A=@n54, B=@n55)) {name=InteriorCubeMapUV, nodeColor=<rgba 1 0 0.2344 1>}
   Custom Random Input=Ceil(None=Multiply(A=@n56, B=@n57))
   Tiling=AppendVector(
    A=ScalarParameter() {defaultValue=1, group=Fake Windows, parameterName=InteriorCubeMapScale_X, sortPriority=500}
    B=ScalarParameter() {defaultValue=1, group=Fake Windows, parameterName=InteriorCubeMapScale_Y, sortPriority=500}
   )
  ) {materialFunction=/Game/Environments/Asteria/Sets/NeonCity/Techart/MF_NeonCity_InteriorCubemap}
  n59 = StaticBoolParameter() {defaultValue=true, group=Fake Windows, parameterName=RandomizeCubemapRotation}
  n60 = NamedRerouteDeclaration(None=QualitySwitch(Default=@n58.UVW, Epic=StaticSwitch(True=@n58.UVW Randomized, False=@n58.UVW Randomized, Value=@n59))) {name=CubemapUVW, nodeColor=<rgba 1 0 0.5156 1>}
  n61 = NamedRerouteUsage()
  n62 = MaterialFunctionCall!(Vector=StaticSwitchParameter(True=LinearInterpolate(A=@n53, B=@n60, Alpha=@n61), False=@n60) {group=Fake Windows, parameterName=FakeWindows_UseInsideOutSideMask, sortPriority=4}) {materialFunction=/Engine/Functions/Engine_MaterialFunctions01/Coordinates/LongLatToUV}
  n63 = Round(None=Multiply(A=ScalarParameter() {defaultValue=3, group=Fake Windows, parameterName=TextureArrayLength, sortPriority=600}, B=NamedRerouteUsage()))
  n64 = LinearInterpolate(
   A=ScalarParameter() {defaultValue=1, group=Fake Windows, parameterName=Inside_TextureArray_ID, sortPriority=600}
   B=@n63
   Alpha=NamedRerouteUsage()
  )
  n65 = AppendVector(A=TextureCoordinate(), B=Constant())
  n66 = TextureSampleParameter2DArray!(
   UVs=AppendVector(
    A=Add(A=@n51, B=@n62.Result)
    B=StaticSwitchParameter(True=@n64, False=@n63) {group=Fake Windows, parameterName=FakeWindows_UseInsideOutSideMask, sortPriority=4}
   )
   "DDX(UVs)"=DDX(Value=@n65)
   "DDY(UVs)"=DDY(Value=@n65)
   Apply View MipBias
  ) {
   mipValueMode=TMVM_Derivative
   parameterName=CubeMapArray
   sortPriority=600
   texture=/Game/Environments/Asteria/Sets/NeonCity/Techart/Textures/TA_FakeInterior_Blur_LongLat1
  }
  n67 = MaterialFunctionCall(
   Value_Morning=ScalarParameter() {defaultValue=0.8, group=Fake Windows, parameterName=CubemapContrast_Morning, sortPriority=6}
   Value_Day=ScalarParameter() {defaultValue=0.8, group=Fake Windows, parameterName=CubemapContrast_Day, sortPriority=7}
   Value_Evening=ScalarParameter() {defaultValue=0.9, group=Fake Windows, parameterName=CubemapContrast_Evening, sortPriority=8}
   Value_Night=ScalarParameter() {defaultValue=0.75, group=Fake Windows, parameterName=CubemapContrast_Night, sortPriority=9}
  ) {materialFunction=&Scalar_Time_of_Day}
  n68 = LinearInterpolate(
   A=MaterialFunctionCall(
    Value_Morning=ScalarParameter() {defaultValue=0.8, group=Fake Windows, parameterName=Inside_CubemapContrast_Morning, sortPriority=16}
    Value_Day=ScalarParameter() {defaultValue=0.8, group=Fake Windows, parameterName=Inside_CubemapContrast_Day, sortPriority=17}
    Value_Evening=ScalarParameter() {defaultValue=0.9, group=Fake Windows, parameterName=Inside_CubemapContrast_Evening, sortPriority=18}
    Value_Night=ScalarParameter() {defaultValue=0.75, group=Fake Windows, parameterName=Inside_CubemapContrast_Night, sortPriority=19}
   ) {materialFunction=&Scalar_Time_of_Day}.Result
   B=@n67.Result
   Alpha=NamedRerouteUsage()
  )
  n69 = Desaturation(
   None=Power(
    Base=@n66.RGB
    Exp=StaticSwitchParameter(True=@n68, False=@n67.Result) {group=Fake Windows, parameterName=FakeWindows_UseInsideOutSideMask, sortPriority=4}
   )
   Fraction=ScalarParameter() {defaultValue=0.1, group=Fake Windows, parameterName=Cubemap_Desat, sortPriority=5}
  )
  n70 = LinearInterpolate(
   A=Constant() {r=1}
   B=ScalarParameter() {defaultValue=0.4, group=Fake Windows, parameterName=FakeInteriorDistanceFadeAmt, sortPriority=100}
   Alpha=NamedRerouteUsage()
  )
  n71 = ShadingPathSwitch(Default=Multiply(A=Multiply(A=@n45, B=@n69), B=@n70), Mobile=@n33) {desc=Migrated from FeatureLevelSwitch}
  NamedRerouteDeclaration(None=@n71) {name=WindowDiffuse, nodeColor=<rgba 1 0.5625 0 1>}
  NamedRerouteDeclaration(None=@n71) {name=WindowEmissive, nodeColor=<rgba 0.2422 1 0 1>}
  n72 = ScalarParameter() {defaultValue=20000, group=Fake Windows, parameterName=WindowDistanceFade, sortPriority=100}
  NamedRerouteDeclaration(None=Saturate(None=Divide(A=@n49.R, B=@n72))) {name=DistanceMask, nodeColor=<rgba 0.9219 0 1 1>}
  n73 = FunctionInput(Preview=Constant()) {bUsePreviewValueAsDefault=true, inputName=Mask_Input, sortPriority=1}
  NamedRerouteDeclaration(None=Saturate(None=ChannelMaskParameter(None=@n73) {group=Fake Windows, parameterName=FakeInteriorMask}.R)) {name=Input_Mask, nodeColor=<rgba 0 1 0.7422 1>}
  NamedRerouteDeclaration(None=@n70) {name=DistanceFace, nodeColor=<rgba 0.3125 0 1 1>}
  n74 = ScalarParameter() {group=Fake Windows, parameterName=FakeWindow_InsideOutsideMask_B, sortPriority=500}
  n75 = ScalarParameter() {defaultValue=1, group=Fake Windows, parameterName=FakeWindow_InsideOutsideMask_A, sortPriority=500}
  n76 = FunctionInput(Preview=Constant()) {bUsePreviewValueAsDefault=true, inputName=InsideOutSideMask, inputType=FunctionInput_Scalar, sortPriority=55}
  NamedRerouteDeclaration(None=Saturate(None=LinearInterpolate(A=@n74, B=@n75, Alpha=@n76))) {name=InsideOutsideMask, nodeColor=<rgba 0 1 0.9063 1>}
  n77 = LocalPosition()
  n78 = AppendVector(
   A=ScalarParameter() {defaultValue=512, group=Fake Windows, parameterName=LocalUVs_Scale_X}
   B=ScalarParameter() {defaultValue=384, group=Fake Windows, parameterName=LocalUVs_Scale_Y}
  )
  NamedRerouteDeclaration(None=StaticSwitchParameter(True=Divide(A=@n77.XY, B=@n78), False=TextureCoordinate()) {defaultValue=true, group=Fake Windows, parameterName=FakeWIndowUseLocalUVsPorjection}) {name=UVSForCube, nodeColor=<rgba 1 0 0.5391 1>}
  NamedRerouteUsage()
  n79 = TextureCoordinate()
  n80 = AppendVector(A=NamedRerouteUsage(), B=NamedRerouteUsage())
  n81 = MaterialFunctionCall(Vector=AppendVector(A=Subtract(A=@n80) {constB=0.5}, B=Constant()), Default=Constant3Vector() {constant=<rgba 1 1 0 0>}) {materialFunction=&SafeNormalize}
  n82 = Add(
   A=Constant2Vector() {g=0.5, r=0.5}
   B=Multiply(
    A=ComponentMask(None=@n81.Result) {g=true, r=true}
    B=ScalarParameter() {defaultValue=2, group=Fake Windows, parameterName=Mobile_Window GradientRandom OffsetVector}
   )
  )
  Power(
   Base=Divide(A=Distance(A=@n79, B=@n82), B=ScalarParameter() {defaultValue=3.01, group=Fake Windows, parameterName=Mobile_Window Gradient_Radius})
   Exp=ScalarParameter() {defaultValue=7.36, group=Fake Windows, parameterName=Mobile_Window Gradient_Falloff}
  )
  ReflectionVectorWS()
  n83 = TextureCoordinate()
  n84 = ComponentMask(None=VectorParameter() {group=Fake Windows, parameterName=FadeLocation, defaultValue=<rgba 0.5 0.55 0 1>}.RGB) {g=true, r=true}
  n85 = ComponentMask(None=VectorParameter() {group=Fake Windows, parameterName=WindowBounds, defaultValue=<rgba 1.3 0.4 0 1>}.RGB) {g=true, r=true}
  n86 = ScalarParameter() {defaultValue=0.1, desc=0.075, group=Fake Windows, parameterName=Mobile_WindowEdgeFalloff}
  n87 = BumpOffset(
   Coordinate=@n83
   Height=Constant()
   HeightRatioInput=LinearInterpolate(Alpha=MaterialFunctionCall(A=@n83, B=@n84, Bounds=@n85, Edge Falloff=@n86) {materialFunction=&BoxMask-2D}.Result) {constB=0.5}
  ) {heightRatio=0.5}
  n88 = Add(A=Multiply(A=Subtract(A=@n87, B=@n84), B=Constant2Vector() {g=3, r=0.95})) {constB=0.25}
  n89 = MaterialFunctionCall(Float2=ConstantBiasScale(None=Sine(None=@n88))) {materialFunction=&BreakOutFloat2Components}
  n90 = VectorParameter() {group=Fake Windows, parameterName=Mobile_WindowEmissive, defaultValue=<rgba 0.4167 0.3334 0.2372 1>}
  n91 = Saturate(None=LinearInterpolate(
    A=ScalarParameter() {defaultValue=1, group=Fake Windows, parameterName=Mobile_Window Lit_Low}
    B=ScalarParameter() {defaultValue=1, group=Fake Windows, parameterName=Mobile_Window Lit_High}
    Alpha=NamedRerouteUsage()
   ))
  StaticSwitchParameter(True=@n33, False=Multiply(A=Multiply(A=@n89.R, B=@n89.R), B=Multiply(A=@n90.RGB, B=@n91))) {desc=Temp Fix until we have a definitive solution in 34.10, group=Fake Windows, parameterName=Mobile_CH6Mobile_FakeInterior}
 ]}
