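# Material function for the four-layer base material that modifies the emissive of incoming material attributes, then passes them to a fake-interior function.
# Static switches gate a chain: SFX texture (pan, rotate, flipbook, distort), flicker, parallax depth, sphere mask, fresnel, time-of-day tint and vertex-color filter.
# Texture and flicker controls take per-layer values through UV2 lerp helpers.
# A ray-tracing quality switch picks eye-adaptation-inverse or Lumen-multiplied emissive before the final color multiplier.
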MaterialFunction /Game/Global/Base4Layers_2022/MaterialFunctions/MF_EmissiveControls_2
graph {nodes [
  Constant2Vector() {g=0.5, r=0.5}
  n1 = FunctionInput() {inputName=Material In, inputType=FunctionInput_MaterialAttributes, sortPriority=-3}
  n2 = NamedRerouteDeclaration(None=GetMaterialAttributes(None=NamedRerouteDeclaration(None=@n1) {name=Input, nodeColor=<rgba 1 0.6328 0 1>}) {attributeGetTypes [#1]}.EmissiveColor) {name=Emissive Layers, nodeColor=<rgba 0 1 0.0391 1>}
  n3 = Multiply(A=@n2, B=NamedRerouteUsage())
  n4 = StaticSwitchParameter(True=Add(A=@n2, B=@n3), False=@n3) {
   desc="By default RGB texture mult on emissive inputs (Masks in layers) this switch will allow to ADD emissive layers on top of Emissive*RGB mix. "
   group=A_Emissive_Texture
   parameterName=Mult Add Emissive
   sortPriority=-31
  }
  n5 = StaticSwitchParameter(
   True=SetMaterialAttributes!(MaterialAttributes=NamedRerouteUsage(), EmissiveColor=@n4) {attributeSetTypes [#1]}
   False=NamedRerouteUsage()
  ) {
   desc="Emissive RGB Available if SFX RGB is activated in 0 SFX Global. Will replace layers Emissive unless Emissive RGB or Emissive Texture Add is set= true  ( adds texture sample)"
   group=A_Emissive_Texture
   parameterName=SFX Emissive Texture RGB
   sortPriority=-32
  }
  n6 = StaticSwitchParameter(True=NamedRerouteUsage(), False=NamedRerouteDeclaration(None=@n5) {name=ToFlicker, nodeColor=<rgba 0.7578 0 1 1>}) {
   desc="Uses SFX Functions texture. Go to  0_SFX_Global group for Flicker Texture input. "
   group=A_Emissive_Flicker
   parameterName=Emissive Flicker
   sortPriority=-2
  }
  n7 = GetMaterialAttributes(None=@n6) {attributeGetTypes [#1]}
  n8 = StaticSwitchParameter(
   True=StaticSwitchParameter(True=TextureCoordinate() {coordinateIndex=2}, False=TextureCoordinate() {coordinateIndex=1}) {group=A_Emissive_Depth, parameterName=Emissive Depth UV2, sortPriority=-30}
   False=TextureCoordinate()
  ) {desc="Default is UV0 ", group=A_Emissive_Depth, parameterName=Emissive Depth UV1, sortPriority=-30}
  n9 = ComponentMask(None=VectorParameter() {group=A_Emissive_Depth, parameterName=DepthMaskLocation, defaultValue=<rgba 0 0 0 1>}.RGB) {g=true, r=true}
  n10 = ScalarParameter() {defaultValue=0.25, group=A_Emissive_Depth, parameterName=DepthMaskSphereRadius}
  n11 = ScalarParameter() {group=A_Emissive_Depth, parameterName=DepthMaskSphereHardnes}
  n12 = SphereMask(
   A=BumpOffset(Coordinate=@n8, Height=ScalarParameter() {defaultValue=-0.35, group=A_Emissive_Depth, parameterName=DepthMaskBumpOffset}) {heightRatio=0.5}
   B=@n9
   Radius=@n10
   Hardness=@n11
  )
  n13 = ScalarParameter() {defaultValue=2, group=A_Emissive_Depth, parameterName=DepthMaskSpherepower}
  n14 = AppendVector(
   A=ScalarParameter() {defaultValue=1, group=A_Emissive_Depth, parameterName=TileDepth X}
   B=ScalarParameter() {defaultValue=1, group=A_Emissive_Depth, parameterName=TileDepth Y}
  )
  n15 = Add(
   A=ComponentMask(None=MaterialFunctionCall!(UVs=@n8, Tiling=@n14, Randomize Rotation) {materialFunction=/Engine/Functions/Engine_MaterialFunctions01/Cubemaps/InteriorCubemap}.UVW) {r=true}
   B=Constant() {r=0.98}
  )
  n16 = ScalarParameter() {defaultValue=3, group=A_Emissive_Depth, parameterName=Depth Edge Blur}
  n17 = LinearInterpolate(A=ScalarParameter() {group=A_Emissive_Depth, parameterName=Depth Min}, Alpha=Saturate(None=Divide(A=@n15, B=@n16)))
  n18 = Saturate(None=Power(Base=SphereMask(A=@n8, B=@n9, Radius=@n10, Hardness=@n11), Exp=@n13))
  n19 = StaticSwitchParameter(
   True=Saturate(None=Power(Base=@n12, Exp=@n13))
   False=StaticSwitchParameter(True=Max(A=@n17, B=@n18), False=@n17) {group=A_Emissive_Depth, parameterName=useDepthSphereMask}
  ) {group=A_Emissive_Depth, parameterName=useCheapBumpOffsetDepth}
  n20 = StaticSwitchParameter(
   True=SetMaterialAttributes(MaterialAttributes=@n7.MaterialAttributes, EmissiveColor=Multiply(A=@n7.EmissiveColor, B=@n19)) {attributeSetTypes [#1]}
   False=@n7.MaterialAttributes
  ) {
   desc="Uses SFX Functions texture. Go to  0_SFX_Global group for Flicker Texture input. "
   group=A_Emissive_Depth
   parameterName=Emissive Depth
   sortPriority=-2
  }
  n21 = GetMaterialAttributes(None=@n20) {attributeGetTypes [#1, #2]}
  n22 = ScalarParameter() {defaultValue=1, group=A_Emissive_Sphere, parameterName=Emissive Center Radius, sortPriority=2}
  n23 = ScalarParameter() {defaultValue=-1, group=A_Emissive_Sphere, parameterName=Emissive Center Hardness, sortPriority=3}
  n24 = SphereMask(
   A=LocalPosition().XYZ
   B=VectorParameter() {group=A_Emissive_Sphere, parameterName=Emissive Local Center, sortPriority=1}.RGB
   Radius=@n22
   Hardness=@n23
  )
  n25 = VectorParameter() {group=A_Emissive_Sphere, parameterName=Emissive WS Center, sortPriority=1}
  n26 = ActorPositionWS()
  n27 = ObjectRadius()
  n28 = SphereMask(
   A=WorldPosition().XYZ
   B=StaticSwitchParameter(True=@n25.RGB, False=Add(A=@n25.RGB, B=@n26)) {group=A_Emissive_Sphere, parameterName=Emissive Sphere World, sortPriority=-1}
   Radius=StaticSwitchParameter(True=@n22, False=Multiply(A=@n27, B=@n22)) {group=A_Emissive_Sphere, parameterName=Emissive Sphere World, sortPriority=-1}
   Hardness=@n23
  )
  n29 = LinearInterpolate(
   A=ScalarParameter() {group=A_Emissive_Sphere, parameterName=Emissive Sphere Min, sortPriority=4}
   B=ScalarParameter() {defaultValue=1, group=A_Emissive_Sphere, parameterName=Emissive Sphere Max, sortPriority=4}
   Alpha=Power(
    Base=StaticSwitchParameter(True=@n24, False=@n28) {group=A_Emissive_Sphere, parameterName=Emissive Sphere Use Local, sortPriority=-1}
    Exp=ScalarParameter() {defaultValue=1, group=A_Emissive_Sphere, parameterName=Emissive Sphere Power, sortPriority=5}
   )
  )
  n30 = StaticSwitchParameter(True=Multiply(A=@n29, B=NamedRerouteUsage()), False=@n29) {desc=Mult Emissive with specific color per time of day, group=A_Emissive_TOD, parameterName=A_EmissiveTOD, sortPriority=-4}
  n31 = Multiply(A=@n21.BaseColor, B=Saturate(None=OneMinus(None=@n30)))
  n32 = StaticSwitchParameter(
   True=SetMaterialAttributes!(
    MaterialAttributes=@n21.MaterialAttributes
    EmissiveColor=Multiply(A=@n21.EmissiveColor, B=@n29)
    BaseColor=StaticSwitchParameter(True=@n31, False=@n21.BaseColor) {defaultValue=true, group=A_Emissive_Sphere, parameterName=Multiply Base Color, sortPriority=-2}
   ) {attributeSetTypes [#1, #2]}
   False=@n21.MaterialAttributes
  ) {group=A_Emissive_Sphere, parameterName=Emissive Sphere, sortPriority=-2}
  n33 = GetMaterialAttributes(None=@n32) {attributeGetTypes [#1, #2]}
  n34 = LinearInterpolate(
   A=ScalarParameter() {group=A_Emissive_Fresnel, parameterName=EmissiveFresnel_Min, sortPriority=5}
   B=ScalarParameter() {defaultValue=1, group=A_Emissive_Fresnel, parameterName=EmissiveFresnel_Max, sortPriority=5}
   Alpha=DotProduct(A=CameraVectorWS(), B=VertexNormalWS())
  )
  n35 = StaticSwitchParameter(
   True=SetMaterialAttributes(MaterialAttributes=@n33.MaterialAttributes, EmissiveColor=Multiply(A=@n33.EmissiveColor, B=@n34), BaseColor=@n33.MaterialAttributes) {attributeSetTypes [#1, #2]}
   False=@n32
  ) {
   desc=Enables Fresnel mask to fade emissive on grazing angles. Usefull for screens or any emissive surface behind the glass.
   group=A_Emissive_Fresnel
   parameterName=Emissive Fresnel Mask
   sortPriority=-2
  }
  n36 = GetMaterialAttributes(None=@n35) {attributeGetTypes [#1]}
  n37 = VectorParameter() {
   desc="Mult Emissive by This color per TOD "
   group=A_Emissive_TOD
   parameterName=EmissiveMorning
   sortPriority=1
   defaultValue=<rgba 0.4 0.0568 0.1168 1>
  }
  n38 = VectorParameter() {
   desc="Mult Emissive by This color per TOD "
   group=A_Emissive_TOD
   parameterName=EmissiveDay
   sortPriority=2
   defaultValue=<rgba 0.1199 0.2582 0.4 1>
  }
  n39 = VectorParameter() {
   desc="Mult Emissive by This color per TOD "
   group=A_Emissive_TOD
   parameterName=EmissiveEvening
   sortPriority=3
   defaultValue=<rgba 0.5208 0.4396 0.02 1>
  }
  n40 = VectorParameter() {
   desc="Mult Emissive by This color per TOD "
   group=A_Emissive_TOD
   parameterName=EmissiveNight
   sortPriority=4
   defaultValue=<rgba 0.1248 0.2764 0.5503 0>
  }
  n41 = MaterialFunctionCall(Color_Morning=@n37.RGB, Color_Day=@n38.RGB, Color_Evening=@n39.RGB, Color_Night=@n40.RGB) {materialFunction=&Color_Time_of_Day}
  n42 = VectorParameter() {
   desc="Mult Emissive by This color per TOD "
   group=A_Emissive_TOD
   parameterName=EmissiveMorning_VertexColor
   sortPriority=5
   defaultValue=<rgba 0.4 0.0568 0.1168 1>
  }
  n43 = VectorParameter() {
   desc="Mult Emissive by This color per TOD "
   group=A_Emissive_TOD
   parameterName=EmissiveDay_VertexColor
   sortPriority=6
   defaultValue=<rgba 0.1199 0.2582 0.4 1>
  }
  n44 = VectorParameter() {
   desc="Mult Emissive by This color per TOD "
   group=A_Emissive_TOD
   parameterName=EmissiveEvening_VertexColor
   sortPriority=7
   defaultValue=<rgba 0.5208 0.4396 0.02 1>
  }
  n45 = VectorParameter() {
   desc="Mult Emissive by This color per TOD "
   group=A_Emissive_TOD
   parameterName=EmissiveNight_VertexColor
   sortPriority=8
   defaultValue=<rgba 0.1248 0.2764 0.5503 0>
  }
  n46 = StaticSwitchParameter(
   True=LinearInterpolate(
    A=@n41.Result
    B=MaterialFunctionCall(Color_Morning=@n42.RGB, Color_Day=@n43.RGB, Color_Evening=@n44.RGB, Color_Night=@n45.RGB) {materialFunction=&Color_Time_of_Day}.Result
    Alpha=NamedRerouteUsage()
   )
   False=@n41.Result
  ) {
   desc=Mult Emissive with specific color per time of day
   group=A_Emissive_TOD
   parameterName=TOD Emissive Colors_VertexColorFilter
   sortPriority=-1
  }
  n47 = StaticSwitchParameter(
   True=SetMaterialAttributes!(MaterialAttributes=@n36.MaterialAttributes, EmissiveColor=Multiply(A=@n36.EmissiveColor, B=@n46)) {attributeSetTypes [#1]}
   False=@n35
  ) {desc=Mult Emissive with specific color per time of day, group=A_Emissive_TOD, parameterName=TOD Emissive Colors, sortPriority=-1}
  n48 = GetMaterialAttributes(None=@n47) {attributeGetTypes [#1]}
  n49 = ScalarParameter() {defaultValue=0.1, group=A_Emissive_TOD, parameterName=DayMult, sortPriority=-3}
  n50 = CollectionParameter() {
   collection=&FortniteMaterialParameters
   parameterName=FortniteActiveTimeOfDay
  }
  n51 = DotProduct(
   A=MaterialFunctionCall(X=@n49, Y=@n49, Z=@n49, A=ScalarParameter() {defaultValue=1, group=A_Emissive_TOD, parameterName=NightMult, sortPriority=-2}) {materialFunction=&MakeFloat4}.Result
   B=@n50
  )
  n52 = Multiply(A=@n48.EmissiveColor, B=NamedRerouteDeclaration(None=@n51) {name=TOD, nodeColor=<rgba 0.2656 0 1 1>})
  n53 = StaticSwitchParameter(
   True=StaticSwitchParameter(
    True=@n47
    False=SetMaterialAttributes(MaterialAttributes=@n48.MaterialAttributes, EmissiveColor=@n52) {attributeSetTypes [#1]}
   ) {group=A_Emissive_TOD, parameterName=TOD Emissive Colors, sortPriority=-1}
   False=@n35
  ) {desc=Mult Emissive with specific color per time of day, group=A_Emissive_TOD, parameterName=A_EmissiveTOD, sortPriority=-4}
  n54 = GetMaterialAttributes(None=@n53) {attributeGetTypes [#1]}
  n55 = VectorParameter() {group=Emissive_VertexColorFilter, parameterName=Emissive_ VertexColor_Filter_B, sortPriority=2, defaultValue=<rgba 0 0 0 1>}
  n56 = VectorParameter() {group=Emissive_VertexColorFilter, parameterName=Emissive_ VertexColor_Filter_A, sortPriority=1, defaultValue=<rgba 1 1 1 1>}
  n57 = VectorParameter() {group=Emissive_VertexColorFilter, parameterName=Emissive_VertexColorMask, sortPriority=5, defaultValue=<rgba 1 0 0 0>}
  n58 = VertexColor()
  MaterialExpressionMultiply_24 = Multiply(A=AppendVector(A=@n57.RGB, B=@n57.RGB), B=AppendVector(A=@n58, B=@n58))
  n59 = Convert(None=@MaterialExpressionMultiply_24) {
   nodeName=Break Float4
   convertInputs [{
     type=Vector4
     defaultValue=<rgba 0 0 0 1>
     expressionInput {
      expression=&self:MaterialExpressionMultiply_24
      inputName=None
      mask=0
      maskA=0
      maskB=0
      maskG=0
      maskR=0
      outputIndex=0
     }
    }]
   convertOutputs [{type=Scalar}, {type=Scalar}, {type=Scalar}, {type=Scalar}]
  }
  n60 = Add(A=Add(A=Add(A=@n59, B=@n59), B=@n59), B=@n59)
  n61 = LinearInterpolate(A=@n55.RGB, B=@n56.RGB, Alpha=NamedRerouteDeclaration(None=Saturate(None=@n60)) {name=VertexColorMask, nodeColor=<rgba 0.5234 1 0 1>})
  n62 = StaticSwitchParameter(
   True=SetMaterialAttributes(MaterialAttributes=@n54.MaterialAttributes, EmissiveColor=Multiply(A=@n54.EmissiveColor, B=@n61)) {attributeSetTypes [#1]}
   False=@n53
  ) {
   desc=Mult Emissive with specific color per time of day
   group=Emissive_VertexColorFilter
   parameterName=Emissive VertexColor Filter
   sortPriority=-4
  }
  n63 = GetMaterialAttributes(None=StaticSwitchParameter(
    True=MaterialFunctionCall!(Material=@n62, PerPixel Motion Offset) {materialFunction=&/MF_EmissiveFlicker}.Result
    False=@n62
   ) {group=A_Emissive_Flicker_V2, parameterName=Advanced Emissive Flicker, sortPriority=0}) {attributeGetTypes [#1]}
  n64 = EyeAdaptationInverse(LightValueInput=@n63.EmissiveColor, AlphaInput=ScalarParameter() {group=A__EmissiveFilters, parameterName=EyeAdaptationInverse})
  n65 = Multiply(A=@n63.EmissiveColor, B=ScalarParameter() {defaultValue=1, group=A__EmissiveFilters, parameterName=LumenEmissiveMultiplier})
  n66 = StaticSwitchParameter(
   True=SetMaterialAttributes(MaterialAttributes=@n63.MaterialAttributes, EmissiveColor=RayTracingQualitySwitch(Normal=@n64.EyeAdaptationInverse, RayTraced=@n65)) {attributeSetTypes [#1]}
   False=NamedRerouteUsage()
  ) {
   desc="Enable to access various emissive functions, TOD,Fresnel, Flipbook etc."
   group=A__EmissiveFilters
   parameterName=Emissive Filters
   sortPriority=-9
  }
  n67 = GetMaterialAttributes(None=@n66) {attributeGetTypes [#1]}
  n68 = Multiply(
   A=@n67.EmissiveColor
   B=VectorParameter() {group=A__EmissiveFilters, parameterName=Emissive_Color_Multiplier, defaultValue=<rgba 1 1 1 1>}.RGB
  )
  n69 = ScalarParameter() {group=Fake Windows, parameterName=FakeWindow_VertexColorFilter_B}
  n70 = ScalarParameter() {defaultValue=1, group=Fake Windows, parameterName=FakeWindow_VertexColorFilter_A}
  n71 = NamedRerouteUsage()
  n72 = NamedRerouteUsage()
  n73 = StaticSwitchParameter(True=Multiply(A=LinearInterpolate(A=@n69, B=@n70, Alpha=@n71), B=@n72), False=@n72) {desc=Mult Emissive with specific color per time of day, group=Fake Windows, parameterName=FakeWindow VertexColor Filter, sortPriority=-4}
  n74 = ScalarParameter() {group=Fake Windows, parameterName=FakeWindow_VertexColorInside}
  n75 = ScalarParameter() {defaultValue=1, group=Fake Windows, parameterName=FakeWindow_VertexColorOutside}
  FunctionOutput(None=MaterialFunctionCall!(
    Material=SetMaterialAttributes(MaterialAttributes=@n67.MaterialAttributes, EmissiveColor=@n68) {attributeSetTypes [#1]}
    Mask_Input=@n73
    InsideOutSideMask=LinearInterpolate(A=@n74, B=@n75, Alpha=@n71)
   ) {materialFunction=&/MF_FakeInterior_Generic}.Result)
  Multiply(A=@n33.BaseColor, B=Saturate(None=OneMinus(None=@n34)))
  n76 = ScalarParameter() {
   defaultValue=1
   desc=Select page from 1 to 4 or Flipbook speed if Anim Flipbook is enabled
   group=A_Emissive_Texture_Flipbook
   parameterName=EmissiveTex_Flipbook_Layer_1
   sliderMax=4
   sortPriority=2
  }
  n77 = ScalarParameter() {
   defaultValue=1
   desc=Select page from 1 to 4 or Flipbook speed if Anim Flipbook is enabled
   group=A_Emissive_Texture_Flipbook
   parameterName=EmissiveTex_Flipbook_Layer_2
   sliderMax=4
   sortPriority=2
  }
  n78 = ScalarParameter() {
   defaultValue=1
   desc=Select page from 1 to 4 or Flipbook speed if Anim Flipbook is enabled
   group=A_Emissive_Texture_Flipbook
   parameterName=EmissiveTex_Flipbook_Layer_3
   sliderMax=4
   sortPriority=2
  }
  n79 = ScalarParameter() {
   defaultValue=1
   desc=Select page from 1 to 4 or Flipbook speed if Anim Flipbook is enabled
   group=A_Emissive_Texture_Flipbook
   parameterName=EmissiveTex_Flipbook_Layer_4
   sliderMax=4
   sortPriority=2
  }
  n80 = MaterialFunctionCall!(1=@n76, 2=@n77, 3=@n78, 4=@n79) {materialFunction=&/MF_UV2LerpScalar_Layers_New}
  n81 = Multiply(A=@n80.Result, B=NamedRerouteUsage())
  n82 = ScalarParameter() {defaultValue=2, group=A_Emissive_Texture_Flipbook, parameterName=SFX Emissive Texture Flipbook Rows, sortPriority=3}
  n83 = ScalarParameter() {defaultValue=2, group=A_Emissive_Texture_Flipbook, parameterName=SFX Emissive Texture Anim Flipbook Columns, sortPriority=4}
  n84 = NamedRerouteUsage()
  n85 = Divide(A=Ceil(None=Add(A=@n80.Result) {constB=-1})) {constB=4}
  n86 = Constant() {r=2}
  NamedRerouteDeclaration(None=StaticSwitchParameter(
    True=MaterialFunctionCall("Animation  Phase (0-1)"=@n81, Number of Rows=@n82, Number of Columns=@n83, UVs=@n84) {materialFunction=&FlipBook}.UVs
    False=MaterialFunctionCall("Animation  Phase (0-1)"=@n85, Number of Rows=@n86, Number of Columns=@n86, UVs=@n84) {materialFunction=&FlipBook}.UVs
   ) {group=A_Emissive_Texture_Flipbook, parameterName=Emissive Textures Anim Flipbook, sortPriority=2}) {name=SFX Shared Flipbook, nodeColor=<rgba 0 1 0.1797 1>}
  n87 = Multiply(
   A=DotProduct(
    A=FunctionInput() {inputName=SFX, sortPriority=-3}
    B=VectorParameter() {group=A_Emissive_Texture, parameterName=SFX Emissive Texture Distortion Ch}.RGB
   )
   B=ScalarParameter() {defaultValue=1, group=A_Emissive_Texture, parameterName=SFX Distort UV}
  )
  n88 = NamedRerouteUsage()
  n89 = Multiply(
   A=Time()
   B=MaterialFunctionCall(
    1=ScalarParameter() {defaultValue=1, group=A_Emissive_Texture_Rotate, parameterName=Emissive Rotate Speed Layer 1, sortPriority=1}
    2=ScalarParameter() {defaultValue=1, group=A_Emissive_Texture_Rotate, parameterName=Emissive Rotate Speed Layer 2, sortPriority=2}
    3=ScalarParameter() {defaultValue=1, group=A_Emissive_Texture_Rotate, parameterName=Emissive Rotate Speed Layer 3, sortPriority=3}
    4=ScalarParameter() {defaultValue=1, group=A_Emissive_Texture_Rotate, parameterName=Emissive Rotate Speed Layer 4, sortPriority=4}
   ) {materialFunction=&MF_UV2LerpScalar_Layers_New}.Result
  )
  n90 = Multiply(
   A=NamedRerouteUsage()
   B=MaterialFunctionCall(
    1=ScalarParameter() {defaultValue=1, group=A_Emissive_Texture_Panner, parameterName=Emissive Pan Speed Layer 1, sortPriority=1}
    2=ScalarParameter() {defaultValue=1, group=A_Emissive_Texture_Panner, parameterName=Emissive Pan Speed Layer 2, sortPriority=2}
    3=ScalarParameter() {defaultValue=1, group=A_Emissive_Texture_Panner, parameterName=Emissive Pan Speed Layer 3, sortPriority=3}
    4=ScalarParameter() {defaultValue=1, group=A_Emissive_Texture_Panner, parameterName=Emissive Pan Speed Layer 4, sortPriority=4}
   ) {materialFunction=&MF_UV2LerpScalar_Layers_New}.Result
  )
  n91 = ScalarParameter() {defaultValue=2, group=A_Emissive_Texture_Panner, parameterName=Emissive Pan Number Of Steps, sortPriority=14}
  n92 = Multiply(A=@n90, B=@n91) {constB=2}
  n93 = StaticSwitchParameter(True=Divide(A=Floor(None=@n92), B=@n91), False=@n90) {
   desc="If true, instead of smoothly panning, step time.\r\nie better for buttons/patters that shouldn't smoothly pan."
   group=A_Emissive_Texture_Panner
   parameterName=SFX Emissive Pan Step Time
   sortPriority=9
  }
  n94 = Panner(
   Coordinate=NamedRerouteUsage()
   Time=@n93
   Speed=AppendVector(
    A=ScalarParameter() {group=A_Emissive_Texture_Panner, parameterName=SFX Emissive Pan Speed X, sortPriority=1}
    B=ScalarParameter() {group=A_Emissive_Texture_Panner, parameterName=SFX Emissive Pan Speed Y, sortPriority=2}
   )
  )
  n95 = TextureCoordinate() {coordinateIndex=2}
  n96 = TextureCoordinate() {coordinateIndex=1}
  n97 = NamedRerouteDeclaration(None=StaticSwitchParameter(
    True=StaticSwitchParameter(True=@n95, False=@n96) {group=A_Emissive_Texture, parameterName=SFX Emissive Textures UV2, sortPriority=-30}
    False=TextureCoordinate()
   ) {desc="Default is UV0 ", group=A_Emissive_Texture, parameterName=SFX Emissive Textures UV1, sortPriority=-30}) {name=EmTex UVs, nodeColor=<rgba 0 0.1797 1 1>}
  n98 = StaticSwitchParameter(
   True=NamedRerouteUsage()
   False=StaticSwitchParameter(
    True=Rotator(Coordinate=@n88, Time=@n89)
    False=StaticSwitchParameter(True=@n94, False=@n97) {group=A_Emissive_Texture_Panner, parameterName=SFX Emissive Textures Panner, sortPriority=0}
   ) {group=A_Emissive_Texture_Rotate, parameterName=SFX Emissive Textures Rotator, sortPriority=0}
  ) {group=A_Emissive_Texture_Flipbook, parameterName=SFX Emissive Textures Flipbook UV, sortPriority=1}
  NamedRerouteDeclaration(None=StaticSwitchParameter(True=Add(A=@n87, B=@n98), False=@n98) {desc="Uses SFX Mask Texture ", group=A_Emissive_Texture, parameterName=SFX Emissive Textures Distort UVs, sortPriority=1}) {name=Emissive Textures UV, nodeColor=<rgba 0.5234 0 1 1>}
  n99 = Time()
  n100 = ScalarParameter() {
   desc=Range is a random value between -X and X. Then add 1. A value of 0 is no variaiton between instances
   group=A_Emissive_Texture
   parameterName=SFX Random Speed Variation Multi
   sortPriority=33
  }
  n101 = NamedRerouteUsage()
  n102 = Add(A=LinearInterpolate(A=Multiply(A=@n100) {constB=-1}, B=@n100, Alpha=Frac(None=@n101)))
  n103 = ScalarParameter() {group=A_Emissive_Texture, parameterName=SFX Random Actor Timing Seed}
  n104 = StaticSwitchParameter(True=Add(A=Multiply(A=@n99, B=@n102), B=Multiply(A=@n101, B=@n103)), False=Time()) {desc=Offset time based off actor position, group=A_Emissive_Texture, parameterName=SFX Emissive Use Rand Time Offset, sortPriority=11}
  NamedRerouteDeclaration(None=@n104) {name=TimeRand, nodeColor=<rgba 0 1 0.0625 1>}
  n105 = GetMaterialAttributes(None=NamedRerouteUsage()) {attributeGetTypes [#1]}
  n106 = ScalarParameter() {defaultValue=0.5, desc=Modulate over Layer speed, group=A_Emissive_Flicker, parameterName=FlickerMinSpeed, sortPriority=5}
  n107 = NamedRerouteUsage()
  n108 = MaterialFunctionCall(Float3=ActorPositionWS()) {materialFunction=&BreakOutFloat3Components}
  n109 = Divide(A=Add(A=Add(A=@n108.R, B=@n108.R), B=@n108.B)) {constB=50}
  n110 = LinearInterpolate(
   A=Multiply(A=@n106, B=@n107) {constB=0.7}
   B=@n107
   Alpha=Frac(None=NamedRerouteDeclaration(None=@n109) {name=RandActPosVal, nodeColor=<rgba 0 1 0.6484 1>})
  )
  n111 = Multiply(
   A=Add(A=Sine(None=@n110) {period=0.5651}, B=Sine(None=@n110) {period=0.442})
   B=ScalarParameter() {defaultValue=0.15, group=A_Emissive_Flicker, parameterName=FlickerScale, sortPriority=6}
  )
  n112 = ScalarParameter() {defaultValue=0.25, group=A_Emissive_Flicker, parameterName=FlickerMinBrightnessBias, sortPriority=7}
  n113 = ConstantBiasScale(None=Sine(None=NamedRerouteUsage()))
  n114 = ScalarParameter() {
   desc=Clamp to prevent the emissive multiply from going below this value.
   group=A_Emissive_Flicker
   parameterName=FlickerMinBrightnessValue
   sortPriority=-1
  }
  n115 = Max(
   A=StaticSwitchParameter(True=Add(A=@n111), False=LinearInterpolate(A=@n112, Alpha=@n113)) {group=A_Emissive_Flicker, parameterName=FlickerSpeedVariation, sortPriority=-1}
   B=@n114
  )
  NamedRerouteDeclaration(None=SetMaterialAttributes(MaterialAttributes=@n105.MaterialAttributes, EmissiveColor=Multiply(A=@n105.EmissiveColor, B=@n115)) {attributeSetTypes [#1]}) {name=FlickerNoTex, nodeColor=<rgba 0 0.8594 1 1>}
  NamedRerouteDeclaration(None=@n61) {name=VertexColor Filter, nodeColor=<rgba 1 0 0.4922 1>}
  n116 = ScalarParameter() {defaultValue=1, group=A_Emissive_Texture, parameterName=Emissive Texture L1, sortPriority=1}
  n117 = ScalarParameter() {defaultValue=1, group=A_Emissive_Texture, parameterName=Emissive Texture L2, sortPriority=2}
  n118 = ScalarParameter() {defaultValue=1, group=A_Emissive_Texture, parameterName=Emissive Texture L3, sortPriority=3}
  n119 = ScalarParameter() {defaultValue=1, group=A_Emissive_Texture, parameterName=Emissive Texture L4, sortPriority=4}
  n120 = NamedRerouteUsage()
  n121 = FunctionInput() {inputName=SFX Textures RGB, inputType=FunctionInput_Texture2D, sortPriority=-1}
  NamedRerouteDeclaration(None=Multiply(
    A=MaterialFunctionCall(1=@n116, 2=@n117, 3=@n118, 4=@n119) {materialFunction=&MF_UV2LerpScalar_Layers_New}.Result
    B=TextureSample(UVs=@n120, Tex=@n121).RGB
   )) {name=Custom Emissive RGB, nodeColor=<rgba 0.0781 1 0 1>}
  n122 = ScalarParameter() {defaultValue=1, group=A_Emissive_Flicker, parameterName=Emissive Time Speed Layer 1, sortPriority=1}
  n123 = ScalarParameter() {defaultValue=1, group=A_Emissive_Flicker, parameterName=Emissive Time Speed Layer 2, sortPriority=2}
  n124 = ScalarParameter() {defaultValue=1, group=A_Emissive_Flicker, parameterName=Emissive Time Speed Layer 3, sortPriority=3}
  n125 = ScalarParameter() {defaultValue=1, group=A_Emissive_Flicker, parameterName=Emissive Time Speed Layer 4, sortPriority=4}
  NamedRerouteDeclaration(None=Multiply(
    A=Time()
    B=MaterialFunctionCall(1=@n122, 2=@n123, 3=@n124, 4=@n125) {materialFunction=&MF_UV2LerpScalar_Layers_New}.Result
   )) {name=FlickTime to UV, nodeColor=<rgba 0 0.9297 1 1>}
  Reroute(None=@n51)
  WorldPosition()
 ]}
