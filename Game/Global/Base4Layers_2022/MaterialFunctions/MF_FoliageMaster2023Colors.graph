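# Foliage master material function for plant color and shading, outputting a custom, soft-normal or default attribute set via static switches.
# Diffuse color passes through hue shift, MPColor or flower-field projection, and biome tinting.
# Custom biome colors blend through masks from four quadrants of a world-projected biome texture.
# Two-sided masks and GI quality parameters adjust backfaces; soft and custom variants can blend toward instance-Z normals.
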MaterialFunction /Game/Global/Base4Layers_2022/MaterialFunctions/MF_FoliageMaster2023Colors
graph {nodes [
  Constant3Vector() {constant=<rgba -1 -1 1 0>}
  Constant() {r=0.75}
  Constant() {r=0.85}
  Constant()
  Constant() {r=1}
  Constant() {r=0.65}
  Constant() {r=6000}
  n1 = VertexColor()
  n2 = Add(A=Add(A=@n1.R, B=@n1.R), B=@n1.B)
  Frac(None=Divide(A=@n2) {constB=10})
  n3 = StaticSwitchParameter(True=NamedRerouteUsage(), False=NamedRerouteUsage()) {group=MediumPlantSettings_SoftNormal, parameterName=UseSoftNormalMaterial, sortPriority=-100}
  FunctionOutput(None=StaticSwitchParameter(True=NamedRerouteUsage(), False=@n3) {group=Custom_Material, parameterName=UseCustomMaterial})
  LinearInterpolate(A=NamedRerouteUsage(), Alpha=NamedRerouteUsage()) {desc=Most meshes look better w/o this}
  LinearInterpolate(
   A=ScalarParameter() {defaultValue=0.7, group=SB, parameterName=SB_SpecMod_Far}
   B=ScalarParameter() {defaultValue=0.7, group=SB, parameterName=SB_SpecMod_Near}
   Alpha=NamedRerouteUsage()
  )
  LinearInterpolate(Alpha=ScalarParameter() {defaultValue=1, group=SB, parameterName=SB_DEBUG_test1})
  n4 = ScalarParameter() {defaultValue=900, group=SB, parameterName=SB_ChaparralCoast_GrassScale}
  n5 = TextureObject() {samplerType=SAMPLERTYPE_Normal, texture=/Game/Environments/Helios/Terrain/Textures/T_Helios_Chap_CoastGround_01_N}
  n6 = StaticBool() {value=true}
  n7 = WorldPosition() {worldPositionShaderOffset=WPT_ExcludeAllShaderOffsets}
  n8 = Transform(None=Constant3Vector() {constant=<rgba 0 0 1 0>}) {transformSourceType=TRANSFORMSOURCE_Instance}
  n9 = Multiply(
   A=TwoSidedSign()
   B=Transform(None=MaterialFunctionCall!(
     Diffuse Texture Object
     Spec Texture Object
     TextureSize=@n4
     --------------------------
     Use seperate settings for Z projection
     Use seperate texture for Z projection
     Texture Size Z projection=@n4
     Z Plane Diffuse Texture Object
     Z Plane Spec Texture Object
     "--------------------------      "
     ProjectionTransitionContrast
     Use Texture Alpha Heightmap to Refine the Transistion
     HeightLerpTransitionContrast
     PickTheDiffuseHeightChannel
     TheDiffuseHeightChannel
     "--------------------------       "
     NormalTextureObject=@n5
     World Space Normals=@n6
     "--------------------------  "
     Export Float 4 Diffuse
     Export Float 4 Spec
     "--------------------------     "
     WorldPosition=@n7.XYZ
     World Space Normal=@n8
    ) {materialFunction=/Engine/Functions/Engine_MaterialFunctions02/Texturing/WorldAlignedTextures_Complex}.Normal Z Texture) {transformSourceType=TRANSFORMSOURCE_World, transformType=TRANSFORM_Tangent}
  )
  LinearInterpolate(A=@n9, B=NamedRerouteUsage(), Alpha=ScalarParameter() {defaultValue=0.75, group=SB, parameterName=SB_DEBUG_testp2})
  n10 = VectorParameter() {group=BiomeSettings, parameterName=CustomColor_Grasslands, channelNames {a=VariationHueShift, b="", g="", r=""}, defaultValue=<rgba 1 1 1 0>}
  n11 = VectorParameter() {group=BiomeSettings, parameterName=CustomColor_Mountains, channelNames {a=VariationHueShift, b="", g="", r=""}, defaultValue=<rgba 1 1 1 0>}
  n12 = VectorParameter() {group=BiomeSettings, parameterName=CustomColor_Wetlands, channelNames {a=VariationHueShift, b="", g="", r=""}, defaultValue=<rgba 1 1 1 0>}
  n13 = LinearInterpolate(
   A=LinearInterpolate(A=AppendVector(A=@n10, B=@n10), B=AppendVector(A=@n11, B=@n11), Alpha=NamedRerouteUsage())
   B=AppendVector(A=@n12, B=@n12)
  )
  n14 = VectorParameter() {
   group=BiomeSettings
   parameterName=CustomColor_MossyForest
   channelNames {a=VariationHueShift, b="", g="", r=""}
   defaultValue=<rgba 1 1 1 0>
  }
  LinearInterpolate(A=@n13, B=AppendVector(A=@n14, B=@n14), Alpha=NamedRerouteUsage())
  n15 = NamedRerouteUsage()
  n16 = LinearInterpolate(A=NamedRerouteUsage(), B=NamedRerouteUsage(), Alpha=NamedRerouteUsage())
  MaterialFunctionCall(BaseNormal=@n15, AdditionalNormal=@n16) {materialFunction=&BlendAngleCorrectedNormals}
  n17 = PerInstanceRandom()
  n18 = MaterialFunctionCall() {materialFunction=&/MF_FlowerFieldsColorProjection}
  MaterialFunctionCall(
   Hue Shift Percentage=Multiply(A=ScalarParameter() {defaultValue=0.1, group=FlowerField, parameterName=FlowerFieldColorRandomHueShift}, B=Subtract(A=@n17))
   Texture=@n18.Result
  ) {materialFunction=&HueShift}
  Multiply(A=@n2) {constB=33.33}
  NamedRerouteDeclaration(None=Saturate(None=TwoSidedSign())) {name=2SidedSignMask, nodeColor=<rgba 1 0 1 1>}
  n19 = TextureSampleParameter2D() {
   group=MediumPlantSettings_Textures
   parameterName=Diffuse
   samplerSource=SSM_Wrap_WorldGroupSettings
   texture=/Game/Global/Textures/Default/Blanks/T_MidgreysRGB
  }
  NamedRerouteDeclaration(None=@n19.RGB) {name=DiffuseTextureRGB, nodeColor=<rgba 1 0 1 1>}
  n20 = VectorParameter() {group=MediumPlantSettings, parameterName=GI_LOWQuality_ADD_Backside, defaultValue=<rgba 0.057 0.0885 0.0202 1>}
  n21 = Multiply(A=NamedRerouteUsage(), B=ScalarParameter() {defaultValue=0.65, group=MediumPlantSettings, parameterName=SSSBrightness})
  n22 = Add(A=@n20.RGB, B=@n21)
  n23 = NamedRerouteUsage()
  NamedRerouteDeclaration(None=LinearInterpolate(A=LinearInterpolate(A=@n22, B=@n21, Alpha=@n23), B=@n21, Alpha=NamedRerouteUsage())) {name=FinalSSColor, nodeColor=<rgba 1 0 1 1>}
  n24 = VectorParameter() {group=SB, parameterName=SB_Color_Flower, defaultValue=<rgba 0.0967 0.15 0.03 1>}
  n25 = MaterialFunctionCall() {materialFunction=/Game/Athena/Helios/Environments/Landscape/Materials/MaterialFunctions_Utilities/MF_ScotchBroom_Colors}
  n26 = NamedRerouteUsage()
  n27 = Add(A=Multiply(A=ComponentMask(None=@n26) {r=true}) {constB=1.5}) {constB=-0.5}
  n28 = LinearInterpolate(
   A=ScalarParameter() {defaultValue=1, group=SB, parameterName=SB_PerInstanceNormal_Far}
   B=ScalarParameter() {defaultValue=0.7, group=SB, parameterName=SB_PerInstanceNormal_Near}
   Alpha=NamedRerouteUsage()
  )
  n29 = NamedRerouteUsage()
  n30 = SetMaterialAttributes!(
   MaterialAttributes
   BaseColor=LinearInterpolate(
    A=VectorParameter() {group=SB, parameterName=SB_Color_Stalk, defaultValue=<rgba 0.0967 0.15 0.03 1>}.RGB
    B=LinearInterpolate(A=@n24.RGB, B=@n25.Flower1_Color, Alpha=@n25.Flower1_Color)
    Alpha=Saturate(None=@n27)
   )
   Specular=Multiply(A=ScalarParameter() {defaultValue=0.7, group=SB, parameterName=SB_SpecMod}, B=NamedRerouteUsage())
   Roughness=ScalarParameter() {defaultValue=0.85, group=SB, parameterName=SB_Flat_Roughness}
   Opacity=ScalarParameter() {defaultValue=0.75, group=SB, parameterName=SB_SSSOpacity}
   OpacityMask=NamedRerouteUsage()
   Normal=LinearInterpolate(A=@n9, B=LinearInterpolate(A=@n16, B=@n15, Alpha=@n28), Alpha=@n29)
   SubsurfaceColor=Constant()
  ) {attributeSetTypes [
    #1
    #2
    #3
    #4
    #5
    #6
    #7
   ]}
  NamedRerouteDeclaration(None=@n30) {name=ScotchBroom, nodeColor=<rgba 1 0 1 1>}
  n31 = Constant3Vector() {constant=<rgba 0 0 1 0>}
  NamedRerouteDeclaration(None=Multiply(A=TwoSidedSign(), B=Transform(None=@n31) {transformSourceType=TRANSFORMSOURCE_Instance, transformType=TRANSFORM_Tangent})) {name=InstanceZ_Normals, nodeColor=<rgba 1 0 1 1>}
  n32 = Constant3Vector() {constant=<rgba 1 1 1 0>}
  n33 = Constant3Vector() {constant=<rgba -1 -1 1 0>}
  n34 = ScalarParameter() {defaultValue=1, desc="Default=1", group=MediumPlantSettings_SoftNormal, parameterName=SoftNormal_NegativeFacingNormalsStyle}
  NamedRerouteDeclaration(None=Multiply(A=LinearInterpolate(A=@n32, B=@n33, Alpha=@n34), B=NamedRerouteUsage())) {name=NegativeFacing_NormalsType, nodeColor=<rgba 1 0 1 1>}
  n35 = NamedRerouteUsage()
  n36 = ScalarParameter() {defaultValue=1, group=MediumPlantSettings_SoftNormal, parameterName=SoftNormal_DiffusePower, sortPriority=-90}
  n37 = VectorParameter() {group=MediumPlantSettings_SoftNormal, parameterName=SoftNormal_Basecolor_Tint, sortPriority=-60, defaultValue=<rgba 1 1 1 1>}
  n38 = Saturate(None=Multiply(A=Power(Base=@n35, Exp=@n36), B=@n37.RGB))
  n39 = StaticSwitchParameter(
   True=LinearInterpolate(
    A=NamedRerouteDeclaration(None=@n38) {name=SoftNormal_ModBasecolor, nodeColor=<rgba 1 0 1 1>}
    B=NamedRerouteUsage()
    Alpha=NamedRerouteUsage()
   )
   False=NamedRerouteUsage()
  ) {group=MPColor, parameterName=Use MPColor, sortPriority=-100}
  n40 = NamedRerouteUsage()
  n41 = ScalarParameter() {defaultValue=0.3, group=MediumPlantSettings_SoftNormal, parameterName=SoftNormal_RoughnessAdd}
  n42 = NamedRerouteUsage()
  n43 = NamedRerouteUsage()
  n44 = ScalarParameter() {
   defaultValue=0.5
   desc=A value of 1 will use instance Z normals
   group=MediumPlantSettings_SoftNormal
   parameterName=SoftNormal_NegativeFace_NormalBlend
  }
  n45 = Saturate(None=@n44)
  n46 = NamedRerouteUsage()
  n47 = NamedRerouteUsage()
  n48 = ScalarParameter() {
   defaultValue=0.4
   desc=A value of 1 will use instance Z normals
   group=MediumPlantSettings_SoftNormal
   parameterName=SoftNormal_PositiveFace_NormalBlend
  }
  n49 = Saturate(None=@n48)
  n50 = StaticSwitchParameter(True=NamedRerouteUsage(), False=NamedRerouteUsage()) {group=MPColor, parameterName=Use MPColor, sortPriority=-100}
  n51 = ScalarParameter() {defaultValue=-0.5, group=MediumPlantSettings_SoftNormal, parameterName=SoftNormal_SSSDesaturate}
  n52 = SetMaterialAttributes(
   BaseColor=@n39
   Specular=Multiply(A=NamedRerouteUsage(), B=ScalarParameter() {defaultValue=0.3, group=MediumPlantSettings_SoftNormal, parameterName=SoftNormal_SpecAmount})
   Roughness=Saturate(None=Add(A=@n40, B=@n41))
   Opacity=ScalarParameter() {defaultValue=0.85, group=MediumPlantSettings_SoftNormal, parameterName=SoftNormal_SSSOpacity}
   OpacityMask=NamedRerouteUsage()
   Normal=LinearInterpolate(A=LinearInterpolate(A=@n42, B=@n43, Alpha=@n45), B=LinearInterpolate(A=@n46, B=@n47, Alpha=@n49), Alpha=NamedRerouteUsage())
   SubsurfaceColor=Multiply(
    A=Desaturation(None=@n50, Fraction=@n51)
    B=ScalarParameter() {defaultValue=0.15, group=MediumPlantSettings_SoftNormal, parameterName=SoftNormal_SSSAmount}
   )
  ) {attributeSetTypes [
    #1
    #2
    #3
    #4
    #5
    #6
    #7
   ]}
  NamedRerouteDeclaration(None=@n52) {name=SoftNormal_Material, nodeColor=<rgba 1 0 1 1>}
  n53 = SetMaterialAttributes(
   BaseColor=NamedRerouteUsage()
   Specular=NamedRerouteUsage()
   Roughness=NamedRerouteUsage()
   Opacity=NamedRerouteUsage()
   OpacityMask=NamedRerouteUsage()
   Normal=NamedRerouteUsage()
   SubsurfaceColor=NamedRerouteUsage()
  ) {attributeSetTypes [
    #1
    #2
    #3
    #4
    #5
    #6
    #7
   ]}
  NamedRerouteDeclaration(None=@n53) {name=DefaultBase_Material, nodeColor=<rgba 1 0 1 1>}
  n54 = ScalarParameter() {defaultValue=6000, group=MediumPlantSettings_SoftNormal, parameterName=SoftNormal_PerInstanceNormal_Distance}
  n55 = Divide(A=PixelDepth().R, B=@n54)
  NamedRerouteDeclaration(None=OneMinus(None=Saturate(None=@n55))) {name=DistanceMask, nodeColor=<rgba 1 0 1 1>}
  n56 = TextureSampleParameter2D() {
   group=MediumPlantSettings_Textures
   parameterName=OSSR_Texture
   samplerType=SAMPLERTYPE_Masks
   texture=/Game/Global/Textures/Default/Blanks/Temp_specular_Mask
   channelNames {a="", b=RandomPerBlade, g=SSMask, r=OpacityMask}
  }
  NamedRerouteDeclaration(None=@n56.SSMask) {name=SSMask, nodeColor=<rgba 1 0 1 1>}
  n57 = TextureSampleParameter2D() {
   group=MediumPlantSettings_Textures
   parameterName=SMR_Texture
   samplerSource=SSM_Wrap_WorldGroupSettings
   samplerType=SAMPLERTYPE_Masks
   texture=&Temp_specular_Mask
   channelNames {a="", b=Roughness, g=ColorSystemMask, r=Specular}
  }
  NamedRerouteDeclaration(None=@n57.Specular) {name=SpecTextureChannel, nodeColor=<rgba 1 0 1 1>}
  n58 = ScalarParameter() {defaultValue=0.85, group=MediumPlantSettings, parameterName=SSSOpacity}
  n59 = NamedRerouteUsage()
  n60 = ScalarParameter() {desc=Use a value of 0-1 to blend in the SSMask from the OSSR_Texture, group=MediumPlantSettings, parameterName=UseSSMask}
  NamedRerouteDeclaration(None=Saturate(None=LinearInterpolate(A=@n58, B=@n59, Alpha=@n60))) {name=SSOpacity, nodeColor=<rgba 1 0 1 1>}
  n61 = VertexColor()
  n62 = Power(Base=@n61.B, Exp=ScalarParameter() {defaultValue=1, parameterName=HueGradient Contrast})
  n63 = Add(
   A=LinearInterpolate(
    A=ScalarParameter() {parameterName=HueBottomGradient}
    B=ScalarParameter() {defaultValue=-0.0832, parameterName=Hue}
    Alpha=Saturate(None=@n62)
   )
   B=Multiply(A=@n61.R, B=ScalarParameter() {parameterName=HuePerLeaf})
  )
  n64 = Multiply(A=VectorParameter() {group=MediumPlantSettings, parameterName=Basecolor_Tint, defaultValue=<rgba 1 1 1 1>}.RGB, B=NamedRerouteUsage())
  NamedRerouteDeclaration(None=StaticSwitchParameter(
    True=MaterialFunctionCall(Hue Shift Percentage=@n63, Texture=@n64) {materialFunction=&HueShift}.Result
    False=@n64
   ) {parameterName=HueShiftDiffuse}) {name=ModifiedDiffuse, nodeColor=<rgba 1 0 1 1>}
  n65 = NamedRerouteUsage()
  n66 = LinearInterpolate(A=@n65, B=Multiply(A=@n18.Result, B=@n65), Alpha=NamedRerouteUsage())
  n67 = Saturate(None=MaterialFunctionCall!(
    Fade Offset=ScalarParameter() {defaultValue=4000, group=FlowerField, parameterName=FlowerFieldColorFadeOffset}
    Fade Distance=ScalarParameter() {defaultValue=2000, group=FlowerField, parameterName=FlowerFieldColorFade}
   ) {materialFunction=/Game/Global/Materials/Character_Parent_2023/Anime_Parent_2023/MaterialFunctions/MFC_DistanceFade_FOV_compensation}.Result)
  n68 = MaterialFunctionCall!(
   MA=SetMaterialAttributes!(MaterialAttributes, BaseColor=NamedRerouteUsage()) {attributeSetTypes [#1]}
   InTint
  ) {materialFunction=/Game/Environments/Asteria/Foliage/HeliosMaterials/Masters/Functions/MF_MPColorMA}
  n69 = GetMaterialAttributes(None=@n68."Result (MA)") {attributeGetTypes [#1]}
  n70 = MaterialFunctionCall(
   Hue Shift Percentage=LinearInterpolate(
    A=ScalarParameter() {group=MPColor, parameterName=HueShift1_InstanceRandom}
    B=ScalarParameter() {group=MPColor, parameterName=HueShift2_InstanceRandom}
    Alpha=PerInstanceRandom()
   )
   Texture=@n69.BaseColor
  ) {materialFunction=&HueShift}
  n71 = NamedRerouteUsage()
  NamedRerouteDeclaration(None=StaticSwitchParameter(True=LinearInterpolate(A=@n66, B=@n18.Result, Alpha=@n67) {constA=1}, False=LinearInterpolate(A=@n69.BaseColor, B=@n70.Result, Alpha=@n71)) {parameterName=UseFlowerFieldColorProjection}) {name=DiffuseTextureWithMPColor, nodeColor=<rgba 1 0 1 1>}
  NamedRerouteDeclaration(None=@n68."Mask (S)") {name=MPColor_Mask, nodeColor=<rgba 1 0 1 1>}
  n72 = VectorParameter() {group=BiomeColors, parameterName=Color_Hades, channelNames {a=Desaturation, b="", g="", r=""}, defaultValue=<rgba 1 1 1 1>}
  n73 = Multiply(A=Desaturation(None=NamedRerouteUsage(), Fraction=@n72.Desaturation), B=@n72)
  n74 = MaterialFunctionCall!(UseLandscapeCoords, "UseProjection(translucent)") {materialFunction=/Game/Athena/Helios/Environments/Landscape/Materials/MaterialFunctions_Utilities/MF_TerrainMasks}
  NamedRerouteDeclaration(None=LinearInterpolate(A=NamedRerouteUsage(), B=Saturate(None=@n73), Alpha=ComponentMask(None=@n74.QuadD_RGBA) {r=true})) {name=DiffuseBiomeColors, nodeColor=<rgba 1 0 1 1>}
  n75 = ScalarParameter() {defaultValue=1, group=Custom_Material, parameterName=Custom_Backside_DiffuseMult}
  n76 = NamedRerouteUsage()
  n77 = ComponentMask(None=@n76) {a=true}
  n78 = Power(Base=NamedRerouteUsage(), Exp=ScalarParameter() {defaultValue=1, group=Custom_Material, parameterName=Custom_DiffusePower})
  n79 = Saturate(None=Multiply(A=ComponentMask(None=@n76) {b=true, g=true, r=true}, B=@n78))
  n80 = Multiply(A=VectorParameter() {group=Custom_Material, parameterName=Custom_Basecolor_Tint, defaultValue=<rgba 1 1 1 1>}.RGB, B=@n78)
  n81 = VectorParameter() {group=Custom_Material, parameterName=Custom_Basecolor_Tint_PerInstanceRandom, defaultValue=<rgba 1 1 1 1>}
  n82 = Multiply(A=@n78, B=@n81.RGB)
  n83 = StaticSwitchParameter(
   True=LinearInterpolate(
    A=MaterialFunctionCall(Hue Shift Percentage=@n77, Texture=@n79) {materialFunction=&HueShift}.Result
    B=@n79
    Alpha=PerInstanceRandom()
   )
   False=LinearInterpolate(A=Saturate(None=@n80), B=Saturate(None=@n82), Alpha=PerInstanceRandom())
  ) {group=Custom_Material, parameterName=CustomMaterial_UseCustomBiomeColors}
  n84 = LinearInterpolate(A=Saturate(None=Multiply(A=@n75, B=@n83)), B=@n83, Alpha=NamedRerouteUsage())
  n85 = NamedRerouteUsage()
  n86 = Multiply(A=Constant3Vector() {constant=<rgba 1 1 -1 0>}, B=@n85)
  n87 = NamedRerouteUsage()
  n88 = Constant3Vector() {constant=<rgba 0 0 1 0>}
  n89 = SetMaterialAttributes!(
   MaterialAttributes
   BaseColor=@n84
   Specular=LinearInterpolate(
    A=ScalarParameter() {group=Custom_Material, parameterName=Custom_Specular_Low}
    B=ScalarParameter() {defaultValue=1, group=Custom_Material, parameterName=Custom_Specular_High}
    Alpha=NamedRerouteUsage()
   )
   Roughness=LinearInterpolate(
    A=ScalarParameter() {group=Custom_Material, parameterName=Custom_Rough_Low}
    B=ScalarParameter() {defaultValue=1, group=Custom_Material, parameterName=Custom_Rough_High}
    Alpha=NamedRerouteUsage()
   )
   Opacity=ScalarParameter() {defaultValue=1, group=Custom_Material, parameterName=Custom_SSSOpacity}
   OpacityMask=NamedRerouteUsage()
   Normal=LinearInterpolate(
    A=LinearInterpolate(A=@n86, B=@n85, Alpha=@n87)
    B=Transform(None=@n88) {transformSourceType=TRANSFORMSOURCE_Instance, transformType=TRANSFORM_Tangent}
    Alpha=ScalarParameter() {group=Custom_Material, parameterName=Custom_InstanceNormalAmount}
   )
   SubsurfaceColor=Multiply(A=@n84, B=VectorParameter() {group=Custom_Material, parameterName=Custom_SSS_Color_Tint, defaultValue=<rgba 0 0 0 1>}.RGB)
  ) {attributeSetTypes [
    #1
    #2
    #3
    #4
    #5
    #6
    #7
   ]}
  NamedRerouteDeclaration(None=@n89) {name=CustomMaterial, nodeColor=<rgba 1 0.25 0 1>}
  n90 = NamedRerouteUsage()
  n91 = Constant2Vector() {g=0.5}
  n92 = TextureObjectParameter() {
   group=BiomeSettings
   parameterName=CustomMaterial_BiomeTexture
   samplerType=SAMPLERTYPE_LinearColor
   texture=/Game/Athena/Helios/Environments/Landscape/Materials/Texture/T_HeliosBiome
  }
  n93 = MaterialFunctionCall(Float4=TextureSample(UVs=Add(A=@n90, B=@n91), Tex=@n92) {desc=BottomLeft-Quadrant-C, samplerSource=SSM_Wrap_WorldGroupSettings, samplerType=SAMPLERTYPE_LinearColor}.RGBA) {materialFunction=&BreakOutFloat4Components}
  NamedRerouteDeclaration(None=@n93.G) {name=BM_MaskMaker, nodeColor=<rgba 0.5 0 1 1>}
  n94 = VectorParameter() {
   desc=This is intended to match the default landscape UVs using world position
   group=BiomeSettings
   parameterName=BiomeTextureWorldUV
   channelNames {a=LWUV-Divide, b=LWUV-Subtract, g=LWUV-Y, r=LWUV-X}
   defaultValue=<rgba 0 0 -130048 260096>
  }
  n95 = ComponentMask(None=@n94) {g=true, r=true}
  n96 = TruncateLWC!(None=WorldPosition().XY)
  n97 = Divide(A=Subtract(A=Add(A=@n95, B=@n96), B=@n94.LWUV-Subtract), B=@n94.LWUV-Divide)
  NamedRerouteDeclaration(None=Multiply(A=Saturate(None=@n97), B=Constant() {r=0.5})) {name=BaseCoord, nodeColor=<rgba 1 0 1 1>}
  NamedRerouteDeclaration(None=@n57.Roughness) {name=RoughnessTextureChannel, nodeColor=<rgba 1 0 1 1>}
  NamedRerouteDeclaration(None=@n93.B) {name=BM_Bigsplat, nodeColor=<rgba 0.5 0 1 1>}
  n98 = MaterialFunctionCall(Float4=TextureSample(UVs=NamedRerouteUsage(), Tex=@n92) {desc=UpperLeft-Quadrant-A, samplerSource=SSM_Wrap_WorldGroupSettings, samplerType=SAMPLERTYPE_LinearColor}.RGBA) {materialFunction=&BreakOutFloat4Components}
  n99 = NamedRerouteUsage()
  n100 = Constant2Vector() {r=0.5}
  n101 = MaterialFunctionCall(Float4=TextureSample(UVs=Add(A=@n99, B=@n100), Tex=@n92) {desc=UpperRight-Quadrant-B, samplerSource=SSM_Wrap_WorldGroupSettings, samplerType=SAMPLERTYPE_LinearColor}.RGBA) {materialFunction=&BreakOutFloat4Components}
  n102 = Max(A=Max(A=@n98.G, B=@n98.G), B=Max(A=Max(A=@n98.A, B=@n101.R), B=@n101.G))
  NamedRerouteDeclaration(None=Multiply(A=@n98.R, B=OneMinus(None=@n102))) {name=BM_Mountains, nodeColor=<rgba 0.5 0 1 1>}
  n103 = Max(A=@n98.B, B=@n101.R)
  NamedRerouteDeclaration(None=Multiply(A=@n98.G, B=OneMinus(None=@n103))) {name=BM_Wetlands, nodeColor=<rgba 0.5 0 1 1>}
  n104 = VectorParameter() {group=BiomeSettings, parameterName=CustomColor_BorealFall, channelNames {a=VariationHueShift, b="", g="", r=""}, defaultValue=<rgba 1 1 1 0>}
  n105 = AppendVector(A=@n104, B=@n104)
  n106 = NamedRerouteUsage()
  n107 = VectorParameter() {group=BiomeSettings, parameterName=CustomColor_Goldlands, channelNames {a=VariationHueShift, b="", g="", r=""}, defaultValue=<rgba 1 1 1 0>}
  n108 = VectorParameter() {group=BiomeSettings, parameterName=CustomColor_RedForest, channelNames {a=VariationHueShift, b="", g="", r=""}, defaultValue=<rgba 1 1 1 0>}
  n109 = LinearInterpolate(
   A=LinearInterpolate(A=LinearInterpolate(A=@n13, B=@n105, Alpha=@n106), B=AppendVector(A=@n107, B=@n107), Alpha=NamedRerouteUsage())
   B=AppendVector(A=@n108, B=@n108)
   Alpha=NamedRerouteUsage()
  )
  n110 = VectorParameter() {group=BiomeSettings, parameterName=CustomColor_MaskMaker, channelNames {a=VariationHueShift, b="", g="", r=""}, defaultValue=<rgba 1 1 1 0>}
  n111 = AppendVector(A=@n110, B=@n110)
  n112 = NamedRerouteUsage()
  n113 = VectorParameter() {group=BiomeSettings, parameterName=CustomColor_Bigsplat, channelNames {a=VariationHueShift, b="", g="", r=""}, defaultValue=<rgba 1 1 1 0>}
  NamedRerouteDeclaration(None=LinearInterpolate(A=LinearInterpolate(A=@n109, B=@n111, Alpha=@n112), B=AppendVector(A=@n113, B=@n113), Alpha=NamedRerouteUsage())) {name=CustomBiomeColors, nodeColor=<rgba 0.5 0 1 1>}
  NamedRerouteDeclaration(None=@n98.B) {name=BM_MossyForest, nodeColor=<rgba 0.5 0 1 1>}
  n114 = StaticSwitchParameter(True=NamedRerouteDeclaration(None=@n57.ColorSystemMask) {name=ColorSystemMask, nodeColor=<rgba 1 0 1 1>}, False=@n56.OpacityMask) {defaultValue=true, desc=False uses red channel of OSSR_Texture, group=MediumPlantSettings, parameterName=UseSMR_GreenChannelAsOpacityMask}
  n115 = StaticSwitchParameter(True=NamedRerouteDeclaration(None=@n19.A) {name=DiffuseTextureAlpha, nodeColor=<rgba 1 0 1 1>}, False=@n114) {
   defaultValue=true
   desc=False uses the SpecularMasks Green channel
   group=MediumPlantSettings
   parameterName=UseDiffuseAlphaAsOpacityMask
   sortPriority=-90
  }
  NamedRerouteDeclaration(None=@n115) {name=OpacityMask, nodeColor=<rgba 1 0 1 1>}
  NamedRerouteDeclaration(None=@n98.A) {name=BM_BorealFall, nodeColor=<rgba 0.5 0 1 1>}
  NamedRerouteDeclaration(None=@n101.R) {name=BM_Goldlands, nodeColor=<rgba 0.5 0 1 1>}
  NamedRerouteDeclaration(None=@n101.G) {name=BM_RedForest, nodeColor=<rgba 0.5 0 1 1>}
  n116 = ScalarParameter() {defaultValue=0.375, group=MediumPlantSettings, parameterName=Specular_Backside}
  n117 = ScalarParameter() {defaultValue=0.375, group=MediumPlantSettings, parameterName=Specular_Frontside}
  n118 = NamedRerouteUsage()
  NamedRerouteDeclaration(None=Multiply(A=NamedRerouteUsage(), B=LinearInterpolate(A=@n116, B=@n117, Alpha=@n118))) {name=FinalSpec, nodeColor=<rgba 1 0 1 1>}
  n119 = ScalarParameter() {defaultValue=1, group=MediumPlantSettings, parameterName=Roughness_Backside}
  n120 = ScalarParameter() {defaultValue=0.65, group=MediumPlantSettings, parameterName=Roughness_Frontside}
  n121 = NamedRerouteUsage()
  NamedRerouteDeclaration(None=Multiply(A=NamedRerouteUsage(), B=LinearInterpolate(A=@n119, B=@n120, Alpha=@n121))) {name=FinalRoughness, nodeColor=<rgba 1 0 1 1>}
  NamedRerouteDeclaration(None=@n56.RandomPerBlade) {name=TextureRandomPerBlade, nodeColor=<rgba 1 0 1 1>}
  n122 = ScalarParameter() {defaultValue=0.5, group=MediumPlantSettings, parameterName=GI_Quality_EPIC_BacksideBrightness}
  n123 = CollectionParameter() {
   collection=&FortniteMaterialParameters
   parameterName=GlobalIlluminationEpicQuality
  }
  n124 = CollectionParameter() {
   collection=&FortniteMaterialParameters
   parameterName=GlobalIlluminationEnabled
  }
  NamedRerouteDeclaration(None=LinearInterpolate(A=Constant() {r=1}, B=Multiply(A=@n122, B=@n123), Alpha=@n124) {constA=1}) {name=EpicIs1_HighIs0, nodeColor=<rgba 1 0 1 1>}
  n125 = TwoSidedSign()
  n126 = TextureSampleParameter2D() {
   group=MediumPlantSettings_Textures
   parameterName=Normals
   samplerSource=SSM_Wrap_WorldGroupSettings
   samplerType=SAMPLERTYPE_Normal
   texture=&FlatNormalBlue
  }
  n127 = NamedRerouteDeclaration(None=@n126.RGB) {name=RawTextureNormals, nodeColor=<rgba 1 0 1 1>}
  n128 = CollectionParameter() {
   collection=&FortniteMaterialParameters
   parameterName=GlobalIlluminationEnabled
  }
  NamedRerouteDeclaration(None=LinearInterpolate(A=Multiply(A=@n125, B=@n127), B=@n127, Alpha=@n128)) {name=ModifiedNormal, nodeColor=<rgba 1 0 1 1>}
  n129 = StaticSwitchParameter(True=NamedRerouteUsage(), False=NamedRerouteUsage()) {group=BiomeColors, parameterName=UseBiomeColors, sortPriority=-500}
  n130 = Multiply(A=ScalarParameter() {defaultValue=1.5, group=MediumPlantSettings, parameterName=Basecolor_BacksideBrightness}, B=@n129)
  n131 = LinearInterpolate(A=@n130, B=@n129, Alpha=NamedRerouteUsage())
  n132 = Add(A=@n131, B=@n20.RGB)
  n133 = NamedRerouteUsage()
  n134 = NamedRerouteUsage()
  n135 = StaticSwitchParameter(
   True=LinearInterpolate(A=LinearInterpolate(A=@n132, B=@n131, Alpha=@n133), B=@n131, Alpha=@n134)
   False=LinearInterpolate(A=Add(A=@n130, B=@n20.RGB), B=@n130, Alpha=@n134)
  ) {defaultValue=true, parameterName=UseTwoSideColorModification}
  n136 = ScalarParameter() {defaultValue=1.2, parameterName=Mobile_SSS_BrightnessComp}
  NamedRerouteDeclaration(None=ShadingPathSwitch(Default=@n135, Mobile=Multiply(A=@n135, B=@n136)) {desc=Migrated from FeatureLevelSwitch}) {name=FinalBasecolor, nodeColor=<rgba 1 0 1 1>}
  n137 = NamedRerouteUsage()
  n138 = NamedRerouteUsage()
  n139 = NamedRerouteUsage()
  Normalize(VectorInput=Add(A=LinearInterpolate(A=@n137, B=@n138, Alpha=@n139), B=NamedRerouteUsage()))
  Saturate(None=Multiply(A=@n29, B=ScalarParameter() {defaultValue=1.25, group=SB, parameterName=SB_DEBUG_testpow33}))
  SphereMask(
   A=CameraPositionWS()
   B=TransformPosition(None=Constant3Vector()) {transformSourceType=TRANSFORMPOSSOURCE_Instance, transformType=TRANSFORMPOSSOURCE_World}
   Radius=@n54
   Hardness=Constant()
  )
  StaticSwitchParameter(True=NamedRerouteUsage(), False=@n3) {desc=Do not enable this at it will be removed soon, group=SB, parameterName=UseSB, sortPriority=-100}
  StaticSwitchParameter(True=NamedRerouteUsage(), False=@n39) {group=MPColor, parameterName=DEBUG_MPColorMask, sortPriority=-99}
  n140 = FunctionInput(Preview=StaticBool() {value=true}) {
   bUsePreviewValueAsDefault=true
   desc="Use SpecMask G as opacity mask. False=Use Diffuse Alpha. Default=True"
   description="Use SpecMask G as opacity mask. False=Use Diffuse Alpha. Default=True"
   inputName=UseSpecMaskG_AsOpacityMask
   inputType=FunctionInput_StaticBool
  }
  StaticSwitch(Value=@n140)
  TextureSample() {texture=/Game/Environments/Asteria/Foliage/ChaparralPlants/ScotchBroom/Textures/MI_ScotchBroom_A_A_D}
  TextureSample(UVs=Add(A=NamedRerouteUsage(), B=Constant2Vector() {g=0.5, r=0.5}), Tex=@n92) {desc=BottomRight-Quadrant-D, samplerSource=SSM_Wrap_WorldGroupSettings, samplerType=SAMPLERTYPE_LinearColor}
 ]}
