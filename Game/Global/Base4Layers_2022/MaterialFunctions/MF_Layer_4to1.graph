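# Material function that merges up to four tinted material-attribute layers into one, stacked by successive masked blends.
# Each upper-layer mask starts from UV offset thresholds, then static switches add or substitute SFX-channel, Z-up, Fresnel and landscape RVT height masks.
# Optional overlay blending uses MF_BlendLayers, and MF_DebugViews provides per-layer debug colors.
# Static switches set layer count and mobile layer skipping; low quality flattens four-layer normals.
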MaterialFunction /Game/Global/Base4Layers_2022/MaterialFunctions/MF_Layer_4to1
graph {nodes [
  n1 = FunctionInput() {inputName=2, inputType=FunctionInput_MaterialAttributes, sortPriority=1}
  Composite!(None=@n1)
  n2 = FunctionInput() {inputName=1, inputType=FunctionInput_MaterialAttributes}
  Composite(None=@n2)
  Composite!() {subgraphName=UV Mask}
  n3 = NamedRerouteUsage()
  n4 = ComponentMask(None=TextureCoordinate() {coordinateIndex=2}) {r=true}
  n5 = ComponentMask(None=TextureCoordinate() {coordinateIndex=1}) {r=true}
  n6 = Constant()
  n7 = Constant() {r=1}
  n8 = StaticSwitch(
   True=Constant() {r=1}
   False=If(
    A=Constant() {r=1}
    B=StaticSwitchParameter(True=@n4, False=@n5) {group=2_Layer, parameterName=Use UV2 For Layer 2 Mask, sortPriority=-2}
    "A > B"=@n6
    "A < B"=@n7
   ) {desc=0}
   Value=StaticBoolParameter() {group=2_Layer_Blending, parameterName=Blend Overlay Layer 2, sortPriority=-1}
  )
  MaterialExpressionReroute_72 = Reroute(None=@n8)
  MaterialExpressionReroute_65 = Reroute(None=@MaterialExpressionReroute_72.Output)
  n9 = NamedRerouteUsage()
  MaterialExpressionReroute_64 = Reroute(None=@n9)
  n10 = StaticSwitchParameter(
   True=Saturate(None=Add(A=@MaterialExpressionReroute_65.InputPin2, B=@MaterialExpressionReroute_64.L2 SFX))
   False=@MaterialExpressionReroute_65.InputPin2
  ) {
   desc="Enable to use SFX texture as the Layer mask instead UV2 offset. By default enables SFX to use UV2 coords (Unwrapped) Maximum 3 layers. 2 Layers if Nanite Texure set enabled. If enabled default 2 Layers awailable automatically, where layer 1 is background layer "
   group=2_Layer_Blending
   parameterName=Add SFX Layer 2 Blending
   sortPriority=2
  }
  n11 = StaticSwitchParameter(True=@MaterialExpressionReroute_64.L2 SFX, False=@n10) {
   desc="Enable to use SFX texture as the Layer mask instead UV2 offset. By default enables SFX to use UV2 coords (Unwrapped) Maximum 3 layers. 2 Layers if Nanite Texure set enabled. If enabled default 2 Layers awailable automatically, where layer 1 is background layer "
   group=2_Layer_Blending
   parameterName=Use SFX Layer 2 Blending
   sortPriority=2
  }
  MaterialExpressionReroute_63 = Reroute(None=@n11)
  MaterialExpressionReroute_46 = Reroute(None=@MaterialExpressionReroute_63.Output)
  n12 = StaticSwitchParameter(True=Saturate(None=Add(A=@n3, B=@MaterialExpressionReroute_46.InputPin)), False=@MaterialExpressionReroute_46.InputPin) {
   desc="Enable to use SFX texture as the Layer mask instead UV2 offset. By default enables SFX to use UV2 coords (Unwrapped) Maximum 3 layers. 2 Layers if Nanite Texure set enabled. If enabled default 2 Layers awailable automatically, where layer 1 is background layer "
   group=2_Layer_Blending
   parameterName=Add ZUp to Layer 2 Blending
   sortPriority=4
  }
  n13 = StaticSwitchParameter(True=NamedRerouteUsage(), False=@n12) {group=2_Layer_Blending, parameterName=Use ZUp to Layer 2 Blending, sortPriority=5}
  MaterialExpressionReroute_45 = Reroute(None=@n13)
  n14 = NamedRerouteUsage()
  n15 = StaticSwitch(
   True=Constant() {r=1}
   False=If(
    A=Constant() {r=2}
    B=StaticSwitchParameter(True=@n4, False=@n5) {group=3_Layer, parameterName=Use UV2 For Layer 3 Mask, sortPriority=-2}
    "A > B"=@n6
    "A < B"=@n7
   )
   Value=StaticBoolParameter() {group=3_Layer_Blending, parameterName=Blend Overlay Layer 3, sortPriority=-1}
  )
  MaterialExpressionReroute_73 = Reroute(None=@n15)
  MaterialExpressionReroute_68 = Reroute(None=@MaterialExpressionReroute_73.Output2)
  n16 = NamedRerouteUsage()
  MaterialExpressionReroute_66 = Reroute(None=@n16)
  n17 = StaticSwitchParameter(
   True=Saturate(None=Add(A=@MaterialExpressionReroute_68.InputPin4, B=@MaterialExpressionReroute_66.L3 SFX))
   False=@MaterialExpressionReroute_68.InputPin4
  ) {
   desc="Enable to use SFX texture as the Layer mask instead UV2 offset. By default enables SFX to use UV2 coords (Unwrapped) Maximum 3 layers. 2 Layers if Nanite Texure set enabled. If enabled default 2 Layers awailable automatically, where layer 1 is background layer "
   group=3_Layer_Blending
   parameterName=Add SFX Layer 3 Blending
   sortPriority=2
  }
  n18 = StaticSwitchParameter(True=@MaterialExpressionReroute_66.L3 SFX, False=@n17) {
   desc="Enable to use SFX texture as the Layer mask instead UV2 offset. By default enables SFX to use UV2 coords (Unwrapped) Maximum 3 layers. 2 Layers if Nanite Texure set enabled. If enabled default 2 Layers awailable automatically, where layer 1 is background layer "
   group=3_Layer_Blending
   parameterName=Use SFX Layer 3 Blending
   sortPriority=2
  }
  MaterialExpressionReroute_67 = Reroute(None=@n18)
  MaterialExpressionReroute_48 = Reroute(None=@MaterialExpressionReroute_67.Output2)
  n19 = StaticSwitchParameter(True=Saturate(None=Add(A=@n14, B=@MaterialExpressionReroute_48.InputPin2)), False=@MaterialExpressionReroute_48.InputPin2) {
   desc="Enable to use SFX texture as the Layer mask instead UV2 offset. By default enables SFX to use UV2 coords (Unwrapped) Maximum 3 layers. 2 Layers if Nanite Texure set enabled. If enabled default 2 Layers awailable automatically, where layer 1 is background layer "
   group=3_Layer_Blending
   parameterName=Add ZUp to Layer 3 Blending
   sortPriority=4
  }
  n20 = StaticSwitchParameter(True=NamedRerouteUsage(), False=@n19) {group=3_Layer_Blending, parameterName=Use ZUp to Layer 3 Blending, sortPriority=5}
  MaterialExpressionReroute_47 = Reroute(None=@n20)
  n21 = NamedRerouteUsage()
  n22 = NamedRerouteUsage()
  MaterialExpressionReroute_71 = Reroute(None=@n22)
  n23 = StaticSwitch(
   True=Constant() {r=1}
   False=If(
    A=Constant() {r=3}
    B=StaticSwitchParameter(True=@n4, False=@n5) {group=4_Layer, parameterName=Use UV2 For Layer 4 Mask, sortPriority=-2}
    "A > B"=@n6
    "A < B"=@n7
   )
   Value=StaticBoolParameter() {group=3_Layer_Blending, parameterName=Blend Overlay Layer 4, sortPriority=-1}
  )
  MaterialExpressionReroute_74 = Reroute(None=@n23)
  MaterialExpressionReroute_70 = Reroute(None=@MaterialExpressionReroute_74.Output3)
  n24 = StaticSwitchParameter(
   True=Saturate(None=Add(A=@MaterialExpressionReroute_70.InputPin5, B=@MaterialExpressionReroute_71.L4 SFX))
   False=@MaterialExpressionReroute_70.InputPin5
  ) {
   desc="Enable to use SFX texture as the Layer mask instead UV2 offset. By default enables SFX to use UV2 coords (Unwrapped) Maximum 3 layers. 2 Layers if Nanite Texure set enabled. If enabled default 2 Layers awailable automatically, where layer 1 is background layer "
   group=4_Layer_Blending
   parameterName=Add SFX Layer 4 Blending
   sortPriority=2
  }
  n25 = StaticSwitchParameter(True=@MaterialExpressionReroute_71.L4 SFX, False=@n24) {
   desc="Enable to use SFX texture as the Layer mask instead UV2 offset. By default enables SFX to use UV2 coords (Unwrapped) Maximum 3 layers. 2 Layers if Nanite Texure set enabled. If enabled default 2 Layers awailable automatically, where layer 1 is background layer "
   group=4_Layer_Blending
   parameterName=Use SFX Layer 4 Blending
   sortPriority=2
  }
  MaterialExpressionReroute_69 = Reroute(None=@n25)
  MaterialExpressionReroute_50 = Reroute(None=@MaterialExpressionReroute_69.Output3)
  n26 = StaticSwitchParameter(True=Saturate(None=Add(A=@n21, B=@MaterialExpressionReroute_50.InputPin3)), False=@MaterialExpressionReroute_50.InputPin3) {
   desc="Enable to use SFX texture as the Layer mask instead UV2 offset. By default enables SFX to use UV2 coords (Unwrapped) Maximum 3 layers. 2 Layers if Nanite Texure set enabled. If enabled default 2 Layers awailable automatically, where layer 1 is background layer "
   group=4_Layer_Blending
   parameterName=Add ZUp to Layer4 Blending
   sortPriority=4
  }
  n27 = StaticSwitchParameter(True=NamedRerouteUsage(), False=@n26) {group=4_Layer_Blending, parameterName=Use ZUp to Layer 4 Blending, sortPriority=5}
  MaterialExpressionReroute_49 = Reroute(None=@n27)
  Composite!(
   InputPin=@MaterialExpressionReroute_45.Output
   InputPin2=@MaterialExpressionReroute_47.Output2
   InputPin3=@MaterialExpressionReroute_49.Output3
  ) {subgraphName=Fresnel Mask}
  n28 = FunctionInput() {inputName=3, inputType=FunctionInput_MaterialAttributes, sortPriority=2}
  Composite!(None=@n28)
  n29 = FunctionInput() {inputName=4, inputType=FunctionInput_MaterialAttributes, sortPriority=3}
  Composite(None=@n29)
  MaterialExpressionReroute_23 = Reroute(None=@n29)
  n30 = GetMaterialAttributes(None=@MaterialExpressionReroute_23.Input) {attributeGetTypes [#1, #2, #3]}
  n31 = VectorParameter() {group=4_Layer, parameterName="L4 Tint  (Expand)", channelNames {a=L4 Specular Tweak, b="", g="", r=""}, defaultValue=<rgba 1 1 1 1>}
  n32 = SetMaterialAttributes!(
   MaterialAttributes=@n30.MaterialAttributes
   BaseColor=Multiply(A=@n30.BaseColor, B=@n31)
   Roughness=Multiply(A=@n30.Roughness, B=ScalarParameter() {defaultValue=1, group=4_Layer, parameterName=L4 Roughness Tweak})
   Specular=Multiply(A=@n30.Specular, B=@n31.L4 Specular Tweak)
  ) {attributeSetTypes [#1, #2, #3]}
  MaterialExpressionReroute_22 = Reroute(None=@n32)
  Composite!(
   MaterialAttributes=@MaterialExpressionReroute_22.Output
   False=@MaterialExpressionReroute_22.Output
   False2=@MaterialExpressionReroute_22.Output
  )
  MaterialExpressionReroute_21 = Reroute(None=@n28)
  n33 = GetMaterialAttributes(None=@MaterialExpressionReroute_21.Input) {attributeGetTypes [#1, #2, #3]}
  n34 = VectorParameter() {group=3_Layer, parameterName="L3 Tint  (Expand)", channelNames {a=L3 Specular Tweak, b="", g="", r=""}, defaultValue=<rgba 1 1 1 1>}
  n35 = SetMaterialAttributes(
   MaterialAttributes=@n33.MaterialAttributes
   BaseColor=Multiply(A=@n33.BaseColor, B=@n34)
   Roughness=Multiply(A=@n33.Roughness, B=ScalarParameter() {defaultValue=1, group=3_Layer, parameterName=L3 Roughness Tweak})
   Specular=Multiply(A=@n33.Specular, B=@n34.L3 Specular Tweak)
  ) {attributeSetTypes [#1, #2, #3]}
  MaterialExpressionReroute_20 = Reroute(None=@n35)
  Composite(
   MaterialAttributes=@MaterialExpressionReroute_20.Output
   False=@MaterialExpressionReroute_20.Output
   False2=@MaterialExpressionReroute_20.Output
  )
  n36 = VertexNormalWS()
  n37 = Fresnel(ExponentIn=ScalarParameter() {defaultValue=3, group=2_Layer_Blending, parameterName=L2 Fresnel Mask Exp}, Normal=@n36) {baseReflectFraction=0}
  MaterialExpressionReroute_79 = Reroute(None=@MaterialExpressionReroute_45.Output)
  n38 = StaticSwitchParameter(True=Add(A=@MaterialExpressionReroute_79.InputPin, B=@n37), False=@MaterialExpressionReroute_79.InputPin) {
   desc="Enable to use SFX texture as the Layer mask instead UV2 offset. By default enables SFX to use UV2 coords (Unwrapped) Maximum 3 layers. 2 Layers if Nanite Texure set enabled. If enabled default 2 Layers awailable automatically, where layer 1 is background layer "
   group=2_Layer_Blending
   parameterName=Add Fresnel to Layer 2 Blending
   sortPriority=4
  }
  n39 = StaticSwitchParameter(True=Saturate(None=@n37), False=@n38) {group=2_Layer_Blending, parameterName=Use Fresnel to Layer 2 Blending, sortPriority=5}
  MaterialExpressionReroute_78 = Reroute(None=@n39)
  n40 = Fresnel(ExponentIn=ScalarParameter() {defaultValue=3, group=3_Layer_Blending, parameterName=L3 Fresnel Mask Exp}, Normal=@n36) {baseReflectFraction=0}
  MaterialExpressionReroute_81 = Reroute(None=@MaterialExpressionReroute_47.Output2)
  n41 = StaticSwitchParameter(True=Add(A=@MaterialExpressionReroute_81.InputPin2, B=@n40), False=@MaterialExpressionReroute_81.InputPin2) {
   desc="Enable to use SFX texture as the Layer mask instead UV2 offset. By default enables SFX to use UV2 coords (Unwrapped) Maximum 3 layers. 2 Layers if Nanite Texure set enabled. If enabled default 2 Layers awailable automatically, where layer 1 is background layer "
   group=3_Layer_Blending
   parameterName=Add Fresnel to Layer 3 Blending
   sortPriority=4
  }
  n42 = StaticSwitchParameter(True=@n40, False=@n41) {group=3_Layer_Blending, parameterName=Use Fresnel to Layer 3 Blending, sortPriority=5}
  MaterialExpressionReroute_80 = Reroute(None=@n42)
  n43 = Fresnel(ExponentIn=ScalarParameter() {defaultValue=3, group=4_Layer_Blending, parameterName=L4 Fresnel Mask Exp}, Normal=@n36) {baseReflectFraction=0}
  MaterialExpressionReroute_83 = Reroute(None=@MaterialExpressionReroute_49.Output3)
  n44 = StaticSwitchParameter(True=Add(A=@MaterialExpressionReroute_83.InputPin3, B=@n43), False=@MaterialExpressionReroute_83.InputPin3) {
   desc="Enable to use SFX texture as the Layer mask instead UV2 offset. By default enables SFX to use UV2 coords (Unwrapped) Maximum 3 layers. 2 Layers if Nanite Texure set enabled. If enabled default 2 Layers awailable automatically, where layer 1 is background layer "
   group=4_Layer_Blending
   parameterName=Add Fresnel to Layer 4 Blending
   sortPriority=4
  }
  n45 = StaticSwitchParameter(True=@n43, False=@n44) {group=4_Layer_Blending, parameterName=Use Fresnel to Layer 4 Blending, sortPriority=5}
  MaterialExpressionReroute_82 = Reroute(None=@n45)
  Composite!(
   InputPin=@MaterialExpressionReroute_78.Output
   InputPin2=@MaterialExpressionReroute_80.Output2
   InputPin3=@MaterialExpressionReroute_82.Output3
  ) {subgraphName=RVT Masks}
  Composite(
   InputPin=@MaterialExpressionReroute_63.Output
   InputPin2=@MaterialExpressionReroute_67.Output2
   InputPin3=@MaterialExpressionReroute_69.Output3
  ) {subgraphName=Z Mask}
  Composite!(
   L2 SFX=@n9
   InputPin2=@MaterialExpressionReroute_72.Output
   L3 SFX=@n16
   InputPin4=@MaterialExpressionReroute_73.Output2
   InputPin5=@MaterialExpressionReroute_74.Output3
   L4 SFX=@n22
  ) {subgraphName=SFX Mask}
  n46 = Constant3Vector() {constant=<rgba 0 0 1 0>}
  n47 = VectorParameter() {group=0_Debug, parameterName=L1 Debug, sortPriority=33, channelNames {a=Emissive Value, b="", g="", r=""}, defaultValue=<rgba 1 0 0 1>}
  MaterialExpressionReroute_19 = Reroute(None=@n2)
  n48 = GetMaterialAttributes(None=@MaterialExpressionReroute_19.Input) {attributeGetTypes [#1, #2, #3]}
  n49 = VectorParameter() {group=1_Layer, parameterName="L1 Tint (Expand)", channelNames {a=L1 Specular Tweak, b="", g="", r=""}, defaultValue=<rgba 1 1 1 1>}
  n50 = SetMaterialAttributes(
   MaterialAttributes=@n48.MaterialAttributes
   BaseColor=Multiply(A=@n48.BaseColor, B=@n49)
   Roughness=Multiply(A=@n48.Roughness, B=ScalarParameter() {defaultValue=1, group=1_Layer, parameterName=L1 Roughness Tweak})
   Specular=Multiply(A=@n48.Specular, B=@n49.L1 Specular Tweak)
  ) {attributeSetTypes [#1, #2, #3]}
  MaterialExpressionReroute_18 = Reroute(None=@n50)
  MaterialExpressionReroute_17 = Reroute(None=@n1)
  n51 = GetMaterialAttributes(None=@MaterialExpressionReroute_17.Input) {attributeGetTypes [#1, #2, #3]}
  n52 = VectorParameter() {group=2_Layer, parameterName="L2 Tint  (Expand)", channelNames {a=L2 Specular Tweak, b="", g="", r=""}, defaultValue=<rgba 1 1 1 1>}
  n53 = SetMaterialAttributes(
   MaterialAttributes=@n51.MaterialAttributes
   BaseColor=Multiply(A=@n51.BaseColor, B=@n52)
   Roughness=Multiply(A=@n51.Roughness, B=ScalarParameter() {defaultValue=1, group=2_Layer, parameterName=L2 Roughness Tweak})
   Specular=Multiply(A=@n51.Specular, B=@n52.L2 Specular Tweak)
  ) {attributeSetTypes [#1, #2, #3]}
  MaterialExpressionReroute_16 = Reroute(None=@n53)
  n54 = MaterialFunctionCall!(
   Lower Layer=@MaterialExpressionReroute_18.Output
   "Upper Layer (Blended)"=@MaterialExpressionReroute_16.Output
   Use Simple Swatch=StaticBoolParameter() {desc=Use SFX to mask needed area, group=2_Layer_Blending, parameterName=L2 Simple Swatch, sortPriority=-15}
   "Use  Normals Only"=StaticBoolParameter() {group=2_Layer_Blending, parameterName=L2 Normals Only, sortPriority=-15}
   Blending Value=ScalarParameter() {defaultValue=1, group=2_Layer_Blending, parameterName="Overlay  Layer 2 Weight", sortPriority=0}
  ) {materialFunction=&/MF_BlendLayers}
  n55 = VectorParameter() {group=0_Debug, parameterName=L2 Debug, sortPriority=34, channelNames {a=Emissive Value, b="", g="", r=""}, defaultValue=<rgba 0 1 0 1>}
  MaterialExpressionReroute_37 = Reroute(None=@MaterialExpressionReroute_78.Output)
  n56 = Saturate(None=Add(A=@MaterialExpressionReroute_37.InputPin, B=NamedRerouteUsage()))
  n57 = ShadingPathSwitch(
   Default=StaticSwitchParameter(
    True=NamedRerouteUsage()
    False=StaticSwitchParameter(True=@n56, False=@MaterialExpressionReroute_37.InputPin) {group=2_Layer_Blending, parameterName=Add RVT to Layer 2 Blending, sortPriority=6}
   ) {group=2_Layer_Blending, parameterName=Use RVT for Layer 2 Blending, sortPriority=7}
   Mobile=@MaterialExpressionReroute_37.InputPin
  )
  MaterialExpressionReroute_36 = Reroute(None=@n57)
  n58 = BlendMaterialAttributes(
   A=MaterialFunctionCall!(Layer Debug=@MaterialExpressionReroute_18.Output, Debug Layer Color=AppendVector(A=@n47, B=@n47), Mat) {materialFunction=&/MF_DebugViews}.Layer Debug
   B=MaterialFunctionCall(
    Layer Debug=StaticSwitchParameter(True=@n54.Result, False=@MaterialExpressionReroute_16.Output) {group=2_Layer_Blending, parameterName=Blend Overlay Layer 2, sortPriority=-1}
    Debug Layer Color=AppendVector(A=@n55, B=@n55)
   ) {materialFunction=&MF_DebugViews}.Layer Debug
   Alpha=Saturate(None=@MaterialExpressionReroute_36.Output)
  ) {vertexAttributeBlendType=UseA}
  MaterialExpressionReroute_28 = Reroute(None=@MaterialExpressionReroute_20.Output)
  MaterialExpressionReroute_31 = Reroute(None=@MaterialExpressionReroute_28.MaterialAttributes)
  n59 = MaterialFunctionCall(
   Lower Layer=@n58
   "Upper Layer (Blended)"=@MaterialExpressionReroute_31.Output
   Use Simple Swatch=StaticBoolParameter() {desc=Use SFX to mask needed area, group=3_Layer_Blending, parameterName=L3 Simple Swatch, sortPriority=-15}
   "Use  Normals Only"=StaticBoolParameter() {group=3_Layer_Blending, parameterName="L3  Normals Only", sortPriority=-15}
   Blending Value=ScalarParameter() {defaultValue=1, group=3_Layer_Blending, parameterName=Overlay Layer 3 Weight, sortPriority=0}
  ) {materialFunction=&MF_BlendLayers}
  n60 = VectorParameter() {group=0_Debug, parameterName=L3 Debug, sortPriority=36, channelNames {a=Emissive Value, b="", g="", r=""}, defaultValue=<rgba 0 0 1 1>}
  MaterialExpressionReroute_39 = Reroute(None=@MaterialExpressionReroute_80.Output2)
  n61 = Saturate(None=Add(A=@MaterialExpressionReroute_39.InputPin2, B=NamedRerouteUsage()))
  n62 = ShadingPathSwitch(
   Default=StaticSwitchParameter(
    True=NamedRerouteUsage()
    False=StaticSwitchParameter(True=@n61, False=@MaterialExpressionReroute_39.InputPin2) {group=3_Layer_Blending, parameterName=Add RVT to Layer 3 Blending, sortPriority=6}
   ) {group=3_Layer_Blending, parameterName=Use RVT to Layer 3 Blending, sortPriority=7}
   Mobile=@MaterialExpressionReroute_39.InputPin2
  )
  MaterialExpressionReroute_38 = Reroute(None=@n62)
  n63 = BlendMaterialAttributes(
   A=@n58
   B=MaterialFunctionCall(
    Layer Debug=StaticSwitchParameter(True=@n59.Result, False=@MaterialExpressionReroute_31.Output) {group=3_Layer_Blending, parameterName=Blend Overlay Layer 3, sortPriority=-1}
    Debug Layer Color=AppendVector(A=@n60, B=@n60)
   ) {materialFunction=&MF_DebugViews}.Layer Debug
   Alpha=Saturate(None=@MaterialExpressionReroute_38.Output2)
  ) {vertexAttributeBlendType=UseA}
  MaterialExpressionReroute_24 = Reroute(None=@MaterialExpressionReroute_22.Output)
  MaterialExpressionReroute_27 = Reroute(None=@MaterialExpressionReroute_24.MaterialAttributes)
  n64 = MaterialFunctionCall(
   Lower Layer=@n63
   "Upper Layer (Blended)"=@MaterialExpressionReroute_27.Output
   Use Simple Swatch=StaticBoolParameter() {desc=Use SFX to mask needed area, group=4_Layer_Blending, parameterName=L4 Simple Swatch, sortPriority=-15}
   "Use  Normals Only"=StaticBoolParameter() {group=4_Layer_Blending, parameterName="L4  Normals Only", sortPriority=-15}
   Blending Value=ScalarParameter() {defaultValue=1, group=4_Layer_Blending, parameterName=Overlay Layer 4 Weight, sortPriority=0}
  ) {materialFunction=&MF_BlendLayers}
  n65 = VectorParameter() {group=0_Debug, parameterName=L4 Debug, sortPriority=36, channelNames {a=Emissive Value, b="", g="", r=""}, defaultValue=<rgba 1 0 1 1>}
  MaterialExpressionReroute_41 = Reroute(None=@MaterialExpressionReroute_82.Output3)
  n66 = Saturate(None=Add(A=@MaterialExpressionReroute_41.InputPin3, B=NamedRerouteUsage()))
  n67 = ShadingPathSwitch(
   Default=StaticSwitchParameter(
    True=NamedRerouteUsage()
    False=StaticSwitchParameter(True=@n66, False=@MaterialExpressionReroute_41.InputPin3) {group=4_Layer_Blending, parameterName=Add RVT to Layer 4 Blending, sortPriority=6}
   ) {group=4_Layer_Blending, parameterName=Use RVT to Layer 4 Blending, sortPriority=7}
   Mobile=@MaterialExpressionReroute_41.InputPin3
  )
  MaterialExpressionReroute_40 = Reroute(None=@n67)
  n68 = BlendMaterialAttributes(
   A=@n63
   B=MaterialFunctionCall(
    Layer Debug=StaticSwitchParameter(True=@n64.Result, False=@MaterialExpressionReroute_27.Output) {group=4_Layer_Blending, parameterName=Blend Overlay Layer 4, sortPriority=-1}
    Debug Layer Color=AppendVector(A=@n65, B=@n65)
   ) {materialFunction=&MF_DebugViews}.Layer Debug
   Alpha=Saturate(None=@MaterialExpressionReroute_40.Output3)
  ) {vertexAttributeBlendType=UseA}
  n69 = StaticSwitch(
   True=@n58
   False=@MaterialExpressionReroute_18.Output
   Value=StaticBoolParameter() {group=2_Layer, parameterName=Use 2 Layers, sortPriority=-1}
  )
  n70 = StaticSwitch(
   True=ShadingPathSwitch(
    Default=ShadingPathSwitch(Default=@n69, Mobile=@MaterialExpressionReroute_18.Output) {desc=Migrated from FeatureLevelSwitch}
    Mobile=@MaterialExpressionReroute_18.Output
   )
   False=@n69
   Value=StaticBoolParameter() {group=2_Layer, parameterName=Skip 2 Layer Low Spec/Mobile}
  )
  n71 = StaticSwitch(
   True=QualitySwitch(Default=@n63, Low=@n63, High=@n63, Medium=@n63, Epic=@n63)
   False=@n70
   Value=StaticBoolParameter() {group=3_Layer, parameterName=Use 3 Layers, sortPriority=-1}
  )
  n72 = StaticSwitch(
   True=ShadingPathSwitch(Default=ShadingPathSwitch(Default=@n71, Mobile=@n70) {desc=Migrated from FeatureLevelSwitch}, Mobile=@n70)
   False=@n71
   Value=StaticBoolParameter() {group=3_Layer, parameterName=Skip 3 Layer Low Spec/Mobile}
  )
  n73 = StaticBoolParameter() {desc=Enable to use World Space Mapping for layer. Only XY or Z available in Layer 3, group=3_layer UV, parameterName=Layer 3 WS Mapping}
  n74 = StaticSwitch(
   True=QualitySwitch(Default=@n68, Low=SetMaterialAttributes!(MaterialAttributes=@n68, Normal=@n46) {attributeSetTypes [#4]})
   False=@n72
   Value=StaticSwitch(True=StaticBool(), False=StaticBoolParameter() {group=4_Layer, parameterName=Use 4 Layers, sortPriority=-1}, Value=@n73)
  )
  n75 = ShadingPathSwitch(Default=@n74, Mobile=@n72) {desc=Migrated from FeatureLevelSwitch}
  FunctionOutput(None=StaticSwitch(
    True=ShadingPathSwitch(Default=@n75, Mobile=@n72)
    False=@n74
    Value=StaticBoolParameter() {group=4_Layer, parameterName=Skip 4 Layer Low Spec/Mobile}
   ))
  n76 = ComponentMask(None=VertexNormalWS()) {b=true}
  n77 = NamedRerouteDeclaration(None=FunctionInput() {inputName=SFX V3}) {name=SFX V3, nodeColor=<rgba 1 0 0.1172 1>}
  n78 = VectorParameter() {parameterName=Zup High map Channel, channelNames {a=Z Up Height Contrast, b="", g="", r=""}, defaultValue=<rgba 1 0 0 0>}
  NamedRerouteDeclaration(None=MaterialFunctionCall(Transition Phase=Saturate(None=@n76), Height Texture=DotProduct(A=@n77, B=@n78), Contrast=@n78.Z Up Height Contrast) {materialFunction=&HeightLerp}.Alpha) {name=Zup, nodeColor=<rgba 1 0 0.0937 1>}
  n79 = Constant()
  n80 = RuntimeVirtualTextureSampleParameter() {
   group=D_RVT
   materialType=WorldHeight
   parameterName=RVTHeight
   sortPriority=1
   virtualTexture=/Game/Athena/Apollo/Maps/Landscape/VT/RVT_Apollo_Landscape_Height
  }
  n81 = Subtract(A=WorldPosition().Z, B=@n80.WorldHeight)
  n82 = OneMinus(None=ScalarParameter() {desc="If layer blending enabled, will affect all layers", group=D_RVT, parameterName=RVT Layer blending Offset})
  n83 = ScalarParameter() {defaultValue=256, desc="If layer blending enabled, will affect all layers", group=D_RVT, parameterName=RVT Layer blending Distance}
  n84 = MaterialProxyReplace(Realtime=Divide(A=Add(A=@n81, B=@n82), B=@n83), MaterialProxy=@n79)
  n85 = OneMinus(None=Saturate(None=VirtualTextureFeatureSwitch!(No=@n79, Yes=@n84)))
  NamedRerouteDeclaration(None=@n85) {name=RVT Inverted, nodeColor=<rgba 1 0.8438 0 1>}
  NamedRerouteDeclaration(None=@n85) {name=RVT, nodeColor=<rgba 0 0.8125 1 1>}
  n86 = VectorParameter() {
   group=2_Layer_Blending
   parameterName=SFX Layer 2 Mask Ch
   sortPriority=2
   channelNames {a=Height Offset, b="", g="", r=""}
   defaultValue=<rgba 1 0 0 0.5>
  }
  n87 = DotProduct(A=@n77, B=@n86)
  n88 = VectorParameter() {
   group=2_Layer_Blending
   parameterName=SFX Layer 2 HM Ch
   sortPriority=2
   channelNames {a=Height Contrast, b="", g="", r=""}
   defaultValue=<rgba 1 0 0 0.5>
  }
  n89 = MaterialFunctionCall(
   Transition Phase=@n87
   Height Texture=Multiply(A=DotProduct(A=@n77, B=@n88), B=ScalarParameter() {defaultValue=1, group=2_Layer_Blending, parameterName=L2 Boost sfx HM, sortPriority=2})
   Contrast=@n88.Height Contrast
  ) {materialFunction=&HeightLerp}
  NamedRerouteDeclaration(None=Saturate(None=StaticSwitchParameter(True=@n89.Alpha, False=@n87) {group=2_Layer_Blending, parameterName=SFX Layer2 Hight Blend, sortPriority=3})) {name=L2 SFX mask, nodeColor=<rgba 0 0.1562 1 1>}
  n90 = VectorParameter() {
   group=3_Layer_Blending
   parameterName=SFX Layer 3 Mask Ch
   sortPriority=2
   channelNames {a=Height Offset, b="", g="", r=""}
   defaultValue=<rgba 1 0 0 0.5>
  }
  n91 = DotProduct(A=@n77, B=@n90)
  n92 = VectorParameter() {
   group=3_Layer_Blending
   parameterName=SFX Layer 3 HM Ch
   sortPriority=2
   channelNames {a=Height Contrast, b="", g="", r=""}
   defaultValue=<rgba 1 0 0 0.5>
  }
  n93 = MaterialFunctionCall(
   Transition Phase=@n91
   Height Texture=Multiply(A=DotProduct(A=@n77, B=@n92), B=ScalarParameter() {defaultValue=1, group=3_Layer_Blending, parameterName=L3 Boost sfx HM, sortPriority=2})
   Contrast=@n92.Height Contrast
  ) {materialFunction=&HeightLerp}
  NamedRerouteDeclaration(None=Saturate(None=StaticSwitchParameter(True=@n93.Alpha, False=@n91) {group=3_Layer_Blending, parameterName=SFX Layer3 Hight Blend, sortPriority=3})) {name=L3 SFX Mask, nodeColor=<rgba 1 0.7031 0 1>}
  n94 = VectorParameter() {
   group=4_Layer_Blending
   parameterName=SFX Layer 4 Mask Ch
   sortPriority=2
   channelNames {a=Height Offset, b="", g="", r=""}
   defaultValue=<rgba 1 0 0 0.5>
  }
  n95 = DotProduct(A=@n77, B=@n94)
  n96 = VectorParameter() {
   group=4_Layer_Blending
   parameterName=SFX Layer 4 HM Ch
   sortPriority=2
   channelNames {a=Height Contrast, b="", g="", r=""}
   defaultValue=<rgba 1 0 0 0.5>
  }
  n97 = MaterialFunctionCall(
   Transition Phase=@n95
   Height Texture=Multiply(A=DotProduct(A=@n77, B=@n96), B=ScalarParameter() {defaultValue=1, group=4_Layer_Blending, parameterName=L4 Boost sfx HM, sortPriority=2})
   Contrast=@n96.Height Contrast
  ) {materialFunction=&HeightLerp}
  NamedRerouteDeclaration(None=Saturate(None=StaticSwitchParameter(True=@n97.Alpha, False=@n95) {group=4_Layer_Blending, parameterName=SFX Layer4 Hight Blend, sortPriority=3})) {name=L4 SFX Mask, nodeColor=<rgba 0 0.7187 1 1>}
  PinBase() {reroutePins [Input {expression=&self:MaterialExpressionReroute_17}]}
  PinBase!(Output=@n53) {reroutePins [Output {expression=&self:MaterialExpressionReroute_16}]}
  PinBase!() {reroutePins [
    MaterialAttributes {expression=&self:MaterialExpressionReroute_28}
    False {expression=&self:MaterialExpressionReroute_29}
    False2 {expression=&self:MaterialExpressionReroute_30}
   ]}
  PinBase!(Output=@MaterialExpressionReroute_28.MaterialAttributes) {reroutePins [Output {expression=&self:MaterialExpressionReroute_31}]}
  PinBase!() {reroutePins [
    InputPin {expression=&self:MaterialExpressionReroute_37}
    InputPin2 {expression=&self:MaterialExpressionReroute_39}
    InputPin3 {expression=&self:MaterialExpressionReroute_41}
   ]}
  PinBase!(Output=@n57, Output2=@n62, Output3=@n67) {reroutePins [
    Output {expression=&self:MaterialExpressionReroute_36}
    Output2 {expression=&self:MaterialExpressionReroute_38}
    Output3 {expression=&self:MaterialExpressionReroute_40}
   ]}
  PinBase!() {reroutePins [
    InputPin {expression=&self:MaterialExpressionReroute_46}
    InputPin2 {expression=&self:MaterialExpressionReroute_48}
    InputPin3 {expression=&self:MaterialExpressionReroute_50}
   ]}
  PinBase!(Output=@n13, Output2=@n20, Output3=@n27) {reroutePins [
    Output {expression=&self:MaterialExpressionReroute_45}
    Output2 {expression=&self:MaterialExpressionReroute_47}
    Output3 {expression=&self:MaterialExpressionReroute_49}
   ]}
  PinBase!() {reroutePins [
    L2 SFX {expression=&self:MaterialExpressionReroute_64}
    InputPin2 {expression=&self:MaterialExpressionReroute_65}
    L3 SFX {expression=&self:MaterialExpressionReroute_66}
    InputPin4 {expression=&self:MaterialExpressionReroute_68}
    InputPin5 {expression=&self:MaterialExpressionReroute_70}
    L4 SFX {expression=&self:MaterialExpressionReroute_71}
   ]}
  PinBase!(Output=@n11, Output2=@n18, Output3=@n25) {reroutePins [
    Output {expression=&self:MaterialExpressionReroute_63}
    Output2 {expression=&self:MaterialExpressionReroute_67}
    Output3 {expression=&self:MaterialExpressionReroute_69}
   ]}
  PinBase!() {reroutePins [Input {expression=&self:MaterialExpressionReroute_19}]}
  PinBase()
  PinBase!(Output=@n8, Output2=@n15, Output3=@n23) {reroutePins [
    Output {expression=&self:MaterialExpressionReroute_72}
    Output2 {expression=&self:MaterialExpressionReroute_73}
    Output3 {expression=&self:MaterialExpressionReroute_74}
   ]}
  PinBase!() {reroutePins [
    InputPin {expression=&self:MaterialExpressionReroute_79}
    InputPin2 {expression=&self:MaterialExpressionReroute_81}
    InputPin3 {expression=&self:MaterialExpressionReroute_83}
   ]}
  PinBase!(Output=@n39, Output2=@n42, Output3=@n45) {reroutePins [
    Output {expression=&self:MaterialExpressionReroute_78}
    Output2 {expression=&self:MaterialExpressionReroute_80}
    Output3 {expression=&self:MaterialExpressionReroute_82}
   ]}
  PinBase!(Output=@n50) {reroutePins [Output {expression=&self:MaterialExpressionReroute_18}]}
  PinBase!() {reroutePins [Input {expression=&self:MaterialExpressionReroute_21}]}
  PinBase!(Output=@n35) {reroutePins [Output {expression=&self:MaterialExpressionReroute_20}]}
  PinBase!() {reroutePins [Input {expression=&self:MaterialExpressionReroute_23}]}
  PinBase!(Output=@n32) {reroutePins [Output {expression=&self:MaterialExpressionReroute_22}]}
  PinBase!() {reroutePins [
    MaterialAttributes {expression=&self:MaterialExpressionReroute_24}
    False {expression=&self:MaterialExpressionReroute_25}
    False2 {expression=&self:MaterialExpressionReroute_26}
   ]}
  PinBase!(Output=@MaterialExpressionReroute_24.MaterialAttributes) {reroutePins [Output {expression=&self:MaterialExpressionReroute_27}]}
  SetMaterialAttributes(MaterialAttributes=@n63, Normal=Constant3Vector() {constant=<rgba 0 0 1 0>}) {attributeSetTypes [#4]}
  MaterialExpressionReroute_25 = Reroute(None=@MaterialExpressionReroute_22.Output)
  n98 = StaticSwitchParameter(
   True=SetMaterialAttributes(MaterialAttributes=@MaterialExpressionReroute_24.MaterialAttributes, Normal=Constant3Vector() {constant=<rgba 0 0 1 0>}) {attributeSetTypes [#4]}
   False=@MaterialExpressionReroute_25.False
  ) {
   desc="Emissive RGB Available if SFX RGB is activated in 0 SFX Global. Will replace layers Emissive unless Emissive RGB or Emissive Texture Add is set= true  ( adds texture sample)"
   group=A_Emissive_Texture
   parameterName=SFX Emissive Texture RGB
   sortPriority=-32
  }
  MaterialExpressionReroute_26 = Reroute(None=@MaterialExpressionReroute_22.Output)
  StaticSwitchParameter(True=@n98, False=@MaterialExpressionReroute_26.False2) {
   desc="Enable to access various emissive functions, TOD,Fresnel, Flipbook etc."
   group=A__EmissiveFilters
   parameterName=Emissive Filters
   sortPriority=-9
  }
  MaterialExpressionReroute_29 = Reroute(None=@MaterialExpressionReroute_20.Output)
  n99 = StaticSwitchParameter(
   True=SetMaterialAttributes(MaterialAttributes=@MaterialExpressionReroute_28.MaterialAttributes, Normal=Constant3Vector() {constant=<rgba 0 0 1 0>}) {attributeSetTypes [#4]}
   False=@MaterialExpressionReroute_29.False
  ) {
   desc="Emissive RGB Available if SFX RGB is activated in 0 SFX Global. Will replace layers Emissive unless Emissive RGB or Emissive Texture Add is set= true  ( adds texture sample)"
   group=A_Emissive_Texture
   parameterName=SFX Emissive Texture RGB
   sortPriority=-32
  }
  MaterialExpressionReroute_30 = Reroute(None=@MaterialExpressionReroute_20.Output)
  StaticSwitchParameter(True=@n99, False=@MaterialExpressionReroute_30.False2) {
   desc="Enable to access various emissive functions, TOD,Fresnel, Flipbook etc."
   group=A__EmissiveFilters
   parameterName=Emissive Filters
   sortPriority=-9
  }
 ]}
