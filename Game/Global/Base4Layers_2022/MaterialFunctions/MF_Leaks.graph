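# Material function that layers procedural leaks, grime and stains over incoming material attributes, with static switches to bypass it or choose the new or older method.
# The new path world-projects a packed mask texture, using optional bounding-box-shaped red and blue channel tints and an AO-driven green-channel HeightLerp to alter base color and roughness.
# The older path blends top stains and leak colors by vertex normal and AO.
# Both apply effects only at Epic quality.
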MaterialFunction /Game/Global/Base4Layers_2022/MaterialFunctions/MF_Leaks
graph {nodes [
  n1 = GetMaterialAttributes(None=NamedRerouteUsage()) {attributeGetTypes [#1, #2]}
  n2 = ScalarParameter() {defaultValue=0.8, group=B_LeaksAndStains WIP, parameterName=Top Desaturation, sortPriority=2}
  n3 = VectorParameter() {group=B_LeaksAndStains WIP, parameterName=TopColor Tone, sortPriority=1, defaultValue=<rgba 0.5816 0.5149 0.3908 1>}
  n4 = FunctionInput() {inputName=AO, inputType=FunctionInput_Scalar}
  n5 = TextureObject() {samplerType=SAMPLERTYPE_Masks, texture=/Game/Global/Base4Layers_2022/Textures/T_Stains_Leaks}
  n6 = Constant3Vector() {constant=<rgba 0 0.0186 1 1>}
  n7 = Constant() {r=5}
  n8 = MaterialFunctionCall(TextureObject=@n5, TextureSize=Constant() {r=2000}, Normal=@n6, ProjectionTransitionContrast=@n7) {materialFunction=&WorldAlignedNormal}
  n9 = ScalarParameter() {defaultValue=0.5, group=B_LeaksAndStains WIP, parameterName=Stain B Value}
  n10 = MaterialFunctionCall(TextureObject=@n5, TextureSize=Constant() {r=200}, Normal=@n6, ProjectionTransitionContrast=@n7) {materialFunction=&WorldAlignedNormal}
  n11 = OneMinus(None=LinearInterpolate(Alpha=ComponentMask(None=@n10.Z Texture) {r=true}) {constA=-1})
  n12 = MaterialFunctionCall(
   A=Multiply(
    A=Desaturation(None=@n1.BaseColor, Fraction=@n2)
    B=ScalarParameter() {defaultValue=0.5, group=B_LeaksAndStains WIP, parameterName=Top Value, sortPriority=2}
   )
   B=Multiply(A=@n1.BaseColor, B=@n3.RGB)
   Transition Phase=LinearInterpolate(
    A=ScalarParameter() {defaultValue=0.3, group=B_LeaksAndStains WIP, parameterName=Top Ao Grime, sortPriority=3}
    B=ScalarParameter() {defaultValue=0.5, group=B_LeaksAndStains WIP, parameterName=Top Ao Grime Max, sortPriority=3}
    Alpha=OneMinus(None=@n4)
   )
   Height Texture=Add(A=ComponentMask(None=@n8.XY Texture) {b=true}, B=Multiply(A=@n9, B=@n11))
   Contrast=ScalarParameter() {defaultValue=0.5, group=B_LeaksAndStains WIP, parameterName=Top Stain Contrast}
  ) {materialFunction=&HeightLerp}
  n13 = TextureCoordinate()
  n14 = Constant() {r=4}
  n15 = OneMinus(None=TextureSample(UVs=Multiply(A=@n13, B=@n14), Tex=@n5) {samplerSource=SSM_Wrap_WorldGroupSettings, texture=/Game/Environments/Helios/Sets/GenericSet/Textures/T_GenericSet_Roof_Tiles_D}.R)
  n16 = Multiply(A=Multiply(A=@n12.Results, B=Saturate(None=@n15)), B=Constant() {r=1})
  n17 = ComponentMask(None=VertexNormalWS()) {b=true}
  n18 = BlendMaterialAttributes(
   A=NamedRerouteUsage()
   B=SetMaterialAttributes!(MaterialAttributes=@n1.MaterialAttributes, BaseColor=Add(A=@n12.Results, B=@n16)) {attributeSetTypes [#1]}
   Alpha=Saturate(None=@n17)
  )
  n19 = GetMaterialAttributes(None=@n18) {attributeGetTypes [#1, #3]}
  n20 = VectorParameter() {group=B_LeaksAndStains WIP, parameterName=Noise, sortPriority=5, defaultValue=<rgba 1.2 1.035 0.7658 1>}
  n21 = Saturate(None=LinearInterpolate(Alpha=ComponentMask(None=@n8.XY Texture) {b=true}) {constA=-1})
  n22 = LinearInterpolate(A=Multiply(A=@n19.BaseColor, B=@n20.RGB), B=@n19.BaseColor, Alpha=@n21)
  n23 = VectorParameter() {group=B_LeaksAndStains WIP, parameterName=Leak, sortPriority=5, defaultValue=<rgba 0.0592 0.0851 0.034 1>}
  n24 = VectorParameter() {group=B_LeaksAndStains WIP, parameterName=Leak2, sortPriority=5, defaultValue=<rgba 0.7396 0.5095 0.1745 1>}
  n25 = LinearInterpolate(A=@n23.RGB, B=@n24.RGB, Alpha=@n21)
  n26 = MaterialFunctionCall(Transition Phase=Constant() {r=0.2}, Height Texture=ComponentMask(None=@n8.XY Texture) {g=true}, Contrast=Constant() {r=0.8}) {materialFunction=&HeightLerp}
  n27 = BlendMaterialAttributes(
   A=SetMaterialAttributes(MaterialAttributes=@n19.MaterialAttributes, BaseColor=LinearInterpolate(A=@n22, B=@n25, Alpha=@n26.Alpha)) {attributeSetTypes [#1]}
   B=@n18
   Alpha=Abs(None=@n17)
  )
  BlendMaterialAttributes(B=@n27)
  n28 = NamedRerouteUsage()
  n29 = GetMaterialAttributes(None=NamedRerouteUsage()) {attributeGetTypes [
    #1
    #2
    #4
    #3
   ]}
  n30 = ScalarParameter() {defaultValue=0.8, group=B_LeaksAndStains WIP, parameterName=G Top Desaturation, sortPriority=2}
  n31 = VectorParameter() {group=B_LeaksAndStains WIP, parameterName=TopColor Tone, sortPriority=1, defaultValue=<rgba 1 1 1 1>}
  n32 = Multiply(A=@n29.BaseColor, B=@n31.RGB)
  n33 = VectorParameter() {group=B_LeaksAndStains WIP, parameterName=R Mask Tint, sortPriority=1, defaultValue=<rgba 0.5816 0 0.0164 1>}
  n34 = TextureObjectParameter() {
   group=B_LeaksAndStains WIP
   parameterName=Leaks Texture
   samplerType=SAMPLERTYPE_Masks
   texture=&T_Stains_Leaks
  }
  n35 = ScalarParameter() {group=B_LeaksAndStains WIP, parameterName=WAT Projection Contrast}
  n36 = ComponentMask(None=MaterialFunctionCall(
    TextureObject=@n34
    TextureSize=ScalarParameter() {defaultValue=2000, group=B_LeaksAndStains WIP, parameterName=G Tiling, sortPriority=3}
    RotZAxis=ScalarParameter() {group=B_LeaksAndStains WIP, parameterName=G Rotation Z, sortPriority=3}
    RotXAxis=ScalarParameter() {group=B_LeaksAndStains WIP, parameterName=G Rotation X, sortPriority=3}
    RotYAxis=ScalarParameter() {group=B_LeaksAndStains WIP, parameterName=G Rotation Y, sortPriority=3}
    ProjectionTransitionContrast=@n35
   ) {materialFunction=/Game/Weapons/WeaponSkins/Wraps/MaterialFunctions/WorldAlignedTexture}.XYZ Texture) {g=true}
  n37 = ScalarParameter() {defaultValue=2000, group=B_LeaksAndStains WIP, parameterName=R Tiling}
  n38 = ScalarParameter() {group=B_LeaksAndStains WIP, parameterName=R Rotation Z}
  n39 = ScalarParameter() {group=B_LeaksAndStains WIP, parameterName=R Rotation X}
  n40 = ScalarParameter() {group=B_LeaksAndStains WIP, parameterName=R Rotation Y}
  n41 = LinearInterpolate(Alpha=ComponentMask(None=MaterialFunctionCall(TextureObject=@n34, TextureSize=@n37, RotZAxis=@n38, RotXAxis=@n39, RotYAxis=@n40, ProjectionTransitionContrast=@n35) {materialFunction=&WorldAlignedTexture}.XYZ Texture) {r=true}) {constA=-1}
  n42 = Multiply(A=ScalarParameter() {defaultValue=0.5, group=B_LeaksAndStains WIP, parameterName=R Stain Value}, B=@n41)
  n43 = Power(
   Base=Multiply(A=Add(A=@n36, B=@n42), B=ScalarParameter() {defaultValue=1, group=B_LeaksAndStains WIP, parameterName=R Mask Range})
   Exp=ScalarParameter() {defaultValue=5, group=B_LeaksAndStains WIP, parameterName=R Mask Power}
  )
  n44 = MaterialFunctionCall(
   In=Multiply(
    A=MaterialFunctionCall() {materialFunction=&BoundingBoxBased_0-1_UVW}.B
    B=ScalarParameter() {defaultValue=0.5, group=B_LeaksAndStains WIP, parameterName=R Bounding Box Position}
   )
   Contrast=ScalarParameter() {defaultValue=0.8989, group=B_LeaksAndStains WIP, parameterName=R Bounding Box Contrast}
  ) {materialFunction=&CheapContrast}
  n45 = StaticSwitch(
   True=OneMinus(None=@n44.Result)
   False=@n44.Result
   Value=StaticBoolParameter() {group=B_LeaksAndStains WIP, parameterName=Invert Red Channel Bounding Box}
  )
  n46 = Multiply(A=ComponentMask(None=VertexNormalWS()) {b=true}, B=ScalarParameter() {group=B_LeaksAndStains WIP, parameterName=R Vertex Normal Conrast})
  n47 = Multiply(
   A=Subtract(A=Multiply(A=@n43, B=@n45), B=Saturate(None=@n46))
   B=ScalarParameter() {defaultValue=1, group=B_LeaksAndStains WIP, parameterName=R Tint Intensity}
  )
  n48 = StaticSwitchParameter(True=LinearInterpolate(A=@n32, B=@n33.RGB, Alpha=Saturate(None=@n47)), False=@n32) {group=B_LeaksAndStains WIP, parameterName=Use Red Channel Mask}
  n49 = VectorParameter() {group=B_LeaksAndStains WIP, parameterName=B Mask Tint, sortPriority=1, defaultValue=<rgba 0.0186 0.044 0.5816 1>}
  n50 = ComponentMask(None=MaterialFunctionCall(
    TextureObject=@n34
    TextureSize=ScalarParameter() {defaultValue=2000, group=B_LeaksAndStains WIP, parameterName=B Tiling}
    RotZAxis=ScalarParameter() {group=B_LeaksAndStains WIP, parameterName=B Rotation Z}
    RotXAxis=ScalarParameter() {group=B_LeaksAndStains WIP, parameterName=B Rotation X}
    RotYAxis=ScalarParameter() {group=B_LeaksAndStains WIP, parameterName=B Rotation Y}
    ProjectionTransitionContrast=@n35
   ) {materialFunction=&WorldAlignedTexture}.XYZ Texture) {b=true}
  n51 = Multiply(
   A=ScalarParameter() {defaultValue=0.5, group=B_LeaksAndStains WIP, parameterName=B Stain Value}
   B=OneMinus(None=LinearInterpolate(Alpha=@n50) {constA=-1})
  )
  n52 = Power(
   Base=Multiply(A=Add(A=@n36, B=@n51), B=ScalarParameter() {defaultValue=1, group=B_LeaksAndStains WIP, parameterName=B Mask Range})
   Exp=ScalarParameter() {defaultValue=5, group=B_LeaksAndStains WIP, parameterName=B Mask Power}
  )
  n53 = MaterialFunctionCall(
   In=Multiply(
    A=MaterialFunctionCall() {materialFunction=&BoundingBoxBased_0-1_UVW}.B
    B=ScalarParameter() {defaultValue=0.5, group=B_LeaksAndStains WIP, parameterName=Bounding Box Position}
   )
   Contrast=ScalarParameter() {defaultValue=0.8989, group=B_LeaksAndStains WIP, parameterName=Bounding Box Contrast}
  ) {materialFunction=&CheapContrast}
  n54 = StaticSwitch(
   True=OneMinus(None=@n53.Result)
   False=@n53.Result
   Value=StaticBoolParameter() {group=B_LeaksAndStains WIP, parameterName=Invert Blue Channel Bounding Box}
  )
  n55 = Saturate(None=StaticSwitchParameter(True=Multiply(A=@n52, B=@n54), False=Subtract(A=@n29.Metallic, B=@n52)) {group=B_LeaksAndStains WIP, parameterName=Use Blue Channel Bounding Volume Mask})
  n56 = ScalarParameter() {defaultValue=0.5, group=B_LeaksAndStains WIP, parameterName=B Mask Intensity}
  n57 = LinearInterpolate(A=@n48, B=@n49.RGB, Alpha=Saturate(None=Multiply(A=@n55, B=@n56)))
  n58 = NamedRerouteUsage()
  n59 = MaterialFunctionCall(
   A=Multiply(
    A=Desaturation(None=@n29.BaseColor, Fraction=@n30)
    B=ScalarParameter() {defaultValue=0.5, group=B_LeaksAndStains WIP, parameterName=G Top Value, sortPriority=2}
   )
   B=StaticSwitchParameter(True=@n57, False=@n48) {group=B_LeaksAndStains WIP, parameterName=Use Blue Channel Mask}
   Transition Phase=LinearInterpolate(
    A=ScalarParameter() {defaultValue=0.3, group=B_LeaksAndStains WIP, parameterName=G Top Ao Grime, sortPriority=3}
    B=ScalarParameter() {defaultValue=0.5, group=B_LeaksAndStains WIP, parameterName=G Top Ao Grime Max, sortPriority=3}
    Alpha=OneMinus(None=@n58)
   )
   Height Texture=Multiply(A=@n36, B=ScalarParameter() {defaultValue=1, group=B_LeaksAndStains WIP, parameterName=G Stain Value, sortPriority=3})
   Contrast=ScalarParameter() {defaultValue=1, group=B_LeaksAndStains WIP, parameterName=G Stain Contrast, sortPriority=3}
  ) {materialFunction=&HeightLerp}
  n60 = NamedRerouteDeclaration(None=@n59.Results) {name=Blended DF Dirt AO, nodeColor=<rgba 1 0.8906 0 1>}
  n61 = NamedRerouteUsage()
  n62 = Multiply(
   A=MaterialFunctionCall(
    Input=Saturate(None=@n61)
    Input Low=ScalarParameter() {group=B_LeaksAndStains WIP, parameterName=B Roughness Low}
    Input High=ScalarParameter() {defaultValue=1, group=B_LeaksAndStains WIP, parameterName=B Roughness High}
   ) {materialFunction=&RemapValue_0-1}.Result
   B=ScalarParameter() {defaultValue=0.5, group=B_LeaksAndStains WIP, parameterName=B Roughness Blend}
  )
  n63 = MaterialFunctionCall(
   Input=Saturate(None=OneMinus(None=@n59.Alpha))
   Input Low=ScalarParameter() {group=B_LeaksAndStains WIP, parameterName=G Roughness Low, sortPriority=3}
   Input High=ScalarParameter() {defaultValue=1, group=B_LeaksAndStains WIP, parameterName=G Roughness High, sortPriority=3}
  ) {materialFunction=&RemapValue_0-1}
  n64 = ScalarParameter() {defaultValue=0.5, group=B_LeaksAndStains WIP, parameterName=G Roughness Blend, sortPriority=3}
  n65 = LinearInterpolate(
   A=LinearInterpolate(
    A=NamedRerouteUsage()
    B=ScalarParameter() {defaultValue=0.5, group=B_LeaksAndStains WIP, parameterName=B Rougness Value}
    Alpha=Saturate(None=@n62)
   )
   B=ScalarParameter() {group=B_LeaksAndStains WIP, parameterName=G Roughness Value, sortPriority=3}
   Alpha=Saturate(None=Multiply(A=@n63.Result, B=@n64))
  )
  n66 = BlendMaterialAttributes(
   A=@n28
   B=QualitySwitch(
    Default=@n28
    Epic=SetMaterialAttributes!(MaterialAttributes=@n28, BaseColor=@n60, Roughness=@n65) {attributeSetTypes [#1, #3]}
   )
   Alpha=Multiply(A=ScalarParameter() {defaultValue=1, group=B_LeaksAndStains WIP, parameterName=Add Leaks and Stains, sortPriority=-1})
  )
  n67 = NamedRerouteUsage()
  n68 = ScalarParameter() {defaultValue=1}
  n69 = BlendMaterialAttributes(
   A=@n67
   B=QualitySwitch(Default=@n67, Epic=@n27)
   Alpha=Multiply(
    A=ScalarParameter() {defaultValue=1, group=B_LeaksAndStains WIP, parameterName=Add Leaks and Stains, sortPriority=-1}
    B=MaterialFunctionCall(Transition Phase=@n4, Height Texture=@n19.Roughness, Contrast=@n68) {materialFunction=&HeightLerp}.Alpha
   )
  )
  FunctionOutput(None=StaticSwitchParameter(True=StaticSwitchParameter(True=@n66, False=@n69) {group=B_LeaksAndStains WIP, parameterName=Use New Leaks}, False=NamedRerouteUsage()) {group=B_LeaksAndStains WIP, parameterName=Leaks and Stains WIP, sortPriority=-2})
  n70 = FunctionInput() {inputType=FunctionInput_MaterialAttributes}
  NamedRerouteDeclaration(None=@n70) {name=MatInput, nodeColor=<rgba 1 0.7031 0 1>}
  NamedRerouteDeclaration(None=@n4) {name=SDF AO Input, nodeColor=<rgba 0 0 0 1>}
  NamedRerouteDeclaration(None=@n29.Roughness) {name=Base Mat Roughness, nodeColor=<rgba 0.1323 0.1323 0.1323 1>}
  NamedRerouteDeclaration(None=@n57) {name=B Channel Mask, nodeColor=<rgba 1 0.6797 0 1>}
  OneMinus(None=@n41)
  OneMinus(None=@n61)
  Reroute(None=@n70)
  ScalarParameter() {parameterName=Param_1}
 ]}
