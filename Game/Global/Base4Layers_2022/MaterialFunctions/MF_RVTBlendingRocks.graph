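# Material function that blends an input material, likely for rocks, into the landscape using Runtime Virtual Texture color, normal and height samples.
# Height above the RVT surface builds noise-eroded masks with separate normal distances, varied by slope or vertex color, alongside a HeightLerp blend.
# Static switches add top snow, edge tinting and inside projection, which shifts RVT sampling along horizontal normals on steep sides.
# It also outputs RVT color and height through RVT switch functions.
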MaterialFunction /Game/Global/Base4Layers_2022/MaterialFunctions/MF_RVTBlendingRocks
graph {nodes [
  n1 = NamedRerouteUsage()
  n2 = NamedRerouteUsage()
  n3 = SetMaterialAttributes!(MaterialAttributes, BaseColor=NamedRerouteUsage(), Specular=NamedRerouteUsage(), Roughness=NamedRerouteUsage(), Normal=NamedRerouteUsage()) {attributeSetTypes [
    #1
    #2
    #3
    #4
   ]}
  n4 = GetMaterialAttributes(None=MaterialFunctionCall!(
    ScalarWeightMask
    UseSnowOcclusionTextureMask
    ------
    UseMaterialInput
    AdditionalMaskToRemoveSnow
    OutputTangentNormals
    UseAdditionalMaskToRemoveSnow
    UseSnowNormals
    OuputVertexNormalsAsTheSnowNormals
    UseSparkle
    LQSpark
    AllowSnowMicroNormals
    CliffMaterialOnly
    Material=@n3
    CliffLandscapeLayerMask
    IsLandscape=StaticBool()
    AreInputNormalsTangentSpace=StaticBool()
    UseInputNormalsForFakeSnowShadow=StaticBool() {value=true}
   ) {materialFunction=/Game/Athena/Helios/Environments/Landscape/Materials/QualityLevelLayers/MF_Helios_Snow_Default}.SnowMaterialOnly) {attributeGetTypes [#1]}
  n5 = Power(Base=ComponentMask(None=VertexNormalWS()) {b=true}) {constExponent=4}
  n6 = Multiply(A=Saturate(None=Multiply(A=@n5) {constB=2}), B=OneMinus(None=NamedRerouteUsage()))
  n7 = Constant()
  n8 = SetMaterialAttributes!(
   MaterialAttributes=NamedRerouteUsage()
   BaseColor=StaticSwitch(
    True=LinearInterpolate(A=@n2, B=@n4.BaseColor, Alpha=@n6)
    False=@n2
    Value=StaticBoolParameter() {desc="If set to True, set TopLayer to False", group=A_L4_Landscape, parameterName=TopSnow, sortPriority=-4}
   )
   Specular=NamedRerouteUsage()
   Roughness=NamedRerouteUsage()
   EmissiveColor=@n7
   Metallic=@n7
  ) {attributeSetTypes [
    #1
    #2
    #3
    #5
    #6
   ]}
  n9 = GetMaterialAttributes(None=NamedRerouteUsage()) {attributeGetTypes [#1]}
  n10 = VectorParameter() {group=E_RVT, parameterName=RVT Edge Color, defaultValue=<rgba 0.3524 0.3524 0.3524 0>}
  n11 = NamedRerouteUsage()
  n12 = ScalarParameter() {defaultValue=55, group=E_RVT, parameterName=RVT Edge Size}
  n13 = BlendMaterialAttributes(
   A=SetMaterialAttributes!(MaterialAttributes=@n9.MaterialAttributes, BaseColor=Multiply(A=@n9.BaseColor, B=@n10.RGB)) {attributeSetTypes [#1]}
   B=NamedRerouteUsage()
   Alpha=Saturate(None=Divide(A=@n11, B=@n12))
  ) {vertexAttributeBlendType=UseA}
  n14 = NamedRerouteUsage()
  n15 = MaterialFunctionCall(
   Transition Phase=NamedRerouteUsage()
   Height Texture=NamedRerouteUsage()
   Contrast=ScalarParameter() {group=E_RVT, parameterName=RVT_Contrast, sortPriority=5}
  ) {materialFunction=&HeightLerp}
  n16 = MaterialProxyReplace(
   Realtime=BlendMaterialAttributes(A=@n8, B=StaticSwitchParameter(True=@n13, False=@n14) {group=E_RVT, parameterName=RVT Edge, sortPriority=0}, Alpha=Saturate(None=@n15.Alpha)) {vertexAttributeBlendType=UseA}
   MaterialProxy=@n1
  )
  FunctionOutput(None=StaticSwitchParameter(
    True=MaterialFunctionCall(NoRVT=@n1, RVT=@n16) {materialFunction=&FN_RVT_Switch_Material}.Result
    False=NamedRerouteUsage()
   ) {group=E_RVT, parameterName=Use_RVT Blending, sortPriority=-1}) {sortPriority=-900}
  n17 = NamedRerouteUsage()
  n18 = Constant() {r=1}
  FunctionOutput(None=MaterialFunctionCall!(NoRVT=Constant() {r=1}, RVT=If(A=@n17, "A > B"=@n17, "A < B"=@n18)) {materialFunction=/Game/Packages/Fortress_SharedMaterials/Base_Material_Functions/FN_RVT_Switch_Scalar}.Result) {outputName=RVT H, sortPriority=2}
  n19 = WorldPosition() {worldPositionShaderOffset=WPT_ExcludeAllShaderOffsets}
  n20 = AppendVector(A=ComponentMask(None=VertexNormalWS()) {g=true, r=true}, B=Constant())
  n21 = Constant3Vector() {constant=<rgba 0 0 1 0>}
  n22 = Saturate(None=DotProduct(A=VertexNormalWS(), B=@n21))
  n23 = ScalarParameter() {defaultValue=4, group=E_RVT, parameterName=SidePower}
  n24 = Multiply(
   A=Multiply(
    A=MaterialFunctionCall(Vector=@n20, Default=@n21) {materialFunction=&SafeNormalize}.Result
    B=NamedRerouteUsage()
   )
   B=OneMinus(None=Power(Base=@n22, Exp=@n23))
  )
  n25 = ScalarParameter() {
   defaultValue=1
   desc="Blends how much influence the Inside Projection has on the RVT sampling, a good value around steep areas is ~0.2\r\n0 = No effect\r\n1 = Full effect"
   group=E_RVT
   parameterName=Inside Projection - Blend
   sortPriority=40
  }
  n26 = StaticSwitchParameter(True=LinearInterpolate(A=@n19.XYZ, B=Subtract(A=@n19.XYZ, B=@n24), Alpha=@n25), False=@n19.XYZ) {group=E_RVT, parameterName=UseInsideProjection}
  n27 = RuntimeVirtualTextureSampleParameter(World Position=@n26) {
   desc=Feedback enabled
   group=E_RVT
   materialType=BaseColor_Normal_Specular_YCoCg
   parameterName=RVT
   sortPriority=1
   virtualTexture=/Game/Athena/Apollo/Maps/Landscape/VT/RVT_Apollo_Landscape
  }
  n28 = RuntimeVirtualTextureSampleParameter!(UVs, World Position=@n26, Mip Level=Constant()) {
   bEnableFeedback=false
   desc=Use mip 0 but don't request feedback. Should give mip requested by landscape.
   group=E_RVT
   materialType=BaseColor_Normal_Specular_YCoCg
   mipValueMode=RVTMVM_MipLevel
   parameterName=RVT
   sortPriority=1
   virtualTexture=&RVT_Apollo_Landscape
  }
  n29 = StaticSwitchParameter(True=@n27.BaseColor, False=@n28.BaseColor) {group=E_RVT, parameterName=UseRVTFeedback}
  FunctionOutput(None=MaterialFunctionCall!(NoRVT=Constant() {r=1}, RVT=NamedRerouteDeclaration(None=@n29) {name=RVT Base Color, nodeColor=<rgba 1 1 0 1>}) {materialFunction=/Game/Packages/Fortress_SharedMaterials/Base_Material_Functions/FN_RVT_Switch_Color}.Result) {outputName=RVT Color}
  Multiply(A=@n19.XYZ)
  n30 = RuntimeVirtualTextureSampleParameter!(UVs, World Position) {
   group=E_RVT
   materialType=WorldHeight
   parameterName=RVTHeight
   sortPriority=1
   virtualTexture=/Game/Athena/Apollo/Maps/Landscape/VT/RVT_Apollo_Landscape_Height
  }
  NamedRerouteDeclaration(None=@n30.WorldHeight) {name=RVT Height as is, nodeColor=<rgba 1 1 0 1>}
  n31 = ScalarParameter() {
   defaultValue=10
   desc="Used to control how far colors will blend with the rvt, as well sets blending distance of normals for Up looking surfaces"
   group=E_RVT
   parameterName=Blend RVT Top
   sortPriority=2
  }
  n32 = ScalarParameter() {
   defaultValue=100
   desc="Used to control how far colors will blend with the rvt, as well sets blending distance of normals for Up looking surfaces"
   group=E_RVT
   parameterName=Blend RVT Base
   sortPriority=2
  }
  n33 = VertexNormalWS()
  n34 = Constant3Vector() {constant=<rgba 0 0 1 0>}
  n35 = OneMinus(None=Saturate(None=DotProduct(A=@n33, B=@n34)))
  n36 = LinearInterpolate(A=@n31, B=@n32, Alpha=@n35)
  n37 = VertexColor()
  n38 = LinearInterpolate(A=@n31, B=@n32, Alpha=@n37.G)
  n39 = Constant() {r=4}
  n40 = Abs(None=ComponentMask(None=VertexNormalWS()) {b=true})
  n41 = LinearInterpolate(A=Divide(A=@n38, B=@n39) {constB=4}, B=@n38, Alpha=@n40)
  n42 = Saturate(None=Divide(A=NamedRerouteUsage(), B=StaticSwitchParameter(True=@n36, False=@n41) {group=E_RVT, parameterName=UseDotTopSideLerp}))
  n43 = WorldPosition()
  n44 = ScalarParameter() {defaultValue=1000, group=E_RVT, parameterName=RVT_Blend_Noise_Scale}
  n45 = Multiply(
   A=TextureSampleParameter2D(UVs=Divide(A=@n43.XY, B=@n44)) {group=E_RVT, parameterName=RVT_EdgeNoise, texture=/Game/Effects/Fort_Effects/Textures/Tile/T_GaussianNoise_01_M}.R
   B=ScalarParameter() {defaultValue=0.7542, group=E_RVT, parameterName=ErodeEdgeNoise}
  )
  n46 = OneMinus(None=@n45)
  n47 = Divide(A=Subtract(A=OneMinus(None=@n42), B=@n45), B=@n46)
  NamedRerouteDeclaration(None=OneMinus(None=Saturate(None=@n47))) {name=RVT BlendMask, nodeColor=<rgba 0.9922 1 0 1>}
  NamedRerouteDeclaration(None=StaticSwitchParameter(True=@n27.Specular, False=@n28.Specular) {group=E_RVT, parameterName=UseRVTFeedback}) {name=RVT Specular, nodeColor=<rgba 1 1 0 1>}
  NamedRerouteDeclaration(None=StaticSwitchParameter(True=@n27.Roughness, False=@n28.Roughness) {group=E_RVT, parameterName=UseRVTFeedback}) {name=RVT Rough, nodeColor=<rgba 1 1 0 1>}
  n48 = WorldPosition()
  n49 = NamedRerouteUsage()
  NamedRerouteDeclaration(None=Subtract(A=@n48.Z, B=@n49)) {name=RVTHeightWithoutOffset, nodeColor=<rgba 1 1 0 1>}
  NamedRerouteDeclaration(None=StaticSwitchParameter(True=@n27.Normal, False=@n28.Normal) {group=E_RVT, parameterName=UseRVTFeedback}) {name=RVT N, nodeColor=<rgba 1 1 0 1>}
  NamedRerouteDeclaration(None=FunctionInput() {inputType=FunctionInput_MaterialAttributes, sortPriority=-1}) {name=InputMaterial, nodeColor=<rgba 0 0 1 1>}
  NamedRerouteDeclaration(None=@n37.R) {name=VertexColor-Red, nodeColor=<rgba 0.9062 0.9451 1 1>}
  n50 = FunctionInput() {bUsePreviewValueAsDefault=true, inputName="RVT Height Map (Roughness)", inputType=FunctionInput_Scalar, previewValue=<1 0 0 1>}
  NamedRerouteDeclaration(None=OneMinus(None=@n50)) {name=HeightmapForRVTBlend, nodeColor=<rgba 0 0 1 1>}
  n51 = ScalarParameter() {group=E_RVT, parameterName=Offset RVT WP, sortPriority=4}
  NamedRerouteDeclaration(None=Subtract(A=Add(A=@n51, B=@n48.Z), B=@n49)) {name=RVT Height, nodeColor=<rgba 1 1 0 1>}
  n52 = ScalarParameter() {
   defaultValue=10
   desc="Controls the blending of normals on sides of meshes, but not UP vector surfaces"
   group=E_RVT
   parameterName=Blend RVT Normals Top
   sortPriority=3
  }
  n53 = ScalarParameter() {
   defaultValue=10
   desc="Controls the blending of normals on sides of meshes, but not UP vector surfaces"
   group=E_RVT
   parameterName=Blend RVT Normals Base
   sortPriority=3
  }
  n54 = LinearInterpolate(A=@n52, B=@n53, Alpha=@n35)
  n55 = LinearInterpolate(A=@n52, B=@n53, Alpha=@n37)
  n56 = LinearInterpolate(A=Divide(A=@n55, B=@n39) {constB=4}, B=@n55, Alpha=@n40)
  n57 = Saturate(None=Divide(A=NamedRerouteUsage(), B=StaticSwitchParameter(True=@n54, False=@n56) {group=E_RVT, parameterName=UseDotTopSideLerp}))
  n58 = Divide(A=Subtract(A=OneMinus(None=@n57), B=@n45), B=@n46)
  NamedRerouteDeclaration(None=OneMinus(None=Saturate(None=@n58))) {name=RVT BlendMaskEroded, nodeColor=<rgba 0.9922 1 0 1>}
  n59 = NamedRerouteUsage()
  n60 = NamedRerouteUsage()
  NamedRerouteDeclaration(None=BlendMaterialAttributes(
    A=SetMaterialAttributes!(MaterialAttributes=@n59, Normal=@n60) {attributeSetTypes [#4]}
    B=@n59
    Alpha=NamedRerouteUsage()
   ) {vertexAttributeBlendType=UseA}) {name=RVTMaterialModifiedNormals, nodeColor=<rgba 1 1 0 1>}
  NamedRerouteUsage()
 ]}
