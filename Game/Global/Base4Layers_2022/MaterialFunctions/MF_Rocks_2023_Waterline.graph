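# Material function that adds a wet, tinted waterline to rock materials by world height.
# Below an offset height, adjustable per primitive, it blends toward a tinted variant with lower specular, altered roughness and world-Z noise normals, and outputs that mask.
# A sine-animated HeightLerp using opacity as height adds a darker, lower-roughness band.
# High and Epic add a shore fade against the RVT input; Low and Medium pass the input through.
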MaterialFunction /Game/Global/Base4Layers_2022/MaterialFunctions/MF_Rocks_2023_Waterline
graph {nodes [
  n1 = ScalarParameter() {
   bUseCustomPrimitiveData=true
   desc="Primitive Data Only, Enable per instance in the details panel."
   group=Waterline
   parameterName=PD_WaterlineHeightOffset
   sortPriority=0
  }
  n2 = Add(
   A=Add(A=@n1, B=ScalarParameter() {defaultValue=-155, group=Waterline, parameterName=Waterline Height offset, sortPriority=0})
   B=WorldPosition().Z
  )
  n3 = ScalarParameter() {defaultValue=1100, group=Waterline, parameterName=Waterline Height Blur, sortPriority=1}
  n4 = NamedRerouteDeclaration(None=FunctionInput() {inputName=WaterInput, inputType=FunctionInput_MaterialAttributes}) {name=Input, nodeColor=<rgba 1 0 0.1641 1>}
  n5 = GetMaterialAttributes(None=@n4) {attributeGetTypes [#1]}
  n6 = WorldPosition()
  n7 = Divide(A=@n6.Z, B=Constant() {r=256})
  n8 = TextureSample(UVs=Divide(A=@n6.Z, B=Constant() {r=2000})) {samplerSource=SSM_Wrap_WorldGroupSettings, samplerType=SAMPLERTYPE_Normal, texture=/Game/Global/Textures/Default/Noise/T_Noise_Normal_01}
  n9 = Add(
   A=Multiply(
    A=TextureSample(UVs=@n7) {samplerType=SAMPLERTYPE_Normal, texture=&T_Noise_Normal_01}.RGB
    B=Constant3Vector() {constant=<rgba 0 0.25 1 0>}
   )
   B=Multiply(A=@n8.RGB, B=Constant3Vector() {constant=<rgba 0 0.5 0 0>})
  )
  n10 = GetMaterialAttributes(None=SetMaterialAttributes!(
    MaterialAttributes=@n5.MaterialAttributes
    Normal=MaterialFunctionCall(BaseNormal=@n5.Normal, AdditionalNormal=@n9) {materialFunction=&BlendAngleCorrectedNormals}.Result
   ) {attributeSetTypes [#1]}) {attributeGetTypes [
    #2
    #3
    #1
    #4
    #5
   ]}
  n11 = VectorParameter() {
   group=Waterline
   parameterName=Rocks Shores Tone
   sortPriority=-1
   channelNames {a=S/R Tweak, b="", g="", r=""}
   defaultValue=<rgba 0.1858 0.1638 0.1377 0.65>
  }
  n12 = DotProduct(A=VertexNormalWS(), B=Constant3Vector() {constant=<rgba 0 0 1 0>})
  n13 = Multiply(A=@n10.BaseColor, B=LinearInterpolate(A=@n11, B=Multiply(A=@n11) {constB=1.5}, Alpha=Saturate(None=@n12)))
  n14 = Multiply(A=@n10.BaseColor, B=@n11)
  n15 = ScalarParameter() {desc=Blends BC3vector Color with input color, group=Waterline, parameterName=Water Base Color Simple, sliderMax=1}
  n16 = GetMaterialAttributes(None=@n4) {attributeGetTypes [#1]}
  n17 = MaterialFunctionCall(In=ComponentMask(None=Transform(None=@n16.Normal)) {b=true}, Contrast=Constant() {r=0.5}) {materialFunction=&CheapContrast}
  n18 = SetMaterialAttributes!(
   MaterialAttributes=@n10.MaterialAttributes
   BaseColor=LinearInterpolate(
    A=LinearInterpolate(A=@n13, B=@n14, Alpha=@n15)
    B=@n10.BaseColor
    Alpha=ScalarParameter() {desc=Blends final color logic with input BC, group=Waterline, parameterName=Water Base Color BlendAmount}
   )
   Specular=Multiply(A=@n10.Specular) {constB=0.25}
   Normal=LinearInterpolate(A=@n10.Normal, B=@n16.Normal, Alpha=Saturate(None=@n17.Result))
   Roughness=ComponentMask(None=Multiply(A=@n10.Roughness, B=@n11.S/R Tweak) {constB=0.5}) {r=true}
  ) {attributeSetTypes [
    #2
    #3
    #1
    #4
   ]}
  n19 = BlendMaterialAttributes(A=@n18, B=NamedRerouteUsage(), Alpha=FunctionOutput(None=Divide(A=@n2, B=@n3)) {outputName=Waterline Height Mask, sortPriority=100})
  n20 = GetMaterialAttributes(None=@n19) {attributeGetTypes [#4, #2]}
  n21 = GetMaterialAttributes(None=@n18) {attributeGetTypes [#5]}
  n22 = Multiply(A=Sine(None=Time()) {period=4}, B=Add(A=@n21.Opacity, B=ScalarParameter() {group=Waterline, parameterName=Time, sortPriority=3}))
  n23 = MaterialFunctionCall(Transition Phase=Divide(A=Add(A=@n22, B=@n2), B=Multiply(A=@n3) {constB=0.2}), Height Texture=@n21.Opacity) {materialFunction=&HeightLerp}
  n24 = Saturate(None=@n23.Lerp Alpha No Contrast)
  n25 = ScalarParameter() {group=Waterline, parameterName=Water Height Lerp SaturateAmount, sliderMax=1}
  n26 = BlendMaterialAttributes(
   A=SetMaterialAttributes!(
    MaterialAttributes=@n20.MaterialAttributes
    Roughness=Multiply(A=@n20.Roughness) {constB=0.5}
    BaseColor=Multiply(A=@n20.BaseColor) {constB=0.5}
   ) {attributeSetTypes [#4, #2]}
   B=@n19
   Alpha=Multiply(
    A=LinearInterpolate(A=@n23.Lerp Alpha No Contrast, B=@n24, Alpha=@n25)
    B=ScalarParameter() {defaultValue=1, group=Waterline, parameterName=Water Height Lerp Strength, sliderMax=1}
   )
  )
  n27 = NamedRerouteUsage()
  n28 = FunctionInput() {bUsePreviewValueAsDefault=true, inputName=RVT, inputType=FunctionInput_Scalar, previewValue=<1 0 0 1>}
  n29 = Subtract(
   A=Add(A=@n2, B=ScalarParameter() {defaultValue=250, group=Waterline, parameterName=Water Shores Height, sortPriority=4}) {constB=-15}
   B=@n28
  )
  n30 = ScalarParameter() {defaultValue=255, group=Waterline, parameterName=Shores Coastlal Fade, sortPriority=5}
  n31 = BlendMaterialAttributes(A=@n26, B=@n27, Alpha=Saturate(None=Divide(A=@n29, B=@n30) {constB=512})) {vertexAttributeBlendType=UseB}
  FunctionOutput(None=QualitySwitch(Default=@n26, Low=@n27, High=@n31, Medium=@n27, Epic=@n31)) {outputName=BlendMaterialAttributes}
 ]}
