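# Material function that builds UVs for four SFX layers and samples a mask texture, defaulting to T_SFX_M.
# Each layer takes a shared UV channel and uses static switches to pick a FlipBook trim cell or scaled, offset, optionally time-panned UVs.
# MF_LerpTexturesBasedOnInt_Float3_4Layers merges the four UV sets before a single texture sample.
# It outputs per-layer UVs, merged UVs and the sampled RGB.
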MaterialFunction /Game/Global/Base4Layers_2022/MaterialFunctions/MF_SFXTexture
graph {nodes [
  n1 = ScalarParameter() {defaultValue=1, group=1_Layer SFX, parameterName="SFX L1 Trim ##", sortPriority=2}
  n2 = ScalarParameter() {defaultValue=1, group=1_Layer SFX, parameterName=SFX L1 Trim Rows Count, sliderMin=1, sortPriority=3}
  n3 = ScalarParameter() {defaultValue=1, group=1_Layer SFX, parameterName=SFX L1 Trim Column Count, sortPriority=1}
  n4 = StaticSwitchParameter(
   True=StaticSwitchParameter(True=TextureCoordinate() {coordinateIndex=2}, False=TextureCoordinate() {coordinateIndex=1}) {group=0_SFX Global, parameterName=SFX UV3}
   False=TextureCoordinate()
  ) {group=0_SFX Global, parameterName=SFX UV2}
  n5 = MaterialFunctionCall(
   "Animation  Phase (0-1)"=Divide(A=Add(A=@n1) {constB=-1}, B=Multiply(A=@n2, B=@n3)) {constB=4}
   Number of Rows=@n2
   Number of Columns=@n3
   UVs=@n4
  ) {materialFunction=&FlipBook}
  n6 = Time()
  n7 = AppendVector(
   A=ScalarParameter() {group=1_Layer SFX, parameterName=SFX Offset X, sortPriority=2}
   B=ScalarParameter() {group=1_Layer SFX, parameterName=SFX Offset Y, sortPriority=2}
  )
  n8 = Add(
   A=Multiply(A=@n4, B=ScalarParameter() {defaultValue=1, group=1_Layer SFX, parameterName=SFX Scale, sortPriority=1})
   B=StaticSwitchParameter(True=Multiply(A=@n6, B=@n7), False=@n7) {group=1_Layer SFX, parameterName=SFX Panner}
  )
  n9 = Constant()
  n10 = AppendVector(
   A=FunctionOutput(None=StaticSwitchParameter(True=@n5.UVs, False=@n8) {group=1_Layer SFX, parameterName=Use Trim for SFX}) {outputName=UVs SFX, sortPriority=5}
   B=@n9
  )
  n11 = ScalarParameter() {defaultValue=1, group=2_Layer SFX, parameterName="SFX L2 Trim ##", sortPriority=3}
  n12 = ScalarParameter() {defaultValue=1, group=2_Layer SFX, parameterName=SFX L2 Trim Rows Count, sliderMin=1, sortPriority=3}
  n13 = ScalarParameter() {defaultValue=1, group=2_Layer SFX, parameterName=SFX L2 Trim Column Count, sortPriority=1}
  n14 = MaterialFunctionCall(
   "Animation  Phase (0-1)"=Divide(A=Add(A=@n11) {constB=-1}, B=Multiply(A=@n12, B=@n13)) {constB=4}
   Number of Rows=@n12
   Number of Columns=@n13
   UVs=@n4
  ) {materialFunction=&FlipBook}
  n15 = Time()
  n16 = AppendVector(
   A=ScalarParameter() {group=2_Layer SFX, parameterName=SFX Offset X 2, sortPriority=2}
   B=ScalarParameter() {group=2_Layer SFX, parameterName=SFX Offset Y 2, sortPriority=3}
  )
  n17 = Add(
   A=Multiply(A=@n4, B=ScalarParameter() {defaultValue=1, group=2_Layer SFX, parameterName=SFX Scale 2, sortPriority=1})
   B=StaticSwitchParameter(True=Multiply(A=@n15, B=@n16), False=@n16) {group=2_Layer SFX, parameterName=SFX Panner L2}
  )
  n18 = Constant()
  n19 = AppendVector(
   A=FunctionOutput(None=StaticSwitchParameter(True=@n14.UVs, False=@n17) {group=2_Layer SFX, parameterName=L2 Trim for SFX}) {outputName=UVs SFX 2, sortPriority=6}
   B=@n18
  )
  n20 = Add(A=ScalarParameter() {defaultValue=1, group=3_Layer SFX, parameterName="SFX L3 Trim ##", sortPriority=4}) {constB=-1}
  n21 = ScalarParameter() {defaultValue=1, group=3_Layer SFX, parameterName=SFX L3 Trim Rows Count, sliderMin=1, sortPriority=3}
  n22 = ScalarParameter() {defaultValue=1, group=3_Layer SFX, parameterName=SFX L3 Trim Column Count, sortPriority=1}
  n23 = MaterialFunctionCall(
   "Animation  Phase (0-1)"=Divide(A=Ceil(None=@n20), B=Multiply(A=@n21, B=@n22)) {constB=4}
   Number of Rows=@n21
   Number of Columns=@n22
   UVs=@n4
  ) {materialFunction=&FlipBook}
  n24 = Time()
  n25 = AppendVector(
   A=ScalarParameter() {group=3_Layer SFX, parameterName=SFX Offset X 3, sortPriority=2}
   B=ScalarParameter() {group=3_Layer SFX, parameterName=SFX Offset Y 3, sortPriority=3}
  )
  n26 = Add(
   A=Multiply(A=@n4, B=ScalarParameter() {defaultValue=1, group=3_Layer SFX, parameterName=SFX Scale 3, sortPriority=1})
   B=StaticSwitchParameter(True=Multiply(A=@n24, B=@n25), False=@n25) {group=3_Layer SFX, parameterName=SFX Panner L3}
  )
  n27 = Constant()
  n28 = AppendVector(
   A=FunctionOutput(None=StaticSwitchParameter(True=@n23.UVs, False=@n26) {group=3_Layer SFX, parameterName=L3 Trim for SFX}) {outputName=UVs SFX 3, sortPriority=7}
   B=@n27
  )
  n29 = Add(A=ScalarParameter() {defaultValue=1, group=4_Layer SFX, parameterName="SFX L4 Trim ##", sortPriority=4}) {constB=-1}
  n30 = ScalarParameter() {defaultValue=1, group=4_Layer SFX, parameterName=SFX L4 Trim Rows Count, sliderMin=1, sortPriority=3}
  n31 = ScalarParameter() {defaultValue=1, group=4_Layer SFX, parameterName=SFX L4 Trim Column Count, sortPriority=1}
  n32 = MaterialFunctionCall(
   "Animation  Phase (0-1)"=Divide(A=Ceil(None=@n29), B=Multiply(A=@n30, B=@n31)) {constB=4}
   Number of Rows=@n30
   Number of Columns=@n31
   UVs=@n4
  ) {materialFunction=&FlipBook}
  n33 = Time()
  n34 = AppendVector(
   A=ScalarParameter() {group=4_Layer SFX, parameterName=SFX Offset X 4, sortPriority=2}
   B=ScalarParameter() {group=4_Layer SFX, parameterName=SFX Offset Y 4, sortPriority=3}
  )
  n35 = Add(
   A=Multiply(A=@n4, B=ScalarParameter() {defaultValue=1, group=4_Layer SFX, parameterName=SFX Scale 4, sortPriority=1})
   B=StaticSwitchParameter(True=Multiply(A=@n33, B=@n34), False=@n34) {group=4_Layer SFX, parameterName=SFX Panner L4}
  )
  n36 = Constant()
  n37 = AppendVector(
   A=FunctionOutput(None=StaticSwitchParameter(True=@n32.UVs, False=@n35) {group=4_Layer SFX, parameterName=L4 Trim for SFX}) {outputName=UVs SFX 4, sortPriority=8}
   B=@n36
  )
  n38 = StaticBool() {value=true}
  n39 = ComponentMask(None=MaterialFunctionCall!(
    transitionFactor
    1=AppendVector(A=@n10, B=@n9)
    2=AppendVector(A=@n19, B=@n18)
    3=AppendVector(A=@n28, B=@n27)
    4=AppendVector(A=@n37, B=@n36)
    Use 2 Textures=@n38
    Use 3 Textures=@n38
    Use 4 Textures=@n38
   ) {materialFunction=/Game/Athena/WinterArt/Materials/MF_LerpTexturesBasedOnInt_Float3_4Layers}.Result) {g=true, r=true}
  n40 = TextureObject() {samplerType=SAMPLERTYPE_Masks, texture=/Game/Global/Base4Layers_2022/Textures/T_SFX_M}
  FunctionOutput(None=TextureSample(UVs=FunctionOutput(None=@n39) {outputName=SFX UVs 4 Layers}, Tex=FunctionInput(Preview=@n40) {inputType=FunctionInput_Texture2D}) {samplerType=SAMPLERTYPE_Masks}.RGB) {outputName=SFX}
 ]}
