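# Material function that optionally restyles incoming material attributes as snow-and-ice subsurface architecture, gated by a static switch.
# Base color blends tints by normal Z, with distance-field AO on Epic and a fake screen-space sphere and fresnel on mobile/low, optionally multiplying desaturated diffuse.
# Biplanar textures drive screen-masked emissive sparkle, replaced by flat color on mobile/low.
# Detail normals blend into the input; opacity follows the global ShadowsEnabled parameter.
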MaterialFunction /Game/Global/Base4Layers_2022/MaterialFunctions/MF_SnowAndIceArchitecture
graph {nodes [
  Constant()
  n1 = TextureSample(UVs=Multiply(A=ScreenPosition().ViewportUV, B=ScalarParameter() {defaultValue=1, group=Snow and Ice Architecture, parameterName=SaIA_ScreenMask_Sparkle})) {samplerType=SAMPLERTYPE_Masks, texture=/Game/Global/Base4Layers_2022/Textures/SnowandIceArch_M}
  Divide(A=Subtract(A=@n1.G) {constB=0.5})
  Fresnel()
  n2 = FunctionInput() {inputType=FunctionInput_MaterialAttributes}
  n3 = GetMaterialAttributes(None=@n2) {attributeGetTypes [
    #1
    #2
    #3
    #4
   ]}
  n4 = Desaturation(None=@n3.BaseColor, Fraction=ScalarParameter() {defaultValue=1, group=Snow and Ice Architecture, parameterName=SaIA_Color_Desat})
  n5 = VectorParameter() {group=Snow and Ice Architecture, parameterName=SaIA_Color_Up, defaultValue=<rgba 0.5365 0.5365 0.5365 1>}
  n6 = Abs(None=ComponentMask(None=VertexNormalWS()) {b=true})
  n7 = LinearInterpolate(
   A=VectorParameter() {group=Snow and Ice Architecture, parameterName=SaIA_Color, defaultValue=<rgba 0.1809 0.2802 0.2969 1>}.RGB
   B=@n5.RGB
   Alpha=@n6
  )
  n8 = VectorParameter() {group=Snow and Ice Architecture, parameterName=SaIA_Color_AO, defaultValue=<rgba 0.3043 0.3244 0.3698 1>}
  n9 = LinearInterpolate(
   A=@n8.RGB
   B=@n7
   Alpha=DistanceFieldApproxAO!(
    World Position=WorldPosition().XYZ
    Normal=VertexNormalWS()
    BaseDistance=ScalarParameter() {defaultValue=500, group=Snow and Ice Architecture, parameterName=SaIA_SSS_Distance_AO}
    Radius=ScalarParameter() {defaultValue=50, group=Snow and Ice Architecture, parameterName=SaIA_SSS_Radius_AO}
   ) {numSteps=4}
  )
  n10 = VectorParameter() {group=Snow and Ice Architecture, parameterName=SaIA_Color_Sphere_Mobile, defaultValue=<rgba 0.4147 0.4436 0.5104 1>}
  n11 = Fresnel(ExponentIn=ScalarParameter() {defaultValue=2, group=Snow and Ice Architecture, parameterName=SaIA_Fresnell}, Normal=PixelNormalWS()) {baseReflectFraction=0}
  n12 = Divide(
   A=Distance(A=ScreenPosition().ViewportUV, B=Constant2Vector() {g=0.5, r=0.5})
   B=LinearInterpolate(
    A=ScalarParameter() {defaultValue=0.5554, group=Snow and Ice Architecture, parameterName=SaIA_MobileFakeSphereMask_1}
    B=ScalarParameter() {defaultValue=0.2622, group=Snow and Ice Architecture, parameterName=SaIA_MobileFakeSphereMask_2}
    Alpha=@n11
   )
  )
  n13 = VectorParameter() {group=Snow and Ice Architecture, parameterName=SaIA_Color_Mobile, defaultValue=<rgba 0.4656 0.6357 0.7708 1>}
  n14 = VectorParameter() {group=Snow and Ice Architecture, parameterName=SaIA_Color_Up_Mobile, defaultValue=<rgba 0.2199 0.3134 0.4531 1>}
  n15 = VertexNormalWS()
  n16 = Multiply(A=Abs(None=ComponentMask(None=@n15) {b=true}), B=@n6)
  n17 = VectorParameter() {group=Snow and Ice Architecture, parameterName=SaIA_Color_Fresnel_Mobile, defaultValue=<rgba 0.3343 0.5004 0.7552 1>}
  n18 = Multiply(
   A=LinearInterpolate(B=@n10.RGB, Alpha=Saturate(None=@n12)) {constA=1}
   B=LinearInterpolate(A=LinearInterpolate(A=@n13.RGB, B=@n14.RGB, Alpha=@n16), B=@n17.RGB, Alpha=@n11)
  )
  n19 = QualitySwitch(Default=ShadingPathSwitch(Default=QualitySwitch(Default=@n7, Epic=@n9), Mobile=@n18) {desc=Migrated from FeatureLevelSwitch}, Low=@n18)
  n20 = ScalarParameter() {defaultValue=7.1226, group=Snow and Ice Architecture, parameterName=SaIA_SSS_Roughness}
  n21 = ScalarParameter() {defaultValue=0.4, group=Snow and Ice Architecture, parameterName=SaIA_SSS_Roughness2}
  n22 = TextureSample() {samplerType=SAMPLERTYPE_Normal, texture=/Game/Global/Base4Layers_2022/Textures/SnowandIceArch_N}
  n23 = ComponentMask(None=@n22.RGB) {b=true}
  n24 = Saturate(None=Divide(A=Add(A=@n23)))
  n25 = VectorParameter() {group=Snow and Ice Architecture, parameterName=SaIA_SparkleColor, defaultValue=<rgba 2.2812 4.2433 5 1>}
  n26 = MaterialFunctionCall(Float3=MaterialFunctionCall!(
    ColorTex=TextureObject() {samplerType=SAMPLERTYPE_Masks, texture=&SnowandIceArch_M}
    NormalTex=TextureObject() {samplerType=SAMPLERTYPE_Normal, texture=&SnowandIceArch_N}
    Pos
    Nrm
    ColorUVScale=ScalarParameter() {defaultValue=512, group=Snow and Ice Architecture, parameterName=SaIA_SparkleScale}
    ColorBlendFactor
    NormalStrength
    NormalUVScale=ScalarParameter() {defaultValue=512, group=Snow and Ice Architecture, parameterName=SaIA_NormalScale}
    NormalBlendFactor
    Local Space Inputs
   ) {materialFunction=/Game/Global/MaterialFunctions/MF_BiPlanar}.Color) {materialFunction=&BreakOutFloat3Components}
  n27 = OneMinus(None=@n1.G)
  n28 = Multiply(A=@n25.RGB, B=Add(A=Multiply(A=@n1.G, B=@n26.R), B=Multiply(A=@n27, B=@n26.B)))
  n29 = VectorParameter() {group=Snow and Ice Architecture, parameterName=SaIA_Color_Emissive_Mobile, defaultValue=<rgba 0.1384 0.1993 0.3281 1>}
  n30 = CollectionParameter() {collection=&FortniteMaterialParameters, parameterName=ShadowsEnabled}
  n31 = VectorParameter() {group=Snow and Ice Architecture, parameterName=SaIA_SSS_Color, defaultValue=<rgba 0 0.1739 0.3229 1>}
  n32 = ScalarParameter() {defaultValue=0.5, group=Snow and Ice Architecture, parameterName=SaIA_Normal_Flatten_Mobile}
  n33 = QualitySwitch(
   Default=ShadingPathSwitch(Default=ScalarParameter() {group=Snow and Ice Architecture, parameterName=SaIA_Normal_Flatten}, Mobile=@n32) {desc=Migrated from FeatureLevelSwitch}
   Low=@n32
  )
  n34 = SetMaterialAttributes!(
   MaterialAttributes=@n3.MaterialAttributes
   BaseColor=StaticSwitchParameter(True=Multiply(A=@n19, B=@n4), False=@n19) {defaultValue=true, group=Snow and Ice Architecture, parameterName=SnowAndIceArchtecture_UseDiffuse}
   Metallic=Constant()
   Specular=ScalarParameter() {defaultValue=0.5, group=Snow and Ice Architecture, parameterName=SaIA_SSS_Specular}
   Roughness=Saturate(None=LinearInterpolate(A=@n20, B=@n21, Alpha=@n24))
   EmissiveColor=QualitySwitch(Default=ShadingPathSwitch(Default=@n28, Mobile=@n29.RGB) {desc=Migrated from FeatureLevelSwitch}, Low=@n29.RGB)
   Opacity=LinearInterpolate(
    A=ScalarParameter() {defaultValue=0.95, group=Snow and Ice Architecture, parameterName=SaIA_SSS_Opacity_Disabled}
    B=ScalarParameter() {defaultValue=0.6, group=Snow and Ice Architecture, parameterName=SaIA_SSS_Opacity}
    Alpha=ComponentMask(None=@n30) {r=true}
   )
   SubsurfaceColor=StaticSwitchParameter(True=Multiply(A=@n4, B=@n31.RGB), False=@n31.RGB) {defaultValue=true, group=Snow and Ice Architecture, parameterName=SnowAndIceArchtecture_UseDiffuse}
   Normal=MaterialFunctionCall(
    BaseNormal=@n3.Normal
    AdditionalNormal=MaterialFunctionCall(Normal=@n22.RGB, Flatness=@n33) {materialFunction=&FlattenNormal}.Result
   ) {materialFunction=&BlendAngleCorrectedNormals}.Result
   ShadingModel=ShadingModel() {shadingModel=MSM_Subsurface}
  ) {attributeSetTypes [
    #1
    #2
    #3
    #5
    #6
    #7
    #8
    #4
    #9
   ]}
  FunctionOutput(None=StaticSwitchParameter(True=@n34, False=@n2) {parameterName=Use Snow and Ice Architecture})
  LinearInterpolate(A=Constant3Vector() {constant=<rgba 0.285 0.4009 0.5 1>}, B=Constant3Vector() {constant=<rgba 0.2 0.3078 0.4 1>})
  PixelNormalWS()
 ]}
