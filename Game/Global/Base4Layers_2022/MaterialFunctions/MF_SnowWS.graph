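# Material function that layers world-space snow onto incoming material attributes for four-layer environment materials, with optional waterline and snow leak effects.
# HeightLerp blends snow using roughness, world-aligned noise, coverage and normal-based floor, wall and ceiling controls.
# Distance-field AO, per-layer UV2 values and time-of-day sun shadow tinting provide further masks and color.
# Material proxy replacements bypass local-space effects for HLOD, and a quality switch separates lower platforms.
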MaterialFunction /Game/Global/Base4Layers_2022/MaterialFunctions/MF_SnowWS
graph {nodes [
  ComponentMask(None=Transform(None=VertexNormalWS()) {transformSourceType=TRANSFORMSOURCE_World, transformType=TRANSFORM_Local}) {g=true}
  n1 = WorldPosition() {worldPositionShaderOffset=WPT_ExcludeAllShaderOffsets}
  n2 = Abs(None=Constant() {r=7500})
  ComponentMask(None=Divide(A=ComponentMask(None=@n1.XYZ) {g=true, r=true}, B=Multiply(A=@n2) {constB=-1})) {b=true, g=true, r=true}
  n3 = GetMaterialAttributes(None=NamedRerouteUsage()) {attributeGetTypes [#1]}
  Desaturation(None=@n3.Specular, Fraction=Constant() {r=1})
  Divide(A=WorldPosition().XYZ) {constB=2500}
  n4 = NamedRerouteUsage()
  n5 = StaticSwitchParameter(True=NamedRerouteUsage(), False=@n4) {
   desc="Enables World Space detail texture from your SFX Functions texture. Select channel. Default scale is 1 = 128  You can enable DFAO mask inside this function"
   group=C_Snow
   parameterName=Use Snow
   sortPriority=-3
  }
  n6 = StaticSwitchParameter(True=@n4, False=@n5) {
   desc="Enables World Space detail texture from your SFX Functions texture. Select channel. Default scale is 1 = 128  You can enable DFAO mask inside this function"
   group=C_Dirt
   parameterName=Use Dirt
   sortPriority=-3
  }
  FunctionOutput(None=@n6)
  n7 = MaterialFunctionCall(
   TextureObject=TextureObject() {samplerType=SAMPLERTYPE_Normal, texture=/Game/Athena/Environments/Landscape/Textures/T_Athena_Snow_01_N}
   TextureSize=Constant() {r=7500}
   WorldPosition=NamedRerouteUsage()
  ) {materialFunction=/Game/Weapons/WeaponSkins/Wraps/MaterialFunctions/WorldAlignedTexture}
  n8 = MaterialFunctionCall(Float3=@n7.XYZ Texture) {materialFunction=&BreakOutFloat3Components}
  n9 = NamedRerouteUsage()
  n10 = ScalarParameter() {group=C_Snow, parameterName=Snow Normal Soften, sortPriority=16}
  n11 = MaterialFunctionCall(
   Normal=GetMaterialAttributes(None=NamedRerouteUsage()) {attributeGetTypes [#2]}.Normal
   Flatness=Saturate(None=Multiply(A=@n9, B=@n10))
  ) {materialFunction=&FlattenNormal}
  n12 = MaterialFunctionCall(Float3=@n11.Result) {materialFunction=&BreakOutFloat3Components}
  MaterialFunctionCall(X=Add(A=@n8.R, B=@n12.R), Y=Add(A=@n8.G, B=@n12.G), Z=Multiply(A=@n8.B, B=@n12.B)) {materialFunction=&MakeFloat3}
  n13 = ComponentMask(None=VertexNormalWS()) {b=true}
  n14 = OneMinus(None=SphereMask(A=Saturate(None=@n13), B=Constant() {r=1}) {attenuationRadius=1, hardnessPercent=60})
  Multiply(A=@n14)
  NamedRerouteDeclaration(None=FunctionInput() {inputName=Mat In, inputType=FunctionInput_MaterialAttributes}) {name=Mat Input, nodeColor=<rgba 1 0 0.6328 1>}
  n15 = ScalarParameter() {group=C_Snow_Waterline, parameterName=Waterline Height}
  n16 = ComponentMask(None=Add(A=WorldPosition().XYZ, B=OneMinus(None=@n15))) {b=true}
  n17 = ScalarParameter() {defaultValue=1500, group=C_Snow_Waterline, parameterName=Waterline Blur}
  NamedRerouteDeclaration(None=Saturate(None=Divide(A=@n16, B=@n17))) {name=Water Height, nodeColor=<rgba 0.9922 1 0 1>}
  n18 = Multiply(A=@n17) {constB=0.1}
  NamedRerouteDeclaration(None=Saturate(None=Divide(A=@n16, B=@n18) {constB=15})) {name=Waterline Contact, nodeColor=<rgba 0.3828 0 1 1>}
  n19 = ComponentMask(None=WorldPosition().XYZ) {b=true}
  n20 = ScalarParameter() {defaultValue=-9618.5, group=C_Snow_Leak, parameterName=Snow Leak Height}
  n21 = ScalarParameter() {defaultValue=2500, group=C_Snow_Leak, parameterName=Snow Leak Fade down}
  NamedRerouteDeclaration(None=Divide(A=Add(A=@n19, B=@n20), B=OneMinus(None=@n21))) {name=Leak Location, nodeColor=<rgba 0 1 0.4609 1>}
  n22 = VectorParameter() {
   group=C_Dirt_Waterline
   parameterName=Waterline Bottom
   channelNames {a=Water Transparency, b="", g="", r=""}
   defaultValue=<rgba 0.3099 0.3142 0.0182 0.65>
  }
  n23 = Multiply(A=@n22) {constB=0.5}
  n24 = GetMaterialAttributes(None=NamedRerouteUsage()) {attributeGetTypes [#3, #1, #4]}
  n25 = NamedRerouteUsage()
  n26 = SetMaterialAttributes!(
   MaterialAttributes=@n25
   Specular=Constant() {r=0.05}
   BaseColor=LinearInterpolate(A=@n23, B=Multiply(A=@n23, B=@n24.BaseColor)) {constAlpha=0.9}
   Roughness=Constant()
  ) {
   desc=Waterline
   attributeSetTypes [#1, #4, #3]
  }
  n27 = NamedRerouteUsage()
  n28 = LinearInterpolate(A=Multiply(A=@n22) {constB=1.5}, Alpha=@n27)
  n29 = Constant() {r=0.1}
  n30 = Constant() {r=0.15}
  n31 = BlendMaterialAttributes(
   A=@n25
   B=SetMaterialAttributes!(
    MaterialAttributes=@n25
    BaseColor=Multiply(A=@n28, B=@n24.BaseColor)
    Specular=LinearInterpolate(A=@n29, B=@n24.Specular, Alpha=@n27)
    Roughness=LinearInterpolate(A=@n30, B=@n24.Roughness, Alpha=@n27)
   ) {attributeSetTypes [#4, #1, #3]}
   Alpha=@n22.Water Transparency
  )
  n32 = NamedRerouteUsage()
  n33 = NamedRerouteUsage()
  n34 = OneMinus(None=@n24.Roughness)
  n35 = ScalarParameter() {defaultValue=0.5, group=C_Snow_Waterline, parameterName=Waterline Contrast}
  n36 = StaticSwitchParameter(
   True=BlendMaterialAttributes(
    A=BlendMaterialAttributes(A=@n26, B=@n31, Alpha=@n32)
    B=@n25
    Alpha=MaterialFunctionCall(Transition Phase=@n33, Height Texture=@n34, Contrast=@n35) {materialFunction=&HeightLerp}.Alpha
   )
   False=@n25
  ) {
   desc="Enables World Space detail texture from your SFX Functions texture. Select channel. Default scale is 1 = 128  You can enable DFAO mask inside this function"
   group=C_Snow
   parameterName=Use Waterline
   sortPriority=-3
  }
  NamedRerouteDeclaration(None=@n36) {name=After Water, nodeColor=<rgba 0 0.25 1 1>}
  n37 = NamedRerouteUsage()
  n38 = ScalarParameter() {
   desc="Uses Base color as specular replacement for materials using snow\r\nUpdate: Changed this to use the specular instead, so the specular texture stays accurate"
   group=C_Snow_AdvancedControls
   parameterName=Snow Custom Spec Min
   sortPriority=17
  }
  n39 = ScalarParameter() {
   defaultValue=1
   desc="Uses Base color as specular replacement for materials using snow\r\nUpdate: Changed this to use the specular instead, so the specular texture stays accurate"
   group=C_Snow_AdvancedControls
   parameterName=Snow Custom Spec Max
   sortPriority=16
  }
  n40 = StaticSwitchParameter(True=LinearInterpolate(A=@n38, B=@n39, Alpha=@n3.Specular), False=@n3.Specular) {group=C_Snow_AdvancedControls, parameterName=Snow Spec Edit, sortPriority=15}
  n41 = GetMaterialAttributes(None=NamedRerouteUsage()) {attributeGetTypes [#1, #3]}
  n42 = NamedRerouteUsage()
  n43 = ScalarParameter() {defaultValue=1, group=C_Snow_Layer, parameterName=Snow R Min, sortPriority=4}
  n44 = ScalarParameter() {defaultValue=1, group=C_Snow_Layer, parameterName=Snow R Max, sortPriority=5}
  n45 = SetMaterialAttributes!(
   MaterialAttributes=@n42
   BaseColor=NamedRerouteUsage()
   Roughness=StaticSwitchParameter(
    True=LinearInterpolate(A=@n43, B=@n44, Alpha=@n41.Roughness)
    False=ScalarParameter() {defaultValue=0.75, group=C_Snow_Layer, parameterName=Snow R, sortPriority=3}
   ) {desc=Used base material roughness as the base., group=C_Snow_Layer, parameterName=Snow Roughness Edit, sortPriority=3}
   Specular=Reroute()
   Metallic
  ) {attributeSetTypes [
    #4
    #3
    #1
    #5
   ]}
  n46 = NamedRerouteUsage()
  n47 = Constant() {r=0.7}
  n48 = NamedRerouteUsage()
  n49 = SetMaterialAttributes!(
   MaterialAttributes=@n42
   BaseColor=NamedRerouteDeclaration(None=LinearInterpolate(A=@n46, B=@n47, Alpha=@n48)) {name=PostRVT_Basecolor, nodeColor=<rgba 1 0 1 1>}
   Roughness=Constant() {r=0.9}
   Specular=Constant() {r=0.5}
   Metallic
   Normal=NamedRerouteUsage()
  ) {attributeSetTypes [
    #4
    #3
    #1
    #5
    #2
   ]}
  n50 = NamedRerouteUsage()
  n51 = OneMinus(None=@n41.Specular)
  n52 = OneMinus(None=Multiply(A=NamedRerouteUsage(), B=Constant() {r=0.5}) {constB=0.5})
  n53 = BlendMaterialAttributes(
   A=@n41.MaterialAttributes
   B=StaticSwitchParameter(True=@n45, False=@n49) {group=C_Snow_Layer, parameterName=Snow Layer Edit, sortPriority=55}
   Alpha=MaterialFunctionCall(Transition Phase=@n50, Height Texture=Multiply(A=@n51, B=@n52), Contrast=Multiply(A=@n50) {constB=0.5}) {materialFunction=&HeightLerp}.Alpha
  )
  n54 = GetMaterialAttributes(None=@n53) {attributeGetTypes [#4]}
  n55 = ScalarParameter() {defaultValue=3, group=C_Snow_Layer, parameterName=Snow Fresnel Exponent}
  n56 = Transform(None=NamedRerouteUsage())
  n57 = Add(A=@n54.BaseColor, B=Multiply(A=@n54.BaseColor, B=Fresnel(ExponentIn=@n55, Normal=@n56) {baseReflectFraction=0}))
  n58 = NamedRerouteUsage()
  n59 = NamedRerouteUsage()
  n60 = LinearInterpolate(
   A=LinearInterpolate(
    A=ScalarParameter() {group=C_Snow_AdvancedControls, parameterName=Snow Ceiling Balance, sortPriority=2}
    B=ScalarParameter() {defaultValue=0.65, group=C_Snow_AdvancedControls, parameterName=Snow Walls Balance, sortPriority=0}
    Alpha=@n58
   )
   B=ScalarParameter() {defaultValue=0.55, group=C_Snow_AdvancedControls, parameterName=Snow Floors Balance, sortPriority=4}
   Alpha=@n59
  )
  n61 = LinearInterpolate(A=Constant() {r=0.65}, B=LinearInterpolate(A=Constant(), B=Constant() {r=0.55}, Alpha=@n59), Alpha=@n58)
  n62 = OneMinus(None=GetMaterialAttributes(None=NamedRerouteUsage()) {attributeGetTypes [#3, #4]}.Roughness)
  n63 = StaticSwitchParameter(
   True=LinearInterpolate(
    A=ScalarParameter() {group=C_Snow_AdvancedControls, parameterName=Snow Roughness Min, sortPriority=7}
    B=ScalarParameter() {defaultValue=1, group=C_Snow_AdvancedControls, parameterName=Snow Roughness Max, sortPriority=8}
    Alpha=@n62
   )
   False=@n62
  ) {desc=Roughness is snow inut for details, group=C_Snow_AdvancedControls, parameterName=Snow Roughness Texture Edit, sortPriority=6}
  n64 = Saturate(None=@n63)
  n65 = Multiply(A=ScalarParameter() {group=C_Snow, parameterName=Snow Floor Flood}, B=NamedRerouteUsage())
  n66 = PreSkinnedPosition()
  n67 = ShaderStageSwitch(PixelShader=VertexInterpolator(VS=@n66).PS, VertexShader=@n66)
  n68 = MaterialFunctionCall(
   TextureObject=TextureObjectParameter() {group=C_Snow, parameterName=Snow noise mask, texture=&AtmosphericCloudNoise01}
   TextureSize=ScalarParameter() {defaultValue=2000, group=C_Snow, parameterName=Snow Breakup Scale}
   WorldPosition=NamedRerouteDeclaration(None=TransformPosition(None=@n67) {transformType=TRANSFORMPOSSOURCE_World}) {name=preskinned world pos, nodeColor=<rgba 0 1 0.8125 1>}
  ) {materialFunction=&WorldAlignedTexture}
  n69 = ConstantBiasScale(None=ScalarParameter() {defaultValue=0.5, group=C_Snow, parameterName=Snow Coverage, sortPriority=1}) {bias=-0.5, scale=2}
  n70 = ScalarParameter() {
   defaultValue=0.1
   desc=Use the original roughness texture for snow mask
   group=C_Snow
   parameterName=Snow Roughness Texture Strength
   sortPriority=5
  }
  n71 = LinearInterpolate(A=Multiply(A=@n64, B=OneMinus(None=@n65)), B=Saturate(None=Add(A=@n68.XYZ Texture, B=@n69)), Alpha=OneMinus(None=Saturate(None=@n70)))
  n72 = NamedRerouteUsage()
  n73 = NamedRerouteUsage()
  n74 = Saturate(None=Add(A=LinearInterpolate(A=@n71, B=@n64, Alpha=@n72), B=OneMinus(None=@n73)))
  n75 = LinearInterpolate(
   A=LinearInterpolate(
    A=ScalarParameter() {defaultValue=10, group=C_Snow_AdvancedControls, parameterName=Snow Ceiling Contrast, sortPriority=3}
    B=ScalarParameter() {defaultValue=10, group=C_Snow_AdvancedControls, parameterName=Snow Walls Contrast, sortPriority=1}
    Alpha=@n58
   )
   B=ScalarParameter() {defaultValue=10, group=C_Snow_AdvancedControls, parameterName=Snow Floors Contrast, sortPriority=5}
   Alpha=@n59
  )
  n76 = LinearInterpolate(A=Constant() {r=10}, B=LinearInterpolate(A=Constant() {r=10}, B=Constant() {r=10}, Alpha=@n59), Alpha=@n58)
  n77 = BlendMaterialAttributes(
   A=NamedRerouteDeclaration(None=SetMaterialAttributes!(MaterialAttributes=@n3.MaterialAttributes, Specular=@n40) {attributeSetTypes [#1]}) {name=Base Mat Updated Spec, nodeColor=<rgba 0 1 0.3906 1>}
   B=StaticSwitchParameter(
    True=SetMaterialAttributes!(MaterialAttributes=@n54.MaterialAttributes, BaseColor=@n57) {attributeSetTypes [#4]}
    False=@n53
   ) {group=C_Snow_AdvancedControls, parameterName=SnowFresnel, sortPriority=12}
   Alpha=MaterialFunctionCall(
    Transition Phase=StaticSwitchParameter(True=@n60, False=@n61) {group=C_Snow_AdvancedControls, parameterName=Snow Advanced Edit, sortPriority=-1}
    Height Texture=NamedRerouteDeclaration(None=@n74) {name=Snow Height Texture, nodeColor=<rgba 0.125 0 1 1>}
    Contrast=StaticSwitchParameter(True=@n75, False=@n76) {group=C_Snow_AdvancedControls, parameterName=Snow Advanced Edit, sortPriority=-1}
   ) {materialFunction=&HeightLerp}.Alpha
  ) {vertexAttributeBlendType=UseA}
  n78 = WorldPosition()
  n79 = ActorPositionWS()
  n80 = Multiply(A=Transform(None=Subtract(A=@n78.XYZ, B=@n79)) {transformSourceType=TRANSFORMSOURCE_World, transformType=TRANSFORM_Local}) {constB=0.1}
  n81 = ScalarParameter() {
   defaultValue=1
   desc="Mask based on mesh local Y. Will make what is on negative side from pivot free of snow effect. In some cases you can't use it ( Curved walls) so instead use additional material layer to controll snow level inside buildings"
   group=C_Snow_AdvancedControls
   parameterName=Interior Walls Snow Value
   sortPriority=10
  }
  n82 = Max(A=Saturate(None=ComponentMask(None=@n80) {g=true}), B=@n81)
  n83 = NamedRerouteDeclaration(None=MaterialProxyReplace(Realtime=BlendMaterialAttributes(A=@n37, B=@n77, Alpha=@n82) {vertexAttributeBlendType=UseA}, MaterialProxy=@n37) {desc=Local space stuff doesn't work in HLOD}) {name=After Snow, nodeColor=<rgba 0.9922 0 1 1>}
  n84 = GetMaterialAttributes(None=@n83) {attributeGetTypes [#4, #3, #1]}
  n85 = Desaturation(None=NamedRerouteUsage(), Fraction=Constant() {r=0.5})
  n86 = Multiply(A=@n84.BaseColor, B=@n85)
  n87 = MaterialFunctionCall() {materialFunction=&ObjectLocalBounds}
  n88 = Add(A=ScalarParameter() {group=C_Snow_Leak, parameterName=Snow Leak Offset}, B=@n87.Local Bounds Max)
  n89 = TransformPosition(None=WorldPosition() {worldPositionShaderOffset=WPT_ExcludeAllShaderOffsets}.XYZ) {transformSourceType=TRANSFORMPOSSOURCE_World}
  n90 = ScalarParameter() {defaultValue=1, group=C_Snow_Leak, parameterName=Snow Leak ScaleTweak}
  n91 = ComponentMask(None=Divide(A=Subtract(A=@n88, B=@n89), B=Multiply(A=@n87.Local Bounds Size, B=@n90))) {b=true}
  n92 = ComponentMask(None=VertexNormalWS()) {b=true}
  n93 = Saturate(None=Add(A=Saturate(None=@n91), B=@n92))
  n94 = NamedRerouteDeclaration(None=Saturate(None=Multiply(B=@n92) {constA=0.75})) {name=Top Gradient, nodeColor=<rgba 0 0.2734 1 1>}
  n95 = BlendMaterialAttributes(
   A=SetMaterialAttributes!(
    MaterialAttributes=@n84.MaterialAttributes
    BaseColor=LinearInterpolate(A=@n86, B=@n85) {constAlpha=0.15}
    Roughness=Multiply(A=@n84.Roughness) {constB=0.75}
    Specular=Multiply(A=@n84.Specular) {constB=0.5}
   ) {attributeSetTypes [#4, #3, #1]}
   B=@n83
   Alpha=StaticSwitchParameter(True=Multiply(A=@n93, B=@n94), False=@n94) {desc=Only on building meshes, group=C_Snow_Leak, parameterName=Add Actor Space Snow Leak}
  )
  NamedRerouteDeclaration(None=StaticSwitchParameter(True=MaterialProxyReplace(Realtime=@n95, MaterialProxy=@n83) {desc="HLOD can't support  Object Bounds"}, False=@n83) {group=C_Snow_Leak, parameterName=Use Snow Leak, sortPriority=22}) {name=Snow Leak IN, nodeColor=<rgba 1 0.8203 0 1>}
  NamedRerouteDeclaration(None=Saturate(None=@n14)) {name=Wall Mask, nodeColor=<rgba 0.9453 0 1 1>}
  n96 = Sign(None=ComponentMask(None=VertexNormalWS()) {b=true})
  n97 = OneMinus(None=NamedRerouteUsage())
  NamedRerouteDeclaration(None=Power(Base=Multiply(A=@n96, B=@n97)) {constExponent=0.3}) {name=FloorMask, nodeColor=<rgba 0.7813 0 1 1>}
  NamedRerouteDeclaration(None=OneMinus(None=NamedRerouteUsage())) {name=CeilMask, nodeColor=<rgba 1 0 0.7031 1>}
  n98 = NamedRerouteUsage()
  NamedRerouteDeclaration(None=MaterialFunctionCall!(
    Low _ PC Low and Med=@n4
    "High _ XB1, PS4"=@n98
    "Epic _ PC High, XSX, PS5"=@n98
    --------
    "Mobile Low _ PC Perf, Android"=@n4
    Mobile Med _ Switch=@n4
   ) {materialFunction=/Game/Global/MaterialFunctions/TechArt/MF_Fortnite_QualitySwitch_MaterialAttributes}.Result) {name=Mat Output, nodeColor=<rgba 1 0.8672 0 1>}
  NamedRerouteDeclaration(None=@n11.Result) {name=Normal, nodeColor=<rgba 0.6172 1 0 1>}
  n99 = StaticSwitchParameter(True=ScalarParameter() {defaultValue=10, group=C_Snow_AO, parameterName=Snow Details AO Distance, sortPriority=22}, False=Constant() {r=10}) {group=C_Snow_AO, parameterName=Snow AO Edit, sortPriority=21}
  n100 = ScalarParameter() {defaultValue=1, group=C_Snow_AO, parameterName=Snow Ceilings AO Scale, sortPriority=26}
  n101 = ScalarParameter() {defaultValue=1, group=C_Snow_AO, parameterName=Snow Floors AO Scale, sortPriority=27}
  n102 = NamedRerouteUsage()
  n103 = Constant() {r=1}
  n104 = StaticSwitchParameter(
   True=StaticSwitchParameter(True=LinearInterpolate(A=@n100, B=@n101, Alpha=@n102), False=@n103) {group=C_Snow_AO, parameterName=Snow AO Advanced Edit, sortPriority=25}
   False=@n103
  ) {group=C_Snow_AO, parameterName=Snow AO Edit, sortPriority=21}
  n105 = NamedRerouteUsage()
  n106 = StaticSwitchParameter(True=ScalarParameter() {defaultValue=500, group=C_Snow_AO, parameterName=Snow Details AO Fade, sortPriority=23}, False=Constant() {r=500}) {group=C_Snow_AO, parameterName=Snow AO Edit, sortPriority=21}
  n107 = DistanceFieldApproxAO!(
   World Position
   Normal
   BaseDistance=LinearInterpolate(A=@n99, B=Multiply(A=@n99, B=@n104), Alpha=@n105)
   Radius=LinearInterpolate(A=@n106, B=Multiply(A=@n104, B=@n106), Alpha=@n105)
  ) {numSteps=3}
  n108 = Constant() {r=1}
  n109 = Multiply(
   A=GetMaterialAttributes(None=NamedRerouteUsage()) {attributeGetTypes [#5]}.Metallic
   B=ScalarParameter() {defaultValue=0.25, group=C_Snow, parameterName=Metal to Snow Mask Value, sortPriority=6}
  )
  n110 = Saturate(None=Add(
    A=ShadingPathSwitch(Default=@n107, Mobile=@n108) {desc=Migrated from FeatureLevelSwitch}
    B=NamedRerouteDeclaration(None=@n109) {name=metal Mask, nodeColor=<rgba 0.6641 1 0 1>}
   ))
  n111 = OneMinus(None=ScalarParameter() {group=C_Snow_AO, parameterName=AO Strength, sortPriority=22})
  NamedRerouteDeclaration(None=Saturate(None=Add(A=@n110, B=@n111))) {name=Ao, nodeColor=<rgba 1 0.9844 0 1>}
  n112 = CurveAtlasRowParameter(CurveTime=CollectionParameter() {collection=&FortniteMaterialParameters, parameterName=TimeOfDay_0-1}) {
   atlas=/Game/Athena/Apollo/Environments/Landscape/Materials/Biome/UniqueBiomeMaterialFunctions/LandscapeCurveAtlas
   curve=/Game/Athena/Apollo/Environments/Landscape/Materials/Biome/UniqueBiomeMaterialFunctions/LandscapeCurve-TOD
   desc="0=morn 1=day 0=eve 1=night"
  }
  NamedRerouteDeclaration(None=Saturate(None=@n112.A)) {name=TOD-Lerp, nodeColor=<rgba 1 0 1 1>}
  NamedRerouteDeclaration(None=LinearInterpolate(A=Constant3Vector() {constant=<rgba 0.285 0.4009 0.5 1>}, B=Constant3Vector() {constant=<rgba 0.2 0.3078 0.4 1>}, Alpha=NamedRerouteUsage())) {name=ShadowColor, nodeColor=<rgba 1 0 1 1>}
  n113 = LinearInterpolate(A=Constant() {r=-0.15}, B=Constant() {r=-1}, Alpha=NamedRerouteUsage())
  n114 = LinearInterpolate(A=Constant() {r=5}, B=Constant() {r=2}, Alpha=NamedRerouteUsage())
  n115 = LinearInterpolate(
   A=PixelNormalWS()
   B=VertexNormalWS()
   Alpha=ScalarParameter() {defaultValue=0.5, desc="0=PixelNormal, 1=VertexNormal", group=C_Snow_AdvancedControls, parameterName=FSS-BlendNormalType}
  )
  n116 = SkyAtmosphereLightDirection()
  n117 = Power(Base=Saturate(None=DotProduct(A=@n115, B=@n116)), Exp=Constant() {r=3.5})
  NamedRerouteDeclaration(None=Saturate(None=LinearInterpolate(A=@n113, B=@n114, Alpha=@n117))) {name=ShadowMask, nodeColor=<rgba 1 0 1 1>}
  NamedRerouteDeclaration(None=MaterialFunctionCall(BaseNormal=@n7.XYZ Texture, AdditionalNormal=@n11.Result) {materialFunction=&BlendAngleCorrectedNormals}.Result) {name=debug node, nodeColor=<rgba 0.1719 0 1 1>}
  n118 = ScalarParameter() {defaultValue=1, group=C_Snow, parameterName=Snow Value Layer 1, sortPriority=1}
  n119 = ScalarParameter() {group=C_Snow, parameterName=Snow Value Layer 2, sortPriority=2}
  n120 = ScalarParameter() {group=C_Snow, parameterName=Snow Value Layer 3, sortPriority=3}
  n121 = ScalarParameter() {group=C_Snow, parameterName=Snow Value Layer 4, sortPriority=4}
  NamedRerouteDeclaration(None=Saturate(None=MaterialFunctionCall!(1=@n118, 2=@n119, 3=@n120, 4=@n121) {materialFunction=&/MF_UV2LerpScalar_Layers_New}.Result)) {name=Layer Walls Mask, nodeColor=<rgba 0.1484 0 1 1>}
  NamedRerouteUsage()
  Reroute(None=@n11.Result)
  Reroute()
  Saturate(None=MaterialFunctionCall!(Scale) {materialFunction=/Game/Athena/Apollo/Environments/Landscape/Materials/Biome/CommonMaterials/MF_SimpleWorldUV}.Result)
  ScalarParameter() {defaultValue=3.5, group=Snow, parameterName=FSS-BlendPow}
  ScalarParameter() {defaultValue=-0.15, group=C_Snow, parameterName=FSS-A-Morning}
  ScalarParameter() {defaultValue=-1, group=C_Snow, parameterName=FSS-A-Day}
  ScalarParameter() {defaultValue=5, group=C_Snow, parameterName=FSS-B-Morning}
  ScalarParameter() {defaultValue=2, group=C_Snow, parameterName=FSS-B-Day}
  SetMaterialAttributes!(MaterialAttributes=@n6, BaseColor=NamedRerouteUsage(), Normal=Constant3Vector() {constant=<rgba 0 0 1 0>}) {attributeSetTypes [#4, #2]}
  TextureSampleParameter2D() {
   group=C_Snow
   parameterName=SnowNormal
   samplerType=SAMPLERTYPE_Normal
   sortPriority=10
   texture=&T_Athena_Snow_01_N
  }
  VectorParameter() {group=C_Snow, parameterName=Snow Shadow Color-Day, defaultValue=<rgba 0.2 0.3078 0.4 1>}
  VectorParameter() {group=C_Snow, parameterName=Snow Shadow Color-Morning, defaultValue=<rgba 0.285 0.4009 0.5 1>}
  VertexNormalWS()
  WorldPosition()
 ]}
