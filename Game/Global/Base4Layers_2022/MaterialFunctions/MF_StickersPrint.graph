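# Material function that prints a tinted sticker onto input material attributes, behind a switch marked WIP, and also outputs the sticker mask.
# Sticker color replaces or modulates base color, with optional flat specular, roughness and metallic values and flattening of the underlying normal.
# The mask combines a selected texture channel with per-layer visibility from a UV2 layer function.
# An optional HeightLerp uses material roughness as height to break up sticker edges.
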MaterialFunction /Game/Global/Base4Layers_2022/MaterialFunctions/MF_StickersPrint
graph {nodes [
  n1 = NamedRerouteUsage()
  n2 = ScalarParameter() {defaultValue=0.5, desc="SRM - Spec, Roughness, Metallic", group=B_StickerPrint_SRM, parameterName=Sticker Specular, sortPriority=1}
  n3 = ScalarParameter() {defaultValue=0.5, desc="SRM - Spec, Roughness, Metallic", group=B_StickerPrint_SRM, parameterName=Sticker Roughness, sortPriority=1}
  n4 = FunctionInput() {inputName=Material In, inputType=FunctionInput_MaterialAttributes}
  n5 = GetMaterialAttributes(None=NamedRerouteDeclaration(None=@n4) {name=Mat Intput, nodeColor=<rgba 1 0 0.0703 1>}) {attributeGetTypes [#1]}
  n6 = NamedRerouteUsage()
  n7 = VectorParameter() {group=B_StickerPrint, parameterName=Sticker Tint, channelNames {a="Blend >> Modulate", b="", g="", r=""}, defaultValue=<rgba 1 1 1 0>}
  n8 = SetMaterialAttributes!(
   MaterialAttributes=@n5.MaterialAttributes
   BaseColor=LinearInterpolate(A=Multiply(A=@n6, B=@n7), B=Multiply(A=@n5.BaseColor, B=@n6), Alpha=@n7."Blend >> Modulate")
  ) {attributeSetTypes [#1]}
  n9 = SetMaterialAttributes!(
   MaterialAttributes=@n8
   Specular=@n2
   Roughness=@n3
   Metallic=ScalarParameter() {desc="SRM - Spec, Roughness, Metallic", group=B_StickerPrint_SRM, parameterName=Sticker Metallic, sortPriority=1}
   EmissiveColor=Constant3Vector()
  ) {attributeSetTypes [
    #2
    #3
    #4
    #5
   ]}
  n10 = ScalarParameter() {
   desc="Use to blend Flat SRM Values with the underlaying surface "
   group=B_StickerPrint_SRM
   parameterName=Sticker SRM Blend Down
   sortPriority=0
  }
  n11 = StaticSwitchParameter(True=BlendMaterialAttributes(A=@n9, B=@n8, Alpha=@n10), False=@n8) {
   desc="Custom scalar (Spec/Roughness/Metallnes) Values for the sticker area"
   group=B_StickerPrint
   parameterName=Sticker SRM
   sortPriority=-5
  }
  n12 = GetMaterialAttributes(None=@n11) {attributeGetTypes [#6]}
  n13 = SetMaterialAttributes!(
   MaterialAttributes=@n12.MaterialAttributes
   Normal=MaterialFunctionCall(Normal=@n12.Normal, Flatness=ScalarParameter() {parameterName=Sticker Flatten Normal Under}) {materialFunction=&FlattenNormal}.Result
  ) {attributeSetTypes [#6]}
  n14 = NamedRerouteUsage()
  FunctionOutput(None=StaticSwitchParameter(True=BlendMaterialAttributes(A=@n1, B=@n13, Alpha=@n14), False=NamedRerouteUsage()) {desc="DO NOT USE YET> WIP", group=B_StickerPrint, parameterName=Sticker Print, sortPriority=-5})
  FunctionOutput(None=@n14) {outputName=Sticker Mask}
  NamedRerouteDeclaration(None=TextureSample(Tex=FunctionInput() {inputName=SFX RGB, inputType=FunctionInput_Texture2D}).RGB) {name=RGB Color, nodeColor=<rgba 0.4531 1 0 1>}
  n15 = FunctionInput() {inputName=SFX Texture, inputType=FunctionInput_Texture2D}
  n16 = ScalarParameter() {defaultValue=1, desc=StickerPerLayerVisibility, group=B_StickerPrint, parameterName=Sticker_Layer_1, sliderMax=4, sortPriority=2}
  n17 = ScalarParameter() {defaultValue=1, desc=StickerPerLayerVisibility, group=B_StickerPrint, parameterName=Sticker_Layer_2, sliderMax=4, sortPriority=2}
  n18 = ScalarParameter() {defaultValue=1, desc=StickerPerLayerVisibility, group=B_StickerPrint, parameterName=Sticker_Layer_3, sliderMax=4, sortPriority=2}
  n19 = ScalarParameter() {defaultValue=1, desc=StickerPerLayerVisibility, group=B_StickerPrint, parameterName=Sticker_Layer_4, sliderMax=4, sortPriority=2}
  n20 = Multiply(
   A=DotProduct(A=TextureSample(Tex=@n15).RGB, B=VectorParameter() {group=B_StickerPrint, parameterName=Sricker Mask Ch, defaultValue=<rgba 1 0 0 0>}.RGB)
   B=MaterialFunctionCall!(1=@n16, 2=@n17, 3=@n18, 4=@n19) {materialFunction=&/MF_UV2LerpScalar_Layers_New}.Result
  )
  n21 = GetMaterialAttributes(None=NamedRerouteUsage()) {attributeGetTypes [#3]}
  n22 = StaticSwitchParameter(
   True=MaterialFunctionCall(Transition Phase=@n20, Height Texture=LinearInterpolate(Alpha=@n21.Roughness) {constA=-1}, Contrast=Constant() {r=0.5}) {materialFunction=&HeightLerp}.Alpha
   False=@n20
  ) {
   desc=Utilize material Roughness for the Height map details. Can be used to blend-in sticker on edges with surface around
   group=B_StickerPrint
   parameterName=Sticker Use Height Lerp
   sortPriority=-1
  }
  NamedRerouteDeclaration(None=@n22) {name=Sticker Mask, nodeColor=<rgba 0 0.6719 1 1>}
  Reroute(None=@n4)
 ]}
