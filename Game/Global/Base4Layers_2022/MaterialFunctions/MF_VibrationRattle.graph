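# Material function that adds a switchable vibration rattle to Material Attributes through World Position Offset.
# Per-instance random and object position vary the time rate to desynchronize instances.
# Products of detuned parabolic Frac waves drive a biased, scaled translation, though the third wave lacks a time input, leaving only X animated.
# Sine products wobble the axis of a fixed-angle RotateAboutAxis around a pivot parameter, added to any existing offset.
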MaterialFunction /Game/Global/Base4Layers_2022/MaterialFunctions/MF_VibrationRattle
graph {nodes [
  n1 = FunctionInput() {inputType=FunctionInput_MaterialAttributes}
  n2 = GetMaterialAttributes(None=@n1) {attributeGetTypes [#1]}
  n3 = MaterialFunctionCall(Float3=TransformPosition(None=Constant3Vector()) {transformType=TRANSFORMPOSSOURCE_World}) {materialFunction=&BreakOutFloat3Components}
  n4 = Add(A=Add(A=@n3.R, B=@n3.R), B=@n3.B)
  n5 = Frac(None=Add(A=PerInstanceRandom(), B=Abs(None=@n4)))
  n6 = ScalarParameter() {defaultValue=0.1, group=VibrationRattle, parameterName=Vibration_VariationPerinstanceRandom}
  n7 = Multiply(A=Add(A=Multiply(A=@n5, B=@n6) {constB=0.1}), B=Time()) {constA=1}
  n8 = ScalarParameter() {defaultValue=1, group=VibrationRattle, parameterName=Vibration_Speed}
  n9 = Frac(None=Multiply(A=@n7, B=@n8))
  n10 = Multiply(A=Multiply(A=@n9, B=OneMinus(None=@n9))) {constB=4}
  n11 = Frac(None=Multiply(A=@n7, B=Add(A=@n8) {constB=0.02}))
  n12 = Multiply(A=Multiply(A=@n11, B=OneMinus(None=@n11))) {constB=4}
  n13 = Multiply(A=@n10, B=@n12)
  n14 = Frac(None=Multiply(B=Add(A=@n8) {constB=0.05}))
  n15 = Multiply(A=Multiply(A=@n14, B=OneMinus(None=@n14))) {constB=4}
  n16 = Multiply(A=@n10, B=@n15)
  n17 = Add(
   A=AppendVector(A=AppendVector(A=@n13, B=@n16), B=Multiply(A=@n12, B=@n15))
   B=VectorParameter() {group=VibrationRattle, parameterName=VibrationBias, defaultValue=<rgba -0.5 -0.5 0 1>}.RGB
  )
  n18 = VectorParameter() {group=VibrationRattle, parameterName=VibrationOffset, defaultValue=<rgba 1 1 1 1>}
  n19 = Multiply(A=@n18.RGB, B=@n18.RGB)
  n20 = ScalarParameter() {defaultValue=2, group=VibrationRattle, parameterName=Vibration_Rotation_Speed}
  n21 = Sine(None=Multiply(A=@n7, B=@n20))
  n22 = Sine(None=Multiply(A=@n7, B=Add(A=@n20) {constB=0.02}))
  n23 = Sine(None=Multiply(A=@n7, B=Multiply(A=@n20) {constB=0.04}))
  n24 = AppendVector(A=AppendVector(A=Multiply(A=@n21, B=@n22), B=Multiply(A=@n22, B=@n23)), B=Multiply(A=@n21, B=@n23))
  n25 = VectorParameter() {group=VibrationRattle, parameterName=VibrationDirectionRandom, defaultValue=<rgba 1 1 0 0.5>}
  n26 = Normalize(VectorInput=Add(
    A=Multiply(A=@n24, B=@n25.RGB)
    B=VectorParameter() {group=VibrationRattle, parameterName=VibrationDirection, defaultValue=<rgba 0 0 1 0.5>}.RGB
   ))
  n27 = ScalarParameter() {defaultValue=0.005, group=VibrationRattle, parameterName=VibrationRotationMax}
  n28 = TransformPosition(None=VectorParameter() {group=VibrationRattle, parameterName=VibrationPivot, defaultValue=<rgba 1 1 1 0.5>}.RGB) {transformType=TRANSFORMPOSSOURCE_TranslatedWorld}
  n29 = WorldPosition() {worldPositionShaderOffset=WPT_CameraRelative}
  n30 = Add(
   A=@n2.WorldPositionOffset
   B=Add(A=Multiply(A=@n17, B=@n19), B=RotateAboutAxis(NormalizedRotationAxis=@n26, RotationAngle=@n27, PivotPoint=@n28, Position=@n29.XYZ))
  )
  FunctionOutput(None=StaticSwitchParameter(
    True=SetMaterialAttributes!(MaterialAttributes=@n2.MaterialAttributes, WorldPositionOffset=@n30) {attributeSetTypes [#1]}
    False=@n1
   ) {group=VibrationRattle, parameterName=UseVibrationRattle})
 ]}
